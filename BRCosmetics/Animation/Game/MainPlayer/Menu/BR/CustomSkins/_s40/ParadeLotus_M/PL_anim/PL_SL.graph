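# Level Sequence for authoring a player animation of roughly 420 frames for the ParadeLotus_M custom skin in the Battle Royale menu.
# It spawns a FortPlayerMannequin at an identity transform, with a skeletal animation track and a Fortnite_Biped_CtrlRig Control Rig track.
# The rig's FK/IK switches stay constant, set differently for legs and arms.
# Dense keys on both FK and IK leg controls suggest motion baked onto the rig.
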LevelSequence /BRCosmetics/Animation/Game/MainPlayer/Menu/BR/CustomSkins/_s40/ParadeLotus_M/PL_anim/PL_SL
bindings [FortPlayerMannequin {
  id=#1
  possessed_class=FortPlayerMannequin
  tracks [
   {type=MovieSceneSpawnTrack, sections [{type=MovieSceneSpawnSection, channels [None {type=MovieSceneScriptingBoolChannel, default=true}]}]}
   {
    type=MovieScene3DTransformTrack
    sections [{
      type=MovieScene3DTransformSection
      channels [
       Location.X {type=MovieSceneScriptingDoubleChannel, default=0}
       Location.Y {type=MovieSceneScriptingDoubleChannel, default=0}
       Location.Z {type=MovieSceneScriptingDoubleChannel, default=0}
       Rotation.X {type=MovieSceneScriptingDoubleChannel, default=0}
       Rotation.Y {type=MovieSceneScriptingDoubleChannel, default=0}
       Rotation.Z {type=MovieSceneScriptingDoubleChannel, default=0}
       Scale.X {type=MovieSceneScriptingDoubleChannel, default=1}
       Scale.Y {type=MovieSceneScriptingDoubleChannel, default=1}
       Scale.Z {type=MovieSceneScriptingDoubleChannel, default=1}
      ]
     }]
   }
   {type=MovieSceneSkeletalAnimationTrack, sections [{type=MovieSceneSkeletalAnimationSection, end_frame=421, start_frame=-1}]}
   {
    type=MovieSceneControlRigParameterTrack
    display_name=Fortnite_Biped_CtrlRig
    sections [{
      type=MovieSceneControlRigParameterSection
      channels [
       leg_l_stretch_switch {type=MovieSceneScriptingBoolChannel, default=true, const [true,423]}
       leg_r_stretch_switch {type=MovieSceneScriptingBoolChannel, default=true, const [true,423]}
       arm_l_fk_local {type=MovieSceneScriptingBoolChannel, default=true, const [true,423]}
       Auto Interpolate_3 {type=MovieSceneScriptingBoolChannel, default=false, const [false,423]}
       Auto Interpolate {type=MovieSceneScriptingBoolChannel, default=false, const [false,423]}
       arm_l_stretch_switch {type=MovieSceneScriptingBoolChannel, default=false, const [false,423]}
       Keep Aligned {type=MovieSceneScriptingBoolChannel, default=false, const [false,423]}
       arm_r_fk_local {type=MovieSceneScriptingBoolChannel, default=true, const [true,423]}
       Auto Interpolate_4 {type=MovieSceneScriptingBoolChannel, default=false, const [false,423]}
       Auto Interpolate_2 {type=MovieSceneScriptingBoolChannel, default=false, const [false,423]}
       arm_r_stretch_switch {type=MovieSceneScriptingBoolChannel, default=false, const [false,423]}
       Keep Aligned_2 {type=MovieSceneScriptingBoolChannel, default=false, const [false,423]}
       head_local {type=MovieSceneScriptingBoolChannel, default=true, const [true,423]}
       Stretch {type=MovieSceneScriptingBoolChannel, default=false, const [false,423]}
       head_stretch {type=MovieSceneScriptingBoolChannel, default=false, const [false,423]}
       spine_stretch_switch {type=MovieSceneScriptingBoolChannel, default=false, const [false,423]}
       arm_l_fk_ik_switch {type=MovieSceneScriptingBoolChannel, default=false, const [false,423]}
       leg_l_fk_ik_switch {type=MovieSceneScriptingBoolChannel, default=true, const [true,423]}
       arm_r_fk_ik_switch {type=MovieSceneScriptingBoolChannel, default=false, const [false,423]}
       leg_r_fk_ik_switch {type=MovieSceneScriptingBoolChannel, default=true, const [true,423]}
       ShowBodyControls {type=MovieSceneScriptingBoolChannel, default=true, const [true,423]}
       spine_fk_ik_switch {type=MovieSceneScriptingBoolChannel, default=false, const [false,423]}
       neck_fk_ik_switch {type=MovieSceneScriptingBoolChannel, default=false, const [false,423]}
       space {type=MovieSceneScriptingByteChannel, default=1, const [1,423]}
       global_ctrl.Location.X {type=MovieSceneScriptingFloatChannel, default=0, const [0,423]}
       global_ctrl.Location.Y {type=MovieSceneScriptingFloatChannel, default=0, const [0,423]}
       global_ctrl.Location.Z {type=MovieSceneScriptingFloatChannel, default=0, const [0,423]}
       global_ctrl.Rotation.X {type=MovieSceneScriptingFloatChannel, default=0, const [0,423]}
       global_ctrl.Rotation.Y {type=MovieSceneScriptingFloatChannel, default=0, const [0,423]}
       global_ctrl.Rotation.Z {type=MovieSceneScriptingFloatChannel, default=0, const [0,423]}
       root_ctrl.Location.X {type=MovieSceneScriptingFloatChannel, default=0, const [0,423]}
       root_ctrl.Location.Y {type=MovieSceneScriptingFloatChannel, default=0, const [0,423]}
       root_ctrl.Location.Z {type=MovieSceneScriptingFloatChannel, default=0, const [0,423]}
       root_ctrl.Rotation.X {type=MovieSceneScriptingFloatChannel, default=0, const [0,423]}
       root_ctrl.Rotation.Y {type=MovieSceneScriptingFloatChannel, default=0, const [0,423]}
       root_ctrl.Rotation.Z {type=MovieSceneScriptingFloatChannel, default=0, const [0,423]}
       root_ctrl.Scale.X {type=MovieSceneScriptingFloatChannel, default=1, const [1,423]}
       root_ctrl.Scale.Y {type=MovieSceneScriptingFloatChannel, default=1, const [1,423]}
       root_ctrl.Scale.Z {type=MovieSceneScriptingFloatChannel, default=1, const [1,423]}
       body_offset_ctrl.Location.X {type=MovieSceneScriptingFloatChannel, default=0, const [0,423]}
       body_offset_ctrl.Location.Y {type=MovieSceneScriptingFloatChannel, default=0, const [0,423]}
       body_offset_ctrl.Location.Z {type=MovieSceneScriptingFloatChannel, default=0, const [0,423]}
       body_offset_ctrl.Rotation.X {type=MovieSceneScriptingFloatChannel, default=0, const [0,423]}
       body_offset_ctrl.Rotation.Y {type=MovieSceneScriptingFloatChannel, default=0, const [0,423]}
       body_offset_ctrl.Rotation.Z {type=MovieSceneScriptingFloatChannel, default=0, const [0,423]}
       body_offset_ctrl.Scale.X {type=MovieSceneScriptingFloatChannel, default=1, const [1,423]}
       body_offset_ctrl.Scale.Y {type=MovieSceneScriptingFloatChannel, default=1, const [1,423]}
       body_offset_ctrl.Scale.Z {type=MovieSceneScriptingFloatChannel, default=1, const [1,423]}
       thigh_l_fk_ctrl.Location.X {type=MovieSceneScriptingFloatChannel, default=0, thinned_from=423, keys [[0,0],[422,0]]}
       thigh_l_fk_ctrl.Location.Y {type=MovieSceneScriptingFloatChannel, default=0, thinned_from=423, keys [[0,0],[422,0]]}
       thigh_l_fk_ctrl.Location.Z {type=MovieSceneScriptingFloatChannel, default=0, thinned_from=423, keys [[0,0],[422,0]]}
       thigh_l_fk_ctrl.Rotation.X {
        type=MovieSceneScriptingFloatChannel
        default=29.2862
        thinned_from=423
        keys [[0,28.0929],[7,28.1982],[15,27.9294],[16,27.9432],[17,28.0281],[18,28.254],[19,28.7038],[22,30.6264],[23,31.324],[24,31.9019],[25,32.335],[26,32.6942],[27,32.9781],[28,33.189],[30,33.4238],[35,33.473],[52,32.8227],[56,32.8174],[66,33.3985],[68,33.3996],[69,33.347],[70,33.2079],[71,32.9399],[72,32.5397],[73,32.0459],[75,30.8717],[78,28.952],[81,27.243],[82,26.8013],[83,26.4696],[84,26.1981],[85,25.9836],[87,25.7116],[90,25.5871],[93,25.6455],[98,26.0925],[103,26.7371],[127,27.4263],[135,28.4781],[136,28.653],[137,28.9017],[138,29.2511],[139,29.8198],[140,30.6186],[141,31.799],[142,33.2315],[143,34.8129],[144,35.3611],[145,35.5233],[146,35.3797],[147,34.9894],[148,34.4841],[150,32.9597],[151,32.6708],[152,32.6031],[153,32.9558],[156,34.5231],[157,34.749],[226,33.9408],[228,34.0717],[232,34.6617],[233,34.7464],[234,34.6729],[237,34.0737],[239,33.8316],[243,33.6972],[244,33.5995],[245,33.3837],[246,33.0704],[247,32.5973],[249,31.1198],[250,30.7304],[251,30.5387],[252,30.4584],[254,30.4637],[264,31.1582],[273,31.2421],[286,30.8079],[291,30.8797],[293,31.0014],[296,31.5148],[297,31.5976],[298,31.554],[299,31.4133],[300,31.1839],[301,30.8801],[302,30.323],[303,29.5176],[304,28.5996],[305,27.6101],[306,26.69],[307,25.9036],[310,23.8423],[311,23.2457],[312,22.7633],[313,22.3667],[314,22.0622],[315,21.8295],[316,21.7273],[317,21.7823],[319,22.0458],[322,22.5429],[324,22.7516],[344,22.9698],[347,22.8834],[348,22.7992],[349,22.6536],[350,22.4423],[351,22.1333],[352,21.7316],[353,21.2107],[356,19.4061],[357,18.8397],[358,18.3553],[359,17.9388],[360,17.5646],[361,17.2379],[362,16.9672],[363,16.7622],[365,16.4948],[367,16.3444],[371,16.2652],[374,16.3043],[376,16.4054],[377,16.4984],[378,16.6265],[379,16.8031],[380,17.0816],[381,17.5141],[382,18.1389],[383,19.0067],[384,20.359],[385,22.1634],[386,24.0565],[387,25.6725],[388,26.8966],[389,27.8569],[390,28.5558],[391,29.0541],[392,29.4431],[393,29.7396],[394,29.958],[396,30.2164],[400,30.3284],[401,30.275],[402,30.1329],[406,29.22],[409,28.7279],[414,28.2439],[418,28.0977],[422,28.0977]]
       }
       thigh_l_fk_ctrl.Rotation.Y {
        type=MovieSceneScriptingFloatChannel
        default=3.2165
        thinned_from=423
        keys [[0,4.5387],[6,4.5597],[11,4.6195],[13,4.6182],[14,4.6022],[15,4.5696],[16,4.5012],[17,4.3702],[18,4.1599],[19,3.8193],[20,3.4104],[21,3.087],[22,2.9022],[23,2.823],[24,2.8458],[25,2.9385],[27,3.1526],[29,3.3968],[30,3.4986],[31,3.5741],[32,3.6293],[33,3.6675],[34,3.6928],[41,3.7713],[43,3.7791],[52,3.7019],[53,3.7019],[54,3.6829],[55,3.6829],[57,3.6617],[68,3.6768],[70,3.7003],[71,3.7216],[72,3.7593],[73,3.8159],[74,3.9046],[75,4.0121],[76,4.105],[77,4.1699],[78,4.2181],[80,4.2873],[81,4.3127],[82,4.3237],[90,4.2493],[92,4.25],[93,4.2686],[94,4.2986],[96,4.3854],[101,4.6584],[104,4.7894],[107,4.8814],[113,4.9853],[115,5.0028],[117,4.9957],[124,4.8609],[125,4.8503],[126,4.8546],[127,4.9261],[128,5.0384],[129,5.1814],[130,5.2784],[132,5.5929],[133,5.5855],[134,5.389],[135,5.033],[136,4.5427],[137,3.9919],[138,3.3182],[139,2.2617],[140,1.3711],[141,1.4751],[142,1.9645],[143,2.2999],[144,2.7757],[145,3.3523],[146,3.8196],[147,4.4131],[149,5.6303],[150,6.1054],[151,6.3095],[152,6.2331],[153,5.8156],[154,5.3173],[156,4.6073],[157,4.4344],[158,4.2972],[159,4.3193],[162,4.4322],[166,4.6183],[170,4.7566],[174,4.8525],[180,4.9182],[181,4.9182],[182,4.9284],[215,4.9646],[223,5.0075],[225,5.0039],[226,4.9858],[227,4.946],[229,4.8167],[230,4.7344],[231,4.6183],[232,4.3573],[233,4.1965],[234,4.1487],[235,4.1708],[236,4.208],[238,4.2992],[239,4.3087],[240,4.2963],[241,4.3849],[242,4.3582],[243,4.3222],[244,4.3031],[245,4.3708],[246,4.5474],[247,4.831],[248,5.1835],[249,5.4996],[250,5.6128],[251,5.6174],[253,5.4541],[254,5.3965],[255,5.3965],[259,5.479],[262,5.4916],[266,5.4529],[274,5.3198],[295,5.2856],[296,5.3036],[297,5.3642],[298,5.4885],[299,5.6719],[300,5.8973],[301,6.2016],[302,6.6079],[303,7.04],[304,7.4417],[305,7.8088],[306,8.0374],[307,8.1652],[308,8.2529],[309,8.1206],[310,7.7986],[311,7.5454],[312,7.3679],[313,7.1659],[314,6.9873],[315,6.8798],[316,6.8194],[319,6.7061],[321,6.5644],[322,6.5084],[324,6.4426],[325,6.4244],[326,6.4205],[327,6.4358],[328,6.4667],[330,6.564],[334,6.7951],[344,7.2388],[348,7.3439],[350,7.367],[352,7.5377],[353,7.643],[354,7.7833],[355,7.8195],[356,7.7517],[358,7.5263],[359,7.4286],[362,7.0138],[363,6.9153],[364,6.8521],[365,6.8203],[366,6.8026],[368,6.8095],[373,6.9215],[375,6.9543],[376,6.9219],[377,6.7981],[378,6.6074],[379,6.3748],[380,6.0574],[381,5.635],[382,5.1953],[383,4.7898],[384,4.4561],[385,4.1533],[386,3.859],[387,3.6211],[388,3.5201],[390,3.5785],[391,3.707],[394,4.033],[395,4.1304],[396,4.2146],[397,4.2864],[398,4.3458],[399,4.3922],[403,4.4969],[406,4.5504],[408,4.5662],[422,4.5389]]
       }
       thigh_l_fk_ctrl.Rotation.Z {
        type=MovieSceneScriptingFloatChannel
        default=-0.9872
        thinned_from=423
        keys [[0,0.1115],[1,0.1115],[2,0.1179],[3,0.1236],[4,0.129],[5,0.1324],[6,0.1324],[7,0.1275],[8,0.1178],[9,0.1065],[10,0.0956],[11,0.0881],[12,0.078],[13,0.0655],[14,0.063],[15,0.0831],[16,0.1346],[17,0.2211],[18,0.3473],[19,0.534],[20,0.7621],[21,0.9782],[22,1.1887],[23,1.404],[24,1.5815],[25,1.7119],[26,1.815],[27,1.8945],[28,1.9525],[29,1.9914],[30,2.016],[31,2.0294],[32,2.0352],[34,2.0337],[36,2.0144],[42,1.9125],[44,1.8861],[49,1.8447],[52,1.8311],[55,1.828],[56,1.8341],[57,1.8476],[58,1.867],[61,1.9404],[63,1.9758],[64,1.9884],[65,1.9926],[66,1.9618],[67,1.8873],[68,1.7895],[69,1.6758],[70,1.5363],[71,1.3611],[72,1.1497],[73,0.9149],[74,0.663],[75,0.3965],[76,0.1175],[77,-0.1787],[78,-0.4778],[79,-0.7492],[80,-1.0068],[81,-1.2558],[82,-1.4509],[83,-1.5856],[84,-1.687],[85,-1.7609],[86,-1.8128],[87,-1.8475],[88,-1.8692],[89,-1.8803],[91,-1.8836],[92,-1.873],[93,-1.8549],[94,-1.8298],[97,-1.7348],[98,-1.6992],[102,-1.5315],[103,-1.4965],[104,-1.4697],[106,-1.4314],[108,-1.4029],[111,-1.3724],[116,-1.3342],[121,-1.3151],[122,-1.306],[123,-1.2852],[124,-1.2539],[125,-1.2134],[126,-1.1904],[127,-1.2212],[128,-1.5088],[129,-2.0695],[130,-2.705],[131,-3.6242],[132,-4.5227],[133,-5.3368],[134,-6.0728],[135,-6.6934],[136,-6.9386],[137,-6.2299],[138,-4.6392],[139,-3.9842],[140,-3.1118],[141,-1.9225],[142,-0.7542],[143,0.1698],[144,0.7804],[145,0.9938],[146,1.1121],[147,0.9447],[148,0.6842],[149,0.212],[150,-0.2446],[151,-0.4996],[152,-0.5559],[153,-0.367],[154,-0.0849],[155,0.1913],[156,0.5412],[157,0.7065],[158,0.8845],[159,0.868],[160,0.8554],[161,0.8512],[162,0.853],[164,0.8625],[168,0.8889],[174,0.9431],[180,0.9708],[181,0.9708],[182,0.9767],[183,0.9767],[184,0.9807],[189,0.9822],[192,0.9813],[193,0.9788],[195,0.9788],[196,0.975],[199,0.975],[200,0.9688],[204,0.9688],[205,0.9586],[209,0.9586],[210,0.9467],[213,0.9467],[214,0.9361],[215,0.9361],[216,0.9108],[217,0.8929],[218,0.8721],[219,0.8487],[220,0.8231],[222,0.7666],[223,0.7251],[224,0.6641],[225,0.5878],[226,0.5214],[227,0.4749],[228,0.4236],[229,0.4022],[230,0.4068],[231,0.433],[232,0.4684],[233,0.48],[234,0.4341],[235,0.3314],[236,0.2018],[237,0.0398],[238,-0.1524],[239,-0.3483],[240,-0.5611],[241,-1],[242,-1.3232],[243,-1.6263],[244,-1.6715],[245,-1.2666],[246,-0.87],[247,-0.4844],[248,0.0141],[249,0.581],[250,1.3919],[251,2.089],[252,2.4791],[253,2.5304],[254,2.4527],[255,2.3452],[256,2.2673],[257,2.2156],[258,2.2075],[259,2.2225],[262,2.2867],[263,2.3044],[265,2.3268],[269,2.3452],[272,2.3471],[273,2.3421],[283,2.2166],[285,2.1988],[286,2.1934],[289,2.1934],[291,2.2125],[292,2.2276],[293,2.248],[294,2.2874],[295,2.3422],[296,2.3861],[297,2.3932],[298,2.3496],[299,2.266],[300,2.1539],[301,2.0151],[302,1.7889],[303,1.4764],[304,1.1067],[305,0.6955],[306,0.3085],[307,-0.0309],[308,-0.3422],[309,-0.6268],[310,-0.88],[311,-1.0876],[312,-1.2545],[313,-1.3883],[314,-1.4865],[315,-1.5558],[316,-1.5816],[317,-1.5576],[318,-1.5151],[321,-1.3624],[322,-1.3172],[323,-1.2856],[324,-1.2652],[325,-1.2571],[326,-1.2604],[327,-1.271],[328,-1.2863],[330,-1.326],[337,-1.482],[339,-1.5197],[343,-1.5788],[344,-1.6001],[345,-1.6251],[347,-1.6913],[348,-1.7317],[349,-1.7917],[350,-1.8784],[351,-2.2002],[352,-2.668],[353,-3.3118],[354,-4.1751],[355,-4.9306],[356,-5.5579],[357,-6.0828],[358,-6.5438],[359,-6.9801],[360,-7.3186],[361,-7.5985],[362,-7.8239],[363,-7.9901],[365,-8.2106],[368,-8.4536],[371,-8.6097],[372,-8.6402],[373,-8.6325],[374,-8.581],[375,-8.548],[376,-8.4939],[377,-8.3741],[378,-8.1604],[379,-7.8604],[380,-7.4056],[381,-6.7465],[382,-6.0036],[383,-5.1871],[384,-4.2927],[385,-3.2994],[386,-2.2833],[387,-1.4211],[388,-0.7969],[389,-0.3418],[390,0.0962],[391,0.2745],[392,0.4155],[393,0.523],[394,0.6042],[395,0.6629],[396,0.7026],[397,0.7272],[398,0.7405],[399,0.7466],[400,0.7494],[401,0.7365],[402,0.6985],[403,0.6426],[404,0.5762],[405,0.5071],[406,0.4425],[407,0.387],[408,0.3385],[409,0.2962],[410,0.259],[411,0.226],[412,0.1965],[413,0.1689],[414,0.1462],[415,0.1304],[416,0.1202],[417,0.1146],[418,0.1121],[422,0.1121]]
       }
       thigh_l_fk_ctrl.Scale.X {type=MovieSceneScriptingFloatChannel, default=1, const [1,423]}
       thigh_l_fk_ctrl.Scale.Y {type=MovieSceneScriptingFloatChannel, default=1, const [1,423]}
       thigh_l_fk_ctrl.Scale.Z {type=MovieSceneScriptingFloatChannel, default=1, const [1,423]}
       calf_l_fk_ctrl.Location.X {
        type=MovieSceneScriptingFloatChannel
        default=-1.7066
        thinned_from=423
        keys [[0,-0.2679],[1,-0.2687],[2,-0.2707],[6,-0.2827],[10,-0.2922],[11,-0.2942],[12,-0.2956],[13,-0.296],[14,-0.2954],[15,-0.2938],[16,-0.2902],[17,-0.2835],[18,-0.2734],[19,-0.2603],[20,-0.2444],[21,-0.2247],[22,-0.2032],[23,-0.1809],[24,-0.1574],[25,-0.1345],[26,-0.1147],[27,-0.0981],[28,-0.0833],[29,-0.0711],[30,-0.0622],[31,-0.056],[32,-0.0516],[33,-0.0486],[34,-0.0467],[35,-0.0455],[37,-0.0438],[38,-0.0431],[40,-0.0424],[42,-0.0422],[43,-0.0426],[45,-0.0439],[47,-0.0455],[50,-0.0487],[55,-0.0552],[56,-0.0567],[57,-0.0584],[58,-0.0604],[60,-0.0651],[61,-0.0672],[62,-0.0688],[63,-0.0701],[64,-0.0716],[65,-0.0735],[66,-0.0732],[67,-0.0697],[68,-0.065],[69,-0.0608],[70,-0.0573],[71,-0.054],[72,-0.0502],[73,-0.0458],[74,-0.04],[75,-0.034],[76,-0.0298],[77,-0.0283],[78,-0.029],[79,-0.0312],[81,-0.0366],[82,-0.0395],[83,-0.0417],[84,-0.0426],[85,-0.0425],[87,-0.0409],[88,-0.0404],[90,-0.04],[91,-0.0396],[92,-0.0387],[93,-0.0373],[94,-0.0353],[95,-0.0329],[96,-0.0301],[99,-0.0208],[100,-0.0147],[101,-0.0091],[102,-0.0042],[104,0.0036],[105,0.0069],[106,0.0101],[108,0.0153],[109,0.017],[119,0.017],[120,0.0163],[123,0.0075],[124,0.0048],[125,0.0024],[126,0.0004],[127,-0.0011],[128,-0.0047],[129,-0.0087],[130,-0.0171],[131,-0.0395],[132,-0.0184],[133,0.0032],[134,0.044],[135,0.0757],[136,0.0496],[137,-0.0295],[138,-0.1624],[139,-0.3843],[140,-0.6919],[141,-0.8875],[142,-0.9664],[143,-1.0818],[144,-1.1248],[145,-1.13],[146,-1.1188],[147,-1.0884],[148,-0.9728],[149,-0.877],[150,-0.8069],[151,-0.764],[152,-0.7533],[153,-0.7989],[154,-0.8845],[155,-0.9618],[156,-1.0231],[157,-1.0602],[158,-1.0768],[159,-1.0838],[160,-1.0853],[161,-1.0838],[162,-1.0798],[163,-1.0724],[168,-1.0183],[171,-0.9928],[174,-0.9753],[179,-0.9602],[218,-0.9636],[225,-0.9531],[226,-0.9572],[227,-0.9757],[229,-1.0212],[231,-1.0484],[232,-1.0746],[233,-1.0929],[234,-1.1031],[235,-1.1038],[236,-1.0972],[237,-1.0837],[238,-1.0609],[239,-1.025],[240,-0.9752],[241,-0.9135],[242,-0.8464],[243,-0.7764],[244,-0.6978],[245,-0.613],[246,-0.533],[247,-0.4624],[248,-0.3942],[249,-0.3503],[250,-0.3407],[251,-0.3555],[252,-0.3907],[253,-0.4339],[254,-0.4718],[255,-0.4961],[256,-0.5022],[257,-0.4984],[258,-0.4964],[261,-0.495],[267,-0.4992],[272,-0.5047],[274,-0.5041],[287,-0.4938],[293,-0.4932],[296,-0.4966],[297,-0.4961],[298,-0.4939],[299,-0.4901],[300,-0.4839],[301,-0.4699],[302,-0.4446],[303,-0.4129],[304,-0.3748],[305,-0.3287],[306,-0.2839],[307,-0.2413],[308,-0.2004],[309,-0.1735],[311,-0.1495],[312,-0.135],[313,-0.1245],[314,-0.1167],[315,-0.108],[316,-0.0984],[317,-0.0891],[318,-0.0808],[319,-0.0742],[320,-0.0703],[322,-0.0659],[325,-0.0589],[326,-0.0563],[327,-0.053],[328,-0.0491],[331,-0.0366],[332,-0.0326],[333,-0.029],[334,-0.0258],[335,-0.0237],[336,-0.0228],[337,-0.0228],[338,-0.0233],[339,-0.0244],[340,-0.0259],[341,-0.0279],[342,-0.0307],[344,-0.0367],[345,-0.0391],[346,-0.0405],[347,-0.0412],[348,-0.0414],[349,-0.0402],[350,-0.0371],[351,-0.0298],[352,-0.0187],[353,-0.004],[354,0.0099],[355,0.0206],[356,0.0285],[357,0.0353],[359,0.0479],[360,0.0569],[361,0.0664],[362,0.0752],[364,0.091],[366,0.1057],[367,0.1124],[368,0.1183],[369,0.1236],[370,0.1283],[371,0.1326],[372,0.1364],[373,0.1398],[374,0.1422],[375,0.1397],[376,0.1309],[377,0.1168],[378,0.1001],[379,0.0809],[380,0.0577],[381,0.03],[382,-0.0034],[383,-0.0386],[384,-0.0767],[385,-0.1187],[386,-0.161],[387,-0.2001],[388,-0.233],[389,-0.2594],[390,-0.2813],[391,-0.2975],[392,-0.3105],[393,-0.3197],[394,-0.3256],[395,-0.3286],[396,-0.3292],[397,-0.3281],[400,-0.3205],[401,-0.3173],[403,-0.3086],[406,-0.2943],[409,-0.2829],[411,-0.2768],[413,-0.2721],[415,-0.2693],[418,-0.268],[422,-0.268]]
       }
       calf_l_fk_ctrl.Location.Y {
        type=MovieSceneScriptingFloatChannel
        default=0.1276
        thinned_from=423
        keys [[0,0.0215],[12,0.0238],[15,0.0236],[16,0.0233],[17,0.0228],[18,0.022],[19,0.0209],[20,0.0196],[23,0.0145],[25,0.0108],[26,0.0092],[28,0.0067],[29,0.0057],[30,0.005],[32,0.0041],[43,0.0034],[65,0.0059],[66,0.0059],[73,0.0037],[76,0.0024],[78,0.0023],[83,0.0033],[93,0.003],[99,0.0016],[104,-0.0003],[109,-0.0014],[120,-0.0013],[129,0.0007],[130,0.0013],[131,0.0031],[133,-0.0005],[134,-0.0037],[135,-0.0062],[136,-0.0039],[137,0.0025],[138,0.0131],[139,0.0309],[140,0.0556],[141,0.0713],[142,0.0777],[143,0.087],[144,0.0906],[145,0.091],[146,0.0901],[147,0.0876],[148,0.0782],[149,0.0704],[150,0.0648],[151,0.0613],[152,0.0605],[153,0.0641],[154,0.071],[155,0.0773],[156,0.0823],[157,0.0853],[158,0.0867],[159,0.0873],[160,0.0874],[161,0.0873],[162,0.0869],[164,0.0856],[168,0.082],[171,0.0799],[174,0.0785],[180,0.0772],[219,0.0775],[225,0.0767],[226,0.0771],[227,0.0786],[229,0.0822],[231,0.0844],[232,0.0865],[233,0.088],[234,0.0888],[235,0.0889],[236,0.0884],[237,0.0873],[238,0.0854],[239,0.0825],[240,0.0785],[241,0.0735],[242,0.0681],[243,0.0624],[244,0.0561],[245,0.0493],[246,0.0428],[247,0.0371],[248,0.0317],[249,0.0282],[250,0.0274],[251,0.0286],[252,0.0314],[253,0.0349],[254,0.038],[255,0.0399],[256,0.0404],[258,0.0399],[274,0.0406],[293,0.0397],[297,0.0399],[299,0.0394],[300,0.0389],[301,0.0378],[302,0.0358],[303,0.0332],[304,0.0302],[306,0.0228],[308,0.0161],[309,0.0139],[311,0.012],[312,0.0108],[313,0.01],[318,0.0065],[319,0.0059],[335,0.0019],[340,0.002],[346,0.0032],[349,0.0032],[350,0.003],[351,0.0023],[352,0.0013],[354,-0.001],[355,-0.0018],[359,-0.0035],[367,-0.0081],[373,-0.0101],[374,-0.0103],[375,-0.0101],[376,-0.0094],[378,-0.007],[379,-0.0056],[380,-0.004],[381,-0.002],[382,0.0004],[383,0.003],[384,0.006],[386,0.0127],[387,0.0159],[388,0.0186],[389,0.0208],[390,0.0226],[391,0.0239],[392,0.025],[393,0.0257],[394,0.0262],[396,0.0265],[401,0.0255],[411,0.0223],[422,0.0216]]
       }
       calf_l_fk_ctrl.Location.Z {type=MovieSceneScriptingFloatChannel, default=-0.0366, thinned_from=423, keys [[0,0.0002],[138,0.0001],[147,0.0007],[422,0.0002]]}
       calf_l_fk_ctrl.Rotation.X {
        type=MovieSceneScriptingFloatChannel
        default=-0.0451
        thinned_from=423
        keys [[0,0.0033],[130,0.0033],[133,0.002],[135,0.0003],[136,-0.0002],[137,-0.0004],[139,0.0001],[140,0.0008],[141,0.0017],[142,0.0024],[144,0.0033],[146,0.0033],[148,0.0026],[150,0.0018],[152,0.0017],[158,0.0033],[240,0.0032],[241,0.003],[247,0.003],[248,0.0032],[350,0.0033],[352,0.0026],[353,0.002],[355,0.0005],[359,-0.0012],[376,-0.0028],[378,-0.0026],[380,-0.0019],[384,0.0009],[387,0.0025],[390,0.0033],[422,0.0033]]
       }
       calf_l_fk_ctrl.Rotation.Y {
        type=MovieSceneScriptingFloatChannel
        default=1.2242
        thinned_from=423
        keys [[0,0.0047],[130,0.0047],[132,0.0035],[133,0.0027],[135,0.0005],[136,-0.0003],[137,-0.0006],[139,0.0001],[140,0.001],[141,0.0023],[142,0.0034],[144,0.0047],[146,0.0047],[148,0.0037],[149,0.0029],[150,0.0024],[151,0.0022],[152,0.0023],[155,0.0034],[156,0.0039],[158,0.0047],[240,0.0046],[241,0.0042],[246,0.0041],[247,0.0042],[248,0.0046],[350,0.0046],[351,0.0042],[352,0.0036],[353,0.0028],[354,0.0016],[355,0.0007],[359,-0.0015],[363,-0.0024],[376,-0.0034],[378,-0.0032],[379,-0.0029],[380,-0.0023],[386,0.0028],[387,0.0035],[390,0.0046],[422,0.0047]]
       }
       calf_l_fk_ctrl.Rotation.Z {
        type=MovieSceneScriptingFloatChannel
        default=-4.277
        thinned_from=423
        keys [[0,-4.9605],[130,-4.9607],[131,-4.453],[132,-3.7435],[133,-2.9209],[134,-1.7393],[135,-0.4942],[136,0.3285],[137,0.6252],[138,0.2636],[139,-0.0953],[140,-1.1274],[141,-2.5117],[142,-3.6168],[143,-4.2788],[144,-4.9607],[146,-4.9607],[147,-4.4964],[148,-3.9128],[149,-3.1503],[150,-2.6157],[151,-2.4055],[152,-2.5005],[153,-2.8186],[154,-3.1893],[155,-3.6009],[156,-4.1811],[157,-4.4905],[158,-4.956],[238,-4.9576],[240,-4.9347],[241,-4.4962],[242,-4.4027],[243,-4.3272],[244,-4.273],[245,-4.3006],[246,-4.3865],[247,-4.4981],[248,-4.9362],[250,-4.9607],[349,-4.9569],[350,-4.9473],[351,-4.519],[352,-3.8774],[353,-2.9907],[354,-1.7485],[355,-0.7555],[356,-0.0038],[357,0.5882],[358,1.1239],[359,1.6659],[360,2.0424],[361,2.3403],[362,2.5755],[363,2.746],[364,2.8501],[365,2.9764],[367,3.1922],[368,3.3071],[369,3.4097],[372,3.6659],[373,3.7137],[374,3.7157],[375,3.8021],[376,3.8777],[377,3.8338],[378,3.6141],[379,3.2418],[380,2.6088],[381,1.6463],[382,0.6244],[383,-0.4076],[384,-1.3133],[385,-2.1588],[386,-3.0162],[387,-3.7424],[388,-4.2205],[389,-4.5109],[390,-4.9455],[392,-4.9607],[422,-4.9607]]
       }
       calf_l_fk_ctrl.Scale.X {type=MovieSceneScriptingFloatChannel, default=1, const [1,423]}
       calf_l_fk_ctrl.Scale.Y {type=MovieSceneScriptingFloatChannel, default=1, const [1,423]}
       calf_l_fk_ctrl.Scale.Z {type=MovieSceneScriptingFloatChannel, default=1, const [1,423]}
       foot_l_fk_ctrl.Location.X {
        type=MovieSceneScriptingFloatChannel
        default=-1.8654
        thinned_from=423
        keys [[0,-0.2719],[1,-0.2728],[2,-0.2749],[6,-0.288],[10,-0.2984],[11,-0.3006],[12,-0.3021],[13,-0.3026],[14,-0.302],[15,-0.3002],[16,-0.2963],[17,-0.289],[18,-0.2779],[19,-0.2635],[20,-0.2461],[21,-0.2247],[22,-0.2011],[23,-0.1767],[24,-0.1511],[25,-0.126],[26,-0.1044],[27,-0.0862],[28,-0.07],[29,-0.0567],[30,-0.0469],[31,-0.0402],[32,-0.0354],[33,-0.0321],[34,-0.03],[35,-0.0287],[37,-0.0268],[38,-0.0261],[40,-0.0253],[42,-0.0251],[43,-0.0255],[45,-0.0269],[47,-0.0287],[50,-0.0322],[55,-0.0393],[57,-0.0427],[58,-0.045],[60,-0.0501],[61,-0.0524],[62,-0.0542],[63,-0.0556],[64,-0.0572],[65,-0.0593],[66,-0.059],[67,-0.0551],[68,-0.05],[69,-0.0454],[70,-0.0416],[71,-0.0379],[72,-0.0339],[73,-0.029],[74,-0.0227],[75,-0.0161],[76,-0.0115],[77,-0.0099],[78,-0.0106],[79,-0.013],[81,-0.019],[82,-0.0221],[83,-0.0245],[84,-0.0255],[85,-0.0254],[87,-0.0236],[88,-0.0232],[90,-0.0227],[91,-0.0222],[92,-0.0213],[93,-0.0197],[94,-0.0175],[95,-0.0149],[96,-0.0118],[99,-0.0016],[100,0.005],[101,0.0112],[102,0.0165],[104,0.025],[105,0.0287],[106,0.0321],[107,0.035],[108,0.0378],[109,0.0396],[119,0.0396],[120,0.0389],[123,0.0293],[124,0.0263],[125,0.0237],[126,0.0215],[127,0.0198],[128,0.0159],[129,0.0116],[130,0.0024],[131,-0.0243],[133,0.0159],[134,0.0555],[135,0.0849],[136,0.0529],[137,-0.0349],[138,-0.1788],[139,-0.4198],[140,-0.7518],[141,-0.9598],[142,-1.0414],[143,-1.1648],[144,-1.2089],[145,-1.2146],[146,-1.2024],[147,-1.171],[148,-1.0472],[149,-0.9456],[150,-0.8712],[151,-0.8252],[152,-0.8131],[153,-0.8616],[154,-0.9536],[155,-1.0365],[156,-1.101],[157,-1.1403],[158,-1.1565],[159,-1.1641],[160,-1.1657],[161,-1.1641],[162,-1.1597],[163,-1.1516],[168,-1.0925],[171,-1.0646],[174,-1.0455],[178,-1.0312],[218,-1.0327],[225,-1.0212],[226,-1.0257],[227,-1.0459],[229,-1.0957],[231,-1.1254],[232,-1.1541],[233,-1.174],[234,-1.1852],[235,-1.186],[236,-1.1788],[237,-1.164],[238,-1.139],[239,-1.0999],[240,-1.0455],[241,-0.9798],[242,-0.9069],[243,-0.8306],[244,-0.7449],[245,-0.6521],[246,-0.5642],[247,-0.4865],[248,-0.4101],[249,-0.3621],[250,-0.3515],[251,-0.3677],[252,-0.4062],[253,-0.4534],[254,-0.4949],[255,-0.5215],[256,-0.5281],[257,-0.5239],[258,-0.5217],[261,-0.5202],[267,-0.5248],[272,-0.5308],[274,-0.5301],[286,-0.5192],[293,-0.5183],[296,-0.522],[297,-0.5214],[298,-0.519],[299,-0.5149],[300,-0.5081],[301,-0.4928],[302,-0.4651],[303,-0.4304],[304,-0.3887],[305,-0.3383],[306,-0.2893],[307,-0.2428],[308,-0.1981],[309,-0.1686],[311,-0.1424],[312,-0.1265],[313,-0.1151],[314,-0.1065],[315,-0.0971],[316,-0.0865],[317,-0.0764],[318,-0.0673],[319,-0.0601],[320,-0.0558],[322,-0.051],[325,-0.0433],[326,-0.0405],[327,-0.0369],[328,-0.0326],[331,-0.019],[332,-0.0146],[333,-0.0106],[334,-0.0071],[335,-0.0048],[336,-0.0038],[337,-0.0038],[338,-0.0044],[339,-0.0056],[340,-0.0072],[341,-0.0095],[342,-0.0126],[344,-0.0191],[345,-0.0217],[346,-0.0232],[347,-0.024],[348,-0.0242],[349,-0.0229],[350,-0.0196],[351,-0.0134],[352,-0.004],[353,0.0083],[354,0.0183],[355,0.0257],[356,0.0312],[357,0.0361],[359,0.0453],[360,0.0535],[361,0.0626],[362,0.0713],[364,0.0874],[366,0.1024],[367,0.1093],[368,0.1152],[369,0.1206],[370,0.1255],[371,0.1297],[373,0.1371],[374,0.1396],[375,0.1365],[376,0.1267],[377,0.1114],[378,0.0941],[379,0.0746],[380,0.0519],[381,0.0258],[382,-0.0064],[383,-0.0405],[384,-0.0783],[385,-0.1206],[386,-0.1633],[387,-0.2029],[388,-0.2369],[389,-0.2645],[390,-0.2866],[391,-0.3043],[392,-0.3185],[393,-0.3286],[394,-0.335],[395,-0.3382],[396,-0.3389],[397,-0.3377],[400,-0.3294],[401,-0.3259],[403,-0.3163],[406,-0.3008],[408,-0.2921],[410,-0.2847],[412,-0.2788],[414,-0.2747],[416,-0.2726],[422,-0.272]]
       }
       foot_l_fk_ctrl.Location.Y {
        type=MovieSceneScriptingFloatChannel
        default=0.0398
        thinned_from=423
        keys [[0,-0.0017],[20,-0.0015],[130,0],[131,-0.0002],[135,0.0005],[136,0.0003],[137,-0.0002],[138,-0.0011],[139,-0.0026],[140,-0.0047],[141,-0.0059],[142,-0.0065],[143,-0.0072],[144,-0.0075],[146,-0.0075],[147,-0.0073],[149,-0.0059],[151,-0.0051],[152,-0.005],[153,-0.0053],[156,-0.0068],[226,-0.0064],[234,-0.0073],[238,-0.0071],[243,-0.0051],[248,-0.0025],[249,-0.0022],[251,-0.0023],[255,-0.0032],[301,-0.0031],[309,-0.001],[376,0.0008],[392,-0.002],[422,-0.0017]]
       }
       foot_l_fk_ctrl.Location.Z {type=MovieSceneScriptingFloatChannel, default=-0.8885, thinned_from=423, keys [[0,0.0002],[137,0],[146,0.0008],[422,0.0002]]}
       foot_l_fk_ctrl.Rotation.X {
        type=MovieSceneScriptingFloatChannel
        default=-1.5204
        thinned_from=423
        keys [[0,-0.3343],[12,-0.3403],[15,-0.3426],[16,-0.3423],[17,-0.3402],[18,-0.3357],[19,-0.328],[21,-0.3085],[22,-0.301],[23,-0.2956],[24,-0.2913],[25,-0.2879],[27,-0.2832],[29,-0.2806],[41,-0.2742],[42,-0.2742],[52,-0.2725],[56,-0.2735],[58,-0.2752],[65,-0.2859],[66,-0.2858],[67,-0.2841],[69,-0.2787],[70,-0.2748],[71,-0.27],[72,-0.2643],[77,-0.2302],[78,-0.2237],[79,-0.2181],[81,-0.2081],[82,-0.2041],[83,-0.2009],[84,-0.1983],[85,-0.1963],[86,-0.1949],[88,-0.1934],[91,-0.193],[94,-0.1942],[103,-0.2011],[109,-0.2032],[113,-0.2032],[114,-0.204],[125,-0.2044],[126,-0.2048],[128,-0.2069],[129,-0.2085],[130,-0.2108],[131,-0.2137],[132,-0.2181],[133,-0.2244],[134,-0.2375],[135,-0.2552],[136,-0.2775],[137,-0.3067],[138,-0.3471],[139,-0.3908],[140,-0.4241],[141,-0.4532],[142,-0.4808],[143,-0.5009],[144,-0.5143],[145,-0.524],[146,-0.5308],[147,-0.5352],[148,-0.5408],[149,-0.5429],[156,-0.553],[227,-0.554],[229,-0.5557],[231,-0.5601],[233,-0.568],[234,-0.5709],[236,-0.5726],[238,-0.5715],[240,-0.5667],[242,-0.5581],[243,-0.5524],[244,-0.5436],[245,-0.5328],[250,-0.4747],[251,-0.4648],[252,-0.4573],[253,-0.4519],[254,-0.4478],[255,-0.445],[256,-0.4436],[261,-0.4437],[273,-0.4448],[290,-0.4386],[294,-0.44],[300,-0.4481],[301,-0.4482],[302,-0.4458],[303,-0.441],[304,-0.4331],[305,-0.4216],[306,-0.408],[309,-0.364],[310,-0.3504],[311,-0.3386],[312,-0.328],[313,-0.3185],[314,-0.3106],[315,-0.3044],[316,-0.2993],[317,-0.2953],[319,-0.2895],[321,-0.2857],[327,-0.2812],[334,-0.2811],[336,-0.2822],[344,-0.2919],[345,-0.2927],[347,-0.2923],[348,-0.2914],[349,-0.2897],[350,-0.2868],[351,-0.2821],[352,-0.2748],[353,-0.2644],[354,-0.2516],[356,-0.2233],[357,-0.2096],[358,-0.1974],[360,-0.1743],[361,-0.1633],[362,-0.1539],[363,-0.1468],[364,-0.1417],[365,-0.1376],[366,-0.1345],[367,-0.1322],[368,-0.1306],[369,-0.1298],[370,-0.1296],[371,-0.13],[372,-0.131],[373,-0.1328],[374,-0.1354],[375,-0.139],[376,-0.1433],[377,-0.1486],[378,-0.155],[379,-0.1628],[380,-0.1726],[381,-0.1844],[382,-0.1975],[383,-0.2111],[384,-0.2261],[386,-0.2608],[387,-0.277],[388,-0.2912],[389,-0.3041],[390,-0.3153],[391,-0.3243],[392,-0.3318],[393,-0.3379],[394,-0.3427],[395,-0.3463],[396,-0.3487],[397,-0.3502],[400,-0.3513],[402,-0.3501],[411,-0.3378],[415,-0.3349],[422,-0.3343]]
       }
       foot_l_fk_ctrl.Rotation.Y {
        type=MovieSceneScriptingFloatChannel
        default=-5.8188
        thinned_from=423
        keys [[0,-6.3705],[15,-6.4855],[16,-6.4818],[17,-6.4534],[18,-6.3926],[19,-6.2876],[21,-6.0201],[22,-5.9168],[23,-5.8419],[25,-5.736],[27,-5.6708],[29,-5.6353],[41,-5.5475],[55,-5.5316],[58,-5.5602],[65,-5.7074],[66,-5.7066],[68,-5.6501],[69,-5.609],[70,-5.5567],[71,-5.4909],[73,-5.3238],[78,-4.8515],[80,-4.7005],[81,-4.6323],[82,-4.5755],[84,-4.4958],[86,-4.4487],[89,-4.4234],[93,-4.4309],[105,-4.5489],[113,-4.5638],[114,-4.5755],[126,-4.5862],[128,-4.6143],[130,-4.6671],[131,-4.7101],[132,-4.7724],[133,-4.8629],[134,-5.0497],[135,-5.3007],[136,-5.6132],[137,-6.0184],[138,-6.5759],[139,-7.1791],[140,-7.6378],[141,-8.0355],[142,-8.4152],[143,-8.6939],[144,-8.879],[145,-9.0153],[146,-9.1106],[148,-9.2556],[149,-9.2884],[156,-9.4243],[228,-9.4439],[230,-9.4857],[231,-9.5217],[233,-9.6323],[234,-9.6727],[237,-9.6946],[239,-9.653],[241,-9.5599],[243,-9.4157],[244,-9.2925],[250,-8.3263],[251,-8.1873],[252,-8.0828],[253,-8.0086],[254,-7.9508],[255,-7.9121],[257,-7.8848],[273,-7.9094],[292,-7.8256],[300,-7.9545],[301,-7.9552],[302,-7.9206],[303,-7.8525],[304,-7.7409],[305,-7.5798],[309,-6.775],[310,-6.5866],[311,-6.4233],[312,-6.2759],[313,-6.1441],[314,-6.0362],[315,-5.9499],[316,-5.8799],[318,-5.7829],[320,-5.7183],[324,-5.6568],[334,-5.6333],[341,-5.7251],[344,-5.7812],[346,-5.7922],[348,-5.773],[349,-5.7482],[350,-5.7074],[351,-5.645],[352,-5.5463],[353,-5.4051],[354,-5.2337],[357,-4.6631],[358,-4.4974],[360,-4.1849],[361,-4.0369],[362,-3.9094],[363,-3.8135],[364,-3.7446],[365,-3.6896],[366,-3.647],[367,-3.616],[368,-3.5954],[369,-3.5845],[370,-3.5821],[371,-3.5879],[372,-3.6028],[373,-3.6281],[374,-3.6652],[375,-3.7151],[376,-3.7765],[377,-3.8494],[378,-3.9371],[379,-4.0431],[380,-4.1757],[381,-4.3341],[383,-4.6914],[384,-4.8956],[386,-5.367],[387,-5.5873],[388,-5.7822],[389,-5.9586],[390,-6.1104],[391,-6.2355],[392,-6.3385],[393,-6.4231],[394,-6.4898],[395,-6.5392],[397,-6.593],[400,-6.6082],[403,-6.5726],[413,-6.3944],[422,-6.3708]]
       }
       foot_l_fk_ctrl.Rotation.Z {
        type=MovieSceneScriptingFloatChannel
        default=4.2926
        thinned_from=423
        keys [[0,4.511],[11,4.6197],[13,4.6216],[14,4.6061],[15,4.5692],[16,4.4966],[17,4.3841],[18,4.2396],[19,4.0451],[20,3.8169],[21,3.6142],[22,3.4486],[23,3.3005],[24,3.1816],[25,3.0934],[26,3.0284],[27,2.982],[28,2.9524],[29,2.9357],[31,2.9255],[41,2.9709],[42,2.9709],[43,2.9827],[56,3.016],[64,2.994],[65,2.9965],[66,3.0215],[67,3.0774],[68,3.1487],[69,3.229],[70,3.3214],[71,3.4288],[72,3.5507],[73,3.6853],[74,3.8318],[76,4.1518],[78,4.5192],[79,4.6844],[81,4.9789],[82,5.0851],[83,5.1487],[84,5.1893],[85,5.213],[87,5.2329],[92,5.2139],[99,5.1182],[109,5.1925],[115,5.1928],[120,5.2674],[126,5.2656],[127,5.2492],[130,5.174],[131,4.8341],[132,4.5098],[133,4.1025],[134,3.3852],[135,2.5671],[136,1.9186],[137,1.3746],[138,0.9343],[139,0.7475],[140,1.1457],[141,1.7159],[142,2.1688],[143,2.4737],[144,2.9124],[145,3.0532],[146,3.2077],[147,3.1646],[148,3.06],[149,2.8726],[150,2.7754],[151,2.787],[152,2.9027],[153,3.0595],[154,3.2045],[155,3.3656],[156,3.598],[157,3.7181],[158,3.9278],[174,3.8359],[214,3.864],[221,3.9578],[224,3.9724],[225,3.9696],[226,3.9423],[227,3.8867],[228,3.807],[229,3.7076],[230,3.5935],[233,3.2203],[234,3.1155],[235,3.0272],[236,2.9521],[237,2.9141],[240,2.8728],[241,2.6644],[242,2.6241],[243,2.6015],[244,2.6418],[245,2.746],[247,3.0453],[248,3.3587],[249,3.4533],[250,3.5055],[253,3.5726],[255,3.602],[257,3.6071],[261,3.5938],[292,3.583],[295,3.5579],[296,3.5598],[297,3.5818],[298,3.6423],[299,3.7386],[301,3.9665],[302,4.1078],[305,4.5962],[306,4.7338],[307,4.8439],[308,4.9408],[309,5.0172],[310,5.0679],[312,5.1218],[314,5.1392],[315,5.1376],[316,5.1241],[322,4.9491],[325,4.8963],[337,4.8812],[342,4.9099],[344,4.9396],[347,5.0213],[350,5.1467],[351,5.0045],[352,4.774],[353,4.4472],[354,3.9712],[355,3.613],[356,3.3529],[357,3.1522],[358,2.9605],[359,2.7515],[360,2.6115],[361,2.5002],[362,2.4095],[363,2.3445],[364,2.3061],[365,2.2533],[367,2.1585],[368,2.1053],[369,2.0571],[370,2.0186],[371,1.9722],[373,1.8543],[374,1.8002],[375,1.6842],[376,1.5576],[377,1.4837],[378,1.4953],[379,1.5855],[380,1.8031],[381,2.174],[382,2.563],[383,2.943],[384,3.229],[385,3.4569],[386,3.6939],[387,3.9013],[388,4.0289],[389,4.0983],[390,4.2618],[391,4.2414],[393,4.2307],[396,4.2564],[402,4.3548],[407,4.4631],[414,4.5196],[422,4.5122]]
       }
       foot_l_fk_ctrl.Scale.X {type=MovieSceneScriptingFloatChannel, default=1, const [1,423]}
       foot_l_fk_ctrl.Scale.Y {type=MovieSceneScriptingFloatChannel, default=1, const [1,423]}
       foot_l_fk_ctrl.Scale.Z {type=MovieSceneScriptingFloatChannel, default=1, const [1,423]}
       ball_l_fk_ctrl.Location.X {type=MovieSceneScriptingFloatChannel, default=0, thinned_from=423, keys [[0,0],[422,0]]}
       ball_l_fk_ctrl.Location.Y {type=MovieSceneScriptingFloatChannel, default=0, thinned_from=423, keys [[0,0],[422,0]]}
       ball_l_fk_ctrl.Location.Z {type=MovieSceneScriptingFloatChannel, default=0, thinned_from=423, keys [[0,0],[422,0]]}
       ball_l_fk_ctrl.Rotation.X {
        type=MovieSceneScriptingFloatChannel
        default=0
        thinned_from=423
        keys [[0,0],[99,0],[100,-0.0069],[101,-0.0133],[102,-0.0189],[103,-0.0189],[104,-0.028],[105,-0.028],[106,-0.0371],[107,-0.0371],[108,-0.046],[112,-0.046],[113,-0.0517],[115,-0.0517],[116,-0.0576],[117,-0.0576],[118,-0.0664],[119,-0.0718],[127,-0.0718],[128,-0.059],[129,-0.0466],[130,-0.0345],[131,0],[422,0]]
       }
       ball_l_fk_ctrl.Rotation.Y {
        type=MovieSceneScriptingFloatChannel
        default=0
        thinned_from=423
        keys [[0,0],[99,0],[100,-0.0042],[101,-0.008],[102,-0.0115],[103,-0.0115],[104,-0.017],[105,-0.017],[106,-0.0225],[107,-0.0225],[108,-0.0279],[112,-0.0279],[113,-0.0314],[115,-0.0314],[116,-0.035],[117,-0.035],[118,-0.0404],[119,-0.0437],[127,-0.0437],[128,-0.0359],[130,-0.0209],[131,0],[422,0]]
       }
       ball_l_fk_ctrl.Rotation.Z {
        type=MovieSceneScriptingFloatChannel
        default=0
        thinned_from=423
        keys [[0,0],[99,0],[100,-0.0231],[101,-0.0446],[102,-0.0635],[103,-0.0635],[104,-0.0939],[105,-0.0939],[106,-0.1244],[107,-0.1244],[108,-0.1541],[112,-0.1541],[113,-0.173],[115,-0.173],[116,-0.193],[117,-0.193],[118,-0.2221],[119,-0.2404],[127,-0.2404],[128,-0.1972],[129,-0.1556],[130,-0.1149],[131,0],[422,0]]
       }
       ball_l_fk_ctrl.Scale.X {type=MovieSceneScriptingFloatChannel, default=1, const [1,423]}
       ball_l_fk_ctrl.Scale.Y {type=MovieSceneScriptingFloatChannel, default=1, const [1,423]}
       ball_l_fk_ctrl.Scale.Z {type=MovieSceneScriptingFloatChannel, default=1, const [1,423]}
       foot_l_ik_ctrl.Location.X {
        type=MovieSceneScriptingFloatChannel
        default=3.8166
        thinned_from=423
        keys [[0,3.8166],[5,3.8166],[6,3.807],[7,3.8166],[52,3.8166],[53,3.8309],[54,3.8166],[55,3.8257],[56,3.8166],[90,3.8166],[91,3.822],[130,3.8107],[180,3.8166],[181,3.8076],[182,3.8166],[183,3.8109],[184,3.8166],[288,3.8195],[422,3.8173]]
       }
       foot_l_ik_ctrl.Location.Y {
        type=MovieSceneScriptingFloatChannel
        default=0
        thinned_from=423
        keys [[0,0],[1,0.0016],[2,0],[5,0],[6,-0.0037],[7,0],[41,0],[42,0.0002],[52,0],[53,0.0038],[54,0],[55,0.0055],[56,0],[90,0],[91,0.0047],[92,0],[100,0.0001],[126,0.0015],[130,0.0006],[131,0],[180,0],[181,0.0009],[182,0],[183,0.001],[184,0],[185,0.0003],[187,-0.0005],[188,-0.0001],[189,-0.001],[190,0],[191,-0.0017],[192,-0.0029],[193,0],[195,-0.0039],[196,0],[197,-0.0023],[198,-0.005],[199,-0.0073],[200,0],[203,-0.0092],[204,-0.0126],[205,0],[207,-0.0072],[209,-0.0151],[210,0],[212,-0.0081],[213,-0.0114],[214,0],[215,-0.016],[216,0],[260,0],[261,-0.0002],[262,0],[272,0],[273,-0.0002],[286,0],[287,-0.0022],[288,-0.0013],[289,0.002],[290,0],[291,-0.0001],[336,0],[348,-0.0001],[418,0],[419,-0.0005],[422,-0.0005]]
       }
       foot_l_ik_ctrl.Location.Z {
        type=MovieSceneScriptingFloatChannel
        default=0
        thinned_from=423
        keys [[0,0],[1,-0.0006],[2,0],[5,0],[6,-0.0018],[7,0],[52,0],[53,0.0024],[54,0],[55,0.0014],[56,0],[90,0],[91,0.0008],[92,0],[99,0],[100,0.0067],[101,0.013],[102,0.0185],[103,0.0228],[104,0.0274],[105,0.0318],[106,0.0364],[108,0.0451],[109,0.0484],[110,0.0485],[111,0.049],[113,0.0507],[114,0.052],[115,0.0538],[116,0.0566],[117,0.0604],[118,0.0651],[119,0.0705],[120,0.0749],[126,0.0749],[127,0.0694],[128,0.0578],[129,0.0456],[130,0.0336],[131,0.0001],[132,0.0004],[133,0],[134,0.0002],[135,0],[136,0.0002],[137,0],[138,0.0002],[141,0],[142,0.0003],[143,0.0001],[147,0.0002],[154,0],[155,0.0003],[156,0],[157,0.0002],[158,-0.0001],[180,-0.0001],[181,-0.0024],[182,0],[183,-0.0015],[184,0],[185,-0.0008],[186,0],[187,-0.0004],[188,-0.0001],[189,-0.0002],[190,0],[192,-0.0002],[193,0],[199,0.0003],[200,0],[204,0.0007],[205,0],[209,0.0009],[210,0],[212,0.0006],[213,0.0011],[214,0.0001],[215,0.0015],[216,0],[238,0],[240,-0.0004],[241,0.0002],[246,0.0001],[247,0.0002],[248,-0.0004],[250,0],[260,0],[261,0.0001],[272,0],[273,0.0001],[286,0],[287,0.0008],[288,0.0007],[289,-0.0004],[290,0],[349,-0.0001],[350,-0.0002],[351,0.0002],[353,0],[354,0.0003],[355,0],[359,0],[361,0.0004],[362,0],[381,0.0001],[386,0],[387,0.0004],[388,0],[389,0.0002],[390,-0.0002],[422,0.0001]]
       }
       foot_l_ik_ctrl.Rotation.X {
        type=MovieSceneScriptingFloatChannel
        default=0
        thinned_from=423
        keys [[0,0],[1,0.0008],[2,0],[5,0],[6,0.0048],[7,0],[41,0],[42,-0.0064],[43,0],[52,0],[53,-0.0056],[54,0],[55,-0.0051],[56,0],[90,0],[91,-0.0042],[92,0],[99,0],[100,0.0232],[101,0.0447],[102,0.0637],[103,0.0781],[104,0.0942],[105,0.1091],[106,0.1248],[107,0.1394],[108,0.1547],[109,0.166],[110,0.1691],[111,0.1718],[114,0.1782],[115,0.1801],[116,0.194],[117,0.2071],[118,0.2233],[119,0.2418],[120,0.2567],[126,0.2567],[127,0.2379],[128,0.1981],[129,0.1562],[130,0.1154],[131,0.0005],[132,0.0013],[133,0.0002],[134,0.0009],[135,0],[136,0.0009],[137,0],[138,0.0006],[139,0.0001],[141,0.0001],[142,0.001],[143,0.0003],[146,0.0003],[147,0.0006],[148,0.0001],[154,0.0001],[155,0.001],[156,0.0002],[157,0.0006],[158,-0.0002],[180,-0.0002],[181,-0.0029],[182,0],[183,-0.0029],[184,0],[185,-0.0023],[186,0],[187,-0.0016],[188,-0.0001],[189,-0.0011],[190,0],[192,-0.0009],[193,0],[195,0.0004],[196,0],[197,0.0006],[199,0.0024],[200,0],[202,0.0027],[204,0.0058],[205,0],[208,0.0056],[209,0.0076],[210,0],[212,0.0038],[213,0.0051],[214,-0.002],[215,0.0084],[216,0],[238,0],[240,-0.0014],[241,0.0006],[243,0.0006],[244,0.0003],[246,0.0003],[247,0.0006],[248,-0.0013],[250,0],[260,0],[261,0.0052],[262,0],[272,0],[273,0.0002],[274,0],[287,0],[288,0.0003],[289,0.0012],[290,0],[291,0.0001],[292,0.0035],[336,0.0035],[337,0.0001],[348,0.0001],[350,-0.0007],[351,0.0006],[352,0],[353,0],[354,0.0009],[355,0],[359,0],[360,0.0008],[361,0.0012],[362,0],[380,0],[381,0.0004],[382,0],[386,0],[387,0.0013],[388,0.0002],[389,0.0006],[390,-0.0008],[392,0],[418,0],[419,0.0004],[422,0.0004]]
       }
       foot_l_ik_ctrl.Rotation.Y {
        type=MovieSceneScriptingFloatChannel
        default=0
        thinned_from=423
        keys [[0,0],[1,-0.0019],[2,0],[5,0],[6,-0.0065],[7,0],[41,0],[42,-0.0053],[43,0],[52,0],[53,0.0101],[54,0],[55,0.006],[56,0],[90,0],[91,0.0034],[92,0],[99,0],[102,0.0026],[108,0.0063],[109,0.0068],[110,0.0097],[112,0.0146],[113,0.0169],[114,0.0073],[115,0.0094],[116,0.0079],[118,0.0091],[120,0.0105],[126,0.0105],[127,0.0098],[130,0.0048],[131,0],[180,0],[181,-0.0061],[182,0],[183,-0.0039],[184,0],[185,-0.0021],[186,0],[187,-0.0007],[188,-0.0004],[189,-0.0004],[190,0],[192,-0.0005],[193,0],[197,0.0001],[199,0.0005],[200,0],[204,0.0014],[205,0],[209,0.0019],[210,0],[211,0.0005],[212,0.0013],[213,0.0026],[214,-0.0005],[215,0.0034],[216,0],[240,-0.0001],[247,0.0001],[260,0],[261,0.0054],[262,0],[286,0],[287,0.0024],[288,0.0022],[289,-0.0011],[290,0],[291,0],[292,0.0034],[336,0.0034],[337,0],[418,0],[419,0.0003],[422,0.0003]]
       }
       foot_l_ik_ctrl.Rotation.Z {
        type=MovieSceneScriptingFloatChannel
        default=-7.2277
        thinned_from=423
        keys [[0,-7.2277],[1,-7.2193],[52,-7.2278],[53,-7.2366],[128,-7.2624],[130,-7.248],[131,-7.2278],[180,-7.2277],[181,-7.2396],[182,-7.2277],[183,-7.2379],[184,-7.2277],[185,-7.2366],[186,-7.2277],[187,-7.2355],[188,-7.2278],[192,-7.2393],[193,-7.2278],[195,-7.2369],[199,-7.2385],[204,-7.2393],[209,-7.239],[210,-7.2277],[213,-7.2381],[214,-7.2272],[215,-7.2379],[288,-7.2294],[289,-7.2169],[422,-7.2327]]
       }
       foot_l_ik_ctrl.Scale.X {type=MovieSceneScriptingFloatChannel, default=1, const [1,423]}
       foot_l_ik_ctrl.Scale.Y {type=MovieSceneScriptingFloatChannel, default=1, const [1,423]}
       foot_l_ik_ctrl.Scale.Z {type=MovieSceneScriptingFloatChannel, default=1, const [1,423]}
       ball_l_ik_ctrl.Location.X {type=MovieSceneScriptingFloatChannel, default=0, thinned_from=423, keys [[0,0],[292,0],[293,-0.0006],[336,-0.0006],[337,0],[422,0]]}
       ball_l_ik_ctrl.Location.Y {
        type=MovieSceneScriptingFloatChannel
        default=0
        thinned_from=423
        keys [[0,0],[292,0],[293,0.0004],[336,0.0004],[337,0],[348,0.0001],[353,-0.0001],[422,0]]
       }
       ball_l_ik_ctrl.Location.Z {type=MovieSceneScriptingFloatChannel, default=0, thinned_from=423, keys [[0,0],[292,0],[293,0.0008],[336,0.0008],[337,0],[422,0]]}
       ball_l_ik_ctrl.Rotation.X {
        type=MovieSceneScriptingFloatChannel
        default=0
        thinned_from=423
        keys [[0,0],[99,0],[100,-0.0234],[101,-0.0451],[102,-0.0642],[103,-0.0642],[104,-0.095],[105,-0.095],[106,-0.1259],[107,-0.1259],[108,-0.156],[112,-0.156],[113,-0.1751],[115,-0.1751],[116,-0.1954],[117,-0.1954],[118,-0.2249],[119,-0.2434],[127,-0.2434],[128,-0.1996],[129,-0.1575],[130,-0.1163],[131,0],[292,0],[293,-0.0035],[336,-0.0035],[337,0],[348,-0.0001],[352,0],[353,0.0002],[354,0],[422,0]]
       }
       ball_l_ik_ctrl.Rotation.Y {
        type=MovieSceneScriptingFloatChannel
        default=0
        thinned_from=423
        keys [[0,0],[99,0],[101,-0.0043],[102,-0.0061],[103,-0.0061],[104,-0.009],[105,-0.009],[106,-0.012],[107,-0.012],[108,-0.0148],[112,-0.0148],[113,-0.0167],[115,-0.0167],[116,-0.0186],[117,-0.0186],[118,-0.0214],[119,-0.0232],[127,-0.0232],[130,-0.0112],[131,0],[292,0],[293,-0.0034],[336,-0.0034],[337,0],[422,0]]
       }
       ball_l_ik_ctrl.Rotation.Z {
        type=MovieSceneScriptingFloatChannel
        default=0
        thinned_from=423
        keys [[0,0],[99,0],[100,0.0068],[101,0.0132],[102,0.0188],[103,0.0188],[104,0.0278],[105,0.0278],[106,0.0369],[107,0.0369],[108,0.0458],[112,0.0458],[113,0.0515],[115,0.0515],[116,0.0574],[117,0.0574],[118,0.0661],[119,0.0716],[127,0.0716],[128,0.0588],[129,0.0465],[130,0.0344],[131,0],[292,0],[293,0.0008],[336,0.0008],[337,0],[422,0]]
       }
       ball_l_ik_ctrl.Scale.X {type=MovieSceneScriptingFloatChannel, default=1, const [1,423]}
       ball_l_ik_ctrl.Scale.Y {type=MovieSceneScriptingFloatChannel, default=1, const [1,423]}
       ball_l_ik_ctrl.Scale.Z {type=MovieSceneScriptingFloatChannel, default=1, const [1,423]}
       foot_bk1_l_ctrl.Location.X {type=MovieSceneScriptingFloatChannel, default=0, const [0,423]}
       foot_bk1_l_ctrl.Location.Y {type=MovieSceneScriptingFloatChannel, default=0, const [0,423]}
       foot_bk1_l_ctrl.Location.Z {type=MovieSceneScriptingFloatChannel, default=0, const [0,423]}
       foot_bk1_l_ctrl.Rotation.X {type=MovieSceneScriptingFloatChannel, default=0, const [0,423]}
       foot_bk1_l_ctrl.Rotation.Y {type=MovieSceneScriptingFloatChannel, default=0, const [0,423]}
       foot_bk1_l_ctrl.Rotation.Z {type=MovieSceneScriptingFloatChannel, default=0, const [0,423]}
       foot_bk1_l_ctrl.Scale.X {type=MovieSceneScriptingFloatChannel, default=1, const [1,423]}
       foot_bk1_l_ctrl.Scale.Y {type=MovieSceneScriptingFloatChannel, default=1, const [1,423]}
       foot_bk1_l_ctrl.Scale.Z {type=MovieSceneScriptingFloatChannel, default=1, const [1,423]}
       foot_roll_l_ctrl.Location.X {type=MovieSceneScriptingFloatChannel, default=0, const [0,423]}
       foot_roll_l_ctrl.Location.Y {type=MovieSceneScriptingFloatChannel, default=0, const [0,423]}
       foot_roll_l_ctrl.Location.Z {type=MovieSceneScriptingFloatChannel, default=0, const [0,423]}
       foot_roll_l_ctrl.Rotation.X {type=MovieSceneScriptingFloatChannel, default=0, const [0,423]}
       foot_roll_l_ctrl.Rotation.Y {type=MovieSceneScriptingFloatChannel, default=0, const [0,423]}
       foot_roll_l_ctrl.Rotation.Z {type=MovieSceneScriptingFloatChannel, default=0, const [0,423]}
       foot_roll_l_ctrl.Scale.X {type=MovieSceneScriptingFloatChannel, default=1, const [1,423]}
       foot_roll_l_ctrl.Scale.Y {type=MovieSceneScriptingFloatChannel, default=1, const [1,423]}
       foot_roll_l_ctrl.Scale.Z {type=MovieSceneScriptingFloatChannel, default=1, const [1,423]}
       roll_blend_l_ctrl {type=MovieSceneScriptingFloatChannel, default=1}
       tip_l_ctrl.Location.X {type=MovieSceneScriptingFloatChannel, default=0, const [0,423]}
       tip_l_ctrl.Location.Y {type=MovieSceneScriptingFloatChannel, default=0, const [0,423]}
       tip_l_ctrl.Location.Z {type=MovieSceneScriptingFloatChannel, default=0, const [0,423]}
       tip_l_ctrl.Rotation.X {type=MovieSceneScriptingFloatChannel, default=0, const [0,423]}
       tip_l_ctrl.Rotation.Y {type=MovieSceneScriptingFloatChannel, default=0, const [0,423]}
       tip_l_ctrl.Rotation.Z {type=MovieSceneScriptingFloatChannel, default=0, const [0,423]}
       tip_l_ctrl.Scale.X {type=MovieSceneScriptingFloatChannel, default=1, const [1,423]}
       tip_l_ctrl.Scale.Y {type=MovieSceneScriptingFloatChannel, default=1, const [1,423]}
       tip_l_ctrl.Scale.Z {type=MovieSceneScriptingFloatChannel, default=1, const [1,423]}
       heel_l_ctrl.Location.X {type=MovieSceneScriptingFloatChannel, default=0, const [0,423]}
       heel_l_ctrl.Location.Y {type=MovieSceneScriptingFloatChannel, default=0, const [0,423]}
       heel_l_ctrl.Location.Z {type=MovieSceneScriptingFloatChannel, default=0, const [0,423]}
       heel_l_ctrl.Rotation.X {type=MovieSceneScriptingFloatChannel, default=0, const [0,423]}
       heel_l_ctrl.Rotation.Y {type=MovieSceneScriptingFloatChannel, default=0, const [0,423]}
       heel_l_ctrl.Rotation.Z {type=MovieSceneScriptingFloatChannel, default=0, const [0,423]}
       heel_l_ctrl.Scale.X {type=MovieSceneScriptingFloatChannel, default=1, const [1,423]}
       heel_l_ctrl.Scale.Y {type=MovieSceneScriptingFloatChannel, default=1, const [1,423]}
       heel_l_ctrl.Scale.Z {type=MovieSceneScriptingFloatChannel, default=1, const [1,423]}
       leg_l_softness {type=MovieSceneScriptingFloatChannel, default=0, const [0,423]}
       leg_l_pv_ik_ctrl.Location.X {
        type=MovieSceneScriptingFloatChannel
        default=-0.1191
        thinned_from=423
        keys [[0,-0.0685],[1,-0.0689],[2,-0.0698],[3,-0.0711],[6,-0.0758],[7,-0.0775],[9,-0.0816],[10,-0.0834],[11,-0.0844],[12,-0.0849],[13,-0.0846],[14,-0.0824],[15,-0.077],[16,-0.0664],[17,-0.05],[18,-0.0291],[19,-0.0014],[20,0.0302],[21,0.0575],[22,0.0794],[23,0.0989],[24,0.1143],[25,0.1256],[26,0.1338],[27,0.1396],[28,0.1433],[29,0.1452],[30,0.1461],[31,0.1463],[33,0.1457],[37,0.1427],[41,0.1386],[42,0.1386],[43,0.1368],[49,0.1336],[54,0.1323],[56,0.1323],[58,0.1335],[62,0.1373],[64,0.1384],[65,0.1384],[66,0.1349],[67,0.1268],[68,0.1164],[69,0.1048],[70,0.0914],[71,0.0759],[72,0.0586],[73,0.0396],[74,0.0194],[75,-0.0018],[76,-0.0235],[77,-0.0468],[78,-0.0707],[79,-0.0913],[80,-0.1099],[81,-0.1273],[82,-0.14],[83,-0.1474],[84,-0.152],[85,-0.1547],[86,-0.1562],[87,-0.1568],[88,-0.157],[90,-0.1561],[92,-0.1544],[94,-0.1517],[98,-0.1452],[99,-0.1434],[101,-0.1453],[102,-0.1459],[103,-0.1463],[105,-0.1482],[108,-0.1521],[109,-0.153],[115,-0.1532],[117,-0.1562],[118,-0.1582],[119,-0.1605],[120,-0.1623],[126,-0.1622],[127,-0.1603],[128,-0.1567],[129,-0.154],[130,-0.1515],[131,-0.1573],[132,-0.1842],[133,-0.2118],[134,-0.2347],[135,-0.2496],[136,-0.2432],[137,-0.1936],[138,-0.0819],[139,-0.0022],[140,0.057],[141,0.1259],[142,0.1825],[143,0.2117],[144,0.2169],[145,0.1944],[146,0.1683],[147,0.1222],[148,0.0725],[149,0.0151],[150,-0.0322],[151,-0.0601],[152,-0.07],[153,-0.0599],[154,-0.0412],[155,-0.0208],[156,0.0071],[157,0.0225],[158,0.0402],[159,0.0424],[160,0.0444],[161,0.0461],[163,0.0492],[165,0.0519],[167,0.0539],[171,0.0566],[174,0.058],[183,0.0601],[189,0.0603],[199,0.0581],[213,0.0526],[214,0.0526],[216,0.0481],[219,0.0399],[220,0.0373],[221,0.0352],[222,0.0336],[223,0.0327],[224,0.0325],[225,0.033],[226,0.0381],[227,0.0484],[228,0.0632],[229,0.0818],[230,0.1033],[231,0.127],[232,0.1516],[233,0.1752],[234,0.1956],[235,0.2126],[236,0.227],[237,0.2341],[239,0.2373],[240,0.237],[241,0.222],[242,0.2168],[243,0.2102],[244,0.1937],[245,0.1747],[246,0.1545],[247,0.1381],[248,0.1311],[249,0.1136],[250,0.1039],[251,0.0979],[253,0.0884],[254,0.0848],[255,0.0823],[256,0.0813],[257,0.0811],[258,0.0818],[260,0.0835],[261,0.0835],[262,0.0853],[264,0.0866],[267,0.0876],[270,0.0881],[273,0.0881],[284,0.0839],[286,0.0835],[288,0.0835],[290,0.084],[291,0.0844],[292,0.0844],[293,0.0852],[295,0.0883],[296,0.0883],[297,0.0848],[298,0.0751],[299,0.0594],[300,0.0413],[301,0.0218],[302,-0.0018],[303,-0.028],[304,-0.0553],[305,-0.0822],[306,-0.1038],[307,-0.1204],[308,-0.1344],[309,-0.1449],[310,-0.1512],[311,-0.1548],[312,-0.1569],[313,-0.1578],[314,-0.1577],[315,-0.1568],[316,-0.1544],[317,-0.1504],[320,-0.137],[321,-0.1329],[322,-0.1292],[323,-0.1262],[324,-0.1238],[325,-0.1219],[326,-0.1206],[328,-0.1191],[330,-0.1185],[333,-0.1185],[336,-0.1194],[337,-0.1194],[340,-0.1215],[342,-0.1236],[343,-0.1253],[344,-0.1278],[345,-0.131],[346,-0.1347],[347,-0.1391],[348,-0.1441],[349,-0.1501],[350,-0.157],[351,-0.1799],[352,-0.2118],[353,-0.2548],[354,-0.3132],[355,-0.3622],[356,-0.4],[357,-0.4296],[358,-0.4549],[359,-0.4787],[360,-0.495],[361,-0.5073],[362,-0.5164],[363,-0.5229],[369,-0.5471],[371,-0.553],[372,-0.5544],[373,-0.5527],[375,-0.5424],[376,-0.5355],[377,-0.5235],[378,-0.505],[379,-0.4817],[380,-0.4491],[381,-0.404],[382,-0.3557],[384,-0.2552],[385,-0.2032],[386,-0.1507],[387,-0.1067],[388,-0.0763],[389,-0.0563],[390,-0.0349],[391,-0.0307],[392,-0.0283],[393,-0.0277],[394,-0.0281],[395,-0.0292],[396,-0.0308],[397,-0.033],[399,-0.038],[401,-0.0423],[402,-0.0453],[405,-0.0558],[406,-0.0589],[407,-0.0615],[408,-0.0635],[409,-0.0652],[411,-0.0676],[413,-0.0693],[414,-0.0698],[415,-0.0698],[418,-0.0687],[422,-0.0687]]
       }
       leg_l_pv_ik_ctrl.Location.Y {
        type=MovieSceneScriptingFloatChannel
        default=-1.702
        thinned_from=423
        keys [[0,-1.945],[130,-1.9438],[131,-1.7408],[132,-1.4582],[133,-1.1326],[134,-0.6691],[135,-0.1857],[136,0.1306],[137,0.2433],[138,0.1028],[139,-0.0366],[140,-0.4364],[141,-0.978],[142,-1.415],[143,-1.6785],[144,-1.9509],[146,-1.9497],[147,-1.7632],[148,-1.5299],[149,-1.2269],[150,-1.0155],[151,-0.9326],[152,-0.9696],[153,-1.0948],[154,-1.2413],[155,-1.4044],[156,-1.6352],[157,-1.7588],[158,-1.9452],[238,-1.9501],[240,-1.941],[241,-1.7654],[242,-1.728],[243,-1.6978],[244,-1.6758],[245,-1.6863],[246,-1.7201],[247,-1.7642],[248,-1.9391],[250,-1.9483],[349,-1.9423],[350,-1.9383],[351,-1.7668],[352,-1.5109],[353,-1.1595],[354,-0.6716],[355,-0.2851],[356,0.0052],[357,0.2326],[358,0.4372],[359,0.6433],[360,0.7859],[361,0.8983],[362,0.9869],[363,1.051],[364,1.09],[365,1.1374],[367,1.2182],[368,1.2612],[369,1.2995],[372,1.3949],[373,1.4127],[374,1.4134],[375,1.4454],[376,1.4735],[377,1.457],[378,1.3751],[379,1.2361],[380,0.9986],[381,0.635],[382,0.2455],[383,-0.1515],[384,-0.5029],[385,-0.8335],[386,-1.1711],[387,-1.459],[388,-1.6494],[389,-1.7655],[390,-1.9396],[392,-1.9458],[422,-1.9451]]
       }
       leg_l_pv_ik_ctrl.Location.Z {
        type=MovieSceneScriptingFloatChannel
        default=-1.5876
        thinned_from=423
        keys [[0,-1.797],[18,-1.7933],[72,-1.7851],[130,-1.8048],[131,-1.6262],[132,-1.3767],[133,-1.0849],[134,-0.6598],[135,-0.2057],[136,0.0989],[137,0.2138],[138,0.09],[139,-0.0354],[140,-0.4094],[141,-0.9059],[142,-1.2972],[143,-1.53],[144,-1.7704],[146,-1.7749],[147,-1.6154],[148,-1.413],[149,-1.146],[150,-0.9582],[151,-0.885],[152,-0.9202],[153,-1.0339],[154,-1.1653],[155,-1.3106],[156,-1.5145],[157,-1.6226],[158,-1.7853],[229,-1.7829],[239,-1.7645],[240,-1.7593],[241,-1.606],[242,-1.5734],[243,-1.5473],[244,-1.5296],[245,-1.5412],[246,-1.5735],[247,-1.6145],[248,-1.7698],[250,-1.7809],[301,-1.7886],[349,-1.8033],[350,-1.8006],[351,-1.6516],[352,-1.427],[353,-1.114],[354,-0.6705],[355,-0.3122],[356,-0.0388],[357,0.178],[358,0.3754],[359,0.5763],[360,0.7165],[361,0.8278],[362,0.9159],[363,0.98],[364,1.0191],[365,1.0667],[367,1.148],[368,1.1915],[369,1.2303],[372,1.3275],[373,1.3459],[374,1.3472],[375,1.3807],[376,1.4104],[377,1.3947],[378,1.3123],[379,1.1723],[380,0.9348],[381,0.5759],[383,-0.179],[384,-0.5065],[385,-0.8091],[386,-1.1134],[387,-1.3691],[388,-1.5362],[389,-1.6372],[390,-1.7886],[422,-1.7971]]
       }
       leg_l_pv_ik_ctrl.Rotation.X {
        type=MovieSceneScriptingFloatChannel
        default=2.161
        thinned_from=423
        keys [[0,2.6062],[18,2.601],[72,2.5895],[130,2.6171],[131,2.3519],[132,1.9838],[133,1.5565],[134,0.9404],[135,0.2901],[136,-0.1416],[137,-0.3032],[138,-0.1278],[139,0.0502],[140,0.583],[141,1.2988],[142,1.8699],[143,2.2126],[144,2.5688],[146,2.5752],[147,2.3382],[148,2.0391],[149,1.6474],[150,1.3736],[151,1.2671],[152,1.3181],[153,1.4834],[154,1.6751],[155,1.888],[156,2.1881],[157,2.3481],[158,2.5896],[230,2.5837],[239,2.5602],[240,2.5526],[241,2.325],[242,2.2768],[243,2.2381],[244,2.2119],[245,2.2288],[246,2.2765],[247,2.337],[248,2.5672],[250,2.5836],[301,2.5944],[349,2.615],[350,2.6109],[351,2.3895],[352,2.0575],[353,1.5986],[354,0.9552],[355,0.4412],[356,0.052],[357,-0.2545],[358,-0.532],[359,-0.8131],[360,-1.0084],[361,-1.163],[362,-1.2851],[363,-1.3736],[364,-1.4277],[365,-1.4934],[367,-1.6054],[368,-1.6652],[369,-1.7185],[372,-1.8518],[373,-1.8771],[374,-1.8788],[375,-1.9246],[376,-1.9651],[377,-1.9436],[378,-1.8307],[379,-1.6385],[380,-1.3108],[381,-0.812],[382,-0.2826],[383,0.2517],[384,0.7199],[385,1.1562],[386,1.5986],[387,1.9732],[388,2.2196],[389,2.3691],[390,2.5939],[422,2.6063]]
       }
       leg_l_pv_ik_ctrl.Rotation.Y {
        type=MovieSceneScriptingFloatChannel
        default=-1.2191
        thinned_from=423
        keys [[0,-0.0142],[12,-0.0153],[14,-0.0151],[15,-0.0147],[16,-0.0141],[17,-0.013],[18,-0.0117],[19,-0.01],[20,-0.008],[21,-0.0063],[23,-0.0037],[24,-0.0027],[26,-0.0015],[28,-0.0009],[65,-0.0012],[66,-0.0014],[68,-0.0026],[70,-0.0042],[72,-0.0062],[76,-0.0114],[78,-0.0144],[81,-0.018],[82,-0.0187],[84,-0.0195],[99,-0.0189],[126,-0.0201],[131,-0.0193],[132,-0.0199],[133,-0.0202],[134,-0.0191],[135,-0.0166],[136,-0.0137],[137,-0.0097],[138,-0.0041],[139,-0.0004],[140,0.0003],[141,0.0014],[142,0.0031],[143,0.004],[144,0.0037],[145,0.0023],[146,0.0007],[149,-0.0066],[150,-0.0085],[151,-0.0098],[152,-0.0106],[153,-0.0106],[155,-0.0095],[156,-0.0086],[158,-0.0074],[214,-0.0066],[223,-0.0078],[225,-0.0078],[226,-0.0075],[227,-0.0068],[228,-0.0059],[230,-0.0034],[233,0.0011],[235,0.0034],[236,0.0043],[237,0.0048],[240,0.005],[241,0.0044],[243,0.0039],[244,0.0029],[247,-0.0008],[248,-0.0016],[249,-0.0027],[250,-0.0034],[254,-0.0045],[261,-0.0046],[296,-0.0044],[297,-0.0046],[298,-0.0052],[300,-0.0073],[301,-0.0085],[303,-0.0117],[305,-0.0151],[306,-0.0165],[308,-0.0184],[309,-0.0191],[311,-0.0197],[315,-0.0198],[325,-0.0176],[343,-0.0178],[348,-0.019],[350,-0.0198],[353,-0.0231],[354,-0.024],[355,-0.0243],[357,-0.0242],[358,-0.0238],[362,-0.0216],[374,-0.0181],[376,-0.0166],[377,-0.0161],[378,-0.0161],[379,-0.0165],[380,-0.0174],[381,-0.0186],[382,-0.0195],[383,-0.0198],[384,-0.0193],[385,-0.0181],[387,-0.0151],[388,-0.0139],[390,-0.0121],[392,-0.0117],[413,-0.0143],[422,-0.0142]]
       }
       leg_l_pv_ik_ctrl.Rotation.Z {
        type=MovieSceneScriptingFloatChannel
        default=-1.1927
        thinned_from=423
        keys [[0,0.0858],[1,0.0863],[2,0.0876],[3,0.0895],[6,0.0961],[7,0.0985],[9,0.1044],[10,0.1068],[11,0.1083],[12,0.1091],[13,0.1087],[14,0.1055],[15,0.0978],[16,0.0827],[17,0.0595],[18,0.0299],[19,-0.0092],[20,-0.054],[21,-0.0926],[22,-0.1237],[23,-0.1513],[24,-0.1732],[25,-0.1892],[26,-0.2008],[27,-0.209],[28,-0.2142],[29,-0.2169],[30,-0.2182],[31,-0.2185],[33,-0.2177],[37,-0.2133],[41,-0.2076],[42,-0.2076],[43,-0.2051],[49,-0.2005],[54,-0.1986],[56,-0.1987],[58,-0.2003],[62,-0.2057],[64,-0.2073],[65,-0.2073],[66,-0.2024],[67,-0.1909],[68,-0.1762],[69,-0.1597],[70,-0.1407],[71,-0.1188],[72,-0.0942],[73,-0.0674],[74,-0.0387],[75,-0.0088],[76,0.022],[77,0.0551],[78,0.0888],[79,0.1181],[80,0.1444],[81,0.1691],[82,0.187],[83,0.1975],[84,0.2041],[85,0.2079],[86,0.2099],[87,0.2109],[88,0.2111],[90,0.2099],[92,0.2075],[94,0.2035],[98,0.1944],[99,0.1919],[101,0.1945],[102,0.1955],[103,0.196],[105,0.1987],[108,0.2041],[109,0.2055],[115,0.2057],[117,0.21],[118,0.2128],[119,0.216],[120,0.2187],[126,0.2185],[127,0.2157],[128,0.2108],[129,0.2068],[130,0.2034],[131,0.2112],[132,0.2486],[133,0.2866],[134,0.3176],[135,0.3378],[136,0.3294],[137,0.2627],[138,0.1115],[139,0.0027],[140,-0.0812],[141,-0.1811],[142,-0.2644],[143,-0.3082],[144,-0.3185],[145,-0.2867],[146,-0.2497],[147,-0.1829],[148,-0.1114],[149,-0.0293],[150,0.0375],[151,0.0765],[152,0.09],[153,0.0756],[154,0.0491],[155,0.0199],[156,-0.0201],[157,-0.0423],[158,-0.0681],[159,-0.0713],[160,-0.0741],[161,-0.0765],[163,-0.0809],[165,-0.0847],[166,-0.0863],[168,-0.0888],[171,-0.0915],[174,-0.0934],[180,-0.0956],[184,-0.0965],[185,-0.0965],[186,-0.0967],[189,-0.0967],[195,-0.0952],[213,-0.0858],[214,-0.0858],[216,-0.0794],[217,-0.0756],[219,-0.0677],[220,-0.0641],[221,-0.0611],[222,-0.0589],[223,-0.0576],[224,-0.0573],[225,-0.0579],[226,-0.0652],[227,-0.0798],[228,-0.1008],[229,-0.127],[230,-0.1575],[231,-0.1911],[232,-0.2259],[233,-0.2595],[234,-0.2884],[235,-0.3125],[236,-0.3327],[237,-0.3428],[239,-0.3473],[240,-0.3468],[241,-0.3237],[242,-0.316],[243,-0.3063],[244,-0.2828],[245,-0.2563],[246,-0.2281],[247,-0.2053],[248,-0.1969],[249,-0.1722],[250,-0.1584],[251,-0.1499],[253,-0.1365],[254,-0.1313],[255,-0.1279],[256,-0.1263],[257,-0.1262],[258,-0.1271],[260,-0.1295],[261,-0.1295],[262,-0.132],[264,-0.1338],[267,-0.1353],[270,-0.136],[273,-0.136],[284,-0.1301],[286,-0.1295],[288,-0.1295],[290,-0.1302],[291,-0.1308],[292,-0.1308],[293,-0.132],[295,-0.1364],[296,-0.1364],[297,-0.1314],[298,-0.1176],[299,-0.0954],[300,-0.0698],[301,-0.0422],[302,-0.0087],[303,0.0285],[304,0.0671],[305,0.1051],[306,0.1357],[307,0.1593],[308,0.1792],[309,0.1939],[310,0.2029],[311,0.208],[312,0.211],[313,0.2122],[314,0.212],[315,0.2108],[316,0.2075],[317,0.2018],[320,0.1828],[321,0.177],[322,0.1717],[323,0.1675],[324,0.164],[325,0.1614],[326,0.1596],[328,0.1574],[330,0.1565],[333,0.1565],[336,0.1578],[337,0.1578],[340,0.1608],[342,0.1638],[343,0.1662],[344,0.1698],[345,0.1743],[346,0.1796],[347,0.1857],[348,0.1929],[349,0.2014],[350,0.2111],[351,0.2431],[352,0.2874],[353,0.3464],[354,0.4256],[355,0.4913],[356,0.5416],[357,0.5805],[358,0.6137],[359,0.6446],[360,0.6657],[361,0.6814],[362,0.6931],[363,0.7015],[369,0.7322],[371,0.7396],[372,0.7414],[373,0.739],[375,0.7254],[376,0.7162],[377,0.7005],[378,0.6766],[379,0.6463],[380,0.6039],[381,0.545],[382,0.481],[384,0.3457],[385,0.2743],[386,0.2016],[387,0.1401],[388,0.0971],[389,0.0687],[390,0.0382],[391,0.0323],[392,0.0288],[393,0.0279],[394,0.0286],[395,0.0301],[396,0.0324],[397,0.0355],[399,0.0426],[400,0.0455],[401,0.0487],[402,0.0529],[403,0.0578],[405,0.0678],[406,0.0722],[407,0.0758],[408,0.0787],[409,0.081],[410,0.0829],[411,0.0844],[413,0.0869],[414,0.0876],[415,0.0876],[418,0.0861],[422,0.0861]]
       }
       leg_l_pv_ik_ctrl.Scale.X {type=MovieSceneScriptingFloatChannel, default=1, const [1,423]}
       leg_l_pv_ik_ctrl.Scale.Y {type=MovieSceneScriptingFloatChannel, default=1, const [1,423]}
       leg_l_pv_ik_ctrl.Scale.Z {type=MovieSceneScriptingFloatChannel, default=1, const [1,423]}
       thigh_r_fk_ctrl.Location.X {type=MovieSceneScriptingFloatChannel, default=0, thinned_from=423, keys [[0,-0.0002],[422,-0.0002]]}
       thigh_r_fk_ctrl.Location.Y {type=MovieSceneScriptingFloatChannel, default=0, thinned_from=423, keys [[0,0],[422,0]]}
       thigh_r_fk_ctrl.Location.Z {type=MovieSceneScriptingFloatChannel, default=0, thinned_from=423, keys [[0,0],[422,0]]}
       thigh_r_fk_ctrl.Rotation.X {
        type=MovieSceneScriptingFloatChannel
        default=43.0271
        thinned_from=423
        keys [[0,41.7621],[12,41.8008],[16,41.9235],[17,41.8543],[18,41.6482],[19,41.2243],[22,39.3473],[23,38.6495],[24,38.0685],[26,37.2689],[28,36.7673],[30,36.5266],[35,36.4671],[55,37.1219],[66,36.5526],[69,36.5981],[70,36.7312],[71,36.9901],[72,37.3782],[74,38.3992],[80,41.9661],[81,42.4978],[82,42.9283],[84,43.5195],[86,43.8902],[90,44.1289],[97,43.7302],[104,42.8849],[135,42.1933],[136,42.1251],[137,41.8809],[139,41.0823],[140,40.3429],[141,39.1586],[142,37.699],[143,36.0692],[144,35.4777],[145,35.2604],[146,35.3453],[147,35.6619],[148,36.0901],[150,37.4902],[151,37.7681],[152,37.8475],[153,37.5238],[156,36.0272],[157,35.8132],[227,36.6736],[230,36.4416],[232,36.2118],[233,36.1732],[234,36.2932],[237,37.0266],[239,37.3533],[245,37.8075],[246,37.971],[247,38.2963],[248,38.8961],[249,39.3949],[250,39.5339],[271,38.7079],[289,39.1519],[293,38.9603],[296,38.4517],[297,38.3665],[298,38.4007],[299,38.5249],[301,39.0098],[302,39.5275],[303,40.2829],[306,42.929],[310,45.6408],[311,46.2292],[313,47.0955],[315,47.6294],[316,47.7283],[324,46.7554],[347,46.6376],[349,46.8431],[351,47.3042],[352,47.6483],[358,50.6451],[361,51.7105],[363,52.1796],[368,52.6094],[378,52.4145],[380,52.1468],[381,51.8459],[382,51.3526],[383,50.6022],[384,49.3495],[385,47.6303],[386,45.8037],[387,44.2317],[388,43.0302],[389,42.0796],[390,41.3888],[391,40.8906],[392,40.5009],[394,39.9789],[396,39.7101],[400,39.5743],[401,39.6227],[403,39.9569],[407,40.8361],[414,41.6101],[422,41.7571]]
       }
       thigh_r_fk_ctrl.Rotation.Y {
        type=MovieSceneScriptingFloatChannel
        default=-0.9994
        thinned_from=423
        keys [[0,-0.0695],[1,-0.0695],[2,-0.0726],[3,-0.0764],[4,-0.0814],[5,-0.0876],[6,-0.0876],[7,-0.1053],[8,-0.1188],[9,-0.1333],[10,-0.1467],[11,-0.156],[12,-0.1614],[13,-0.1598],[14,-0.1456],[15,-0.1136],[16,-0.0443],[17,0.0918],[18,0.3149],[19,0.6795],[20,1.1181],[21,1.4698],[22,1.6804],[23,1.782],[24,1.7744],[25,1.6914],[26,1.592],[27,1.4886],[29,1.2493],[30,1.1489],[31,1.0743],[32,1.0193],[33,0.9813],[34,0.9558],[35,0.9381],[36,0.9237],[38,0.9003],[41,0.8734],[42,0.867],[43,0.8642],[44,0.8654],[45,0.8699],[47,0.8849],[52,0.9373],[53,0.9373],[54,0.9559],[55,0.9559],[56,0.9699],[57,0.9775],[58,0.9827],[59,0.9845],[64,0.9751],[67,0.9756],[68,0.9735],[69,0.9651],[70,0.9516],[71,0.9284],[72,0.8863],[73,0.822],[74,0.722],[75,0.5992],[76,0.4868],[77,0.3988],[78,0.3247],[79,0.2601],[80,0.1976],[81,0.141],[82,0.1023],[83,0.0816],[84,0.0695],[85,0.0641],[86,0.0634],[87,0.0671],[88,0.0747],[89,0.0833],[90,0.0913],[92,0.0913],[93,0.0762],[94,0.0513],[95,0.018],[96,-0.0221],[97,-0.0678],[98,-0.1157],[99,-0.163],[100,-0.2086],[101,-0.2517],[102,-0.2913],[103,-0.3274],[104,-0.3609],[105,-0.3919],[106,-0.4189],[107,-0.4425],[108,-0.4631],[109,-0.4811],[110,-0.497],[111,-0.5112],[113,-0.5363],[114,-0.5459],[115,-0.5516],[116,-0.5511],[117,-0.543],[118,-0.529],[119,-0.5106],[120,-0.4894],[122,-0.4454],[123,-0.4257],[124,-0.4095],[125,-0.3988],[126,-0.4018],[127,-0.471],[128,-0.5807],[129,-0.7202],[130,-0.812],[131,-0.8325],[132,-0.6224],[133,-0.2623],[134,0.5037],[135,1.7701],[136,3.0867],[137,4.0427],[138,4.7282],[139,5.7206],[140,6.1141],[141,5.1897],[142,3.9578],[143,3.0734],[144,2.1129],[145,1.4003],[146,0.8062],[147,0.245],[148,-0.2965],[149,-0.6562],[150,-0.8129],[151,-0.8116],[152,-0.6462],[153,-0.2653],[154,-0.0058],[155,0.112],[156,0.2007],[157,0.1146],[158,0.1016],[159,0.0715],[160,0.0344],[161,-0.0039],[162,-0.0426],[163,-0.0857],[164,-0.1308],[165,-0.1753],[166,-0.2166],[167,-0.2521],[168,-0.2835],[169,-0.3134],[170,-0.3416],[171,-0.3675],[172,-0.3909],[173,-0.4115],[174,-0.4293],[175,-0.4439],[176,-0.456],[177,-0.4666],[178,-0.4756],[179,-0.4832],[180,-0.4895],[181,-0.4895],[182,-0.499],[183,-0.499],[184,-0.505],[185,-0.505],[186,-0.5085],[187,-0.5085],[188,-0.5107],[189,-0.5107],[190,-0.5128],[192,-0.5128],[193,-0.5163],[196,-0.5163],[197,-0.5208],[201,-0.5208],[202,-0.5272],[208,-0.5272],[209,-0.5363],[213,-0.5363],[214,-0.54],[215,-0.54],[216,-0.5477],[217,-0.5477],[218,-0.5588],[220,-0.57],[222,-0.5781],[223,-0.5792],[224,-0.5783],[225,-0.575],[226,-0.5592],[227,-0.5274],[228,-0.4773],[229,-0.4173],[230,-0.3413],[231,-0.2299],[232,0.024],[233,0.1789],[234,0.2216],[235,0.1962],[236,0.1574],[237,0.1106],[238,0.0769],[239,0.22],[240,0.3609],[241,0.5818],[242,0.907],[243,1.1582],[244,1.3996],[245,1.4907],[246,1.3677],[247,0.939],[248,0.3055],[249,-0.2445],[250,-0.5255],[251,-0.7373],[252,-0.8265],[253,-0.9136],[254,-0.8725],[255,-0.8835],[256,-0.9103],[257,-0.9339],[258,-0.9487],[259,-0.9616],[260,-0.9694],[261,-0.9703],[262,-0.9674],[263,-0.9609],[264,-0.9517],[265,-0.94],[266,-0.9248],[267,-0.9065],[269,-0.8619],[270,-0.8438],[271,-0.8281],[273,-0.8033],[274,-0.7926],[275,-0.7853],[277,-0.7789],[281,-0.7737],[289,-0.7716],[291,-0.7739],[292,-0.7739],[293,-0.7718],[295,-0.7533],[296,-0.7666],[297,-0.8248],[298,-0.9487],[299,-1.1338],[300,-1.3626],[301,-1.6711],[302,-2.0879],[303,-2.5407],[304,-2.9717],[305,-3.3766],[306,-3.6472],[307,-3.8158],[308,-3.944],[309,-3.8593],[310,-3.5936],[311,-3.3934],[312,-3.2595],[313,-3.096],[314,-2.9481],[315,-2.8639],[318,-2.7205],[319,-2.664],[320,-2.5804],[321,-2.4886],[322,-2.4171],[323,-2.3664],[324,-2.331],[325,-2.3071],[326,-2.2998],[327,-2.3125],[328,-2.3409],[329,-2.3829],[330,-2.4342],[333,-2.6072],[334,-2.6589],[344,-3.1024],[345,-3.139],[350,-3.2751],[351,-3.2955],[352,-3.2997],[353,-3.2876],[354,-3.2608],[355,-3.1932],[356,-3.081],[357,-2.9528],[359,-2.7227],[361,-2.465],[362,-2.3516],[363,-2.2688],[364,-2.2201],[365,-2.1922],[366,-2.1748],[367,-2.1662],[368,-2.1651],[369,-2.1697],[370,-2.1697],[371,-2.1891],[373,-2.2078],[374,-2.2126],[375,-2.2122],[376,-2.1584],[377,-2.0287],[378,-1.8548],[379,-1.6547],[380,-1.3939],[381,-1.0565],[382,-0.6834],[383,-0.3164],[384,0.0566],[385,0.4348],[386,0.7604],[387,0.9812],[388,1.0583],[389,1.0176],[390,0.9105],[391,0.7994],[392,0.7041],[394,0.5023],[395,0.4101],[396,0.3298],[397,0.2607],[398,0.2033],[399,0.1584],[400,0.1264],[401,0.1001],[402,0.0726],[403,0.0446],[404,0.017],[405,-0.0093],[406,-0.0334],[407,-0.0518],[408,-0.063],[409,-0.0685],[410,-0.07],[411,-0.0692],[412,-0.0679],[413,-0.0676],[415,-0.0688],[417,-0.0694],[422,-0.0695]]
       }
       thigh_r_fk_ctrl.Rotation.Z {
        type=MovieSceneScriptingFloatChannel
        default=1.795
        thinned_from=423
        keys [[0,3.1979],[1,3.1979],[5,3.1638],[8,3.1595],[11,3.177],[13,3.2014],[14,3.2213],[15,3.2517],[17,3.3456],[18,3.3618],[19,3.3261],[20,3.2484],[21,3.1444],[22,3.0196],[23,2.8851],[24,2.773],[25,2.687],[26,2.6109],[27,2.5489],[28,2.502],[29,2.47],[30,2.4498],[31,2.439],[33,2.4314],[35,2.4352],[43,2.5183],[52,2.5614],[53,2.5614],[54,2.5668],[56,2.5672],[64,2.4773],[65,2.4585],[66,2.4117],[67,2.3314],[68,2.2376],[69,2.149],[70,2.0772],[71,2.0331],[72,2.0185],[73,2.0265],[74,2.0512],[76,2.1198],[77,2.1413],[78,2.1512],[79,2.1683],[82,2.2309],[86,2.3492],[87,2.3692],[88,2.3823],[90,2.3952],[93,2.3959],[95,2.3773],[97,2.349],[99,2.3086],[102,2.2354],[103,2.2148],[105,2.187],[110,2.1463],[118,2.1109],[121,2.1069],[122,2.1124],[123,2.1292],[124,2.1555],[125,2.1897],[126,2.1973],[127,2.1445],[128,1.8261],[129,1.2251],[130,0.5444],[131,-0.1924],[132,-1.0066],[133,-1.626],[134,-2.1142],[135,-2.8013],[136,-3.377],[137,-3.0235],[138,-1.9373],[139,-1.7658],[140,-1.5983],[141,-1.2182],[142,-0.8039],[143,-0.6004],[144,-0.2202],[145,0.0819],[146,0.4468],[147,0.8237],[148,1.2564],[149,1.4856],[150,1.4188],[151,1.2227],[152,1.0183],[153,0.818],[154,0.8034],[155,0.7839],[156,0.747],[157,0.8841],[158,0.8759],[159,0.8796],[160,0.8889],[161,0.9035],[162,0.9201],[164,0.96],[171,1.1155],[173,1.1549],[174,1.1718],[177,1.2113],[180,1.2428],[181,1.2428],[182,1.2599],[183,1.2599],[184,1.2736],[185,1.2736],[186,1.2836],[187,1.2836],[188,1.2908],[189,1.2908],[190,1.2963],[192,1.2963],[193,1.3017],[196,1.3017],[201,1.3041],[208,1.3022],[209,1.2939],[213,1.2939],[214,1.2879],[215,1.2879],[216,1.2711],[217,1.2711],[218,1.2441],[220,1.2079],[222,1.1636],[223,1.1272],[224,1.0699],[225,0.997],[226,0.9301],[227,0.8701],[228,0.7839],[229,0.7129],[230,0.6557],[231,0.6063],[232,0.5825],[233,0.5643],[234,0.5672],[235,0.5748],[236,0.5631],[237,0.501],[238,0.3904],[239,0.0697],[240,-0.2396],[241,-0.6692],[242,-1.2511],[243,-1.7148],[244,-1.9259],[245,-1.6379],[246,-1.2234],[247,-0.509],[248,0.3694],[249,1.5416],[250,2.6897],[251,3.684],[252,4.2748],[253,4.5112],[254,4.4188],[255,4.2855],[256,4.1747],[257,4.0859],[258,4.0394],[262,3.9695],[268,3.9312],[289,3.9935],[293,3.9527],[294,3.9251],[296,3.8379],[297,3.8046],[298,3.7822],[299,3.7678],[300,3.7679],[301,3.7782],[302,3.8247],[304,3.9814],[305,4.048],[306,4.1029],[307,4.1392],[308,4.1658],[309,4.2044],[310,4.2733],[311,4.3516],[313,4.4715],[315,4.561],[316,4.5836],[317,4.5801],[319,4.5448],[323,4.4268],[338,4.0726],[347,3.9234],[348,3.9217],[350,3.9487],[353,4.0065],[355,4.007],[356,4.0294],[361,4.182],[363,4.2279],[366,4.2624],[371,4.2635],[372,4.2762],[373,4.3026],[374,4.3437],[378,4.5826],[379,4.6283],[380,4.6556],[381,4.6547],[382,4.6135],[383,4.5188],[384,4.3208],[385,4.025],[386,3.7019],[387,3.426],[388,3.2233],[389,3.0698],[390,2.966],[391,2.8954],[392,2.8412],[393,2.8025],[394,2.776],[395,2.7582],[397,2.7387],[400,2.7346],[401,2.7476],[402,2.779],[403,2.8234],[405,2.9261],[406,2.9727],[407,3.011],[409,3.0686],[414,3.1572],[417,3.1905],[422,3.1959]]
       }
       thigh_r_fk_ctrl.Scale.X {type=MovieSceneScriptingFloatChannel, default=1, const [1,423]}
       thigh_r_fk_ctrl.Scale.Y {type=MovieSceneScriptingFloatChannel, default=1, const [1,423]}
       thigh_r_fk_ctrl.Scale.Z {type=MovieSceneScriptingFloatChannel, default=1, const [1,423]}
       calf_r_fk_ctrl.Location.X {
        type=MovieSceneScriptingFloatChannel
        default=-1.5504
        thinned_from=423
        keys [[0,-0.1059],[1,-0.1059],[4,-0.1066],[6,-0.1066],[7,-0.1061],[11,-0.1032],[12,-0.1017],[14,-0.0963],[15,-0.095],[16,-0.0967],[19,-0.1088],[20,-0.115],[21,-0.1238],[22,-0.1345],[23,-0.1479],[24,-0.1644],[25,-0.1821],[26,-0.198],[27,-0.2117],[28,-0.2242],[29,-0.2349],[30,-0.2428],[31,-0.2486],[32,-0.2533],[33,-0.257],[35,-0.2623],[38,-0.2677],[42,-0.2725],[44,-0.2735],[48,-0.2727],[56,-0.2664],[58,-0.2637],[60,-0.2589],[62,-0.2529],[64,-0.2456],[65,-0.2427],[66,-0.2382],[67,-0.2309],[68,-0.2231],[69,-0.2166],[70,-0.212],[71,-0.2094],[72,-0.2091],[73,-0.2112],[74,-0.2158],[75,-0.2219],[76,-0.2269],[77,-0.2282],[78,-0.227],[79,-0.2273],[81,-0.2295],[82,-0.2321],[84,-0.24],[85,-0.243],[86,-0.2447],[88,-0.2457],[91,-0.2457],[92,-0.2461],[93,-0.2472],[95,-0.2511],[104,-0.2768],[110,-0.2901],[114,-0.2972],[116,-0.2996],[120,-0.3005],[123,-0.3003],[126,-0.3024],[127,-0.3069],[128,-0.3132],[129,-0.3203],[130,-0.3236],[131,-0.3205],[132,-0.2777],[133,-0.2057],[134,-0.0762],[135,0.0715],[136,0.1679],[137,0.2],[138,0.1746],[139,0.1752],[140,0.063],[141,-0.0878],[142,-0.1702],[143,-0.2738],[144,-0.331],[145,-0.3489],[146,-0.3452],[147,-0.3262],[148,-0.2196],[149,-0.1357],[150,-0.0757],[151,-0.0254],[152,0.0096],[153,0.0142],[154,-0.0194],[155,-0.0663],[156,-0.1048],[157,-0.1353],[158,-0.1534],[159,-0.166],[160,-0.1743],[161,-0.1794],[162,-0.1817],[163,-0.1811],[164,-0.1783],[165,-0.1739],[166,-0.1685],[168,-0.1572],[169,-0.1524],[170,-0.1481],[171,-0.1444],[172,-0.1412],[173,-0.1384],[174,-0.1361],[175,-0.1342],[176,-0.1326],[178,-0.1303],[180,-0.1289],[182,-0.1282],[185,-0.1281],[190,-0.1294],[207,-0.1385],[211,-0.1394],[213,-0.1394],[214,-0.1391],[215,-0.1379],[216,-0.1363],[217,-0.1344],[219,-0.1299],[222,-0.122],[224,-0.1174],[225,-0.116],[226,-0.1218],[227,-0.1436],[228,-0.1709],[229,-0.1973],[230,-0.2149],[231,-0.2259],[232,-0.2256],[233,-0.2314],[234,-0.2435],[235,-0.2579],[236,-0.2685],[237,-0.2688],[238,-0.2577],[239,-0.2317],[240,-0.1922],[241,-0.1426],[242,-0.0883],[243,-0.0331],[244,0.0247],[245,0.0711],[246,0.0973],[247,0.1024],[248,0.0878],[249,0.0636],[250,0.0256],[251,-0.0209],[252,-0.0688],[253,-0.1161],[254,-0.157],[255,-0.1883],[256,-0.2015],[257,-0.2029],[259,-0.2045],[260,-0.2048],[263,-0.2039],[266,-0.2015],[270,-0.1965],[272,-0.1947],[274,-0.194],[286,-0.1971],[292,-0.1975],[293,-0.1971],[294,-0.1961],[296,-0.1922],[297,-0.1908],[298,-0.1886],[299,-0.1856],[300,-0.1844],[301,-0.1892],[302,-0.2017],[303,-0.2184],[304,-0.2361],[305,-0.2558],[306,-0.2725],[307,-0.2865],[308,-0.2998],[309,-0.3019],[310,-0.2949],[311,-0.2932],[312,-0.2958],[314,-0.2961],[315,-0.2994],[316,-0.3041],[317,-0.3082],[318,-0.3115],[319,-0.3132],[320,-0.3123],[322,-0.3079],[324,-0.3057],[325,-0.3051],[326,-0.3054],[327,-0.3068],[329,-0.3118],[334,-0.3279],[342,-0.3429],[344,-0.3449],[348,-0.3453],[349,-0.3461],[352,-0.3518],[353,-0.3544],[355,-0.3579],[358,-0.3673],[361,-0.3744],[362,-0.3758],[365,-0.3765],[366,-0.3755],[367,-0.3736],[368,-0.3708],[370,-0.363],[372,-0.3537],[373,-0.3502],[374,-0.3485],[375,-0.3492],[376,-0.3487],[377,-0.3438],[378,-0.3352],[379,-0.3227],[380,-0.3032],[381,-0.2766],[382,-0.2468],[384,-0.1891],[385,-0.1595],[386,-0.1332],[387,-0.1138],[388,-0.104],[389,-0.102],[391,-0.1065],[395,-0.1085],[396,-0.1088],[397,-0.1088],[398,-0.1084],[399,-0.1077],[401,-0.1055],[402,-0.1047],[403,-0.1042],[404,-0.1039],[408,-0.1042],[412,-0.1031],[413,-0.1029],[414,-0.1031],[418,-0.1056],[419,-0.1059],[422,-0.1059]]
       }
       calf_r_fk_ctrl.Location.Y {
        type=MovieSceneScriptingFloatChannel
        default=0.1186
        thinned_from=423
        keys [[0,0.0087],[12,0.0083],[15,0.0078],[19,0.0089],[20,0.0094],[22,0.011],[23,0.0121],[26,0.0163],[28,0.0184],[29,0.0193],[31,0.0204],[42,0.0224],[58,0.0217],[65,0.0199],[66,0.0196],[69,0.0178],[71,0.0172],[73,0.0173],[76,0.0186],[81,0.0188],[85,0.02],[94,0.0204],[115,0.0245],[126,0.0248],[129,0.0263],[130,0.0266],[131,0.0263],[132,0.0227],[133,0.0168],[134,0.0061],[135,-0.0061],[136,-0.0138],[137,-0.0163],[138,-0.0141],[139,-0.0141],[140,-0.005],[141,0.0072],[142,0.0138],[143,0.0223],[144,0.027],[145,0.0285],[146,0.0282],[147,0.0267],[148,0.018],[149,0.0111],[150,0.0061],[151,0.0019],[152,-0.001],[153,-0.0014],[154,0.0014],[155,0.0053],[156,0.0085],[157,0.0111],[158,0.0126],[159,0.0136],[160,0.0143],[161,0.0147],[162,0.0149],[164,0.0146],[171,0.0118],[181,0.0105],[214,0.0114],[225,0.0095],[226,0.01],[227,0.0118],[229,0.0162],[230,0.0176],[231,0.0185],[232,0.0185],[233,0.019],[235,0.0212],[236,0.0221],[237,0.0221],[238,0.0212],[239,0.019],[240,0.0157],[241,0.0115],[243,0.0025],[244,-0.0022],[245,-0.006],[246,-0.0082],[247,-0.0086],[248,-0.0075],[249,-0.0055],[250,-0.0023],[253,0.0095],[254,0.0129],[255,0.0155],[256,0.0165],[294,0.0161],[300,0.0151],[301,0.0155],[302,0.0166],[304,0.0194],[305,0.021],[306,0.0224],[308,0.0246],[309,0.0248],[310,0.0242],[311,0.0241],[314,0.0243],[318,0.0256],[319,0.0257],[325,0.0251],[329,0.0256],[343,0.0283],[350,0.0286],[362,0.0309],[366,0.0309],[370,0.0298],[373,0.0288],[374,0.0286],[376,0.0287],[377,0.0283],[378,0.0275],[379,0.0265],[380,0.0249],[381,0.0227],[385,0.0131],[386,0.0109],[387,0.0093],[388,0.0085],[389,0.0084],[398,0.0089],[422,0.0087]]
       }
       calf_r_fk_ctrl.Location.Z {
        type=MovieSceneScriptingFloatChannel
        default=-0.0325
        thinned_from=423
        keys [[0,0.0001],[133,0.0002],[137,-0.0002],[147,0.0004],[422,0.0001]]
       }
       calf_r_fk_ctrl.Rotation.X {
        type=MovieSceneScriptingFloatChannel
        default=-0.0409
        thinned_from=423
        keys [[0,0.0055],[131,0.0055],[133,0.0046],[134,0.0038],[136,0.0002],[137,-0.0009],[138,-0.0014],[139,-0.0016],[140,-0.001],[142,0.0017],[144,0.0036],[148,0.0055],[149,0.0055],[152,0.0041],[153,0.0039],[156,0.0049],[157,0.0054],[238,0.0055],[241,0.0041],[242,0.0033],[245,0.0016],[246,0.0014],[248,0.0025],[249,0.0033],[253,0.0055],[422,0.0055]]
       }
       calf_r_fk_ctrl.Rotation.Y {
        type=MovieSceneScriptingFloatChannel
        default=1.1988
        thinned_from=423
        keys [[0,0.0049],[131,0.0049],[133,0.0041],[134,0.0034],[136,0.0001],[137,-0.0008],[138,-0.0011],[139,-0.0013],[140,-0.0008],[142,0.0014],[144,0.0032],[148,0.0049],[149,0.0049],[152,0.0036],[153,0.0035],[156,0.0044],[157,0.0049],[238,0.0049],[241,0.0036],[242,0.0028],[245,0.0014],[246,0.0012],[248,0.0021],[249,0.0029],[253,0.0049],[422,0.0049]]
       }
       calf_r_fk_ctrl.Rotation.Z {
        type=MovieSceneScriptingFloatChannel
        default=-4.3747
        thinned_from=423
        keys [[0,-4.9607],[130,-4.9602],[131,-4.9509],[132,-4.459],[133,-4.1132],[134,-3.44],[135,-1.8504],[136,-0.1492],[137,0.8098],[138,1.1915],[139,1.3475],[140,0.8823],[141,-0.2661],[142,-1.4838],[143,-2.4176],[144,-3.2266],[145,-3.6439],[146,-4.0761],[147,-4.4703],[148,-4.9607],[149,-4.9607],[150,-4.4676],[151,-4.0424],[152,-3.6958],[153,-3.5131],[154,-3.8741],[155,-4.1884],[156,-4.4324],[157,-4.9347],[159,-4.9605],[237,-4.9602],[238,-4.953],[239,-4.5162],[240,-4.1827],[241,-3.6833],[242,-2.9158],[243,-2.3994],[244,-1.8483],[245,-1.4169],[246,-1.2532],[247,-1.6676],[248,-2.2032],[249,-2.9183],[250,-3.4123],[251,-4.0232],[253,-4.9604],[422,-4.9607]]
       }
       calf_r_fk_ctrl.Scale.X {type=MovieSceneScriptingFloatChannel, default=1, const [1,423]}
       calf_r_fk_ctrl.Scale.Y {type=MovieSceneScriptingFloatChannel, default=1, const [1,423]}
       calf_r_fk_ctrl.Scale.Z {type=MovieSceneScriptingFloatChannel, default=1, const [1,423]}
       foot_r_fk_ctrl.Location.X {
        type=MovieSceneScriptingFloatChannel
        default=-1.6949
        thinned_from=423
        keys [[0,-0.0948],[1,-0.0948],[4,-0.0956],[5,-0.0957],[6,-0.0955],[7,-0.0951],[10,-0.0926],[11,-0.0919],[12,-0.0902],[14,-0.0844],[15,-0.083],[16,-0.0848],[19,-0.098],[20,-0.1048],[21,-0.1144],[22,-0.1261],[23,-0.1408],[24,-0.1588],[25,-0.1782],[26,-0.1956],[27,-0.2106],[28,-0.2243],[29,-0.2359],[30,-0.2446],[31,-0.2509],[32,-0.256],[33,-0.2601],[34,-0.2633],[36,-0.2682],[40,-0.2747],[42,-0.277],[44,-0.2782],[48,-0.2773],[56,-0.2704],[58,-0.2674],[60,-0.2622],[62,-0.2557],[64,-0.2477],[65,-0.2444],[66,-0.2395],[67,-0.2316],[68,-0.223],[69,-0.216],[70,-0.2109],[71,-0.208],[72,-0.2077],[73,-0.21],[74,-0.2151],[75,-0.2217],[76,-0.2272],[77,-0.2286],[78,-0.2274],[79,-0.2277],[80,-0.2291],[81,-0.2301],[82,-0.2329],[84,-0.2416],[85,-0.2448],[86,-0.2467],[88,-0.2478],[91,-0.2477],[92,-0.2482],[93,-0.2494],[94,-0.2513],[96,-0.2565],[103,-0.2789],[108,-0.2919],[114,-0.3041],[116,-0.3067],[118,-0.3077],[123,-0.3075],[125,-0.3087],[126,-0.3098],[127,-0.3147],[128,-0.3215],[129,-0.3294],[130,-0.3329],[131,-0.3297],[132,-0.2849],[133,-0.2076],[134,-0.0688],[135,0.0859],[136,0.1841],[137,0.2152],[138,0.1858],[139,0.1858],[140,0.0651],[141,-0.095],[142,-0.1799],[143,-0.2893],[144,-0.3484],[145,-0.3662],[146,-0.3603],[147,-0.3379],[148,-0.2192],[149,-0.1274],[150,-0.064],[151,-0.0107],[152,0.0261],[153,0.0303],[154,-0.0048],[155,-0.0548],[156,-0.0959],[157,-0.1271],[158,-0.1468],[159,-0.1606],[160,-0.1696],[161,-0.1752],[162,-0.1777],[163,-0.1771],[164,-0.1741],[165,-0.1693],[166,-0.1634],[167,-0.157],[168,-0.151],[169,-0.1457],[170,-0.141],[171,-0.137],[172,-0.1334],[173,-0.1304],[174,-0.1279],[175,-0.1258],[176,-0.1241],[177,-0.1227],[178,-0.1215],[180,-0.12],[182,-0.1193],[185,-0.1192],[190,-0.1205],[206,-0.13],[210,-0.1314],[213,-0.1316],[214,-0.1312],[215,-0.1299],[216,-0.1282],[217,-0.1261],[219,-0.1211],[222,-0.1125],[224,-0.1074],[225,-0.1059],[226,-0.1123],[227,-0.1361],[228,-0.166],[229,-0.1948],[230,-0.2141],[231,-0.2261],[232,-0.2257],[233,-0.2321],[234,-0.2454],[235,-0.2611],[236,-0.2727],[237,-0.273],[238,-0.2609],[239,-0.2344],[240,-0.1925],[241,-0.1404],[242,-0.0843],[243,-0.0261],[244,0.0348],[245,0.0836],[246,0.1117],[247,0.119],[248,0.1054],[249,0.0819],[250,0.0424],[251,-0.0059],[252,-0.0562],[253,-0.106],[254,-0.1508],[255,-0.185],[256,-0.1994],[257,-0.2009],[259,-0.2027],[260,-0.203],[262,-0.2026],[265,-0.2005],[267,-0.1982],[270,-0.194],[272,-0.192],[274,-0.1913],[286,-0.1946],[292,-0.195],[293,-0.1947],[294,-0.1936],[296,-0.1893],[297,-0.1877],[298,-0.1853],[299,-0.182],[300,-0.1807],[301,-0.1859],[302,-0.1997],[303,-0.2179],[304,-0.2373],[305,-0.2588],[306,-0.277],[307,-0.2924],[308,-0.3069],[309,-0.3092],[310,-0.3016],[311,-0.2998],[312,-0.3025],[314,-0.3029],[315,-0.3064],[316,-0.3116],[317,-0.3161],[318,-0.3197],[319,-0.3216],[320,-0.3206],[322,-0.3158],[324,-0.3133],[325,-0.3128],[326,-0.3131],[327,-0.3146],[329,-0.3201],[333,-0.3345],[335,-0.3402],[342,-0.3541],[344,-0.3562],[348,-0.3567],[349,-0.3576],[352,-0.3638],[353,-0.3667],[355,-0.3705],[358,-0.3808],[361,-0.3885],[362,-0.3901],[365,-0.3908],[366,-0.3897],[367,-0.3876],[368,-0.3846],[369,-0.3807],[372,-0.3659],[373,-0.3621],[374,-0.3601],[375,-0.3609],[376,-0.3605],[377,-0.3551],[378,-0.3456],[379,-0.332],[380,-0.3107],[381,-0.2815],[382,-0.249],[384,-0.1859],[385,-0.1535],[386,-0.1247],[387,-0.1035],[388,-0.0928],[389,-0.0906],[391,-0.0955],[395,-0.0977],[396,-0.098],[397,-0.098],[398,-0.0976],[399,-0.0968],[401,-0.0944],[402,-0.0935],[403,-0.093],[404,-0.0927],[405,-0.0927],[407,-0.093],[409,-0.0927],[412,-0.0918],[413,-0.0916],[414,-0.0918],[418,-0.0946],[419,-0.0948],[422,-0.0948]]
       }
       foot_r_fk_ctrl.Location.Y {
        type=MovieSceneScriptingFloatChannel
        default=-0.0329
        thinned_from=423
        keys [[0,-0.0004],[131,-0.0015],[133,-0.0009],[135,0.0004],[136,0.0008],[137,0.001],[139,0.0008],[141,-0.0004],[143,-0.0013],[145,-0.0016],[147,-0.0015],[149,-0.0006],[152,0.0001],[154,0],[227,-0.0006],[237,-0.0012],[246,0.0005],[250,0.0002],[255,-0.0008],[379,-0.0015],[389,-0.0004],[422,-0.0004]]
       }
       foot_r_fk_ctrl.Location.Z {
        type=MovieSceneScriptingFloatChannel
        default=-0.9075
        thinned_from=423
        keys [[0,0.0001],[133,0.0002],[137,-0.0003],[147,0.0004],[422,0.0001]]
       }
       foot_r_fk_ctrl.Rotation.X {
        type=MovieSceneScriptingFloatChannel
        default=-1.4665
        thinned_from=423
        keys [[0,-0.2771],[1,-0.2771],[9,-0.2685],[11,-0.2673],[13,-0.2676],[14,-0.2689],[15,-0.2715],[16,-0.2767],[17,-0.2846],[18,-0.2942],[19,-0.3064],[20,-0.3201],[21,-0.3318],[22,-0.34],[23,-0.346],[24,-0.3508],[25,-0.3544],[26,-0.3568],[28,-0.3592],[56,-0.3605],[65,-0.3533],[66,-0.3513],[71,-0.3327],[76,-0.3172],[77,-0.3128],[78,-0.3078],[79,-0.3036],[81,-0.2971],[82,-0.2952],[83,-0.295],[88,-0.2997],[89,-0.2997],[90,-0.3008],[91,-0.3008],[95,-0.3018],[126,-0.294],[127,-0.2931],[128,-0.2914],[130,-0.2861],[132,-0.2799],[133,-0.2784],[134,-0.285],[135,-0.2948],[136,-0.3039],[137,-0.3189],[138,-0.3433],[139,-0.3474],[140,-0.3315],[142,-0.2961],[143,-0.2744],[144,-0.2566],[145,-0.2418],[146,-0.2299],[147,-0.2195],[148,-0.212],[149,-0.2071],[150,-0.2031],[151,-0.1978],[154,-0.1839],[155,-0.1804],[156,-0.1779],[157,-0.1774],[168,-0.1872],[183,-0.1925],[185,-0.1925],[213,-0.1911],[215,-0.1901],[221,-0.1853],[223,-0.1846],[225,-0.1849],[226,-0.1867],[227,-0.19],[228,-0.1944],[229,-0.1998],[232,-0.2193],[233,-0.2262],[234,-0.2335],[236,-0.2489],[237,-0.2538],[241,-0.2653],[243,-0.2701],[245,-0.2694],[246,-0.2687],[247,-0.2703],[249,-0.2776],[250,-0.2804],[251,-0.2821],[252,-0.2824],[256,-0.2786],[268,-0.275],[273,-0.2752],[286,-0.2797],[292,-0.2796],[294,-0.2778],[295,-0.2763],[297,-0.2714],[298,-0.2671],[300,-0.255],[301,-0.2495],[302,-0.2457],[303,-0.2438],[304,-0.2437],[305,-0.2457],[306,-0.2497],[308,-0.2603],[309,-0.2667],[311,-0.2823],[312,-0.2893],[313,-0.2957],[314,-0.3012],[315,-0.3059],[316,-0.3097],[318,-0.3143],[320,-0.3171],[328,-0.3217],[335,-0.3213],[341,-0.3171],[348,-0.3066],[349,-0.3055],[351,-0.3053],[352,-0.3062],[354,-0.3095],[355,-0.312],[356,-0.316],[359,-0.3305],[361,-0.3413],[362,-0.346],[363,-0.3497],[365,-0.3548],[367,-0.358],[369,-0.3596],[371,-0.3602],[373,-0.3629],[374,-0.3654],[377,-0.3764],[378,-0.3794],[379,-0.3811],[380,-0.3809],[381,-0.3789],[382,-0.3751],[383,-0.3693],[384,-0.3601],[385,-0.3476],[386,-0.3343],[387,-0.3224],[388,-0.3124],[389,-0.3035],[390,-0.296],[391,-0.2897],[392,-0.2843],[393,-0.2795],[394,-0.2755],[395,-0.2722],[397,-0.2674],[399,-0.2643],[400,-0.2633],[401,-0.2629],[403,-0.2639],[415,-0.2754],[418,-0.2769],[422,-0.2771]]
       }
       foot_r_fk_ctrl.Rotation.Y {
        type=MovieSceneScriptingFloatChannel
        default=-4.3339
        thinned_from=423
        keys [[0,-4.8127],[1,-4.8127],[10,-4.6882],[12,-4.6809],[13,-4.6856],[14,-4.7026],[15,-4.7377],[16,-4.8066],[17,-4.9122],[18,-5.0414],[19,-5.2044],[20,-5.3887],[21,-5.5461],[22,-5.6561],[23,-5.7375],[24,-5.8014],[25,-5.8501],[27,-5.9033],[56,-5.9317],[65,-5.8346],[66,-5.8084],[71,-5.5567],[76,-5.3472],[79,-5.1646],[81,-5.0758],[82,-5.0512],[83,-5.0482],[88,-5.1098],[89,-5.1098],[90,-5.1253],[91,-5.1253],[98,-5.1289],[126,-5.0342],[127,-5.0223],[128,-5.0001],[132,-4.85],[133,-4.8329],[134,-4.9289],[136,-5.2073],[137,-5.4144],[138,-5.7397],[139,-5.7873],[140,-5.5633],[142,-5.0625],[143,-4.7646],[144,-4.5191],[145,-4.3173],[146,-4.1553],[147,-4.0118],[148,-3.9074],[149,-3.8409],[150,-3.7892],[151,-3.7213],[154,-3.5402],[155,-3.4925],[156,-3.4579],[157,-3.4484],[170,-3.5925],[183,-3.6442],[213,-3.6256],[222,-3.5439],[225,-3.5449],[226,-3.5688],[227,-3.6113],[228,-3.6687],[229,-3.7398],[233,-4.0814],[236,-4.3769],[237,-4.4399],[243,-4.6837],[245,-4.691],[246,-4.6871],[247,-4.7098],[249,-4.8042],[250,-4.8406],[251,-4.8631],[252,-4.8656],[256,-4.8155],[272,-4.7685],[289,-4.8319],[292,-4.8303],[294,-4.8071],[296,-4.7577],[297,-4.7218],[298,-4.6639],[300,-4.502],[301,-4.4293],[302,-4.3778],[303,-4.3538],[304,-4.3531],[305,-4.3804],[306,-4.4365],[308,-4.5808],[309,-4.6686],[311,-4.8786],[313,-5.0602],[314,-5.1342],[315,-5.1971],[316,-5.2477],[318,-5.3106],[321,-5.36],[330,-5.4141],[337,-5.3911],[342,-5.3338],[348,-5.2075],[350,-5.1885],[352,-5.2029],[354,-5.2486],[355,-5.281],[359,-5.5283],[362,-5.7338],[363,-5.7829],[365,-5.8502],[368,-5.9056],[371,-5.9215],[373,-5.9579],[374,-5.9921],[377,-6.1431],[378,-6.184],[379,-6.2072],[380,-6.2058],[381,-6.1795],[382,-6.1278],[383,-6.0504],[384,-5.9268],[386,-5.5804],[387,-5.421],[388,-5.2869],[389,-5.1672],[390,-5.0662],[391,-4.9825],[393,-4.846],[394,-4.7915],[396,-4.7124],[398,-4.66],[400,-4.6288],[402,-4.6265],[417,-4.806],[422,-4.8127]]
       }
       foot_r_fk_ctrl.Rotation.Z {
        type=MovieSceneScriptingFloatChannel
        default=2.1737
        thinned_from=423
        keys [[0,2.2437],[5,2.2493],[8,2.2368],[11,2.2078],[12,2.1931],[16,2.1086],[17,2.1094],[18,2.1504],[19,2.2487],[20,2.3891],[21,2.5413],[22,2.6833],[23,2.8092],[24,2.9123],[25,2.9935],[26,3.0611],[27,3.1162],[28,3.1587],[29,3.1891],[31,3.2203],[36,3.2563],[58,3.2229],[65,3.1442],[66,3.1587],[67,3.2015],[78,3.8336],[79,3.8816],[81,3.9586],[82,3.9874],[84,4.015],[89,4.0192],[129,4.0077],[130,3.9972],[131,3.973],[132,3.7044],[133,3.4714],[134,2.9497],[135,1.9058],[136,0.7325],[137,-0.2186],[138,-1.1057],[139,-1.8335],[140,-2.0238],[141,-1.7587],[142,-1.3773],[143,-0.9918],[144,-0.6743],[145,-0.5354],[146,-0.3879],[147,-0.2503],[148,-0.081],[149,-0.1431],[150,-0.4269],[151,-0.648],[152,-0.8166],[153,-0.8821],[154,-0.6714],[155,-0.4912],[156,-0.3496],[157,-0.1021],[158,-0.0987],[159,-0.102],[160,-0.1117],[161,-0.1205],[163,-0.1373],[165,-0.1544],[166,-0.1625],[167,-0.1702],[169,-0.1832],[170,-0.189],[171,-0.1943],[172,-0.199],[173,-0.2032],[175,-0.2102],[176,-0.2132],[179,-0.2206],[182,-0.2263],[183,-0.2277],[185,-0.2277],[186,-0.2311],[190,-0.2336],[195,-0.2332],[197,-0.2327],[211,-0.2216],[213,-0.2216],[214,-0.2193],[215,-0.2155],[216,-0.2106],[219,-0.1946],[220,-0.19],[221,-0.1868],[222,-0.1851],[223,-0.1847],[224,-0.1855],[225,-0.1882],[226,-0.2016],[227,-0.2248],[228,-0.2554],[229,-0.2945],[230,-0.3424],[231,-0.3983],[233,-0.5669],[234,-0.658],[235,-0.7406],[236,-0.8057],[237,-0.8457],[238,-0.8607],[239,-1.0547],[240,-1.1774],[241,-1.3592],[242,-1.6535],[243,-1.8103],[244,-1.9258],[245,-1.9604],[246,-1.8614],[247,-1.5222],[248,-1.1678],[249,-0.7226],[250,-0.3587],[251,0.0557],[252,0.3761],[253,0.6672],[254,0.7227],[255,0.7642],[256,0.7926],[257,0.8107],[258,0.8211],[262,0.8398],[266,0.8499],[271,0.8544],[272,0.8537],[273,0.8554],[276,0.8676],[286,0.8816],[287,0.8816],[290,0.8879],[292,0.8879],[293,0.8976],[294,0.9054],[296,0.9258],[297,0.9283],[298,0.9237],[299,0.9162],[300,0.9051],[301,0.9036],[302,0.9098],[303,0.9208],[304,0.9634],[305,1.0474],[306,1.1616],[307,1.2967],[308,1.4403],[309,1.5759],[310,1.6854],[312,1.8628],[313,1.9426],[314,2.0083],[317,2.1608],[319,2.2482],[320,2.2859],[321,2.3186],[322,2.3449],[323,2.3645],[325,2.3897],[329,2.415],[334,2.4178],[336,2.4068],[341,2.3439],[343,2.3128],[344,2.302],[345,2.2982],[346,2.3033],[348,2.3297],[349,2.3509],[350,2.3815],[351,2.4268],[352,2.5025],[353,2.613],[354,2.7531],[356,3.0573],[357,3.1978],[360,3.5592],[361,3.6719],[362,3.7694],[363,3.8425],[364,3.8948],[365,3.937],[366,3.9702],[368,4.0118],[370,4.0243],[371,4.0183],[372,3.9996],[373,3.9665],[374,3.9169],[375,3.8487],[376,3.7654],[377,3.6701],[378,3.563],[379,3.4448],[380,3.3116],[382,3.0278],[383,2.9069],[384,2.8164],[385,2.7515],[387,2.6535],[389,2.5339],[390,2.469],[391,2.4097],[392,2.3607],[393,2.3206],[394,2.2894],[395,2.2672],[396,2.2546],[397,2.25],[401,2.2639],[407,2.2439],[414,2.2536],[422,2.2437]]
       }
       foot_r_fk_ctrl.Scale.X {type=MovieSceneScriptingFloatChannel, default=1, const [1,423]}
       foot_r_fk_ctrl.Scale.Y {type=MovieSceneScriptingFloatChannel, default=1, const [1,423]}
       foot_r_fk_ctrl.Scale.Z {type=MovieSceneScriptingFloatChannel, default=1, const [1,423]}
       ball_r_fk_ctrl.Location.X {type=MovieSceneScriptingFloatChannel, default=0, const [0,423]}
       ball_r_fk_ctrl.Location.Y {type=MovieSceneScriptingFloatChannel, default=0, const [0,423]}
       ball_r_fk_ctrl.Location.Z {type=MovieSceneScriptingFloatChannel, default=0, const [0,423]}
       ball_r_fk_ctrl.Rotation.X {type=MovieSceneScriptingFloatChannel, default=0, const [0,423]}
       ball_r_fk_ctrl.Rotation.Y {type=MovieSceneScriptingFloatChannel, default=0, const [0,423]}
       ball_r_fk_ctrl.Rotation.Z {type=MovieSceneScriptingFloatChannel, default=0, const [0,423]}
       ball_r_fk_ctrl.Scale.X {type=MovieSceneScriptingFloatChannel, default=1, const [1,423]}
       ball_r_fk_ctrl.Scale.Y {type=MovieSceneScriptingFloatChannel, default=1, const [1,423]}
       ball_r_fk_ctrl.Scale.Z {type=MovieSceneScriptingFloatChannel, default=1, const [1,423]}
       foot_r_ik_ctrl.Location.X {
        type=MovieSceneScriptingFloatChannel
        default=0
        thinned_from=423
        keys [[0,0],[5,0],[6,0.0081],[7,0],[44,0],[46,0.0003],[47,0],[52,0],[53,-0.0114],[54,0],[55,-0.0084],[56,0],[88,0],[89,-0.0002],[90,0.0001],[91,-0.0045],[92,0.002],[93,0],[119,0],[120,0.0001],[180,0],[181,0.012],[182,0],[183,0.0086],[184,0],[185,0.0057],[186,0],[187,0.0034],[188,0],[189,0.003],[190,0],[191,0.0027],[192,0.0051],[193,0],[194,0.002],[196,0.0056],[197,0],[198,0.0018],[201,0.0056],[202,0],[207,0.0045],[208,0.0052],[209,0],[210,0.0005],[211,0.0015],[212,0.0018],[213,0.0015],[214,0],[215,-0.0004],[216,0],[217,-0.001],[218,0],[286,0],[288,-0.0029],[289,-0.0025],[290,0],[346,0],[347,-0.0002],[369,0],[370,0.0112],[371,0],[418,0],[419,0.0008],[422,0.0008]]
       }
       foot_r_ik_ctrl.Location.Y {type=MovieSceneScriptingFloatChannel, default=-17.1651, thinned_from=423, keys [[0,-17.1652],[216,-17.1651],[217,-17.184],[422,-17.1627]]}
       foot_r_ik_ctrl.Location.Z {
        type=MovieSceneScriptingFloatChannel
        default=0.0002
        thinned_from=423
        keys [[0,0.0002],[1,0.001],[2,0.0002],[5,0.0002],[6,0.002],[7,0.0002],[44,0.0002],[45,0.0003],[46,0.0001],[52,0.0002],[53,-0.0024],[54,0.0002],[55,-0.0017],[56,0.0002],[90,0.0001],[91,-0.0013],[92,-0.0007],[93,0.0002],[119,0.0002],[120,0],[121,0.0002],[122,0],[123,0.0002],[130,0.0002],[131,0],[132,0.0003],[133,0.0002],[135,0.0005],[136,0.0002],[140,0.0002],[142,0.0004],[144,0.0002],[145,0.0005],[146,0.0002],[147,0.0003],[149,0.0002],[150,0.0003],[151,0.0002],[152,0.0005],[154,0.0002],[156,0.0003],[157,-0.0002],[159,0.0001],[180,0.0001],[181,-0.0014],[182,0.0002],[183,-0.0012],[184,0.0001],[185,-0.0007],[186,0.0002],[187,-0.0006],[188,0.0002],[189,-0.0004],[190,0.0002],[192,-0.0005],[193,0.0002],[196,0],[201,0.001],[202,0.0002],[208,0.0026],[209,0.0002],[212,0.0015],[213,0.0016],[214,0.0002],[215,0.0022],[216,0.0002],[217,0.0036],[218,0.0002],[237,0.0002],[238,0],[239,0.0004],[240,0.0002],[241,0.0005],[242,0.0002],[243,0.0002],[244,0.0005],[245,0.0003],[249,0.0002],[250,0.0003],[251,0.0002],[252,0.0003],[287,0],[288,0.0002],[289,0.001],[290,0.0003],[291,0.0003],[292,0],[369,0.0002],[370,0.0025],[371,0.0002],[418,0.0002],[419,-0.0002],[422,-0.0002]]
       }
       foot_r_ik_ctrl.Rotation.X {
        type=MovieSceneScriptingFloatChannel
        default=0
        thinned_from=423
        keys [[0,0],[1,-0.0011],[2,0],[5,0],[6,-0.0035],[7,0],[44,0],[46,0.0011],[47,0],[52,0],[53,0.0058],[54,0],[55,0.0035],[56,0],[88,0],[89,0.0022],[90,-0.0002],[91,0.004],[92,-0.0054],[93,0],[119,0],[120,-0.0016],[121,0],[122,-0.0005],[123,-0.0008],[125,-0.0008],[126,0],[130,0],[131,-0.0005],[132,0.0005],[133,0.0001],[135,0.0012],[136,0.0003],[140,0.0003],[141,0.0008],[142,0.0008],[143,0],[144,0],[145,0.0011],[146,0.0001],[147,0.0006],[148,0],[149,0],[150,0.0006],[151,0.0001],[152,0.0012],[153,0.0008],[154,0.0001],[155,0.0001],[156,0.0005],[157,-0.0014],[158,-0.0005],[159,0],[180,0],[181,-0.0083],[182,0],[183,-0.0061],[184,0.0013],[185,-0.0015],[186,0],[187,-0.0023],[188,0],[189,-0.0018],[190,0],[191,-0.0015],[192,-0.0027],[193,0],[194,-0.0008],[196,-0.0018],[197,0],[198,-0.0005],[202,0],[207,0.0026],[208,0.0033],[209,0],[210,0.0009],[211,0.0015],[212,0.0017],[213,0.0022],[214,0],[215,0.0051],[216,0],[217,0.0087],[218,0],[237,0],[238,-0.0004],[239,0.0007],[240,0.0002],[241,0.0012],[242,0.0002],[243,0.0002],[244,0.0012],[245,0.0004],[247,0.0004],[248,0],[249,0],[250,0.0006],[251,0.0001],[252,0.0006],[253,0],[286,0],[287,0.0002],[288,0.0022],[290,0],[291,0],[292,-0.0055],[293,0],[346,0],[347,0.0003],[348,0],[369,0],[370,-0.0034],[371,0],[418,0],[419,-0.0004],[422,-0.0004]]
       }
       foot_r_ik_ctrl.Rotation.Y {
        type=MovieSceneScriptingFloatChannel
        default=0
        thinned_from=423
        keys [[0,0],[1,-0.0011],[2,0],[5,0],[6,0.0086],[7,0],[44,0],[46,0.0012],[47,0],[52,0],[53,-0.0099],[54,0],[55,-0.0072],[56,0],[88,0],[89,0.0085],[90,0],[91,0.0003],[92,-0.003],[93,0],[119,0],[120,-0.0036],[121,0],[123,-0.0034],[125,-0.0034],[126,0],[180,0],[181,-0.0022],[182,0],[183,-0.0025],[184,0.0019],[185,0.0011],[186,0],[187,-0.0021],[188,0],[189,-0.0015],[190,0],[191,-0.0009],[192,-0.0016],[193,0],[194,-0.0001],[195,0],[196,0.0003],[197,0],[199,0.0016],[201,0.0038],[202,0],[207,0.0075],[208,0.0092],[209,0],[211,0.0032],[212,0.0029],[213,0.002],[214,0],[215,0.0062],[216,0],[217,0.0105],[218,0],[286,0],[288,-0.0008],[289,0.0029],[290,0],[291,0],[292,-0.0067],[293,0],[369,0],[370,0.0086],[371,0],[418,0],[419,-0.0015],[422,-0.0015]]
       }
       foot_r_ik_ctrl.Rotation.Z {type=MovieSceneScriptingFloatChannel, default=-61.8415, thinned_from=423, keys [[0,-61.8415],[422,-61.8363]]}
       foot_r_ik_ctrl.Scale.X {type=MovieSceneScriptingFloatChannel, default=1, const [1,423]}
       foot_r_ik_ctrl.Scale.Y {type=MovieSceneScriptingFloatChannel, default=1, const [1,423]}
       foot_r_ik_ctrl.Scale.Z {type=MovieSceneScriptingFloatChannel, default=1, const [1,423]}
       ball_r_ik_ctrl.Location.X {
        type=MovieSceneScriptingFloatChannel
        default=0
        thinned_from=423
        keys [[0,0],[118,0],[119,-0.0003],[120,0],[123,0],[124,0.0004],[125,0.0004],[126,0],[290,0],[291,-0.0004],[292,0],[422,0]]
       }
       ball_r_ik_ctrl.Location.Y {
        type=MovieSceneScriptingFloatChannel
        default=0
        thinned_from=423
        keys [[0,0],[118,0],[119,0.0001],[277,0],[286,-0.0001],[290,0],[291,0.0003],[292,0],[422,0]]
       }
       ball_r_ik_ctrl.Location.Z {
        type=MovieSceneScriptingFloatChannel
        default=0
        thinned_from=423
        keys [[0,0],[118,0],[119,0.0003],[120,0],[123,0],[125,-0.0002],[290,0],[291,0.0005],[292,0],[422,0]]
       }
       ball_r_ik_ctrl.Rotation.X {
        type=MovieSceneScriptingFloatChannel
        default=0
        thinned_from=423
        keys [[0,0],[118,0],[119,-0.0008],[120,0],[123,0],[124,0.0008],[125,0.0008],[126,0],[290,0],[291,-0.0024],[292,0],[422,0]]
       }
       ball_r_ik_ctrl.Rotation.Y {
        type=MovieSceneScriptingFloatChannel
        default=0
        thinned_from=423
        keys [[0,0],[118,0],[119,-0.0019],[120,0],[123,0],[124,0.0034],[125,0.0034],[126,0],[290,0],[291,-0.0027],[292,0],[422,0]]
       }
       ball_r_ik_ctrl.Rotation.Z {type=MovieSceneScriptingFloatChannel, default=0, thinned_from=423, keys [[0,0],[118,0],[119,0.0001],[290,0],[291,0.0005],[292,0],[422,0]]}
       ball_r_ik_ctrl.Scale.X {type=MovieSceneScriptingFloatChannel, default=1, const [1,423]}
       ball_r_ik_ctrl.Scale.Y {type=MovieSceneScriptingFloatChannel, default=1, const [1,423]}
       ball_r_ik_ctrl.Scale.Z {type=MovieSceneScriptingFloatChannel, default=1, const [1,423]}
       foot_bk1_r_ctrl.Location.X {type=MovieSceneScriptingFloatChannel, default=0, const [0,423]}
       foot_bk1_r_ctrl.Location.Y {type=MovieSceneScriptingFloatChannel, default=0, const [0,423]}
       foot_bk1_r_ctrl.Location.Z {type=MovieSceneScriptingFloatChannel, default=0, const [0,423]}
       foot_bk1_r_ctrl.Rotation.X {type=MovieSceneScriptingFloatChannel, default=0, const [0,423]}
       foot_bk1_r_ctrl.Rotation.Y {type=MovieSceneScriptingFloatChannel, default=0, const [0,423]}
       foot_bk1_r_ctrl.Rotation.Z {type=MovieSceneScriptingFloatChannel, default=0, const [0,423]}
       foot_bk1_r_ctrl.Scale.X {type=MovieSceneScriptingFloatChannel, default=1, const [1,423]}
       foot_bk1_r_ctrl.Scale.Y {type=MovieSceneScriptingFloatChannel, default=1, const [1,423]}
       foot_bk1_r_ctrl.Scale.Z {type=MovieSceneScriptingFloatChannel, default=1, const [1,423]}
       foot_roll_r_ctrl.Location.X {type=MovieSceneScriptingFloatChannel, default=0, const [0,423]}
       foot_roll_r_ctrl.Location.Y {type=MovieSceneScriptingFloatChannel, default=0, const [0,423]}
       foot_roll_r_ctrl.Location.Z {type=MovieSceneScriptingFloatChannel, default=0, const [0,423]}
       foot_roll_r_ctrl.Rotation.X {type=MovieSceneScriptingFloatChannel, default=0, const [0,423]}
       foot_roll_r_ctrl.Rotation.Y {type=MovieSceneScriptingFloatChannel, default=0, const [0,423]}
       foot_roll_r_ctrl.Rotation.Z {type=MovieSceneScriptingFloatChannel, default=0, const [0,423]}
       foot_roll_r_ctrl.Scale.X {type=MovieSceneScriptingFloatChannel, default=1, const [1,423]}
       foot_roll_r_ctrl.Scale.Y {type=MovieSceneScriptingFloatChannel, default=1, const [1,423]}
       foot_roll_r_ctrl.Scale.Z {type=MovieSceneScriptingFloatChannel, default=1, const [1,423]}
       roll_blend_r_ctrl {type=MovieSceneScriptingFloatChannel, default=1}
       tip_r_ctrl.Location.X {type=MovieSceneScriptingFloatChannel, default=0, const [0,423]}
       tip_r_ctrl.Location.Y {type=MovieSceneScriptingFloatChannel, default=0, const [0,423]}
       tip_r_ctrl.Location.Z {type=MovieSceneScriptingFloatChannel, default=0, const [0,423]}
       tip_r_ctrl.Rotation.X {type=MovieSceneScriptingFloatChannel, default=0, const [0,423]}
       tip_r_ctrl.Rotation.Y {type=MovieSceneScriptingFloatChannel, default=0, const [0,423]}
       tip_r_ctrl.Rotation.Z {type=MovieSceneScriptingFloatChannel, default=0, const [0,423]}
       tip_r_ctrl.Scale.X {type=MovieSceneScriptingFloatChannel, default=1, const [1,423]}
       tip_r_ctrl.Scale.Y {type=MovieSceneScriptingFloatChannel, default=1, const [1,423]}
       tip_r_ctrl.Scale.Z {type=MovieSceneScriptingFloatChannel, default=1, const [1,423]}
       heel_r_ctrl.Location.X {type=MovieSceneScriptingFloatChannel, default=0, const [0,423]}
       heel_r_ctrl.Location.Y {type=MovieSceneScriptingFloatChannel, default=0, const [0,423]}
       heel_r_ctrl.Location.Z {type=MovieSceneScriptingFloatChannel, default=0, const [0,423]}
       heel_r_ctrl.Rotation.X {type=MovieSceneScriptingFloatChannel, default=0, const [0,423]}
       heel_r_ctrl.Rotation.Y {type=MovieSceneScriptingFloatChannel, default=0, const [0,423]}
       heel_r_ctrl.Rotation.Z {type=MovieSceneScriptingFloatChannel, default=0, const [0,423]}
       heel_r_ctrl.Scale.X {type=MovieSceneScriptingFloatChannel, default=1, const [1,423]}
       heel_r_ctrl.Scale.Y {type=MovieSceneScriptingFloatChannel, default=1, const [1,423]}
       heel_r_ctrl.Scale.Z {type=MovieSceneScriptingFloatChannel, default=1, const [1,423]}
       leg_r_softness {type=MovieSceneScriptingFloatChannel, default=0, const [0,423]}
       leg_r_pv_ik_ctrl.Location.X {
        type=MovieSceneScriptingFloatChannel
        default=0.1761
        thinned_from=423
        keys [[0,0.2485],[1,0.2485],[5,0.246],[8,0.2463],[10,0.2478],[11,0.2491],[12,0.2509],[13,0.2535],[14,0.2567],[15,0.2607],[16,0.2653],[17,0.2685],[18,0.2673],[19,0.2594],[20,0.2462],[22,0.2134],[23,0.198],[24,0.1852],[25,0.1749],[26,0.1662],[27,0.1589],[28,0.1531],[29,0.149],[30,0.1463],[32,0.1432],[34,0.1409],[35,0.1401],[37,0.1394],[46,0.1403],[53,0.142],[58,0.1442],[59,0.1452],[64,0.1528],[65,0.1537],[66,0.1512],[67,0.1445],[68,0.136],[70,0.1179],[71,0.1092],[74,0.0847],[75,0.0767],[76,0.069],[78,0.0523],[79,0.0455],[80,0.04],[81,0.0348],[82,0.0308],[83,0.0286],[84,0.0274],[85,0.0269],[87,0.0269],[88,0.0272],[89,0.0272],[90,0.0279],[91,0.0279],[92,0.0287],[96,0.0301],[99,0.0301],[103,0.0288],[107,0.0282],[117,0.0274],[119,0.0271],[120,0.0271],[121,0.0269],[126,0.0269],[127,0.0273],[129,0.0277],[130,0.0287],[131,0.0305],[132,0.0183],[133,0.0153],[134,0.0197],[135,0.0062],[136,0.0007],[137,0.0423],[138,0.1415],[139,0.2342],[140,0.2978],[141,0.3597],[142,0.4095],[144,0.4467],[145,0.4547],[146,0.4649],[147,0.4739],[148,0.4894],[149,0.4925],[150,0.4809],[151,0.4667],[152,0.4534],[153,0.4419],[156,0.4495],[157,0.4625],[170,0.4829],[183,0.4903],[214,0.488],[221,0.48],[224,0.4793],[225,0.4799],[226,0.4828],[227,0.488],[228,0.4949],[229,0.5037],[230,0.5143],[231,0.5262],[233,0.5569],[235,0.5906],[236,0.6052],[237,0.6143],[238,0.6184],[240,0.5966],[241,0.5785],[242,0.5507],[243,0.5271],[245,0.4581],[246,0.43],[247,0.4253],[248,0.4309],[249,0.4412],[250,0.4411],[251,0.4442],[252,0.4457],[253,0.4499],[254,0.4421],[255,0.4362],[256,0.4322],[257,0.4295],[264,0.4233],[291,0.4203],[292,0.4207],[294,0.4174],[298,0.4084],[303,0.3945],[304,0.3893],[305,0.3804],[306,0.3689],[307,0.3554],[308,0.3407],[309,0.3274],[310,0.3176],[313,0.2948],[314,0.2887],[316,0.2789],[318,0.2682],[320,0.2588],[321,0.2548],[322,0.2515],[323,0.2492],[324,0.2475],[328,0.2439],[334,0.243],[335,0.2432],[341,0.2511],[343,0.2542],[344,0.2551],[345,0.2549],[346,0.2536],[348,0.2489],[349,0.2458],[350,0.2416],[351,0.2357],[352,0.2261],[353,0.2123],[354,0.1944],[355,0.1748],[356,0.1561],[357,0.1386],[358,0.1232],[359,0.1083],[360,0.0931],[361,0.0786],[362,0.066],[363,0.0564],[364,0.0496],[365,0.0439],[366,0.0396],[367,0.0363],[368,0.034],[369,0.0327],[370,0.0323],[371,0.0332],[372,0.0359],[373,0.0407],[374,0.0479],[375,0.0579],[376,0.0702],[377,0.0844],[378,0.1001],[379,0.1174],[380,0.1365],[381,0.1567],[382,0.1757],[383,0.1913],[384,0.2014],[385,0.2065],[387,0.2117],[388,0.2159],[389,0.2211],[391,0.2322],[392,0.2364],[393,0.2397],[394,0.2421],[395,0.2436],[396,0.2441],[398,0.243],[400,0.2408],[401,0.2403],[402,0.2405],[407,0.245],[414,0.2462],[418,0.2483],[422,0.2485]]
       }
       leg_r_pv_ik_ctrl.Location.Y {
        type=MovieSceneScriptingFloatChannel
        default=-1.7104
        thinned_from=423
        keys [[0,-1.9548],[130,-1.9498],[131,-1.9461],[132,-1.749],[133,-1.611],[134,-1.3436],[135,-0.7178],[136,-0.0575],[137,0.3095],[138,0.4527],[139,0.5096],[140,0.3305],[141,-0.1117],[142,-0.5851],[143,-0.9507],[144,-1.27],[145,-1.4354],[146,-1.6074],[147,-1.7648],[148,-1.9614],[149,-1.9615],[150,-1.7639],[151,-1.594],[152,-1.4559],[153,-1.3831],[154,-1.5265],[155,-1.6516],[156,-1.7489],[157,-1.9502],[159,-1.9607],[237,-1.9653],[238,-1.9625],[239,-1.7874],[240,-1.6541],[241,-1.455],[242,-1.1506],[243,-0.9467],[244,-0.7298],[245,-0.5605],[246,-0.4959],[247,-0.6572],[248,-0.8667],[249,-1.1481],[250,-1.3432],[251,-1.5857],[253,-1.9601],[422,-1.9548]]
       }
       leg_r_pv_ik_ctrl.Location.Z {
        type=MovieSceneScriptingFloatChannel
        default=-1.5597
        thinned_from=423
        keys [[0,-1.7671],[130,-1.7875],[131,-1.784],[132,-1.6116],[133,-1.4893],[134,-1.2491],[135,-0.6776],[136,-0.0551],[137,0.3046],[138,0.4564],[139,0.5235],[140,0.3555],[141,-0.0646],[142,-0.5064],[143,-0.8435],[144,-1.1327],[145,-1.2812],[146,-1.4341],[147,-1.5729],[148,-1.7445],[149,-1.7442],[150,-1.5713],[151,-1.4219],[152,-1.2998],[153,-1.2357],[154,-1.3642],[155,-1.4756],[156,-1.5618],[157,-1.7379],[159,-1.7466],[237,-1.7326],[238,-1.7297],[239,-1.5766],[240,-1.4595],[241,-1.2836],[242,-1.0114],[243,-0.8276],[244,-0.6312],[245,-0.4775],[246,-0.4203],[247,-0.5718],[248,-0.7657],[249,-1.0227],[250,-1.1997],[251,-1.4172],[253,-1.7481],[378,-1.7809],[422,-1.7671]]
       }
       leg_r_pv_ik_ctrl.Rotation.X {
        type=MovieSceneScriptingFloatChannel
        default=2.1218
        thinned_from=423
        keys [[0,2.5646],[130,2.5933],[131,2.5881],[132,2.3322],[133,2.1515],[134,1.7985],[135,0.9679],[136,0.078],[137,-0.4292],[138,-0.6414],[139,-0.7345],[140,-0.499],[141,0.094],[142,0.7247],[143,1.2107],[144,1.6315],[145,1.8487],[146,2.0734],[147,2.2785],[148,2.5329],[149,2.5324],[150,2.2761],[151,2.0555],[152,1.876],[153,1.7819],[154,1.9705],[155,2.1345],[156,2.2618],[157,2.5228],[159,2.5359],[237,2.5162],[238,2.5118],[239,2.2849],[240,2.1119],[241,1.8532],[242,1.4554],[243,1.1884],[244,0.9046],[245,0.6835],[246,0.6014],[247,0.8188],[248,1.0982],[249,1.4709],[250,1.7293],[251,2.0485],[253,2.538],[378,2.5841],[422,2.5646]]
       }
       leg_r_pv_ik_ctrl.Rotation.Y {
        type=MovieSceneScriptingFloatChannel
        default=-1.2014
        thinned_from=423
        keys [[0,0.0056],[12,0.0058],[17,0.0069],[18,0.0068],[19,0.0063],[20,0.0055],[23,0.0025],[25,0.0011],[28,-0.0003],[59,-0.0008],[65,-0.0003],[66,-0.0004],[78,-0.0066],[82,-0.008],[130,-0.0081],[131,-0.008],[132,-0.0082],[133,-0.008],[134,-0.0068],[135,-0.0045],[136,-0.0004],[137,0.0051],[138,0.0123],[139,0.0183],[140,0.0205],[141,0.0207],[142,0.0207],[143,0.0199],[145,0.0197],[147,0.02],[148,0.0205],[149,0.0207],[154,0.0188],[156,0.0185],[157,0.0189],[216,0.0203],[225,0.02],[227,0.0205],[229,0.0214],[231,0.0228],[236,0.0277],[237,0.0282],[238,0.0284],[240,0.0277],[243,0.0259],[245,0.0237],[247,0.0212],[251,0.0186],[252,0.0182],[253,0.0181],[255,0.0173],[294,0.0161],[303,0.0147],[305,0.0138],[309,0.0106],[313,0.0086],[322,0.0059],[336,0.0054],[345,0.006],[350,0.0052],[352,0.0043],[353,0.0034],[357,-0.0012],[362,-0.0058],[365,-0.0072],[368,-0.0078],[371,-0.0078],[373,-0.0073],[374,-0.0069],[376,-0.0055],[378,-0.0036],[382,0.0011],[383,0.0021],[384,0.0027],[389,0.004],[393,0.0051],[422,0.0057]]
       }
       leg_r_pv_ik_ctrl.Rotation.Z {
        type=MovieSceneScriptingFloatChannel
        default=-1.6101
        thinned_from=423
        keys [[0,-0.3633],[1,-0.3633],[5,-0.3597],[8,-0.3603],[10,-0.3624],[11,-0.3641],[12,-0.3667],[13,-0.3704],[14,-0.375],[15,-0.3806],[16,-0.3872],[17,-0.3917],[18,-0.39],[19,-0.3788],[20,-0.36],[22,-0.3136],[23,-0.2918],[24,-0.2736],[25,-0.2591],[26,-0.2467],[27,-0.2363],[28,-0.2282],[29,-0.2224],[30,-0.2186],[32,-0.2141],[34,-0.2109],[36,-0.2091],[39,-0.2087],[44,-0.2101],[46,-0.2101],[53,-0.2125],[58,-0.2156],[59,-0.217],[64,-0.2277],[65,-0.229],[66,-0.2255],[67,-0.216],[68,-0.2039],[70,-0.1783],[71,-0.1659],[74,-0.1313],[75,-0.12],[76,-0.1091],[77,-0.0973],[78,-0.0853],[79,-0.0757],[80,-0.068],[81,-0.0605],[82,-0.055],[83,-0.0518],[84,-0.05],[85,-0.0494],[87,-0.0494],[88,-0.0498],[89,-0.0498],[90,-0.0508],[91,-0.0508],[92,-0.052],[95,-0.0535],[96,-0.0539],[99,-0.0539],[103,-0.0522],[107,-0.0513],[108,-0.0513],[109,-0.0511],[110,-0.0511],[111,-0.0509],[112,-0.0509],[113,-0.0506],[114,-0.0506],[116,-0.0502],[117,-0.0502],[119,-0.0497],[120,-0.0497],[121,-0.0494],[126,-0.0493],[127,-0.0499],[129,-0.0505],[130,-0.0519],[131,-0.0545],[132,-0.0364],[133,-0.0316],[134,-0.0363],[135,-0.0139],[136,-0.0015],[137,-0.0543],[138,-0.1856],[139,-0.309],[140,-0.3973],[141,-0.4891],[142,-0.5662],[144,-0.6312],[145,-0.6455],[146,-0.6632],[147,-0.679],[148,-0.7046],[149,-0.7092],[150,-0.6888],[151,-0.6654],[152,-0.6441],[153,-0.6266],[155,-0.6393],[156,-0.6442],[157,-0.6664],[170,-0.6954],[183,-0.706],[214,-0.7027],[221,-0.6914],[224,-0.6904],[225,-0.6912],[226,-0.6954],[227,-0.7027],[228,-0.7125],[229,-0.7249],[230,-0.7399],[231,-0.7568],[233,-0.8003],[235,-0.8481],[236,-0.8687],[237,-0.8817],[238,-0.8874],[240,-0.8493],[241,-0.8192],[242,-0.7732],[243,-0.7358],[245,-0.6324],[246,-0.5926],[247,-0.5894],[248,-0.6012],[249,-0.6211],[250,-0.6247],[251,-0.6337],[252,-0.6393],[253,-0.6486],[254,-0.6377],[255,-0.6293],[256,-0.6235],[257,-0.6199],[264,-0.6111],[291,-0.6067],[292,-0.6074],[294,-0.6027],[300,-0.5825],[302,-0.5736],[303,-0.5701],[304,-0.5628],[305,-0.5502],[306,-0.534],[307,-0.5148],[308,-0.494],[309,-0.4751],[310,-0.4612],[313,-0.4289],[314,-0.4203],[316,-0.4064],[318,-0.3912],[320,-0.3779],[321,-0.3722],[322,-0.3676],[323,-0.3643],[324,-0.3619],[328,-0.3568],[334,-0.3555],[335,-0.3559],[341,-0.367],[343,-0.3715],[344,-0.3727],[345,-0.3724],[346,-0.3706],[348,-0.3639],[349,-0.3595],[350,-0.3536],[351,-0.3453],[352,-0.3317],[353,-0.312],[354,-0.2867],[355,-0.259],[356,-0.2324],[357,-0.2077],[358,-0.1857],[359,-0.1647],[360,-0.1432],[361,-0.1227],[362,-0.1048],[363,-0.0912],[364,-0.0815],[365,-0.0736],[366,-0.0673],[367,-0.0627],[368,-0.0594],[369,-0.0576],[370,-0.0571],[371,-0.0583],[372,-0.0621],[373,-0.0689],[374,-0.0791],[375,-0.0933],[376,-0.1108],[377,-0.1308],[378,-0.1531],[379,-0.1776],[380,-0.2046],[381,-0.2332],[382,-0.2601],[383,-0.2822],[384,-0.2965],[385,-0.3037],[387,-0.3112],[388,-0.3172],[389,-0.3245],[391,-0.3402],[392,-0.3462],[393,-0.3509],[394,-0.3542],[395,-0.3563],[396,-0.3571],[398,-0.3555],[400,-0.3524],[401,-0.3517],[402,-0.352],[407,-0.3584],[414,-0.3601],[418,-0.3631],[422,-0.3633]]
       }
       leg_r_pv_ik_ctrl.Scale.X {type=MovieSceneScriptingFloatChannel, default=1, const [1,423]}
       leg_r_pv_ik_ctrl.Scale.Y {type=MovieSceneScriptingFloatChannel, default=1, const [1,423]}
       leg_r_pv_ik_ctrl.Scale.Z {type=MovieSceneScriptingFloatChannel, default=1, const [1,423]}
       upperarm_l_fk_ctrl.Location.X {type=MovieSceneScriptingFloatChannel, default=0, thinned_from=423, keys [[0,0],[422,0]]}
       upperarm_l_fk_ctrl.Location.Y {type=MovieSceneScriptingFloatChannel, default=0, thinned_from=423, keys [[0,0],[422,0]]}
       upperarm_l_fk_ctrl.Location.Z {type=MovieSceneScriptingFloatChannel, default=0, thinned_from=423, keys [[0,0],[422,0]]}
       upperarm_l_fk_ctrl.Rotation.X {
        type=MovieSceneScriptingFloatChannel
        default=49.8134
        thinned_from=423
        keys [[0,52.5525],[12,52.9556],[14,52.8139],[15,52.8226],[17,52.6234],[20,53.4714],[21,54.404],[22,55.7475],[24,58.1046],[25,58.9213],[28,60.7358],[29,61.235],[31,61.866],[35,62.5036],[39,62.8065],[47,63.9403],[60,64.5702],[62,64.9252],[63,65.2985],[69,68.436],[70,69.1622],[74,72.5802],[75,73.1299],[76,73.0098],[77,71.7792],[78,69.7984],[79,67.3903],[81,62.1599],[82,59.7623],[83,57.5978],[84,55.5794],[85,53.861],[86,52.6168],[87,51.7598],[89,50.4142],[91,49.4507],[93,48.8202],[95,48.5082],[97,48.4798],[99,48.7209],[104,49.7102],[116,50.4582],[119,50.3765],[120,50.2249],[121,49.8997],[123,48.9479],[125,47.7546],[127,46.3707],[128,45.5807],[129,44.4664],[130,43.0492],[131,40.1974],[132,34.9894],[133,24.0377],[134,-16.8717],[135,-32.571],[136,-35.2535],[137,-38.063],[138,-37.941],[139,-38.3321],[140,-38.5383],[141,-36.1569],[142,-40.7177],[143,-103.6026],[144,-110.8347],[145,-103.7748],[146,-94.7572],[147,-93.0486],[148,-89.6512],[149,-82.835],[150,-77.138],[151,-72.4329],[152,-70.4973],[153,-71.1804],[154,-71.5582],[156,-71.9277],[157,-71.8903],[160,-70.8797],[173,-69.7428],[216,-68.9915],[218,-68.6093],[219,-68.2272],[220,-67.6816],[221,-66.9432],[222,-65.9926],[223,-65.274],[224,-64.7898],[226,-66.2039],[227,-66.1917],[228,-64.9559],[229,-61.7098],[230,-53.973],[231,-39.4762],[232,-7.0074],[233,23.8413],[234,43.5651],[235,52.6294],[236,55.0272],[237,56.8802],[238,57.7304],[239,57.9693],[241,58.0342],[244,58.0091],[245,58.9398],[246,57.8039],[247,57.3276],[248,57.6102],[249,58.4912],[250,59.0502],[251,59.2965],[255,58.6262],[257,58.5169],[259,58.7246],[262,59.4552],[269,61.7766],[275,62.9176],[295,63.3599],[296,63.5073],[297,63.8164],[298,64.5349],[299,65.5307],[301,67.0952],[303,69.0555],[305,70.6831],[306,70.8979],[307,70.4672],[308,69.6867],[309,68.64],[310,67.3427],[311,65.7645],[317,54.7841],[318,53.1725],[319,51.8697],[320,50.8577],[321,50.1915],[322,49.7469],[323,49.4368],[325,49.1308],[326,49.0668],[327,49.1749],[329,49.7296],[334,51.5329],[336,51.9641],[340,52.3889],[343,52.4042],[346,52.1152],[350,51.2759],[351,51.1481],[354,51.203],[355,51.4534],[357,52.5072],[358,52.7791],[359,52.756],[361,52.3877],[367,50.9779],[369,50.6775],[370,50.6134],[378,51.3935],[381,51.3244],[382,51.1861],[383,50.858],[384,50.1906],[386,48.4448],[387,47.8854],[388,47.6578],[389,47.5842],[391,47.7228],[392,47.9736],[400,51.6505],[402,52.2308],[405,52.5603],[422,52.5569]]
       }
       upperarm_l_fk_ctrl.Rotation.Y {
        type=MovieSceneScriptingFloatChannel
        default=26.0475
        thinned_from=423
        keys [[0,26.4212],[15,26.0603],[16,25.9541],[17,25.772],[18,25.6431],[20,25.187],[21,25.2948],[24,26.3363],[25,26.5806],[29,27.306],[35,27.7356],[57,28.1323],[62,28.1825],[69,28.759],[73,29.3736],[75,29.756],[76,29.883],[78,29.955],[79,29.8967],[81,29.5734],[83,29.0449],[85,28.4134],[86,28.2055],[93,27.5511],[98,27.443],[117,27.5712],[120,27.471],[122,27.2385],[124,26.8762],[126,26.3671],[127,26.0299],[128,25.6136],[129,25.0088],[130,23.875],[131,21.6049],[132,15.7585],[133,1.9024],[134,-5.7775],[135,-8.7274],[136,-9.0278],[137,-5.8381],[138,-1.8642],[139,2.8272],[140,7.7724],[141,25.8125],[142,58.6064],[143,74.6643],[144,71.0496],[146,69.4379],[147,69.1973],[148,65.6993],[149,55.9693],[150,49.6215],[151,47.4027],[153,48.6263],[154,50.2581],[155,52.0874],[156,53.6199],[157,54.6924],[158,55.5654],[159,56.2821],[160,56.7992],[161,57.1479],[163,57.4725],[220,57.4585],[222,57.2708],[224,56.7751],[225,54.3662],[226,50.2033],[227,45.0342],[228,39.6711],[229,33.8527],[230,28.4668],[231,23.7236],[232,21.7441],[233,20.9209],[234,20.4244],[235,21.3201],[236,23.9299],[237,26.997],[239,30.04],[240,30.9636],[241,31.4379],[242,31.1792],[243,29.2084],[244,25.1096],[245,23.3501],[246,23.6414],[247,24.727],[248,26.1843],[249,27.2923],[250,27.803],[251,28.0113],[253,27.9546],[256,27.8047],[277,28.6897],[296,28.776],[298,28.9422],[301,29.4622],[305,30.4011],[306,30.5586],[308,30.624],[310,30.4475],[311,30.2696],[313,29.73],[318,28.0441],[319,27.7678],[320,27.5525],[322,27.3199],[326,27.2174],[335,27.914],[343,28.0933],[350,28.0405],[354,28.378],[355,28.5395],[357,29.0061],[358,29.1458],[360,29.13],[370,28.3087],[378,28.1211],[381,27.9056],[382,27.76],[383,27.5091],[384,27.08],[386,25.9546],[387,25.5367],[388,25.3032],[389,25.1664],[391,25.0686],[392,25.1204],[393,25.2753],[399,26.3386],[401,26.58],[404,26.7202],[422,26.4296]]
       }
       upperarm_l_fk_ctrl.Rotation.Z {
        type=MovieSceneScriptingFloatChannel
        default=-0.4292
        thinned_from=423
        keys [[0,-5.5783],[2,-5.5677],[4,-5.5156],[10,-5.419],[11,-5.4139],[12,-5.3951],[15,-5.4489],[16,-5.6486],[17,-6.1149],[18,-6.7209],[19,-7.7262],[20,-8.9252],[21,-9.5553],[22,-9.5912],[23,-9.5256],[26,-9.1606],[27,-8.947],[29,-8.4076],[31,-7.9167],[33,-7.385],[35,-6.8299],[36,-6.5708],[40,-5.6567],[41,-5.4408],[42,-5.2492],[43,-5.0742],[45,-4.7682],[46,-4.6549],[47,-4.53],[48,-4.4452],[49,-4.3344],[50,-4.2717],[51,-4.1852],[52,-4.1209],[54,-4.0329],[55,-3.9953],[56,-3.9726],[57,-3.9726],[58,-3.997],[59,-4.0073],[60,-4.0075],[61,-3.9987],[62,-3.971],[63,-3.913],[67,-3.5354],[68,-3.4492],[69,-3.2187],[70,-2.7416],[71,-2.1041],[72,-1.416],[73,-0.6316],[74,0.2459],[75,1.0368],[76,1.5197],[77,1.557],[78,1.2994],[79,0.8429],[80,0.2237],[81,-0.468],[82,-1.1151],[84,-2.3502],[85,-2.8514],[86,-3.1225],[87,-3.2263],[88,-3.307],[89,-3.3722],[90,-3.4299],[91,-3.4761],[92,-3.5108],[94,-3.5623],[95,-3.5771],[97,-3.5884],[98,-3.584],[100,-3.5447],[103,-3.468],[108,-3.3991],[114,-3.3731],[115,-3.3754],[116,-3.3679],[117,-3.3679],[118,-3.3925],[119,-3.43],[120,-3.5079],[121,-3.6244],[122,-3.7842],[123,-3.989],[124,-4.248],[125,-4.5649],[126,-4.9658],[127,-5.4903],[128,-6.0622],[129,-6.795],[130,-7.8907],[131,-9.3189],[132,-9.3069],[133,-6.2315],[134,-3.2239],[135,-1.4156],[136,-0.047],[137,1.8387],[138,-0.2867],[139,-0.5292],[140,0.1941],[141,0.7645],[142,-9.1946],[143,-77.7105],[144,-49.8131],[145,-30.8181],[146,-18.4835],[147,-11.823],[148,-4.3298],[149,7.4673],[150,11.3417],[151,12.1113],[152,12.7647],[153,12.5694],[154,12.591],[155,12.8144],[156,12.9014],[157,13.1706],[158,13.738],[159,14.2722],[160,14.7083],[161,15.0252],[162,15.2366],[163,15.3817],[171,15.9619],[172,15.9907],[173,16.0627],[174,16.0627],[175,16.1374],[176,16.1374],[178,16.2329],[179,16.2329],[180,16.2829],[181,16.2829],[182,16.3307],[183,16.3307],[184,16.37],[185,16.37],[186,16.403],[188,16.403],[189,16.4425],[194,16.4425],[195,16.4797],[196,16.4358],[197,16.5061],[215,16.5266],[217,16.6529],[218,16.8187],[219,17.0814],[220,17.4624],[221,17.9805],[222,18.6419],[223,19.0474],[224,19.1505],[225,17.5685],[226,15.0987],[227,12.3746],[228,9.4852],[229,6.3712],[230,3.5394],[231,1.172],[232,0.5305],[233,-1.0922],[234,-2.8112],[235,-3.6714],[236,-3.157],[237,-3.1122],[238,-2.6692],[239,-2.7322],[240,-3.0877],[241,-3.0142],[242,-2.5692],[243,-2.0142],[244,-0.6818],[245,-1.1699],[246,-2.3394],[247,-3.2329],[248,-3.546],[249,-3.4846],[250,-3.6528],[251,-3.8683],[252,-4.1485],[253,-4.2006],[254,-4.1803],[257,-4.0977],[258,-4.0623],[259,-4.0144],[260,-3.9496],[266,-3.4672],[267,-3.3915],[268,-3.3262],[269,-3.2747],[270,-3.2364],[272,-3.1789],[273,-3.1656],[276,-3.1423],[280,-3.1006],[283,-3.0872],[286,-3.0844],[288,-3.0871],[291,-3.0525],[292,-3.0525],[293,-3.0242],[295,-2.9952],[296,-2.9677],[297,-2.9111],[298,-2.7553],[299,-2.4887],[300,-2.16],[301,-1.5784],[302,-0.6911],[303,0.2145],[304,1.1745],[305,2.1471],[306,2.7223],[307,2.7937],[308,2.613],[309,2.2936],[310,1.9075],[311,1.4395],[312,0.8906],[313,0.3265],[314,-0.1989],[315,-0.6728],[316,-1.0822],[317,-1.4508],[318,-1.7532],[319,-1.9857],[320,-2.1549],[321,-2.2702],[322,-2.357],[323,-2.4243],[324,-2.4752],[326,-2.5308],[327,-2.5281],[328,-2.513],[329,-2.4922],[334,-2.3412],[335,-2.317],[337,-2.2903],[338,-2.2811],[339,-2.2786],[340,-2.2656],[341,-2.2834],[342,-2.2834],[343,-2.2627],[345,-2.2594],[347,-2.235],[348,-2.2017],[349,-2.1494],[350,-2.0627],[351,-1.8961],[352,-1.6465],[353,-1.3664],[354,-1.1092],[355,-0.751],[356,-0.243],[357,0.2461],[358,0.5294],[359,0.6102],[360,0.6021],[361,0.5308],[362,0.4211],[363,0.2971],[364,0.1825],[365,0.0737],[366,-0.0433],[367,-0.1665],[368,-0.286],[369,-0.3924],[370,-0.4913],[371,-0.5675],[372,-0.6387],[373,-0.7047],[374,-0.792],[375,-0.8858],[376,-1.0197],[377,-1.2053],[378,-1.4266],[379,-1.6884],[380,-1.9442],[381,-2.2047],[382,-2.534],[383,-3.0018],[384,-3.7211],[385,-4.614],[386,-5.4889],[387,-6.1591],[388,-6.5993],[389,-6.9306],[390,-7.1725],[391,-7.3441],[392,-7.3789],[393,-7.2533],[395,-6.8606],[396,-6.6848],[398,-6.2755],[399,-6.0867],[400,-5.934],[401,-5.8238],[402,-5.7441],[403,-5.6894],[410,-5.5737],[414,-5.5539],[422,-5.5762]]
       }
       upperarm_l_fk_ctrl.Scale.X {type=MovieSceneScriptingFloatChannel, default=1, const [1,423]}
       upperarm_l_fk_ctrl.Scale.Y {type=MovieSceneScriptingFloatChannel, default=1, const [1,423]}
       upperarm_l_fk_ctrl.Scale.Z {type=MovieSceneScriptingFloatChannel, default=1, const [1,423]}
       lowerarm_l_fk_ctrl.Location.X {
        type=MovieSceneScriptingFloatChannel
        default=0
        thinned_from=423
        keys [[0,0],[1,0.0006],[2,0],[56,0],[57,0.0006],[58,0],[173,0],[174,0.0009],[175,0],[176,0.0025],[177,0],[178,0],[179,0.0023],[180,0],[181,0.0019],[182,0],[183,0.001],[184,0],[185,0.0011],[186,0],[187,0.0006],[188,0.0017],[189,0],[190,0.002],[191,0.0015],[194,0.0015],[195,0],[197,0],[199,-0.0014],[201,-0.002],[202,0],[204,-0.0011],[205,0],[206,-0.0003],[207,0],[208,-0.0005],[209,0],[210,-0.0004],[211,0],[212,-0.0003],[213,0],[283,0],[284,0.0002],[285,0],[291,0],[292,0.0005],[293,0],[341,0],[342,0.0024],[343,0],[418,0],[419,0.0019],[422,0.0019]]
       }
       lowerarm_l_fk_ctrl.Location.Y {
        type=MovieSceneScriptingFloatChannel
        default=0
        thinned_from=423
        keys [[0,0],[1,0.0011],[2,0],[56,0],[57,0.001],[58,0],[116,0],[117,-0.0003],[118,0],[173,0],[174,0.0016],[175,0],[176,0.0044],[177,0],[178,0],[179,0.0041],[180,0],[181,0.0032],[182,0],[183,0.0017],[184,0],[185,0.002],[186,0],[187,0.001],[188,0.0029],[189,0],[190,0.0037],[191,0.0026],[194,0.0026],[195,0],[197,0],[199,-0.0024],[201,-0.0037],[202,0],[204,-0.0021],[205,0],[206,-0.0007],[207,0],[208,-0.0009],[209,0],[210,-0.0007],[211,0],[212,-0.0006],[213,0],[283,0],[284,0.0004],[285,0],[291,0],[292,0.0012],[293,0],[341,0],[342,0.0044],[343,0],[418,0],[419,0.0034],[422,0.0034]]
       }
       lowerarm_l_fk_ctrl.Location.Z {
        type=MovieSceneScriptingFloatChannel
        default=0
        thinned_from=423
        keys [[0,0],[1,-0.0029],[2,0],[56,0],[57,-0.0044],[58,0],[116,0],[117,-0.0045],[118,0],[173,0],[174,-0.0037],[175,0],[176,-0.0029],[177,0],[178,0],[179,-0.0039],[180,0],[181,-0.005],[182,0],[183,-0.0053],[184,0],[185,-0.0037],[186,0],[187,-0.0035],[188,-0.003],[189,0],[190,0.0014],[191,-0.0042],[194,-0.0042],[195,0],[197,0],[198,0.0044],[199,0.0051],[200,0.0019],[201,0],[203,-0.0001],[204,-0.0012],[205,0],[206,-0.0026],[207,0],[208,-0.0003],[209,0],[210,-0.0003],[211,0],[212,-0.0004],[213,0],[283,0],[284,0.0004],[285,0],[291,0],[292,0.0057],[293,0],[341,0],[342,0.0025],[343,0],[418,0],[419,-0.0011],[422,-0.0011]]
       }
       lowerarm_l_fk_ctrl.Rotation.X {
        type=MovieSceneScriptingFloatChannel
        default=-4.6845
        thinned_from=423
        keys [[0,-4.6883],[1,-4.6961],[15,-4.617],[17,-4.6244],[19,-4.6779],[24,-4.8653],[28,-4.9649],[31,-5.0084],[34,-5.0201],[56,-4.8937],[67,-4.9305],[70,-4.9112],[73,-4.8609],[75,-4.7976],[79,-4.6261],[83,-4.4306],[86,-4.3197],[89,-4.2476],[93,-4.2056],[116,-4.2093],[117,-4.2192],[118,-4.2093],[125,-4.2058],[126,-4.2164],[127,-4.2432],[128,-4.283],[129,-4.3331],[130,-4.3594],[131,-4.3016],[132,-3.5714],[133,-1.4408],[134,-0.1685],[135,0.074],[137,0.1603],[138,0.1532],[139,0.1166],[140,0.1166],[141,-0.371],[142,-1.0487],[143,-1.2001],[144,-1.9842],[145,-2.4913],[146,-2.8875],[147,-3.2019],[148,-3.6125],[149,-4.059],[150,-4.4292],[151,-4.5326],[152,-4.5688],[153,-4.5611],[154,-4.5326],[155,-4.4818],[158,-4.2947],[159,-4.2526],[160,-4.2364],[173,-4.2364],[174,-4.2458],[175,-4.2364],[176,-4.2436],[178,-4.2364],[179,-4.246],[180,-4.2364],[181,-4.2483],[182,-4.2364],[183,-4.2489],[184,-4.2364],[185,-4.2449],[186,-4.2364],[187,-4.2445],[190,-4.2326],[191,-4.2459],[194,-4.2459],[195,-4.2364],[197,-4.2364],[199,-4.2263],[201,-4.2391],[205,-4.2364],[206,-4.244],[207,-4.2364],[216,-4.2328],[218,-4.2078],[219,-4.1805],[220,-4.1399],[221,-4.0832],[222,-4.0078],[223,-3.9138],[224,-3.786],[225,-3.5989],[226,-3.3287],[227,-2.9547],[228,-2.4641],[229,-1.5993],[230,-1.0354],[231,-0.6395],[232,1.5662],[233,1.2457],[234,1.1307],[235,-0.5557],[236,-2.2732],[237,-3.7227],[238,-4.398],[239,-4.7433],[240,-4.9336],[241,-5.062],[242,-4.987],[243,-4.7654],[244,-4.3933],[245,-4.3028],[246,-4.3553],[248,-4.6901],[249,-4.7753],[250,-4.8002],[263,-4.9367],[291,-4.9307],[292,-4.9199],[293,-4.9371],[299,-4.9569],[301,-4.9371],[305,-4.8495],[306,-4.813],[307,-4.7585],[317,-4.1613],[318,-4.1068],[319,-4.0632],[320,-4.032],[322,-3.9999],[324,-3.9909],[326,-3.9994],[334,-4.1239],[341,-4.1665],[348,-4.1633],[350,-4.1431],[351,-4.1163],[357,-3.8428],[358,-3.8072],[360,-3.7578],[364,-3.7011],[367,-3.6811],[370,-3.6831],[373,-3.7143],[376,-3.7707],[377,-3.7978],[378,-3.8353],[379,-3.8875],[380,-3.9582],[382,-4.1276],[383,-4.2024],[384,-4.2615],[387,-4.4078],[390,-4.5692],[392,-4.6538],[395,-4.7382],[398,-4.7873],[403,-4.8112],[418,-4.6902],[422,-4.69]]
       }
       lowerarm_l_fk_ctrl.Rotation.Y {
        type=MovieSceneScriptingFloatChannel
        default=-4.4488
        thinned_from=423
        keys [[0,-3.3335],[1,-3.3407],[2,-3.3339],[21,-3.328],[36,-3.2953],[56,-3.3118],[57,-3.2996],[58,-3.3116],[131,-3.3485],[132,-3.285],[133,2.0604],[134,-15.1811],[135,-6.8612],[136,2.5505],[137,4.9005],[138,4.2057],[139,3.8447],[140,3.8447],[141,1.1863],[142,-0.8419],[143,-2.0755],[144,-2.6877],[145,-2.9561],[146,-3.112],[147,-3.2059],[148,-3.2919],[149,-3.3409],[150,-3.3473],[173,-3.3478],[174,-3.3527],[175,-3.3478],[176,-3.3512],[178,-3.3478],[179,-3.3516],[180,-3.3478],[181,-3.3514],[182,-3.3478],[190,-3.3451],[201,-3.3594],[202,-3.3478],[204,-3.3573],[205,-3.3478],[206,-3.356],[207,-3.3478],[208,-3.3567],[209,-3.3478],[210,-3.3589],[211,-3.3478],[212,-3.3613],[213,-3.3478],[221,-3.3422],[223,-3.3299],[224,-3.3163],[225,-3.2897],[226,-3.2367],[227,-3.1341],[228,-2.9438],[229,-2.4224],[230,-1.4102],[231,0.3474],[232,7.5678],[233,7.8938],[234,7.971],[235,3.9083],[236,-2.0497],[237,-4.0823],[238,-4.091],[239,-3.1459],[240,-2.504],[241,-2.4701],[242,-3.2989],[244,-3.3481],[247,-3.3442],[249,-3.3255],[305,-3.3173],[315,-3.3484],[341,-3.3458],[342,-3.351],[373,-3.307],[386,-3.3484],[422,-3.3311]]
       }
       lowerarm_l_fk_ctrl.Rotation.Z {
        type=MovieSceneScriptingFloatChannel
        default=80.1518
        thinned_from=423
        keys [[0,80.2324],[16,79.3139],[17,79.4093],[21,81.0337],[25,82.9095],[30,84.2334],[34,84.5265],[57,82.8894],[68,83.3321],[73,82.461],[78,80.0403],[84,76.383],[87,75.1271],[90,74.3592],[97,73.9467],[125,74.0283],[126,74.1642],[127,74.5083],[129,75.6633],[130,76.0018],[131,75.259],[132,65.8312],[133,35.6234],[134,9.9339],[135,-3.5854],[136,-7.4414],[137,-11.8373],[138,-16.7329],[139,-21.668],[140,-21.668],[141,13.4903],[142,28.4015],[143,31.123],[144,43.9195],[145,51.2929],[146,56.7635],[147,60.98],[148,66.3703],[149,72.1437],[150,76.9026],[151,78.2315],[152,78.6977],[153,78.5981],[154,78.2313],[155,77.5785],[158,75.1745],[159,74.6328],[160,74.425],[216,74.379],[218,74.0574],[219,73.7071],[220,73.1845],[221,72.4551],[222,71.4846],[223,70.2734],[224,68.6218],[225,66.1943],[226,62.656],[227,57.6721],[228,50.909],[229,37.9146],[230,23.4881],[231,10.6342],[232,2.6204],[233,3.9358],[234,13.6476],[235,29.5446],[236,48.5276],[237,66.6586],[238,75.8572],[239,80.8252],[240,83.3466],[241,85.0633],[242,84.0994],[243,81.2311],[244,76.4412],[245,75.2747],[246,75.9484],[248,80.256],[249,81.3546],[250,81.6766],[267,83.6465],[300,83.6374],[305,82.3138],[306,81.8412],[318,72.7552],[319,72.1944],[320,71.7927],[322,71.379],[325,71.3011],[336,73.2151],[348,73.482],[350,73.2225],[351,72.8769],[357,69.3528],[359,68.5406],[364,67.5178],[369,67.238],[373,67.6893],[376,68.4196],[378,69.2554],[379,69.9299],[380,70.8421],[382,73.0222],[383,73.984],[391,79.2896],[394,80.5651],[398,81.5101],[405,81.759],[418,80.2565],[422,80.2409]]
       }
       lowerarm_l_fk_ctrl.Scale.X {type=MovieSceneScriptingFloatChannel, default=1, const [1,423]}
       lowerarm_l_fk_ctrl.Scale.Y {type=MovieSceneScriptingFloatChannel, default=1, const [1,423]}
       lowerarm_l_fk_ctrl.Scale.Z {type=MovieSceneScriptingFloatChannel, default=1, const [1,423]}
       hand_l_fk_ctrl.Location.X {type=MovieSceneScriptingFloatChannel, default=0.0043, thinned_from=423, keys [[0,0],[422,0]]}
       hand_l_fk_ctrl.Location.Y {type=MovieSceneScriptingFloatChannel, default=-0.431, thinned_from=423, keys [[0,0],[422,0]]}
       hand_l_fk_ctrl.Location.Z {type=MovieSceneScriptingFloatChannel, default=0.0668, thinned_from=423, keys [[0,0],[422,0]]}
       hand_l_fk_ctrl.Rotation.X {
        type=MovieSceneScriptingFloatChannel
        default=-13.7246
        thinned_from=423
        keys [[0,-13.7122],[14,-13.6127],[15,-13.586],[16,-13.4026],[17,-12.958],[18,-12.3122],[19,-11.3013],[20,-10.1166],[21,-9.3464],[22,-9.041],[23,-8.8438],[24,-8.7184],[25,-8.6386],[26,-8.6174],[27,-8.6654],[28,-8.7605],[30,-9.0072],[31,-9.1464],[32,-9.3154],[37,-10.3348],[39,-10.7047],[41,-11.0261],[43,-11.2844],[45,-11.494],[49,-11.7916],[53,-11.9765],[56,-12.0378],[63,-11.907],[65,-11.8414],[66,-11.773],[68,-11.5254],[69,-11.517],[70,-11.7083],[71,-12.0979],[72,-12.6709],[73,-13.5657],[74,-14.8099],[75,-16.1617],[76,-17.3309],[77,-18.1803],[78,-18.8023],[79,-19.2635],[80,-19.5846],[81,-19.7687],[82,-19.86],[83,-19.8416],[85,-19.6187],[86,-19.6193],[88,-19.7398],[117,-19.6492],[119,-19.5747],[121,-19.3988],[122,-19.2598],[123,-19.079],[124,-18.8471],[125,-18.559],[126,-18.1969],[127,-17.7656],[128,-17.4515],[129,-17.174],[130,-16.3809],[131,-14.899],[132,-23.5999],[133,-49.3279],[134,-68.5953],[135,-83.6876],[137,-88.8688],[138,-88.9001],[139,-81.2012],[141,-81.2012],[143,-89.5788],[144,-89.0005],[148,-88.7935],[149,-89.7207],[150,-90.2663],[151,-90.5835],[218,-90.75],[221,-90.1474],[224,-89.124],[225,-88.4971],[226,-87.2869],[227,-85.1973],[228,-81.9182],[229,-74.9151],[230,-63.1596],[231,-57.686],[232,-74.8844],[233,-44.3077],[234,-26.833],[235,-25.2195],[236,-13.6269],[237,-13.663],[238,-13.805],[239,-13.8598],[241,-15.4166],[242,-15.8597],[244,-17.0845],[245,-16.6621],[246,-15.4082],[247,-14.6709],[249,-14.3189],[250,-13.7646],[251,-13.1492],[252,-12.6343],[253,-12.4882],[254,-12.4851],[256,-12.6128],[257,-12.6573],[276,-12.3909],[298,-12.4034],[299,-12.4471],[300,-12.6189],[301,-12.8347],[302,-12.9847],[303,-13.1085],[306,-13.359],[308,-13.6089],[311,-13.7323],[312,-13.7289],[314,-13.623],[316,-13.6238],[317,-13.5841],[321,-13.2058],[323,-13.072],[325,-13.0137],[349,-13.3319],[351,-13.4277],[354,-13.71],[355,-13.9117],[357,-14.6014],[358,-14.7893],[359,-14.8008],[360,-14.7371],[361,-14.6202],[369,-13.4693],[374,-12.921],[375,-12.786],[376,-12.7831],[378,-13.1634],[379,-13.2254],[380,-13.1109],[381,-12.9113],[383,-12.4404],[384,-12.1832],[385,-11.8971],[386,-11.6646],[387,-11.5676],[388,-11.6201],[389,-11.7496],[391,-12.0611],[392,-12.2432],[394,-12.7233],[395,-12.8963],[399,-13.4268],[400,-13.5146],[403,-13.6499],[422,-13.7122]]
       }
       hand_l_fk_ctrl.Rotation.Y {
        type=MovieSceneScriptingFloatChannel
        default=15.6421
        thinned_from=423
        keys [[0,15.4712],[3,15.5532],[14,16.0145],[16,16.0199],[17,15.9732],[18,15.89],[20,15.6471],[22,15.5181],[26,15.1224],[32,14.3214],[34,14.1322],[35,14.0765],[36,14.0498],[37,14.0554],[41,14.2896],[57,14.5179],[61,14.4404],[63,14.4661],[65,14.5631],[66,14.6739],[68,15.0534],[69,15.219],[72,15.5495],[74,15.6917],[75,15.8748],[77,16.4124],[78,16.7298],[79,17.106],[83,18.8663],[84,19.2596],[86,19.9115],[87,20.169],[88,20.3646],[89,20.5107],[90,20.6106],[94,20.7303],[117,20.8115],[124,20.5293],[125,20.4496],[126,20.2459],[127,19.8581],[128,19.1576],[130,17.1092],[131,16.3068],[132,9.2821],[133,6.6683],[134,-3.4024],[135,-10.1784],[136,-11.2276],[137,-12.2518],[138,-11.9062],[139,-10.3065],[141,-10.3065],[142,-7.6811],[143,-4.9867],[144,-12.1112],[145,-13.8912],[146,-14.1879],[147,-14.3362],[148,-14.4103],[149,-3.1231],[150,4.9499],[151,10.2382],[152,12.0013],[153,14.1321],[154,14.8424],[155,15.3131],[156,15.4682],[158,15.6561],[160,15.7438],[217,15.7724],[219,15.7101],[221,15.5725],[224,15.2774],[225,15.0993],[226,14.7614],[227,14.1953],[228,13.3505],[229,11.7127],[230,9.4144],[231,13.1965],[232,1.1833],[233,0.9292],[234,-7.1168],[235,-13.2811],[236,-12.1192],[237,-11.8189],[238,-11.1837],[239,-8.8868],[240,-4.4551],[241,0.1721],[242,3.5783],[244,8.7002],[245,12.2435],[246,12.8951],[248,12.4843],[249,12.858],[250,13.5367],[251,14.0693],[252,14.176],[254,13.7317],[257,13.1842],[258,13.0488],[259,12.9673],[262,12.8603],[266,12.9083],[278,13.4402],[283,13.5504],[287,13.5682],[294,13.396],[295,13.3878],[296,13.4105],[297,13.4833],[299,13.735],[300,13.81],[301,13.917],[302,14.0895],[303,14.1921],[304,14.082],[305,13.9162],[306,14.0094],[307,14.5037],[308,15.1746],[309,15.7806],[310,16.3442],[312,17.6328],[314,18.7614],[316,19.9898],[317,20.5546],[318,21.0349],[319,21.3915],[320,21.6032],[322,21.7857],[324,21.8236],[326,21.7344],[334,20.9223],[341,20.6529],[348,20.598],[349,20.6374],[350,20.7265],[351,20.9547],[352,21.3417],[357,23.5485],[358,23.8943],[359,24.1488],[361,24.5121],[364,24.8357],[367,24.9778],[369,24.9486],[371,24.7851],[373,24.495],[374,24.2915],[375,24.0393],[376,23.6715],[377,23.1416],[378,22.4443],[379,21.5837],[380,20.4792],[381,19.1934],[382,17.9778],[383,17.0865],[384,16.691],[385,16.6449],[386,16.6783],[387,16.5344],[388,16.1775],[390,15.3301],[391,14.9736],[392,14.7101],[393,14.5087],[394,14.3487],[395,14.233],[396,14.1594],[397,14.1195],[399,14.1202],[404,14.3441],[407,14.5665],[412,15.0597],[417,15.4161],[419,15.4712],[422,15.4712]]
       }
       hand_l_fk_ctrl.Rotation.Z {
        type=MovieSceneScriptingFloatChannel
        default=-5.0908
        thinned_from=423
        keys [[0,-3.9825],[1,-3.9957],[2,-4.0299],[5,-4.1766],[6,-4.2127],[8,-4.259],[10,-4.2825],[11,-4.283],[12,-4.2433],[14,-4.0922],[15,-4.0758],[16,-3.9224],[17,-3.7307],[18,-3.8901],[19,-3.9105],[20,-3.8899],[21,-4.6145],[22,-5.8529],[23,-7.0048],[24,-8.0705],[25,-8.8463],[27,-10.0421],[28,-10.6192],[29,-11.1191],[30,-11.4947],[31,-11.7716],[32,-12.0078],[33,-12.2071],[34,-12.3753],[35,-12.5106],[36,-12.6148],[38,-12.768],[39,-12.8841],[42,-13.3927],[44,-13.6742],[48,-14.0717],[52,-14.3475],[55,-14.4635],[58,-14.4873],[59,-14.5252],[60,-14.6044],[61,-14.7309],[62,-14.9175],[63,-15.2481],[66,-16.6666],[67,-17.1828],[68,-17.6361],[69,-18.1537],[70,-18.8787],[74,-22.4632],[75,-23.1333],[76,-23.1864],[77,-22.2359],[78,-20.6058],[79,-18.5578],[81,-13.9696],[82,-11.8167],[83,-9.8379],[84,-7.9624],[85,-6.3495],[86,-5.1827],[87,-4.3814],[88,-3.6984],[89,-3.098],[90,-2.5869],[91,-2.1618],[92,-1.8173],[93,-1.5421],[94,-1.344],[95,-1.2361],[96,-1.1971],[97,-1.2069],[98,-1.2804],[99,-1.4379],[100,-1.6378],[101,-1.8519],[102,-2.0609],[103,-2.2304],[104,-2.3723],[105,-2.4839],[107,-2.6701],[108,-2.7513],[109,-2.8171],[110,-2.8706],[111,-2.9119],[113,-2.9773],[114,-3.0037],[115,-3.0221],[116,-3.0221],[117,-3.0314],[118,-3.0044],[119,-2.9345],[120,-2.7769],[121,-2.4448],[122,-1.9936],[123,-1.473],[124,-0.8819],[125,-0.2416],[126,0.4727],[127,1.2312],[128,2.148],[129,3.4794],[130,5.3167],[131,8.9326],[132,11.6533],[133,-7.2041],[134,-4.5517],[135,-1.8735],[136,-1.3747],[137,-0.8621],[138,-0.7167],[139,-2.0457],[141,-2.0457],[143,8.7291],[144,8.2611],[145,8.13],[148,8.0906],[149,9.0659],[150,9.6512],[151,9.9977],[153,10.2377],[155,10.3084],[216,10.3319],[218,10.2989],[220,10.2097],[221,10.1353],[224,9.8601],[225,9.6933],[226,9.3749],[227,8.8365],[228,8.0208],[229,6.3986],[230,4.0509],[231,3.1788],[232,-12.1648],[233,-7.4565],[234,1.8444],[235,1.0977],[236,-5.1053],[237,-19.1587],[238,-30.9271],[239,-18.8288],[240,-6.5007],[241,-2.8745],[242,-0.8474],[243,-0.7061],[244,-0.5916],[245,-6.9859],[246,-6.3501],[247,-6.4915],[248,-7.4175],[249,-8.7291],[250,-9.5617],[251,-10.004],[252,-9.9293],[253,-9.7548],[254,-9.6043],[255,-9.4783],[256,-9.4014],[257,-9.3987],[258,-9.4796],[259,-9.6318],[260,-9.8378],[261,-10.0801],[262,-10.3535],[267,-11.9271],[268,-12.2107],[269,-12.4648],[270,-12.6777],[273,-13.1635],[275,-13.4156],[277,-13.5826],[280,-13.7279],[283,-13.7866],[291,-13.6897],[293,-13.724],[295,-13.8324],[296,-13.974],[297,-14.2617],[298,-14.9126],[299,-15.8127],[300,-16.528],[301,-17.1181],[302,-17.7936],[303,-18.3369],[304,-18.761],[305,-19.0474],[306,-18.9246],[307,-18.3357],[308,-17.4782],[309,-16.4039],[310,-15.1405],[311,-13.669],[312,-12.0045],[313,-10.2501],[314,-8.5203],[315,-6.81],[317,-3.4239],[318,-1.862],[319,-0.5785],[320,0.4348],[321,1.112],[322,1.5673],[323,1.8828],[324,2.0755],[325,2.1868],[326,2.2241],[327,2.0962],[328,1.8239],[329,1.4981],[330,1.1221],[331,0.7243],[332,0.3388],[333,-0.0247],[334,-0.3729],[335,-0.6373],[336,-0.8141],[337,-0.9536],[338,-1.0662],[339,-1.1624],[340,-1.2524],[341,-1.3036],[342,-1.3067],[343,-1.2746],[344,-1.2117],[345,-1.1373],[346,-1.031],[347,-0.8738],[348,-0.6869],[349,-0.4863],[350,-0.294],[351,-0.1697],[352,-0.1232],[353,-0.1207],[354,-0.1263],[355,-0.3379],[356,-0.8153],[357,-1.3038],[358,-1.5356],[359,-1.4776],[360,-1.3024],[361,-1.0501],[362,-0.7555],[363,-0.4516],[364,-0.1827],[365,0.0584],[366,0.2941],[367,0.4929],[368,0.6583],[369,0.8332],[370,0.9099],[371,0.854],[372,0.7718],[374,0.5811],[375,0.4991],[376,0.3528],[377,0.1145],[378,-0.1153],[379,-0.2429],[380,-0.2968],[381,-0.3239],[382,-0.2727],[383,-0.0885],[384,0.3569],[385,0.9878],[386,1.5768],[387,1.891],[388,1.8899],[389,1.7413],[390,1.5009],[391,1.2226],[392,0.8222],[393,0.2592],[394,-0.3485],[395,-0.8941],[396,-1.3888],[397,-1.8994],[398,-2.4039],[399,-2.8802],[400,-3.3038],[401,-3.6423],[402,-3.8893],[403,-4.0575],[404,-4.1601],[405,-4.2157],[406,-4.2431],[407,-4.2553],[408,-4.2577],[410,-4.2374],[412,-4.1912],[414,-4.1265],[417,-4.0149],[418,-3.9914],[419,-3.9825],[422,-3.9825]]
       }
       hand_l_fk_ctrl.Scale.X {type=MovieSceneScriptingFloatChannel, default=1, const [1,423]}
       hand_l_fk_ctrl.Scale.Y {type=MovieSceneScriptingFloatChannel, default=1, const [1,423]}
       hand_l_fk_ctrl.Scale.Z {type=MovieSceneScriptingFloatChannel, default=1, const [1,423]}
       thumb_01_l_ctrl.Location.X {
        type=MovieSceneScriptingFloatChannel
        default=0.0686
        thinned_from=423
        keys [[0,0.0686],[224,0.0686],[225,0.0669],[226,0.0627],[228,0.0515],[229,0.0416],[230,0.027],[231,0.0129],[232,0.0032],[233,0.0011],[234,0.0008],[235,0.0008],[236,0.0018],[237,0.0041],[238,0.0272],[239,0.0348],[240,0.0391],[241,0.0418],[243,0.0459],[245,0.061],[247,0.0686],[422,0.0686]]
       }
       thumb_01_l_ctrl.Location.Y {
        type=MovieSceneScriptingFloatChannel
        default=-0.3985
        thinned_from=423
        keys [[0,-0.3985],[224,-0.3985],[225,-0.3888],[226,-0.3644],[227,-0.3321],[228,-0.2989],[229,-0.2416],[230,-0.1566],[231,-0.0747],[232,-0.0187],[233,-0.0064],[234,-0.0047],[235,-0.0047],[236,-0.0107],[237,-0.0236],[238,-0.158],[239,-0.2024],[240,-0.2269],[241,-0.243],[243,-0.2667],[245,-0.3546],[247,-0.3985],[422,-0.3985]]
       }
       thumb_01_l_ctrl.Location.Z {
        type=MovieSceneScriptingFloatChannel
        default=-0.4179
        thinned_from=423
        keys [[0,-0.4179],[224,-0.4179],[225,-0.4078],[226,-0.3822],[227,-0.3483],[228,-0.3135],[229,-0.2533],[230,-0.1642],[231,-0.0784],[232,-0.0196],[233,-0.0067],[234,-0.0049],[235,-0.0049],[236,-0.0112],[237,-0.0248],[238,-0.1657],[239,-0.2122],[240,-0.238],[241,-0.2548],[243,-0.2797],[245,-0.3718],[247,-0.4179],[422,-0.4179]]
       }
       thumb_01_l_ctrl.Rotation.X {
        type=MovieSceneScriptingFloatChannel
        default=7.2456
        thinned_from=423
        keys [[0,7.2456],[130,7.2456],[131,7.5456],[132,7.9206],[134,9.0456],[142,9.0456],[143,11.8999],[144,17.9942],[145,23.6257],[146,28.247],[147,32.9985],[148,36.7086],[149,38.2056],[224,38.2056],[225,33.7853],[227,14.3359],[228,9.9155],[229,10.4159],[230,11.158],[231,11.8729],[232,12.3622],[233,12.4692],[235,12.4845],[236,12.4318],[237,12.3191],[238,11.1455],[239,10.7583],[240,10.5438],[241,10.4034],[243,10.1968],[245,8.2294],[247,7.2456],[422,7.2456]]
       }
       thumb_01_l_ctrl.Rotation.Y {
        type=MovieSceneScriptingFloatChannel
        default=-12.6986
        thinned_from=423
        keys [[0,-12.6986],[131,-12.6986],[132,-13.3931],[133,-14.683],[134,-15.3775],[142,-15.3775],[143,-16.3144],[144,-18.315],[145,-20.1636],[147,-23.2404],[148,-24.4583],[149,-24.9497],[224,-24.9497],[225,-23.6702],[226,-20.6323],[227,-17.0371],[228,-14.0857],[229,-13.3426],[230,-12.2407],[231,-11.1792],[232,-10.4526],[233,-0.0319],[234,9.2989],[235,9.2989],[236,7.305],[237,3.0476],[238,-9.2925],[239,-11.8661],[240,-12.4211],[241,-12.6292],[243,-12.6986],[422,-12.6986]]
       }
       thumb_01_l_ctrl.Rotation.Z {
        type=MovieSceneScriptingFloatChannel
        default=-0.7946
        thinned_from=423
        keys [[0,-0.7946],[130,-0.7946],[131,-1.7546],[132,-2.9546],[134,-6.5547],[142,-6.5547],[143,-6.2023],[144,-5.4499],[145,-4.7546],[146,-4.1841],[147,-3.5975],[148,-3.1395],[149,-2.9546],[224,-2.9546],[225,-3.2733],[227,-4.6754],[228,-4.994],[229,-4.0963],[230,-2.7651],[231,-1.4826],[232,-0.6048],[233,4.7154],[234,8.7187],[235,8.7187],[236,7.9327],[237,6.2546],[238,-0.4355],[239,-2.1309],[240,-2.8056],[241,-3.2057],[243,-3.7312],[245,-1.7735],[247,-0.7946],[422,-0.7946]]
       }
       thumb_01_l_ctrl.Scale.X {type=MovieSceneScriptingFloatChannel, default=1, const [1,423]}
       thumb_01_l_ctrl.Scale.Y {type=MovieSceneScriptingFloatChannel, default=1, const [1,423]}
       thumb_01_l_ctrl.Scale.Z {type=MovieSceneScriptingFloatChannel, default=1, const [1,423]}
       thumb_02_l_ctrl.Location.X {type=MovieSceneScriptingFloatChannel, default=0, thinned_from=423, keys [[0,0],[422,0]]}
       thumb_02_l_ctrl.Location.Y {type=MovieSceneScriptingFloatChannel, default=0, thinned_from=423, keys [[0,0],[422,0]]}
       thumb_02_l_ctrl.Location.Z {type=MovieSceneScriptingFloatChannel, default=0, thinned_from=423, keys [[0,0],[422,0]]}
       thumb_02_l_ctrl.Rotation.X {
        type=MovieSceneScriptingFloatChannel
        default=-28.0109
        thinned_from=423
        keys [[0,-28.0109],[224,-28.0109],[225,-26.9767],[226,-24.3661],[227,-20.9172],[228,-17.3683],[229,-11.2464],[230,-2.1681],[231,6.5775],[232,12.564],[233,13.8736],[234,14.0606],[235,14.0606],[236,13.415],[237,12.0364],[238,-2.3207],[239,-7.058],[240,-9.6818],[241,-11.3993],[243,-13.9278],[245,-23.3165],[247,-28.0109],[422,-28.0109]]
       }
       thumb_02_l_ctrl.Rotation.Y {
        type=MovieSceneScriptingFloatChannel
        default=-14.6405
        thinned_from=423
        keys [[0,-14.6405],[224,-14.6405],[225,-13.6649],[226,-11.2022],[227,-7.9489],[228,-4.6011],[229,1.1738],[230,9.7374],[231,17.9873],[232,23.6345],[233,24.8698],[234,25.0462],[235,25.0462],[236,24.4372],[237,23.1368],[238,9.5935],[239,5.1248],[240,2.6497],[241,1.0296],[243,-1.3556],[245,-10.2122],[247,-14.6405],[422,-14.6405]]
       }
       thumb_02_l_ctrl.Rotation.Z {
        type=MovieSceneScriptingFloatChannel
        default=7.2188
        thinned_from=423
        keys [[0,7.2188],[224,7.2188],[225,8.1408],[226,10.6433],[227,14.3313],[228,18.8095],[230,28.5562],[231,33.0344],[232,36.7224],[233,39.2249],[234,40.1469],[235,40.1469],[236,39.4138],[237,37.8484],[238,17.404],[239,10.4562],[240,8.2979],[241,7.4886],[243,7.2188],[422,7.2188]]
       }
       thumb_02_l_ctrl.Scale.X {type=MovieSceneScriptingFloatChannel, default=1, const [1,423]}
       thumb_02_l_ctrl.Scale.Y {type=MovieSceneScriptingFloatChannel, default=1, const [1,423]}
       thumb_02_l_ctrl.Scale.Z {type=MovieSceneScriptingFloatChannel, default=1, const [1,423]}
       thumb_03_l_ctrl.Location.X {type=MovieSceneScriptingFloatChannel, default=0, thinned_from=423, keys [[0,0],[422,0]]}
       thumb_03_l_ctrl.Location.Y {type=MovieSceneScriptingFloatChannel, default=0, thinned_from=423, keys [[0,0],[422,0]]}
       thumb_03_l_ctrl.Location.Z {type=MovieSceneScriptingFloatChannel, default=0, thinned_from=423, keys [[0,0],[422,0]]}
       thumb_03_l_ctrl.Rotation.X {
        type=MovieSceneScriptingFloatChannel
        default=-4.6289
        thinned_from=423
        keys [[0,-4.6289],[224,-4.6289],[225,-4.5403],[226,-4.3167],[227,-4.0213],[228,-3.7174],[229,-3.193],[230,-2.4155],[231,-1.6665],[232,-1.1537],[233,-1.0416],[234,-1.0256],[235,-1.0256],[236,-1.0808],[237,-1.1989],[238,-2.4286],[239,-2.8343],[240,-3.059],[241,-3.2061],[243,-3.4227],[245,-4.2268],[247,-4.6289],[422,-4.6289]]
       }
       thumb_03_l_ctrl.Rotation.Y {
        type=MovieSceneScriptingFloatChannel
        default=-5.8424
        thinned_from=423
        keys [[0,-5.8424],[224,-5.8424],[225,-5.7903],[226,-5.6588],[228,-5.3063],[229,-4.9978],[230,-4.5405],[231,-4.0999],[232,-3.7984],[233,-3.7324],[234,-3.723],[235,-3.723],[236,-3.7555],[237,-3.8249],[238,-4.5482],[239,-4.7869],[240,-4.919],[241,-5.0055],[243,-5.1329],[245,-5.6059],[247,-5.8424],[422,-5.8424]]
       }
       thumb_03_l_ctrl.Rotation.Z {
        type=MovieSceneScriptingFloatChannel
        default=-69.3636
        thinned_from=423
        keys [[0,-69.3636],[224,-69.3636],[225,-66.8967],[226,-60.6698],[227,-52.4436],[228,-43.9786],[229,-29.3764],[230,-7.7229],[231,13.1374],[232,27.4164],[233,35.9994],[234,40.0902],[235,40.0902],[236,39.4135],[237,37.9688],[238,3.4059],[239,-8.9566],[240,-16.2967],[241,-21.1784],[243,-28.4836],[245,-55.7368],[247,-69.3636],[422,-69.3636]]
       }
       thumb_03_l_ctrl.Scale.X {type=MovieSceneScriptingFloatChannel, default=1, const [1,423]}
       thumb_03_l_ctrl.Scale.Y {type=MovieSceneScriptingFloatChannel, default=1, const [1,423]}
       thumb_03_l_ctrl.Scale.Z {type=MovieSceneScriptingFloatChannel, default=1, const [1,423]}
       index_metacarpal_l_ctrl.Location.X {type=MovieSceneScriptingFloatChannel, default=0, thinned_from=423, keys [[0,0],[422,0]]}
       index_metacarpal_l_ctrl.Location.Y {type=MovieSceneScriptingFloatChannel, default=0, thinned_from=423, keys [[0,0],[422,0]]}
       index_metacarpal_l_ctrl.Location.Z {type=MovieSceneScriptingFloatChannel, default=0, const [0,423]}
       index_metacarpal_l_ctrl.Rotation.X {
        type=MovieSceneScriptingFloatChannel
        default=-2.9693
        thinned_from=423
        keys [[0,-2.9693],[215,-2.9693],[217,-2.9509],[219,-2.9138],[220,-2.8861],[221,-2.8517],[223,-2.7726],[224,-2.7401],[225,-2.6963],[226,-2.6247],[227,-2.5088],[228,-2.3323],[229,-1.9659],[230,-1.4225],[231,-0.8991],[232,-0.5408],[233,-0.4624],[234,-0.4512],[235,-0.4512],[236,-0.4898],[237,-0.5723],[238,-1.125],[239,-1.7152],[240,-1.8722],[241,-1.975],[243,-2.1264],[245,-2.6883],[247,-2.9693],[422,-2.9693]]
       }
       index_metacarpal_l_ctrl.Rotation.Y {
        type=MovieSceneScriptingFloatChannel
        default=4.9934
        thinned_from=423
        keys [[0,4.9934],[130,4.9934],[131,4.9559],[132,4.7853],[134,3.7784],[135,3.5534],[215,3.5534],[216,3.5282],[217,3.4942],[218,3.4435],[219,3.3745],[220,3.2854],[221,3.1746],[222,3.0403],[223,2.9194],[224,2.8149],[225,2.6737],[226,2.4429],[227,2.0695],[228,1.5007],[229,0.3199],[230,-1.4312],[231,-3.118],[232,-4.2727],[233,-4.5253],[234,-4.5614],[235,-4.5614],[236,-4.4368],[237,-4.1709],[238,-2.3901],[239,-0.488],[240,0.0181],[241,0.3494],[243,0.8371],[245,3.608],[247,4.9934],[422,4.9934]]
       }
       index_metacarpal_l_ctrl.Rotation.Z {
        type=MovieSceneScriptingFloatChannel
        default=-25.32
        thinned_from=423
        keys [[0,-25.32],[130,-25.32],[131,-23.3201],[132,-20.8201],[134,-13.3204],[215,-13.3204],[217,-13.2247],[218,-13.1429],[219,-13.0315],[220,-12.8877],[221,-12.7088],[223,-12.297],[224,-12.1283],[225,-11.9004],[226,-11.5278],[227,-10.9252],[228,-10.007],[229,-8.1011],[230,-5.2748],[231,-2.552],[232,-0.6883],[233,5.0495],[234,8.8817],[235,8.8817],[236,8.1038],[237,6.4428],[238,0.5671],[239,-4.786],[240,-5.9504],[241,-6.6796],[243,-7.703],[245,-19.4476],[247,-25.32],[422,-25.32]]
       }
       index_metacarpal_l_ctrl.Scale.X {type=MovieSceneScriptingFloatChannel, default=1, const [1,423]}
       index_metacarpal_l_ctrl.Scale.Y {type=MovieSceneScriptingFloatChannel, default=1, const [1,423]}
       index_metacarpal_l_ctrl.Scale.Z {type=MovieSceneScriptingFloatChannel, default=1, const [1,423]}
       index_01_l_ctrl.Location.X {
        type=MovieSceneScriptingFloatChannel
        default=-0.0198
        thinned_from=423
        keys [[0,-0.0198],[130,-0.0198],[131,-0.0306],[132,-0.08],[134,-0.3712],[135,-0.4363],[215,-0.436],[217,-0.4332],[218,-0.4305],[219,-0.4268],[220,-0.4221],[221,-0.4162],[223,-0.4026],[224,-0.3971],[225,-0.3896],[226,-0.3773],[227,-0.3575],[228,-0.3272],[229,-0.2645],[230,-0.1714],[231,-0.0818],[232,-0.0205],[233,-0.007],[234,-0.0051],[235,-0.0051],[236,-0.0117],[237,-0.0259],[238,-0.1205],[239,-0.2216],[240,-0.2485],[241,-0.2661],[243,-0.292],[245,-0.1105],[247,-0.0198],[422,-0.0198]]
       }
       index_01_l_ctrl.Location.Y {
        type=MovieSceneScriptingFloatChannel
        default=-1.1896
        thinned_from=423
        keys [[0,-1.1896],[130,-1.1896],[131,-1.1854],[132,-1.1659],[134,-1.0509],[135,-1.0252],[215,-1.0244],[217,-1.0178],[218,-1.0115],[219,-1.0028],[220,-0.9917],[221,-0.9779],[223,-0.946],[224,-0.933],[225,-0.9153],[226,-0.8865],[227,-0.8399],[228,-0.7689],[229,-0.6215],[230,-0.4028],[231,-0.1922],[232,-0.0481],[233,-0.0165],[234,-0.012],[235,-0.012],[236,-0.0276],[237,-0.0608],[238,-0.2831],[239,-0.5206],[240,-0.5838],[241,-0.6251],[243,-0.686],[245,-1.0218],[247,-1.1896],[422,-1.1896]]
       }
       index_01_l_ctrl.Location.Z {
        type=MovieSceneScriptingFloatChannel
        default=0.3879
        thinned_from=423
        keys [[0,0.3879],[130,0.3879],[131,0.3769],[132,0.3269],[134,0.0319],[135,-0.034],[216,-0.0339],[219,-0.0332],[224,-0.0309],[225,-0.0303],[226,-0.0294],[227,-0.0278],[228,-0.0255],[229,-0.0206],[230,-0.0133],[231,-0.0064],[232,-0.0016],[233,-0.0005],[235,-0.0004],[236,-0.0009],[237,-0.002],[238,-0.0094],[239,-0.0172],[240,-0.0193],[241,-0.0207],[243,-0.0227],[245,0.251],[247,0.3879],[422,0.3879]]
       }
       index_01_l_ctrl.Rotation.X {
        type=MovieSceneScriptingFloatChannel
        default=8.4656
        thinned_from=423
        keys [[0,8.4656],[214,8.4656],[215,8.4555],[216,8.4231],[217,8.3657],[218,8.2803],[219,8.164],[220,8.0139],[221,7.8271],[222,7.6008],[223,7.3972],[224,7.2211],[225,6.9831],[226,6.5942],[227,5.9651],[228,5.0065],[229,3.0167],[230,0.0661],[231,-2.7765],[232,-4.7222],[233,-5.1478],[234,-5.2086],[235,-5.2086],[236,-4.9988],[237,-4.5507],[238,-1.5498],[239,1.6554],[240,2.5082],[241,3.0664],[243,3.8882],[245,6.9398],[247,8.4656],[422,8.4656]]
       }
       index_01_l_ctrl.Rotation.Y {
        type=MovieSceneScriptingFloatChannel
        default=-1.3715
        thinned_from=423
        keys [[0,-1.3715],[215,-1.372],[217,-1.3766],[219,-1.3869],[221,-1.4041],[225,-1.4473],[226,-1.4672],[227,-1.4994],[228,-1.5485],[229,-1.6503],[230,-1.8013],[231,-1.9468],[232,-2.0464],[233,-2.0682],[235,-2.0713],[236,-2.0605],[237,-2.0376],[238,-1.884],[239,-1.72],[240,-1.6763],[241,-1.6478],[243,-1.6057],[245,-1.4495],[247,-1.3715],[422,-1.3715]]
       }
       index_01_l_ctrl.Rotation.Z {
        type=MovieSceneScriptingFloatChannel
        default=-86.8568
        thinned_from=423
        keys [[0,-86.8568],[130,-86.8568],[131,-85.5268],[132,-82.5968],[134,-68.7368],[135,-66.4322],[215,-66.3755],[216,-66.1948],[217,-65.8737],[218,-65.3961],[219,-64.7458],[220,-63.9066],[221,-62.8624],[222,-61.5969],[223,-60.4585],[224,-59.4735],[225,-58.1432],[226,-55.9686],[227,-52.4508],[228,-47.091],[229,-35.9654],[230,-19.4672],[231,-3.5734],[232,7.306],[233,15.1099],[234,19.1299],[235,19.1299],[236,18.4255],[237,16.9216],[238,-0.4515],[239,-20.2971],[240,-25.9406],[241,-29.6807],[243,-35.2574],[245,-69.6568],[247,-86.8568],[422,-86.8568]]
       }
       index_01_l_ctrl.Scale.X {type=MovieSceneScriptingFloatChannel, default=1, const [1,423]}
       index_01_l_ctrl.Scale.Y {type=MovieSceneScriptingFloatChannel, default=1, const [1,423]}
       index_01_l_ctrl.Scale.Z {type=MovieSceneScriptingFloatChannel, default=1, const [1,423]}
       index_02_l_ctrl.Location.X {type=MovieSceneScriptingFloatChannel, default=0, thinned_from=423, keys [[0,0],[422,0]]}
       index_02_l_ctrl.Location.Y {type=MovieSceneScriptingFloatChannel, default=0, thinned_from=423, keys [[0,0],[422,0]]}
       index_02_l_ctrl.Location.Z {type=MovieSceneScriptingFloatChannel, default=0, thinned_from=423, keys [[0,0],[422,0]]}
       index_02_l_ctrl.Rotation.X {
        type=MovieSceneScriptingFloatChannel
        default=2.7848
        thinned_from=423
        keys [[0,2.7848],[215,2.7828],[217,2.7647],[218,2.7475],[219,2.7241],[220,2.6939],[221,2.6563],[223,2.5698],[224,2.5343],[225,2.4864],[226,2.4082],[227,2.2815],[228,2.0886],[229,1.6881],[230,1.0943],[231,0.5222],[232,0.1305],[233,0.0449],[234,0.0326],[235,0.0326],[236,0.0749],[237,0.1651],[238,0.769],[239,1.4141],[240,1.5858],[241,1.6981],[243,1.8635],[245,2.4777],[247,2.7848],[422,2.7848]]
       }
       index_02_l_ctrl.Rotation.Y {
        type=MovieSceneScriptingFloatChannel
        default=-2.6518
        thinned_from=423
        keys [[0,-2.6518],[215,-2.6498],[217,-2.6326],[218,-2.6162],[219,-2.594],[220,-2.5652],[221,-2.5294],[223,-2.447],[224,-2.4133],[225,-2.3677],[226,-2.2931],[227,-2.1725],[228,-1.9888],[229,-1.6075],[230,-1.042],[231,-0.4972],[232,-0.1243],[233,-0.0427],[234,-0.0311],[235,-0.0311],[236,-0.0713],[237,-0.1572],[238,-0.7323],[239,-1.3466],[240,-1.51],[241,-1.617],[243,-1.7745],[245,-2.3593],[247,-2.6518],[422,-2.6518]]
       }
       index_02_l_ctrl.Rotation.Z {
        type=MovieSceneScriptingFloatChannel
        default=-67.2511
        thinned_from=423
        keys [[0,-67.2511],[214,-67.2511],[215,-67.1743],[216,-66.9293],[217,-66.4941],[218,-65.8469],[219,-64.9655],[220,-63.8283],[221,-62.413],[222,-60.698],[223,-59.155],[224,-57.8201],[225,-56.0171],[226,-53.07],[227,-48.3025],[228,-41.0385],[229,-25.9603],[230,-3.6008],[231,17.9396],[232,32.6841],[233,41.3738],[234,45.4743],[235,45.4743],[236,44.8015],[237,43.3648],[238,21.1431],[239,-26.4924],[240,-17.8628],[241,-17.5002],[243,-25.0421],[245,-53.1813],[247,-67.2511],[422,-67.2511]]
       }
       index_02_l_ctrl.Scale.X {type=MovieSceneScriptingFloatChannel, default=1, const [1,423]}
       index_02_l_ctrl.Scale.Y {type=MovieSceneScriptingFloatChannel, default=1, const [1,423]}
       index_02_l_ctrl.Scale.Z {type=MovieSceneScriptingFloatChannel, default=1, const [1,423]}
       index_03_l_ctrl.Location.X {type=MovieSceneScriptingFloatChannel, default=-0.0001, thinned_from=423, keys [[0,-0.0001],[422,-0.0001]]}
       index_03_l_ctrl.Location.Y {type=MovieSceneScriptingFloatChannel, default=0.0003, thinned_from=423, keys [[0,0.0003],[422,0.0003]]}
       index_03_l_ctrl.Location.Z {type=MovieSceneScriptingFloatChannel, default=0, thinned_from=423, keys [[0,0],[422,0]]}
       index_03_l_ctrl.Rotation.X {
        type=MovieSceneScriptingFloatChannel
        default=0.046
        thinned_from=423
        keys [[0,0.046],[214,0.046],[215,0.0459],[216,0.0454],[217,0.0445],[218,0.0431],[219,0.0413],[220,0.0389],[221,0.0359],[222,0.0322],[223,0.0288],[224,0.0259],[225,0.0219],[226,0.0153],[227,0.0046],[228,-0.0118],[229,-0.0447],[230,-0.0849],[231,-0.1076],[232,-0.1115],[234,-0.1077],[235,-0.1077],[236,-0.108],[237,-0.1087],[238,-0.1091],[239,-0.0829],[240,-0.0706],[241,-0.0614],[242,-0.0541],[243,-0.0465],[244,-0.0161],[245,0.0156],[246,0.0311],[247,0.046],[422,0.046]]
       }
       index_03_l_ctrl.Rotation.Y {
        type=MovieSceneScriptingFloatChannel
        default=-0.1016
        thinned_from=423
        keys [[0,-0.1016],[215,-0.1017],[217,-0.1023],[219,-0.1036],[221,-0.1056],[222,-0.1068],[223,-0.1078],[225,-0.1094],[227,-0.1115],[228,-0.1109],[229,-0.1022],[230,-0.0723],[231,-0.0296],[232,0.0034],[233,0.0212],[234,0.0291],[235,0.0291],[236,0.0279],[237,0.0253],[238,-0.0235],[239,-0.0746],[240,-0.0864],[241,-0.0931],[242,-0.0976],[243,-0.1014],[244,-0.1104],[245,-0.1105],[246,-0.1071],[247,-0.1016],[422,-0.1016]]
       }
       index_03_l_ctrl.Rotation.Z {
        type=MovieSceneScriptingFloatChannel
        default=-90.9605
        thinned_from=423
        keys [[0,-90.9605],[215,-90.8713],[216,-90.5866],[217,-90.0809],[218,-89.3288],[219,-88.3048],[220,-86.9833],[221,-85.3389],[222,-83.346],[223,-81.5532],[224,-80.0021],[225,-77.9072],[226,-74.4827],[227,-68.943],[228,-60.5026],[229,-42.9823],[230,-17.0015],[231,8.0274],[232,25.1599],[233,34.3968],[234,38.5471],[235,38.5471],[236,37.8937],[237,36.4986],[238,11.2814],[239,-18.5742],[240,-27.3383],[241,-33.1783],[243,-41.9343],[245,-74.6183],[247,-90.9605],[422,-90.9605]]
       }
       index_03_l_ctrl.Scale.X {type=MovieSceneScriptingFloatChannel, default=1, const [1,423]}
       index_03_l_ctrl.Scale.Y {type=MovieSceneScriptingFloatChannel, default=1, const [1,423]}
       index_03_l_ctrl.Scale.Z {type=MovieSceneScriptingFloatChannel, default=1, const [1,423]}
       middle_metacarpal_l_ctrl.Location.X {type=MovieSceneScriptingFloatChannel, default=0, const [0,423]}
       middle_metacarpal_l_ctrl.Location.Y {type=MovieSceneScriptingFloatChannel, default=0, thinned_from=423, keys [[0,0],[422,0]]}
       middle_metacarpal_l_ctrl.Location.Z {type=MovieSceneScriptingFloatChannel, default=0, thinned_from=423, keys [[0,0],[422,0]]}
       middle_metacarpal_l_ctrl.Rotation.X {
        type=MovieSceneScriptingFloatChannel
        default=0.2131
        thinned_from=423
        keys [[0,0.2131],[215,0.2131],[217,0.2116],[218,0.2102],[219,0.2084],[220,0.2061],[221,0.2033],[223,0.1966],[224,0.1939],[225,0.1903],[226,0.1843],[227,0.1746],[228,0.1598],[229,0.1292],[230,0.0837],[231,0.04],[232,0.01],[233,0.0034],[234,0.0025],[235,0.0025],[236,0.0057],[237,0.0126],[238,0.0588],[239,0.1082],[240,0.1213],[241,0.1299],[243,0.1426],[245,0.1896],[247,0.2131],[422,0.2131]]
       }
       middle_metacarpal_l_ctrl.Rotation.Y {
        type=MovieSceneScriptingFloatChannel
        default=-1.1482
        thinned_from=423
        keys [[0,-1.1482],[215,-1.1482],[217,-1.1399],[218,-1.1328],[219,-1.1231],[220,-1.1107],[221,-1.0952],[223,-1.0595],[224,-1.0449],[225,-1.0251],[226,-0.9929],[227,-0.9407],[228,-0.8611],[229,-0.696],[230,-0.4512],[231,-0.2153],[232,-0.0538],[233,-0.0185],[234,-0.0135],[235,-0.0135],[236,-0.0309],[237,-0.0681],[238,-0.3171],[239,-0.583],[240,-0.6538],[241,-0.7001],[243,-0.7683],[245,-1.0215],[247,-1.1482],[422,-1.1482]]
       }
       middle_metacarpal_l_ctrl.Rotation.Z {
        type=MovieSceneScriptingFloatChannel
        default=-11.9648
        thinned_from=423
        keys [[0,-11.9648],[130,-11.9648],[131,-11.7231],[132,-11.4211],[134,-10.5148],[215,-10.5148],[217,-10.4389],[218,-10.374],[219,-10.2856],[220,-10.1716],[221,-10.0296],[223,-9.7029],[224,-9.5691],[225,-9.3882],[226,-9.0927],[227,-8.6146],[228,-7.8861],[229,-6.374],[230,-4.1317],[231,-1.9715],[232,-0.4929],[233,5.1567],[234,8.9808],[235,8.9808],[236,8.1998],[237,6.5321],[238,1.1478],[239,-3.5861],[240,-4.5592],[241,-5.1599],[243,-5.9888],[245,-9.9728],[247,-11.9648],[422,-11.9648]]
       }
       middle_metacarpal_l_ctrl.Scale.X {type=MovieSceneScriptingFloatChannel, default=1, const [1,423]}
       middle_metacarpal_l_ctrl.Scale.Y {type=MovieSceneScriptingFloatChannel, default=1, const [1,423]}
       middle_metacarpal_l_ctrl.Scale.Z {type=MovieSceneScriptingFloatChannel, default=1, const [1,423]}
       middle_01_l_ctrl.Location.X {
        type=MovieSceneScriptingFloatChannel
        default=-0.5532
        thinned_from=423
        keys [[0,-0.5532],[215,-0.5528],[217,-0.5492],[218,-0.5458],[219,-0.5411],[220,-0.5351],[221,-0.5277],[223,-0.5105],[224,-0.5034],[225,-0.4939],[226,-0.4784],[227,-0.4532],[228,-0.4149],[229,-0.3353],[230,-0.2174],[231,-0.1037],[232,-0.0259],[233,-0.0089],[234,-0.0065],[235,-0.0065],[236,-0.0149],[237,-0.0328],[238,-0.1528],[239,-0.2809],[240,-0.315],[241,-0.3373],[243,-0.3702],[245,-0.4922],[247,-0.5532],[422,-0.5532]]
       }
       middle_01_l_ctrl.Location.Y {
        type=MovieSceneScriptingFloatChannel
        default=-0.9634
        thinned_from=423
        keys [[0,-0.9634],[215,-0.9627],[217,-0.9564],[218,-0.9505],[219,-0.9424],[220,-0.9319],[221,-0.9189],[223,-0.889],[224,-0.8767],[225,-0.8602],[226,-0.8331],[227,-0.7893],[228,-0.7225],[229,-0.584],[230,-0.3786],[231,-0.1806],[232,-0.0452],[233,-0.0155],[234,-0.0113],[235,-0.0113],[236,-0.0259],[237,-0.0571],[238,-0.266],[239,-0.4892],[240,-0.5486],[241,-0.5875],[243,-0.6447],[245,-0.8572],[247,-0.9634],[422,-0.9634]]
       }
       middle_01_l_ctrl.Location.Z {
        type=MovieSceneScriptingFloatChannel
        default=-0.0915
        thinned_from=423
        keys [[0,-0.0915],[215,-0.0914],[217,-0.0908],[218,-0.0902],[219,-0.0895],[220,-0.0885],[221,-0.0872],[223,-0.0844],[224,-0.0832],[225,-0.0817],[226,-0.0791],[227,-0.0749],[228,-0.0686],[229,-0.0554],[230,-0.0359],[231,-0.0171],[232,-0.0043],[233,-0.0015],[234,-0.0011],[235,-0.0011],[236,-0.0025],[237,-0.0054],[238,-0.0253],[239,-0.0464],[240,-0.0521],[241,-0.0558],[243,-0.0612],[245,-0.0814],[247,-0.0915],[422,-0.0915]]
       }
       middle_01_l_ctrl.Rotation.X {type=MovieSceneScriptingFloatChannel, default=0, thinned_from=423, keys [[0,0],[422,0]]}
       middle_01_l_ctrl.Rotation.Y {type=MovieSceneScriptingFloatChannel, default=0, thinned_from=423, keys [[0,0],[422,0]]}
       middle_01_l_ctrl.Rotation.Z {
        type=MovieSceneScriptingFloatChannel
        default=-76.2058
        thinned_from=423
        keys [[0,-76.2058],[130,-76.2058],[131,-76.3277],[132,-76.8822],[134,-80.1545],[135,-80.8858],[215,-80.8351],[217,-80.386],[219,-79.3767],[220,-78.6258],[221,-77.6914],[223,-75.5402],[224,-74.6588],[225,-73.4683],[226,-71.5224],[227,-68.3746],[228,-63.5784],[229,-53.6227],[230,-38.8594],[231,-24.637],[232,-14.9016],[233,-7.3599],[234,-3.3637],[235,-3.3637],[236,-4.0775],[237,-5.6014],[238,-21.5221],[239,-39.5307],[240,-43.6152],[241,-45.9826],[243,-49.0003],[245,-67.1372],[247,-76.2058],[422,-76.2058]]
       }
       middle_01_l_ctrl.Scale.X {type=MovieSceneScriptingFloatChannel, default=1, const [1,423]}
       middle_01_l_ctrl.Scale.Y {type=MovieSceneScriptingFloatChannel, default=1, const [1,423]}
       middle_01_l_ctrl.Scale.Z {type=MovieSceneScriptingFloatChannel, default=1, const [1,423]}
       middle_02_l_ctrl.Location.X {type=MovieSceneScriptingFloatChannel, default=-0.0452, thinned_from=423, keys [[0,-0.0452],[422,-0.0452]]}
       middle_02_l_ctrl.Location.Y {type=MovieSceneScriptingFloatChannel, default=0.0507, thinned_from=423, keys [[0,0.0507],[422,0.0507]]}
       middle_02_l_ctrl.Location.Z {type=MovieSceneScriptingFloatChannel, default=0.1146, thinned_from=423, keys [[0,0.1146],[422,0.1146]]}
       middle_02_l_ctrl.Rotation.X {
        type=MovieSceneScriptingFloatChannel
        default=-6.9244
        thinned_from=423
        keys [[0,-6.9244],[221,-6.9163],[225,-6.8709],[226,-6.8362],[227,-6.761],[228,-6.6001],[229,-6.0831],[230,-4.8675],[231,-3.2424],[232,-1.9293],[233,-1.2815],[234,-0.9999],[235,-0.9999],[236,-1.1002],[237,-1.3139],[238,-3.4034],[239,-5.3295],[240,-5.7507],[241,-5.9957],[242,-6.1597],[243,-6.3084],[244,-6.7244],[245,-6.926],[246,-6.9494],[247,-6.9244],[422,-6.9244]]
       }
       middle_02_l_ctrl.Rotation.Y {
        type=MovieSceneScriptingFloatChannel
        default=-1.6646
        thinned_from=423
        keys [[0,-1.6646],[214,-1.6646],[215,-1.6607],[216,-1.6485],[217,-1.6265],[218,-1.5938],[219,-1.5487],[220,-1.49],[221,-1.4159],[222,-1.3246],[223,-1.2413],[224,-1.1683],[225,-1.0685],[226,-0.9024],[227,-0.6271],[228,-0.196],[229,0.7168],[230,2.0009],[231,2.9996],[232,3.4718],[233,3.6807],[234,3.7565],[236,3.7519],[237,3.7404],[238,3.178],[239,1.9803],[240,1.5551],[241,1.2599],[242,1.0342],[243,0.8055],[244,-0.0577],[245,-0.8976],[246,-1.2932],[247,-1.6646],[422,-1.6646]]
       }
       middle_02_l_ctrl.Rotation.Z {
        type=MovieSceneScriptingFloatChannel
        default=-62.576
        thinned_from=423
        keys [[0,-62.576],[214,-62.576],[215,-62.502],[216,-62.2659],[217,-61.8466],[218,-61.2229],[219,-60.3738],[220,-59.2781],[221,-57.9147],[222,-56.2625],[223,-54.7763],[224,-53.4905],[225,-51.7541],[226,-48.9159],[227,-44.3255],[228,-37.3324],[229,-22.8175],[230,-1.2829],[231,19.4898],[232,33.7264],[233,42.3104],[234,46.4079],[235,46.4079],[236,45.7293],[237,44.2804],[238,22.6822],[239,-2.5114],[240,-9.8139],[241,-14.6688],[243,-21.9316],[245,-49.0225],[247,-62.576],[422,-62.576]]
       }
       middle_02_l_ctrl.Scale.X {type=MovieSceneScriptingFloatChannel, default=1, const [1,423]}
       middle_02_l_ctrl.Scale.Y {type=MovieSceneScriptingFloatChannel, default=1, const [1,423]}
       middle_02_l_ctrl.Scale.Z {type=MovieSceneScriptingFloatChannel, default=1, const [1,423]}
       middle_03_l_ctrl.Location.X {type=MovieSceneScriptingFloatChannel, default=0.0206, thinned_from=423, keys [[0,0.0206],[422,0.0206]]}
       middle_03_l_ctrl.Location.Y {type=MovieSceneScriptingFloatChannel, default=-0.0583, thinned_from=423, keys [[0,-0.0583],[422,-0.0583]]}
       middle_03_l_ctrl.Location.Z {type=MovieSceneScriptingFloatChannel, default=0.1324, thinned_from=423, keys [[0,0.1324],[422,0.1324]]}
       middle_03_l_ctrl.Rotation.X {
        type=MovieSceneScriptingFloatChannel
        default=-2.0926
        thinned_from=423
        keys [[0,-2.0926],[215,-2.0912],[217,-2.0792],[218,-2.0676],[219,-2.0514],[220,-2.0298],[221,-2.002],[223,-1.9341],[224,-1.9046],[225,-1.8635],[226,-1.7928],[227,-1.6699],[228,-1.4636],[229,-0.974],[230,-0.1507],[231,0.6625],[232,1.1785],[233,1.4648],[234,1.5919],[235,1.5919],[236,1.5712],[237,1.5263],[238,0.7867],[239,-0.1984],[240,-0.4848],[241,-0.6714],[242,-0.8081],[243,-0.9417],[244,-1.4051],[245,-1.7955],[246,-1.9568],[247,-2.0926],[422,-2.0926]]
       }
       middle_03_l_ctrl.Rotation.Y {
        type=MovieSceneScriptingFloatChannel
        default=1.0729
        thinned_from=423
        keys [[0,1.0729],[214,1.0729],[215,1.0755],[216,1.0838],[217,1.0985],[218,1.1203],[219,1.1497],[220,1.1874],[221,1.2337],[222,1.2889],[223,1.3377],[224,1.3793],[225,1.4344],[226,1.5218],[227,1.6558],[228,1.8407],[229,2.1403],[230,2.3466],[231,2.2563],[232,2.035],[233,1.8397],[234,1.7309],[235,1.7309],[236,1.7497],[237,1.789],[238,2.216],[239,2.3431],[240,2.301],[242,2.2083],[243,2.1548],[244,1.8857],[245,1.5187],[246,1.3043],[247,1.0729],[422,1.0729]]
       }
       middle_03_l_ctrl.Rotation.Z {
        type=MovieSceneScriptingFloatChannel
        default=-76.641
        thinned_from=423
        keys [[0,-76.641],[215,-76.5696],[216,-76.3419],[217,-75.9375],[218,-75.3361],[219,-74.5171],[220,-73.4602],[221,-72.145],[222,-70.5511],[223,-69.1171],[224,-67.8764],[225,-66.2006],[226,-63.4611],[227,-59.0291],[228,-52.2753],[229,-38.2519],[230,-17.4493],[231,2.5927],[232,16.3084],[233,24.7631],[234,28.8426],[235,28.8426],[236,28.1612],[237,26.7063],[238,5.7765],[239,-18.6162],[240,-25.6769],[241,-30.3703],[243,-37.3895],[245,-63.5619],[247,-76.641],[422,-76.641]]
       }
       middle_03_l_ctrl.Scale.X {type=MovieSceneScriptingFloatChannel, default=1, const [1,423]}
       middle_03_l_ctrl.Scale.Y {type=MovieSceneScriptingFloatChannel, default=1, const [1,423]}
       middle_03_l_ctrl.Scale.Z {type=MovieSceneScriptingFloatChannel, default=1, const [1,423]}
       ring_metacarpal_l_ctrl.Location.X {type=MovieSceneScriptingFloatChannel, default=0, thinned_from=423, keys [[0,0],[422,0]]}
       ring_metacarpal_l_ctrl.Location.Y {type=MovieSceneScriptingFloatChannel, default=0, thinned_from=423, keys [[0,0],[422,0]]}
       ring_metacarpal_l_ctrl.Location.Z {type=MovieSceneScriptingFloatChannel, default=0, thinned_from=423, keys [[0,0],[422,0]]}
       ring_metacarpal_l_ctrl.Rotation.X {
        type=MovieSceneScriptingFloatChannel
        default=5.3943
        thinned_from=423
        keys [[0,5.3943],[215,5.3943],[217,5.3565],[218,5.3241],[219,5.2801],[220,5.2232],[221,5.1525],[223,4.9896],[224,4.9228],[225,4.8327],[226,4.6854],[227,4.447],[228,4.0839],[229,3.33],[230,2.2122],[231,1.1353],[232,0.3981],[233,0.2369],[234,0.2138],[235,0.2138],[236,0.2933],[237,0.4631],[238,1.6],[239,2.8143],[240,3.1374],[241,3.3489],[243,3.6602],[245,4.8163],[247,5.3943],[422,5.3943]]
       }
       ring_metacarpal_l_ctrl.Rotation.Y {
        type=MovieSceneScriptingFloatChannel
        default=-2.9431
        thinned_from=423
        keys [[0,-2.9431],[215,-2.9431],[217,-2.9221],[218,-2.904],[219,-2.8794],[220,-2.8478],[221,-2.8083],[223,-2.7175],[224,-2.6803],[225,-2.6301],[226,-2.5479],[227,-2.4151],[228,-2.2126],[229,-1.7924],[230,-1.1693],[231,-0.569],[232,-0.1581],[233,-0.0682],[234,-0.0554],[235,-0.0554],[236,-0.0997],[237,-0.1943],[238,-0.8281],[239,-1.505],[240,-1.685],[241,-1.8029],[243,-1.9765],[245,-2.6209],[247,-2.9431],[422,-2.9431]]
       }
       ring_metacarpal_l_ctrl.Rotation.Z {
        type=MovieSceneScriptingFloatChannel
        default=-10.4667
        thinned_from=423
        keys [[0,-10.4667],[215,-10.4667],[217,-10.3834],[218,-10.3122],[219,-10.2151],[220,-10.09],[221,-9.9342],[223,-9.5755],[224,-9.4286],[225,-9.2301],[226,-8.9057],[227,-8.3809],[228,-7.5813],[229,-5.9216],[230,-3.4603],[231,-1.0893],[232,0.5338],[233,6.2164],[234,10.0436],[235,10.0436],[236,9.2637],[237,7.5985],[238,2.03],[239,-2.936],[240,-3.9809],[241,-4.6297],[243,-5.5315],[245,-8.8216],[247,-10.4667],[422,-10.4667]]
       }
       ring_metacarpal_l_ctrl.Scale.X {type=MovieSceneScriptingFloatChannel, default=1, const [1,423]}
       ring_metacarpal_l_ctrl.Scale.Y {type=MovieSceneScriptingFloatChannel, default=1, const [1,423]}
       ring_metacarpal_l_ctrl.Scale.Z {type=MovieSceneScriptingFloatChannel, default=1, const [1,423]}
       ring_01_l_ctrl.Location.X {
        type=MovieSceneScriptingFloatChannel
        default=-0.3709
        thinned_from=423
        keys [[0,-0.3709],[215,-0.3706],[217,-0.3682],[218,-0.3659],[219,-0.3628],[220,-0.3588],[221,-0.3537],[223,-0.3422],[224,-0.3375],[225,-0.3311],[226,-0.3207],[227,-0.3038],[228,-0.2781],[229,-0.2248],[230,-0.1457],[231,-0.0695],[232,-0.0174],[233,-0.006],[234,-0.0043],[235,-0.0043],[236,-0.01],[237,-0.022],[238,-0.1024],[239,-0.1883],[240,-0.2112],[241,-0.2261],[243,-0.2482],[245,-0.33],[247,-0.3709],[422,-0.3709]]
       }
       ring_01_l_ctrl.Location.Y {
        type=MovieSceneScriptingFloatChannel
        default=-1.0509
        thinned_from=423
        keys [[0,-1.0509],[215,-1.0501],[217,-1.0433],[218,-1.0368],[219,-1.028],[220,-1.0166],[221,-1.0024],[223,-0.9698],[224,-0.9564],[225,-0.9383],[226,-0.9088],[227,-0.861],[228,-0.7882],[229,-0.637],[230,-0.4129],[231,-0.197],[232,-0.0493],[233,-0.0169],[234,-0.0123],[235,-0.0123],[236,-0.0283],[237,-0.0623],[238,-0.2902],[239,-0.5337],[240,-0.5984],[241,-0.6408],[243,-0.7032],[245,-0.935],[247,-1.0509],[422,-1.0509]]
       }
       ring_01_l_ctrl.Location.Z {
        type=MovieSceneScriptingFloatChannel
        default=-0.0244
        thinned_from=423
        keys [[0,-0.0244],[216,-0.0243],[220,-0.0236],[225,-0.0218],[226,-0.0211],[227,-0.02],[228,-0.0183],[229,-0.0148],[231,-0.0046],[232,-0.0011],[233,-0.0004],[235,-0.0003],[236,-0.0007],[237,-0.0014],[238,-0.0067],[239,-0.0124],[240,-0.0139],[241,-0.0149],[243,-0.0163],[245,-0.0217],[247,-0.0244],[422,-0.0244]]
       }
       ring_01_l_ctrl.Rotation.X {
        type=MovieSceneScriptingFloatChannel
        default=-0.8239
        thinned_from=423
        keys [[0,-0.8239],[214,-0.8239],[215,-0.826],[216,-0.8327],[217,-0.8446],[218,-0.8623],[219,-0.8865],[220,-0.9176],[221,-0.9564],[222,-1.0034],[223,-1.0456],[224,-1.0822],[225,-1.1316],[226,-1.2123],[227,-1.3429],[228,-1.5419],[229,-1.9549],[230,-2.5673],[231,-3.1574],[232,-3.5613],[233,-3.6496],[234,-3.6622],[235,-3.6622],[236,-3.6187],[237,-3.5257],[238,-2.9028],[239,-2.2375],[240,-2.0604],[241,-1.9446],[243,-1.774],[245,-1.1406],[247,-0.8239],[422,-0.8239]]
       }
       ring_01_l_ctrl.Rotation.Y {
        type=MovieSceneScriptingFloatChannel
        default=-2.6856
        thinned_from=423
        keys [[0,-2.6856],[216,-2.6883],[219,-2.7049],[225,-2.7804],[226,-2.8053],[227,-2.8456],[228,-2.9069],[229,-3.0342],[230,-3.223],[231,-3.4049],[232,-3.5294],[233,-3.5566],[235,-3.5605],[236,-3.5471],[237,-3.5184],[238,-3.3264],[239,-3.1213],[240,-3.0668],[241,-3.031],[243,-2.9784],[245,-2.7832],[247,-2.6856],[422,-2.6856]]
       }
       ring_01_l_ctrl.Rotation.Z {
        type=MovieSceneScriptingFloatChannel
        default=-61.7969
        thinned_from=423
        keys [[0,-61.7969],[130,-61.7969],[131,-62.378],[132,-65.0216],[134,-80.6221],[135,-84.1083],[215,-84.0517],[217,-83.5506],[218,-83.0738],[219,-82.4245],[220,-81.5867],[221,-80.544],[223,-78.1439],[224,-77.1604],[225,-75.8322],[226,-73.6609],[227,-70.1487],[228,-64.7974],[229,-53.6891],[230,-37.2168],[231,-21.3479],[232,-10.4855],[233,-2.6855],[234,1.3341],[235,1.3341],[236,0.6296],[237,-0.8747],[238,-18.2263],[239,-38.0446],[240,-42.5997],[241,-45.254],[243,-48.6619],[245,-57.4186],[247,-61.7969],[422,-61.7969]]
       }
       ring_01_l_ctrl.Scale.X {type=MovieSceneScriptingFloatChannel, default=1, const [1,423]}
       ring_01_l_ctrl.Scale.Y {type=MovieSceneScriptingFloatChannel, default=1, const [1,423]}
       ring_01_l_ctrl.Scale.Z {type=MovieSceneScriptingFloatChannel, default=1, const [1,423]}
       ring_02_l_ctrl.Location.X {type=MovieSceneScriptingFloatChannel, default=0, thinned_from=423, keys [[0,0],[422,0]]}
       ring_02_l_ctrl.Location.Y {type=MovieSceneScriptingFloatChannel, default=0, thinned_from=423, keys [[0,0],[422,0]]}
       ring_02_l_ctrl.Location.Z {type=MovieSceneScriptingFloatChannel, default=0, thinned_from=423, keys [[0,0],[422,0]]}
       ring_02_l_ctrl.Rotation.X {
        type=MovieSceneScriptingFloatChannel
        default=-12.8636
        thinned_from=423
        keys [[0,-12.8636],[215,-12.8542],[217,-12.7708],[218,-12.6913],[219,-12.5832],[220,-12.4437],[221,-12.2701],[223,-11.8704],[224,-11.7066],[225,-11.4854],[226,-11.1238],[227,-10.5389],[228,-9.6477],[229,-7.7978],[230,-5.0546],[231,-2.4119],[232,-0.603],[233,-0.2073],[234,-0.1507],[235,-0.1507],[236,-0.3458],[237,-0.7624],[238,-3.5524],[239,-6.5322],[240,-7.3251],[241,-7.844],[243,-8.6081],[245,-11.4451],[247,-12.8636],[422,-12.8636]]
       }
       ring_02_l_ctrl.Rotation.Y {
        type=MovieSceneScriptingFloatChannel
        default=-0.6911
        thinned_from=423
        keys [[0,-0.6911],[215,-0.6906],[217,-0.6861],[218,-0.6818],[219,-0.676],[220,-0.6685],[221,-0.6592],[223,-0.6377],[224,-0.6289],[225,-0.617],[226,-0.5976],[227,-0.5662],[228,-0.5183],[229,-0.4189],[230,-0.2715],[231,-0.1296],[232,-0.0324],[233,-0.0111],[234,-0.0081],[235,-0.0081],[236,-0.0186],[237,-0.041],[238,-0.1908],[239,-0.3509],[240,-0.3935],[241,-0.4214],[243,-0.4624],[245,-0.6149],[247,-0.6911],[422,-0.6911]]
       }
       ring_02_l_ctrl.Rotation.Z {
        type=MovieSceneScriptingFloatChannel
        default=-59.6606
        thinned_from=423
        keys [[0,-59.6606],[215,-59.6156],[217,-59.217],[218,-58.8376],[219,-58.321],[220,-57.6544],[221,-56.8249],[223,-54.9152],[224,-54.1328],[225,-53.0759],[226,-51.3485],[227,-48.554],[228,-44.2963],[229,-35.4583],[230,-22.3524],[231,-9.7267],[232,-1.0842],[233,6.2071],[234,10.1804],[235,10.1804],[236,9.4578],[237,7.9149],[238,-6.6111],[239,-22.8626],[240,-27.3942],[241,-30.3867],[243,-34.8326],[245,-51.3846],[247,-59.6606],[422,-59.6606]]
       }
       ring_02_l_ctrl.Scale.X {type=MovieSceneScriptingFloatChannel, default=1, const [1,423]}
       ring_02_l_ctrl.Scale.Y {type=MovieSceneScriptingFloatChannel, default=1, const [1,423]}
       ring_02_l_ctrl.Scale.Z {type=MovieSceneScriptingFloatChannel, default=1, const [1,423]}
       ring_03_l_ctrl.Location.X {type=MovieSceneScriptingFloatChannel, default=0, thinned_from=423, keys [[0,0],[422,0]]}
       ring_03_l_ctrl.Location.Y {type=MovieSceneScriptingFloatChannel, default=0, thinned_from=423, keys [[0,0],[422,0]]}
       ring_03_l_ctrl.Location.Z {type=MovieSceneScriptingFloatChannel, default=0, thinned_from=423, keys [[0,0],[422,0]]}
       ring_03_l_ctrl.Rotation.X {type=MovieSceneScriptingFloatChannel, default=0, thinned_from=423, keys [[0,0],[422,0]]}
       ring_03_l_ctrl.Rotation.Y {type=MovieSceneScriptingFloatChannel, default=0, thinned_from=423, keys [[0,0],[422,0]]}
       ring_03_l_ctrl.Rotation.Z {
        type=MovieSceneScriptingFloatChannel
        default=-88.4683
        thinned_from=423
        keys [[0,-88.4683],[215,-88.3882],[216,-88.1327],[217,-87.6789],[218,-87.0039],[219,-86.0848],[220,-84.8988],[221,-83.4229],[222,-81.6342],[223,-80.0252],[224,-78.633],[225,-76.7528],[226,-73.6793],[227,-68.7074],[228,-61.1321],[229,-45.4075],[230,-22.0894],[231,0.3743],[232,15.7509],[233,24.5855],[234,28.6991],[235,28.6991],[236,28.0314],[237,26.6058],[238,3.5912],[239,-23.4535],[240,-31.3413],[241,-36.5917],[243,-44.4549],[245,-73.7971],[247,-88.4683],[422,-88.4683]]
       }
       ring_03_l_ctrl.Scale.X {type=MovieSceneScriptingFloatChannel, default=1, const [1,423]}
       ring_03_l_ctrl.Scale.Y {type=MovieSceneScriptingFloatChannel, default=1, const [1,423]}
       ring_03_l_ctrl.Scale.Z {type=MovieSceneScriptingFloatChannel, default=1, const [1,423]}
       pinky_metacarpal_l_ctrl.Location.X {type=MovieSceneScriptingFloatChannel, default=0, thinned_from=423, keys [[0,0],[422,0]]}
       pinky_metacarpal_l_ctrl.Location.Y {type=MovieSceneScriptingFloatChannel, default=0, thinned_from=423, keys [[0,0],[422,0]]}
       pinky_metacarpal_l_ctrl.Location.Z {type=MovieSceneScriptingFloatChannel, default=0, thinned_from=423, keys [[0,0],[422,0]]}
       pinky_metacarpal_l_ctrl.Rotation.X {
        type=MovieSceneScriptingFloatChannel
        default=9.8795
        thinned_from=423
        keys [[0,9.8795],[215,9.8795],[217,9.8262],[219,9.7185],[220,9.6383],[221,9.5386],[223,9.309],[224,9.2149],[225,9.0878],[226,8.8801],[227,8.5442],[228,8.0323],[229,6.9697],[230,5.3939],[231,3.876],[232,2.8369],[233,2.6096],[234,2.5771],[235,2.5771],[236,2.6892],[237,2.9284],[238,4.531],[239,6.2427],[240,6.6981],[241,6.9962],[243,7.4351],[245,9.0647],[247,9.8795],[422,9.8795]]
       }
       pinky_metacarpal_l_ctrl.Rotation.Y {
        type=MovieSceneScriptingFloatChannel
        default=-6.9881
        thinned_from=423
        keys [[0,-6.9881],[218,-6.9742],[225,-6.8769],[226,-6.8478],[227,-6.8006],[228,-6.7287],[229,-6.5795],[231,-6.145],[232,-5.9991],[233,-5.9671],[235,-5.9626],[236,-5.9783],[237,-6.0119],[238,-6.237],[239,-6.4774],[240,-6.5413],[243,-6.6448],[245,-6.8737],[247,-6.9881],[422,-6.9881]]
       }
       pinky_metacarpal_l_ctrl.Rotation.Z {
        type=MovieSceneScriptingFloatChannel
        default=-8.607
        thinned_from=423
        keys [[0,-8.607],[216,-8.5993],[219,-8.5523],[225,-8.3382],[226,-8.2677],[227,-8.1536],[228,-7.9798],[229,-7.619],[230,-7.0841],[231,-6.5687],[232,-6.2159],[233,-1.3552],[234,2.9764],[235,2.9764],[236,2.1861],[237,0.4988],[238,-3.4487],[239,-6.3725],[240,-6.7864],[241,-7.0112],[243,-7.2716],[245,-8.1619],[247,-8.607],[422,-8.607]]
       }
       pinky_metacarpal_l_ctrl.Scale.X {type=MovieSceneScriptingFloatChannel, default=1, const [1,423]}
       pinky_metacarpal_l_ctrl.Scale.Y {type=MovieSceneScriptingFloatChannel, default=1, const [1,423]}
       pinky_metacarpal_l_ctrl.Scale.Z {type=MovieSceneScriptingFloatChannel, default=1, const [1,423]}
       pinky_01_l_ctrl.Location.X {
        type=MovieSceneScriptingFloatChannel
        default=-0.3836
        thinned_from=423
        keys [[0,-0.3836],[215,-0.3833],[217,-0.3808],[218,-0.3784],[219,-0.3752],[220,-0.371],[221,-0.3659],[223,-0.354],[224,-0.3491],[225,-0.3425],[226,-0.3317],[227,-0.3142],[228,-0.2877],[229,-0.2325],[230,-0.1507],[231,-0.0719],[232,-0.018],[233,-0.0062],[234,-0.0045],[235,-0.0045],[236,-0.0103],[237,-0.0227],[238,-0.1059],[239,-0.1948],[240,-0.2184],[241,-0.2339],[243,-0.2567],[245,-0.3413],[247,-0.3836],[422,-0.3836]]
       }
       pinky_01_l_ctrl.Location.Y {
        type=MovieSceneScriptingFloatChannel
        default=-1.0453
        thinned_from=423
        keys [[0,-1.0452],[215,-1.0445],[217,-1.0377],[218,-1.0313],[219,-1.0225],[220,-1.0111],[221,-0.997],[223,-0.9645],[224,-0.9512],[225,-0.9333],[226,-0.9039],[227,-0.8564],[228,-0.7839],[229,-0.6336],[230,-0.4107],[231,-0.196],[232,-0.049],[233,-0.0168],[234,-0.0122],[235,-0.0122],[236,-0.0281],[237,-0.062],[238,-0.2887],[239,-0.5308],[240,-0.5952],[241,-0.6374],[243,-0.6995],[245,-0.93],[247,-1.0453],[422,-1.0453]]
       }
       pinky_01_l_ctrl.Location.Z {
        type=MovieSceneScriptingFloatChannel
        default=-0.0533
        thinned_from=423
        keys [[0,-0.0533],[215,-0.0533],[217,-0.0529],[219,-0.0521],[221,-0.0508],[224,-0.0485],[225,-0.0476],[226,-0.0461],[227,-0.0437],[228,-0.04],[229,-0.0323],[230,-0.0209],[231,-0.01],[232,-0.0025],[233,-0.0009],[234,-0.0006],[235,-0.0006],[236,-0.0014],[237,-0.0032],[238,-0.0147],[239,-0.0271],[240,-0.0304],[241,-0.0325],[243,-0.0357],[245,-0.0474],[247,-0.0533],[422,-0.0533]]
       }
       pinky_01_l_ctrl.Rotation.X {
        type=MovieSceneScriptingFloatChannel
        default=-7.5972
        thinned_from=423
        keys [[0,-7.5972],[215,-7.6028],[217,-7.6529],[218,-7.7007],[219,-7.7656],[220,-7.8494],[221,-7.9537],[223,-8.1939],[224,-8.2923],[225,-8.4252],[226,-8.6424],[227,-8.9938],[228,-9.5292],[229,-10.6405],[230,-12.2885],[231,-13.8761],[232,-14.9628],[233,-15.2006],[234,-15.2345],[235,-15.2345],[236,-15.1173],[237,-14.8671],[238,-13.191],[239,-11.4008],[240,-10.9245],[241,-10.6127],[243,-10.1537],[245,-8.4494],[247,-7.5972],[422,-7.5972]]
       }
       pinky_01_l_ctrl.Rotation.Y {
        type=MovieSceneScriptingFloatChannel
        default=-7.928
        thinned_from=423
        keys [[0,-7.928],[215,-7.9218],[217,-7.8668],[218,-7.8144],[219,-7.7431],[220,-7.6511],[221,-7.5366],[223,-7.273],[224,-7.165],[225,-7.0191],[226,-6.7807],[227,-6.395],[228,-5.8072],[229,-4.5873],[230,-2.7782],[231,-1.0355],[232,0.1575],[233,0.4185],[234,0.4557],[235,0.4557],[236,0.3271],[237,0.0524],[238,-1.7875],[239,-3.7527],[240,-4.2755],[241,-4.6178],[243,-5.1216],[245,-6.9926],[247,-7.928],[422,-7.928]]
       }
       pinky_01_l_ctrl.Rotation.Z {
        type=MovieSceneScriptingFloatChannel
        default=-49.5474
        thinned_from=423
        keys [[0,-49.5474],[130,-49.5474],[131,-50.4334],[132,-54.4649],[134,-78.255],[135,-83.5712],[215,-83.5194],[217,-83.0604],[219,-82.0291],[220,-81.2618],[221,-80.307],[223,-78.1089],[224,-77.2082],[225,-75.9918],[226,-74.0033],[227,-70.7868],[228,-65.8859],[229,-55.7128],[230,-40.6271],[231,-26.0941],[232,-16.1461],[233,-8.5557],[234,-4.5551],[235,-4.5551],[236,-5.2671],[237,-6.7873],[238,-22.9792],[239,-41.3298],[240,-45.1563],[241,-47.2311],[243,-49.629],[422,-49.5474]]
       }
       pinky_01_l_ctrl.Scale.X {type=MovieSceneScriptingFloatChannel, default=1, const [1,423]}
       pinky_01_l_ctrl.Scale.Y {type=MovieSceneScriptingFloatChannel, default=1, const [1,423]}
       pinky_01_l_ctrl.Scale.Z {type=MovieSceneScriptingFloatChannel, default=1, const [1,423]}
       pinky_02_l_ctrl.Location.X {
        type=MovieSceneScriptingFloatChannel
        default=0.3529
        thinned_from=423
        keys [[0,0.3529],[215,0.3526],[217,0.3503],[218,0.3481],[219,0.3452],[220,0.3413],[221,0.3366],[223,0.3256],[224,0.3211],[225,0.3151],[226,0.3051],[227,0.2891],[228,0.2646],[229,0.2139],[230,0.1387],[231,0.0662],[232,0.0165],[233,0.0057],[234,0.0041],[235,0.0041],[236,0.0095],[237,0.0209],[238,0.0974],[239,0.1792],[240,0.2009],[241,0.2152],[243,0.2361],[245,0.3139],[247,0.3529],[422,0.3529]]
       }
       pinky_02_l_ctrl.Location.Y {
        type=MovieSceneScriptingFloatChannel
        default=0.0017
        thinned_from=423
        keys [[0,0.0017],[227,0.0014],[229,0.0011],[232,0.0001],[237,0.0001],[239,0.0009],[422,0.0017]]
       }
       pinky_02_l_ctrl.Location.Z {
        type=MovieSceneScriptingFloatChannel
        default=-0.0756
        thinned_from=423
        keys [[0,-0.0756],[215,-0.0756],[217,-0.0751],[219,-0.074],[220,-0.0732],[221,-0.0721],[223,-0.0698],[224,-0.0688],[225,-0.0675],[226,-0.0654],[227,-0.062],[228,-0.0567],[229,-0.0458],[230,-0.0297],[231,-0.0142],[232,-0.0035],[233,-0.0012],[234,-0.0009],[235,-0.0009],[236,-0.002],[237,-0.0045],[238,-0.0209],[239,-0.0384],[240,-0.0431],[241,-0.0461],[243,-0.0506],[245,-0.0673],[247,-0.0756],[422,-0.0756]]
       }
       pinky_02_l_ctrl.Rotation.X {
        type=MovieSceneScriptingFloatChannel
        default=1.9939
        thinned_from=423
        keys [[0,1.9939],[215,1.9925],[217,1.9795],[218,1.9672],[219,1.9505],[220,1.9288],[221,1.9019],[223,1.84],[224,1.8146],[225,1.7803],[226,1.7243],[227,1.6336],[228,1.4955],[229,1.2087],[230,0.7835],[231,0.3739],[232,0.0935],[233,0.0321],[234,0.0234],[235,0.0234],[236,0.0536],[237,0.1182],[238,0.5506],[239,1.0125],[240,1.1354],[241,1.2159],[243,1.3343],[245,1.7741],[247,1.9939],[422,1.9939]]
       }
       pinky_02_l_ctrl.Rotation.Y {
        type=MovieSceneScriptingFloatChannel
        default=10.0966
        thinned_from=423
        keys [[0,10.0966],[215,10.0892],[217,10.0237],[218,9.9614],[219,9.8765],[220,9.767],[221,9.6307],[223,9.317],[224,9.1885],[225,9.0149],[226,8.7311],[227,8.272],[228,7.5725],[229,6.1205],[230,3.9674],[231,1.8931],[232,0.4733],[233,0.1627],[234,0.1183],[235,0.1183],[236,0.2715],[237,0.5984],[238,2.7882],[239,5.1271],[240,5.7494],[241,6.1568],[243,6.7565],[245,8.9832],[247,10.0966],[422,10.0966]]
       }
       pinky_02_l_ctrl.Rotation.Z {
        type=MovieSceneScriptingFloatChannel
        default=-57.9295
        thinned_from=423
        keys [[0,-57.9295],[215,-57.886],[217,-57.5007],[218,-57.1341],[219,-56.635],[220,-55.9908],[221,-55.1893],[223,-53.344],[224,-52.5879],[225,-51.5667],[226,-49.8975],[227,-47.1972],[228,-43.083],[229,-34.5429],[230,-21.8788],[231,-9.6786],[232,-1.3275],[233,5.897],[234,9.8644],[235,9.8644],[236,9.1394],[237,7.5914],[238,-6.5629],[239,-22.3461],[240,-26.7329],[241,-29.6282],[243,-33.927],[245,-49.9286],[247,-57.9295],[422,-57.9295]]
       }
       pinky_02_l_ctrl.Scale.X {type=MovieSceneScriptingFloatChannel, default=1, const [1,423]}
       pinky_02_l_ctrl.Scale.Y {type=MovieSceneScriptingFloatChannel, default=1, const [1,423]}
       pinky_02_l_ctrl.Scale.Z {type=MovieSceneScriptingFloatChannel, default=1, const [1,423]}
       pinky_03_l_ctrl.Location.X {type=MovieSceneScriptingFloatChannel, default=0, thinned_from=423, keys [[0,0],[422,0]]}
       pinky_03_l_ctrl.Location.Y {type=MovieSceneScriptingFloatChannel, default=0, thinned_from=423, keys [[0,0],[422,0]]}
       pinky_03_l_ctrl.Location.Z {type=MovieSceneScriptingFloatChannel, default=0, thinned_from=423, keys [[0,0],[422,0]]}
       pinky_03_l_ctrl.Rotation.X {
        type=MovieSceneScriptingFloatChannel
        default=3.1496
        thinned_from=423
        keys [[0,3.1496],[215,3.1473],[217,3.1268],[218,3.1074],[219,3.0809],[220,3.0468],[221,3.0042],[223,2.9064],[224,2.8663],[225,2.8121],[226,2.7236],[227,2.5804],[228,2.3622],[229,1.9092],[230,1.2376],[231,0.5905],[232,0.1476],[233,0.0507],[234,0.0369],[235,0.0369],[236,0.0847],[237,0.1867],[238,0.8698],[239,1.5994],[240,1.7935],[241,1.9206],[243,2.1076],[245,2.8023],[247,3.1496],[422,3.1496]]
       }
       pinky_03_l_ctrl.Rotation.Y {
        type=MovieSceneScriptingFloatChannel
        default=-4.3172
        thinned_from=423
        keys [[0,-4.3172],[215,-4.3141],[217,-4.2861],[218,-4.2594],[219,-4.2231],[220,-4.1763],[221,-4.118],[223,-3.9839],[224,-3.9289],[225,-3.8547],[226,-3.7333],[227,-3.537],[228,-3.2379],[229,-2.6171],[230,-1.6964],[231,-0.8095],[232,-0.2024],[233,-0.0696],[234,-0.0506],[235,-0.0506],[236,-0.1161],[237,-0.2559],[238,-1.1922],[239,-2.1923],[240,-2.4584],[241,-2.6326],[243,-2.889],[245,-3.8411],[247,-4.3172],[422,-4.3172]]
       }
       pinky_03_l_ctrl.Rotation.Z {
        type=MovieSceneScriptingFloatChannel
        default=-111.6788
        thinned_from=423
        keys [[0,-111.6788],[215,-111.5817],[216,-111.2719],[217,-110.7218],[218,-109.9035],[219,-108.7893],[220,-107.3515],[221,-105.5623],[222,-103.3941],[223,-101.4434],[224,-99.7558],[225,-97.4764],[226,-93.7505],[227,-87.7233],[228,-78.54],[229,-59.4775],[230,-31.2097],[231,-3.9776],[232,14.6629],[233,24.2454],[234,28.4271],[235,28.4271],[236,27.7861],[237,26.4173],[238,-0.6918],[239,-32.9616],[240,-42.4782],[241,-48.8247],[243,-58.3475],[245,-93.9015],[247,-111.6788],[422,-111.6788]]
       }
       pinky_03_l_ctrl.Scale.X {type=MovieSceneScriptingFloatChannel, default=1, const [1,423]}
       pinky_03_l_ctrl.Scale.Y {type=MovieSceneScriptingFloatChannel, default=1, const [1,423]}
       pinky_03_l_ctrl.Scale.Z {type=MovieSceneScriptingFloatChannel, default=1, const [1,423]}
       hand_l_ik_ctrl.Location.X {
        type=MovieSceneScriptingFloatChannel
        default=-21.8393
        thinned_from=423
        keys [[0,-19.212],[14,-19.3816],[16,-19.3184],[18,-19.102],[19,-18.9235],[22,-18.207],[24,-17.6895],[25,-17.4645],[27,-17.1198],[29,-16.8893],[31,-16.7713],[66,-16.8989],[69,-17.0266],[71,-17.2095],[75,-17.7804],[77,-18.1567],[81,-18.961],[82,-19.1043],[84,-19.263],[91,-19.3581],[108,-18.8026],[120,-18.772],[126,-18.8335],[128,-18.7545],[129,-18.6802],[130,-18.4621],[131,-17.8124],[132,-14.1968],[133,-5.0625],[134,0.029],[135,2.7688],[136,3.9731],[137,5.2411],[138,6.1252],[139,4.8831],[140,1.8739],[141,-9.1825],[142,-30.8042],[143,-48.1492],[144,-54.2872],[145,-54.7517],[146,-54.0983],[147,-54.1107],[148,-51.7059],[149,-46.014],[150,-42.372],[151,-40.961],[152,-41.545],[153,-42.4523],[154,-43.7943],[155,-44.9928],[156,-45.8668],[158,-46.8197],[160,-47.4579],[162,-47.7818],[218,-47.6121],[220,-47.4483],[222,-47.0246],[223,-46.6152],[224,-46.0232],[225,-43.8775],[226,-40.2097],[227,-35.2805],[228,-29.3031],[229,-20.4072],[230,-11.25],[231,-4.0449],[232,-0.433],[233,1.2732],[234,0.7592],[235,-2.2002],[236,-7.4118],[237,-13.2845],[238,-16.5189],[239,-18.1675],[240,-18.9874],[241,-19.6039],[242,-19.3253],[243,-18.6007],[244,-17.5796],[245,-17.4189],[246,-17.1935],[247,-17.3354],[248,-17.7863],[249,-18.1457],[250,-18.3393],[253,-18.7078],[254,-18.7882],[255,-18.8205],[267,-18.6846],[293,-18.7303],[297,-18.659],[299,-18.7393],[301,-18.8889],[302,-19.0093],[303,-19.1719],[306,-19.7816],[308,-20.0622],[309,-20.2484],[310,-20.477],[311,-20.6389],[313,-20.822],[315,-20.8979],[336,-20.2534],[347,-20.3158],[353,-20.5982],[358,-20.9897],[364,-21.1025],[378,-20.7812],[379,-20.7941],[382,-20.9859],[383,-20.9536],[384,-20.7738],[386,-20.1518],[387,-19.8793],[389,-19.4988],[391,-19.2578],[394,-19.0839],[401,-18.9917],[422,-19.2112]]
       }
       hand_l_ik_ctrl.Location.Y {
        type=MovieSceneScriptingFloatChannel
        default=9.6599
        thinned_from=423
        keys [[0,10.7295],[15,10.9051],[16,10.8949],[17,10.8481],[18,10.7626],[19,10.6263],[21,10.2801],[22,10.1404],[24,9.9411],[26,9.7889],[28,9.6886],[44,9.4411],[56,9.4596],[59,9.5279],[65,9.7366],[66,9.7454],[68,9.6986],[69,9.6556],[70,9.592],[71,9.5027],[72,9.3879],[77,8.6566],[79,8.4179],[81,8.2199],[83,8.076],[85,7.9817],[87,7.9327],[91,7.9056],[119,8.0653],[126,8.0742],[127,8.0849],[128,8.1183],[129,8.1803],[130,8.3398],[131,8.6388],[132,9.558],[133,10.0728],[134,9.5232],[135,6.9838],[136,4.5054],[137,2.2171],[138,0.0117],[139,-2.8031],[140,-6.891],[141,-19.2681],[142,-29.9689],[143,-33.7636],[144,-32.7612],[145,-31.3227],[146,-29.5489],[147,-28.3972],[148,-26.7154],[149,-24.5015],[150,-21.4401],[151,-19.5445],[152,-18.7577],[153,-18.7754],[154,-18.8979],[155,-19.1565],[158,-20.0239],[159,-20.2507],[160,-20.3821],[168,-20.771],[218,-20.9267],[222,-21.169],[223,-21.3372],[224,-21.642],[225,-22.2429],[226,-22.9101],[227,-23.3526],[228,-23.2805],[229,-22.4512],[230,-19.1874],[231,-13.9592],[232,-9.2108],[233,-3.231],[234,2.6449],[235,8.3473],[236,12.3736],[237,13.9905],[238,14.2294],[240,13.3442],[241,13.101],[242,13.0967],[243,13.6499],[244,15.1757],[245,15.7282],[246,15.4573],[247,15.035],[248,14.459],[249,13.9322],[250,13.5171],[251,13.2039],[252,12.9976],[253,12.8771],[254,12.7962],[255,12.743],[273,12.7841],[291,12.6526],[294,12.6859],[299,12.8251],[301,12.8209],[302,12.7588],[303,12.6398],[304,12.4598],[305,12.2124],[308,11.3357],[309,11.0696],[311,10.6508],[312,10.4611],[313,10.2966],[315,10.0451],[317,9.8579],[319,9.7464],[334,9.5559],[339,9.6103],[344,9.7293],[346,9.7451],[348,9.7207],[349,9.6834],[350,9.624],[351,9.5362],[352,9.4005],[353,9.2091],[357,8.2588],[358,8.0403],[361,7.4511],[362,7.2886],[363,7.1664],[364,7.0773],[365,7.0044],[366,6.9475],[367,6.9057],[368,6.8778],[370,6.8607],[372,6.8875],[374,6.9634],[376,7.0951],[377,7.1762],[378,7.2927],[379,7.4403],[380,7.6447],[381,7.9],[383,8.4621],[384,8.7605],[386,9.4508],[387,9.7782],[388,10.0569],[389,10.2958],[390,10.4921],[391,10.6536],[392,10.7897],[393,10.9008],[395,11.0541],[397,11.1246],[401,11.1243],[403,11.0753],[412,10.791],[422,10.73]]
       }
       hand_l_ik_ctrl.Location.Z {
        type=MovieSceneScriptingFloatChannel
        default=8.5439
        thinned_from=423
        keys [[0,7.273],[7,7.2388],[11,7.1694],[13,7.1645],[14,7.1813],[15,7.2289],[16,7.3239],[17,7.4741],[18,7.6871],[19,8.0022],[20,8.3854],[21,8.7302],[23,9.3163],[24,9.5372],[25,9.689],[26,9.8034],[27,9.8841],[28,9.9326],[30,9.9644],[46,9.7005],[55,9.6686],[58,9.7074],[65,9.8967],[66,9.8774],[67,9.8069],[68,9.7098],[69,9.5905],[70,9.4354],[71,9.2298],[72,8.9709],[73,8.6696],[74,8.3323],[76,7.5968],[78,6.8223],[79,6.4777],[80,6.1582],[81,5.8549],[82,5.6218],[83,5.465],[84,5.3496],[85,5.2674],[86,5.2103],[87,5.1733],[88,5.1541],[90,5.151],[92,5.1722],[95,5.2581],[98,5.3665],[102,5.5502],[104,5.6174],[108,5.6874],[122,5.7999],[125,5.7981],[127,5.8205],[128,5.8957],[129,6.0386],[130,6.115],[131,6.4208],[132,6.9466],[133,6.2587],[134,3.6352],[135,0.9976],[136,0.1062],[137,0.3282],[138,0.3851],[139,1.6924],[140,6.3563],[141,28.5172],[142,41.3783],[143,43.5553],[144,39.4656],[145,37.8911],[146,37.4827],[147,35.5555],[148,33.2528],[149,31.5332],[150,31.7518],[151,33.0633],[152,33.9124],[153,34.3245],[155,35.3966],[156,35.8385],[157,36.0766],[158,36.1719],[159,36.1917],[196,35.3162],[217,35.5023],[219,35.7066],[220,35.9065],[221,36.1888],[225,37.7292],[226,38.045],[227,38.2439],[228,38.0673],[229,36.5458],[230,33.0977],[231,28.0479],[232,24.2062],[233,20.7341],[234,17.5917],[235,15.0638],[236,14.5783],[237,14.8521],[239,15.8085],[240,15.8495],[241,15.7251],[242,15.159],[243,14.3367],[244,13.2667],[245,11.7101],[246,11.1572],[247,10.8258],[248,10.5839],[250,9.8879],[251,9.6255],[252,9.495],[253,9.4719],[265,9.7229],[273,9.76],[285,9.6128],[291,9.6256],[293,9.6649],[296,9.8126],[297,9.8051],[298,9.7158],[299,9.5613],[301,9.16],[302,8.8884],[303,8.5423],[304,8.1663],[305,7.7716],[306,7.4075],[307,7.0855],[308,6.7899],[309,6.5453],[310,6.3511],[311,6.1806],[312,6.0307],[313,5.9116],[314,5.8212],[315,5.7459],[316,5.7086],[317,5.7248],[318,5.7635],[320,5.8811],[322,6.0146],[324,6.1044],[325,6.1358],[340,6.1852],[343,6.1822],[345,6.1402],[346,6.1019],[347,6.0493],[348,5.9761],[349,5.8748],[350,5.7447],[351,5.5755],[352,5.3537],[353,5.0666],[355,4.3898],[356,4.0706],[357,3.7857],[358,3.5438],[359,3.3378],[360,3.1512],[361,2.9885],[362,2.8539],[363,2.752],[364,2.6762],[365,2.6156],[366,2.5684],[367,2.5309],[368,2.5063],[370,2.4844],[371,2.4883],[372,2.5135],[373,2.5674],[374,2.651],[375,2.7751],[377,3.0477],[378,3.1907],[379,3.3588],[380,3.5841],[381,3.8735],[382,4.2185],[383,4.6112],[384,5.1201],[386,6.3853],[387,6.9273],[388,7.3334],[389,7.639],[390,7.8479],[391,7.9897],[392,8.0938],[393,8.1644],[394,8.2095],[395,8.2344],[396,8.2427],[400,8.1591],[401,8.1186],[402,8.0495],[406,7.6742],[409,7.4808],[413,7.3276],[415,7.2857],[422,7.2747]]
       }
       hand_l_ik_ctrl.Rotation.X {
        type=MovieSceneScriptingFloatChannel
        default=62.7795
        thinned_from=423
        keys [[0,62.7795],[17,62.794],[24,63.8589],[69,63.7772],[71,63.526],[72,63.219],[73,62.6832],[75,61.154],[76,60.4907],[77,60.0395],[82,59.0052],[127,59.0223],[128,58.7299],[131,56.9604],[132,41.5048],[133,-8.9225],[134,-66.0441],[135,-98.9109],[136,-104.4924],[137,-109.0444],[138,-107.7188],[139,-98.2018],[141,-95.3149],[142,-92.9347],[143,-88.7584],[144,209.0719],[145,191.1553],[146,182.0941],[147,175.2522],[148,170.4591],[149,156.4314],[150,146.711],[151,139.7833],[152,137.0001],[153,136.0154],[155,135.0835],[221,134.9503],[223,135.4101],[224,135.9981],[225,137.8577],[226,141.8365],[227,149.5159],[228,164.114],[229,201.9845],[230,-108.8029],[231,-75.4401],[232,-61.1014],[233,5.9957],[234,42.1549],[235,49.3181],[236,61.6345],[237,59.6294],[238,56.6393],[239,61.2792],[240,63.3496],[241,63.3207],[242,63.0614],[243,61.8699],[244,60.4409],[245,60.3154],[247,60.9617],[248,61.1514],[249,61.6398],[251,63.127],[252,63.5681],[253,63.6848],[299,63.5612],[300,63.4684],[301,63.5864],[306,65.5302],[310,66.2837],[350,65.9861],[363,64.1181],[375,63.8665],[376,63.6876],[379,62.3877],[380,62.2178],[382,62.2566],[383,62.4072],[386,63.4985],[387,63.7321],[389,63.7384],[394,63.3032],[422,62.7832]]
       }
       hand_l_ik_ctrl.Rotation.Y {
        type=MovieSceneScriptingFloatChannel
        default=17.583
        thinned_from=423
        keys [[0,17.583],[7,17.4991],[14,17.6756],[16,17.6543],[17,17.5588],[18,17.3346],[19,16.8967],[21,15.7376],[23,14.6389],[24,14.2243],[25,13.9478],[26,13.7288],[27,13.566],[28,13.4634],[29,13.4095],[31,13.3833],[35,13.4553],[43,13.843],[46,13.9175],[47,13.9179],[48,13.9528],[49,13.9446],[50,13.9765],[55,13.9724],[58,13.8848],[65,13.466],[67,13.4057],[69,13.427],[70,13.5071],[71,13.6847],[72,13.9696],[73,14.3273],[74,14.7378],[77,16.0648],[78,16.4683],[80,17.1879],[81,17.5149],[82,17.7841],[84,18.178],[86,18.4501],[88,18.5903],[92,18.6128],[95,18.4933],[99,18.2249],[103,17.8799],[108,17.6944],[115,17.5776],[127,17.4046],[128,17.2631],[129,16.9947],[130,16.6602],[131,16.262],[132,16.0786],[133,31.7144],[134,8.5825],[135,-8.6549],[136,-13.8997],[137,-15.8295],[138,-18.8275],[139,-20.3145],[140,-17.3269],[141,26.0597],[142,55.2013],[143,62.59],[144,53.7833],[145,48.8473],[146,48.4279],[147,44.514],[148,41.1506],[149,35.4746],[150,33.4416],[151,33.2657],[152,32.2595],[153,30.3828],[154,29.1944],[155,28.2547],[156,27.8187],[159,27.4209],[161,27.0189],[162,26.8926],[164,26.8106],[216,26.8456],[217,26.9167],[218,27.0512],[219,27.2748],[220,27.6123],[221,28.0909],[222,28.7486],[223,29.6905],[224,31.0622],[225,34.4415],[226,39.7676],[227,46.6881],[228,54.3906],[229,59.6929],[230,53.7081],[231,46.5829],[232,29.6475],[233,24.0395],[234,6.9295],[235,0.3313],[236,3.6157],[237,16.3327],[238,28.0133],[239,18.275],[240,7.8071],[241,5.245],[242,3.1269],[243,3.7078],[244,4.7803],[245,11.1215],[246,11.673],[247,12.4036],[249,14.4628],[250,15.1685],[251,15.6067],[252,15.8029],[253,15.8287],[263,15.2665],[272,15.0903],[285,15.4346],[289,15.4455],[292,15.3563],[293,15.2995],[294,15.1854],[296,14.8757],[297,14.8233],[298,14.8869],[299,15.0194],[300,15.2136],[301,15.3752],[302,15.6227],[304,16.3663],[306,17.0431],[307,17.336],[308,17.5898],[310,17.9961],[312,18.4693],[315,18.9792],[316,19.0654],[317,19.03],[319,18.8409],[322,18.406],[324,18.2235],[325,18.1627],[343,18.1593],[347,18.2551],[348,18.3355],[349,18.4611],[350,18.6403],[351,18.8816],[352,19.1886],[353,19.58],[357,21.3113],[358,21.6593],[360,22.2335],[362,22.6716],[364,22.939],[367,23.1593],[373,23.1978],[376,23.0889],[378,23.0784],[379,23.0125],[380,22.8158],[381,22.4823],[382,22.04],[383,21.515],[384,20.7797],[385,19.8017],[386,18.7578],[387,17.8435],[388,17.1467],[389,16.6249],[390,16.2875],[391,16.0768],[392,15.9173],[394,15.6983],[396,15.5951],[400,15.6206],[401,15.683],[402,15.8182],[406,16.6273],[409,17.0513],[414,17.4565],[417,17.5657],[422,17.5798]]
       }
       hand_l_ik_ctrl.Rotation.Z {
        type=MovieSceneScriptingFloatChannel
        default=93.0959
        thinned_from=423
        keys [[0,93.0959],[18,92.9626],[21,93.6311],[26,95.6626],[31,96.6147],[69,96.4474],[71,96.0201],[74,94.5889],[81,90.5058],[84,89.5458],[92,89.1025],[128,90.6528],[131,90.1213],[132,77.5038],[133,35.3515],[134,15.4315],[135,8.6054],[136,1.622],[137,-1.217],[138,-3.7987],[139,-9.2105],[140,-13.983],[141,-34.0557],[142,-57.7987],[143,-76.1057],[144,239.0562],[145,228.368],[146,221.8426],[147,216.1465],[148,213.5294],[149,203.1474],[150,193.62],[151,186.3797],[152,182.5023],[153,180.2833],[155,177.2118],[156,176.5079],[164,174.8537],[218,175.1651],[221,176.4466],[223,178.3278],[224,179.888],[225,184.3755],[226,192.1119],[227,203.7307],[228,221.2737],[229,259.8538],[230,-57.3321],[231,-30.1525],[232,-27.1067],[233,-1.9041],[234,12.1069],[235,23.4163],[236,41.1727],[237,53.6817],[238,58.6809],[239,68.4033],[240,77.1559],[241,83.746],[242,86.9655],[244,90.2],[245,92.2867],[249,93.4505],[251,94.2754],[299,95.0484],[302,94.6734],[303,94.3513],[309,90.2067],[310,89.3787],[312,88.2564],[314,87.5657],[316,87.3274],[348,88.3207],[352,87.8422],[362,85.1204],[376,84.6414],[380,83.7102],[381,83.6277],[382,83.7668],[383,84.2642],[384,85.3917],[386,88.7518],[387,90.1425],[388,91.0987],[389,91.8103],[392,93.0658],[400,94.1483],[422,93.0996]]
       }
       hand_l_ik_ctrl.Scale.X {type=MovieSceneScriptingFloatChannel, default=1, const [1,423]}
       hand_l_ik_ctrl.Scale.Y {type=MovieSceneScriptingFloatChannel, default=1, const [1,423]}
       hand_l_ik_ctrl.Scale.Z {type=MovieSceneScriptingFloatChannel, default=1, const [1,423]}
       Softness {type=MovieSceneScriptingFloatChannel, default=0, const [0,423]}
       hand_l_ikorient_ctrl.Rotation.Roll {type=MovieSceneScriptingFloatChannel, default=0, const [0,423]}
       hand_l_ikorient_ctrl.Rotation.Pitch {type=MovieSceneScriptingFloatChannel, default=0, const [0,423]}
       hand_l_ikorient_ctrl.Rotation.Yaw {type=MovieSceneScriptingFloatChannel, default=0, const [0,423]}
       arm_l_pv_ik_ctrl.Location.X {
        type=MovieSceneScriptingFloatChannel
        default=27.3261
        thinned_from=423
        keys [[0,27.3261],[12,27.1338],[16,27.0238],[17,27.0662],[18,27.2026],[20,27.6633],[22,28.2988],[24,28.8029],[26,29.1716],[29,29.5169],[39,29.8598],[59,29.6046],[63,29.3522],[72,28.5892],[75,28.4651],[76,28.5102],[79,28.9666],[80,29.006],[81,28.9405],[83,28.6479],[86,28.0686],[89,27.6999],[93,27.4941],[98,27.6429],[108,28.1319],[120,28.1985],[126,27.9243],[128,27.926],[129,27.8818],[130,27.6893],[131,27.0061],[132,24.6399],[133,16.4609],[134,-0.7234],[135,-8.4411],[136,-9.8826],[137,-8.9535],[138,-6.4736],[139,-4.199],[140,-1.8138],[141,8.7419],[142,22.5728],[143,24.8632],[144,24.1555],[145,22.0659],[146,20.729],[147,19.6903],[148,17.3477],[149,11.0676],[150,7.5572],[151,6.7334],[152,6.97],[153,6.8788],[154,7.3553],[155,8.0426],[156,8.6446],[157,9.1156],[158,9.5586],[159,9.9531],[160,10.2657],[161,10.4926],[162,10.6492],[163,10.7619],[165,10.903],[170,11.1177],[188,11.3078],[196,11.3178],[217,11.197],[220,11.1157],[222,11.0035],[224,10.825],[225,9.9362],[226,8.3432],[227,6.3573],[228,4.4534],[229,2.6872],[230,2.1515],[231,3.9115],[232,12.3003],[233,18.8075],[234,21.1084],[235,21.8966],[236,22.5141],[237,23.1767],[238,23.4768],[239,23.8305],[240,24.128],[241,24.3392],[242,24.4424],[243,24.3725],[244,23.7626],[245,23.637],[246,23.9658],[248,24.7764],[250,25.4642],[251,25.7365],[252,25.9343],[255,26.2354],[259,26.3879],[295,26.1236],[297,26.0265],[300,25.7471],[302,25.6592],[304,25.7559],[307,26.1067],[308,26.1597],[309,26.1182],[313,25.5459],[318,24.5695],[320,24.3234],[323,24.2382],[326,24.2932],[334,25.0261],[340,25.1632],[353,24.8986],[361,25.1333],[370,25.0723],[381,25.4492],[383,25.6127],[388,26.5041],[390,26.75],[400,27.3662],[422,27.3267]]
       }
       arm_l_pv_ik_ctrl.Location.Y {
        type=MovieSceneScriptingFloatChannel
        default=19.0338
        thinned_from=423
        keys [[0,19.0338],[2,19.0763],[5,19.2209],[7,19.2437],[12,19.0516],[13,18.9372],[14,18.876],[16,18.9376],[17,19.0879],[18,19.7018],[19,20.5418],[20,21.5803],[21,23.1265],[22,24.9973],[23,26.8085],[24,28.3955],[25,29.571],[26,30.4897],[27,31.3114],[28,32.0088],[29,32.5576],[30,32.9421],[31,33.1978],[33,33.5223],[35,33.6665],[39,33.6262],[49,34.0242],[57,34.2038],[59,34.3849],[62,34.9366],[63,35.2552],[66,36.4704],[67,36.8148],[71,37.7609],[72,37.795],[73,37.7013],[74,37.4672],[75,36.9759],[76,35.9841],[77,34.2111],[78,31.9327],[79,29.4195],[80,26.7818],[81,24.2044],[82,21.9282],[83,19.9878],[84,18.2481],[85,16.8001],[86,15.7466],[87,15.0183],[88,14.4366],[89,13.9613],[90,13.5815],[91,13.2841],[92,13.0699],[93,12.9394],[94,12.8832],[95,12.9107],[97,13.1289],[98,13.313],[99,13.5764],[102,14.4959],[103,14.7512],[104,14.9524],[107,15.3735],[108,15.4851],[111,15.7112],[116,15.9464],[118,15.9638],[119,15.9315],[120,15.8343],[121,15.6164],[122,15.3142],[123,14.9634],[125,14.1409],[127,13.3005],[128,12.9315],[129,12.4339],[130,11.6763],[131,10.1831],[132,8.1004],[133,4.7331],[134,0.5006],[135,2.7575],[136,4.5638],[137,7.0206],[138,9.0605],[139,11.2929],[140,13.9995],[141,19.6685],[142,22.3249],[143,16.687],[144,27.2749],[145,33.4329],[146,35.8236],[147,38.0011],[148,39.7001],[149,42.3369],[150,42.0702],[151,41.2248],[152,41.5006],[153,41.9604],[155,43.236],[156,43.744],[157,44.0255],[159,44.2531],[219,43.9222],[223,44.1728],[224,44.1095],[225,43.2024],[226,41.5061],[227,39.0643],[228,35.9123],[229,31.8317],[230,27.0021],[231,21.9073],[232,19.9102],[233,23.8934],[234,28.5434],[235,31.1945],[236,32.5419],[237,33.4329],[238,33.86],[239,33.8557],[242,33.2307],[243,32.6875],[245,31.3543],[246,30.0807],[247,29.2763],[248,28.7414],[251,28.0088],[252,27.6889],[253,27.4893],[254,27.3989],[255,27.3745],[256,27.4066],[257,27.5103],[259,27.8765],[267,30.108],[269,30.5424],[272,30.9702],[276,31.2215],[289,31.0119],[292,31.1365],[293,31.2348],[294,31.419],[297,32.2409],[299,33.0182],[300,33.1633],[301,33.2012],[302,33.1273],[303,32.7099],[304,32.1117],[305,31.3583],[306,30.3145],[307,28.9755],[308,27.4831],[309,25.8684],[310,24.1309],[312,20.5126],[313,18.7495],[314,17.1195],[315,15.5854],[316,14.2232],[317,13.0585],[318,12.0816],[319,11.3562],[320,10.8529],[321,10.5913],[322,10.4689],[323,10.3991],[324,10.3797],[326,10.4004],[327,10.5268],[328,10.7426],[329,10.9924],[332,11.861],[334,12.3932],[335,12.594],[336,12.7336],[338,12.9402],[340,13.0919],[342,13.128],[343,13.1123],[345,12.9523],[346,12.8177],[347,12.6185],[348,12.3497],[349,11.9998],[350,11.5843],[351,11.1269],[352,10.6202],[353,10.0149],[354,9.3372],[355,8.7648],[356,8.3591],[357,8.0132],[358,7.608],[359,7.1157],[361,6.0879],[362,5.6182],[363,5.2176],[364,4.9018],[365,4.6432],[366,4.4221],[367,4.2476],[368,4.1191],[369,4.0208],[370,3.9885],[371,4.0559],[372,4.165],[373,4.3192],[374,4.5086],[375,4.7435],[376,5.0083],[379,5.8343],[380,6.185],[381,6.6745],[382,7.2975],[383,8.0447],[384,9.0678],[385,10.4082],[386,11.868],[387,13.2451],[388,14.443],[389,15.4666],[390,16.2776],[393,18.1906],[394,18.8001],[395,19.3204],[397,20.1578],[399,20.8414],[400,21.1211],[401,21.2833],[402,21.2825],[403,21.1585],[407,20.2485],[410,19.7532],[414,19.2717],[416,19.118],[419,19.0386],[422,19.0386]]
       }
       arm_l_pv_ik_ctrl.Location.Z {
        type=MovieSceneScriptingFloatChannel
        default=25.055
        thinned_from=423
        keys [[0,21.9466],[15,22.1423],[17,22.1336],[20,22.5819],[22,23.0286],[23,23.1015],[24,23.0821],[30,22.2993],[33,22.0522],[37,21.9178],[57,22.3376],[58,22.3138],[62,22.4108],[65,22.7087],[69,23.3239],[70,23.535],[72,24.127],[73,24.4731],[76,25.7368],[77,26.0742],[78,26.3386],[79,26.4975],[81,26.6007],[82,26.5317],[86,25.7981],[93,24.9757],[97,24.6837],[120,24.6002],[122,24.3941],[125,23.8891],[126,23.6465],[127,23.3154],[128,22.8841],[129,22.2995],[130,21.4423],[131,19.8265],[132,15.2177],[133,5.314],[134,-3.5063],[135,-6.7363],[136,-8.3136],[137,-10.5529],[138,-11.0103],[139,-9.4122],[140,-5.4731],[141,3.1967],[142,23.3055],[143,41.3187],[144,27.1391],[145,22.5073],[146,20.066],[147,17.9992],[148,13.4062],[149,5.69],[150,1.1626],[151,-0.1208],[152,0.5636],[153,1.5165],[154,3.0231],[155,4.4874],[156,5.6551],[157,6.4049],[158,7.0095],[159,7.4862],[160,7.8177],[161,8.0327],[162,8.1534],[163,8.1978],[164,8.1904],[168,8.0501],[169,8.0309],[170,7.9956],[176,7.8837],[189,7.8747],[216,7.9704],[221,8.0603],[222,8.0469],[223,7.9333],[224,7.7474],[225,6.2373],[226,3.8603],[227,1.3483],[228,-0.7491],[229,-2.4313],[230,-3.043],[231,-1.7874],[232,4.6651],[233,12.1251],[234,16.2017],[235,18.3403],[236,20.3539],[237,22.7825],[239,25.0955],[240,25.835],[241,26.1478],[242,25.829],[243,24.2276],[244,20.93],[245,19.6534],[246,19.5124],[247,19.8522],[248,20.6449],[249,21.3542],[250,21.7471],[251,21.9483],[252,21.9991],[255,21.9059],[257,21.7407],[259,21.6901],[262,21.7884],[274,22.6753],[283,22.8847],[295,22.7878],[297,22.8562],[298,22.9714],[299,23.1632],[301,23.4332],[302,23.635],[303,23.8853],[304,24.1977],[305,24.5693],[306,25.0154],[308,26.0568],[310,27.1893],[311,27.5444],[312,27.6689],[313,27.6763],[314,27.5739],[315,27.39],[316,27.1036],[319,26.0402],[320,25.7714],[321,25.5944],[323,25.3901],[326,25.2599],[335,25.7725],[339,25.8562],[345,25.7572],[350,25.432],[351,25.4162],[354,25.575],[355,25.718],[357,26.2256],[358,26.4105],[360,26.5109],[364,26.3931],[370,25.9839],[377,25.8335],[379,25.6658],[381,25.344],[382,25.1067],[383,24.7882],[384,24.3191],[386,23.1255],[387,22.6192],[388,22.2237],[389,21.8772],[390,21.5851],[391,21.3535],[392,21.2109],[393,21.1551],[395,21.157],[397,21.2369],[402,21.6411],[408,21.8569],[422,21.9453]]
       }
       arm_l_pv_ik_ctrl.Rotation.X {
        type=MovieSceneScriptingFloatChannel
        default=38.858
        thinned_from=423
        keys [[0,38.858],[16,38.6748],[17,38.777],[18,39.1625],[20,40.2286],[22,41.6519],[23,42.2014],[24,42.6317],[25,42.8854],[28,43.2233],[32,43.1884],[40,42.4738],[58,42.3666],[62,42.6424],[66,43.5044],[75,45.9682],[76,46.1365],[77,46.0859],[78,45.8757],[79,45.4784],[81,44.4219],[85,41.9135],[86,41.4251],[89,40.3865],[93,39.479],[95,39.2533],[98,39.1626],[117,39.6088],[120,39.5087],[123,38.9183],[128,37.5283],[129,37.0769],[130,36.4352],[131,34.7698],[132,29.0131],[133,15.6903],[134,-1.473],[135,-6.8421],[136,-9.1987],[137,-14.3071],[138,-16.5805],[139,-19.7053],[140,-23.2874],[141,-32.6343],[142,-53.908],[143,-92.2317],[144,-92.1018],[145,-91.1327],[146,-89.5017],[147,-90.2814],[148,-88.0859],[149,-81.1583],[150,-75.444],[151,-72.5679],[152,-72.9803],[153,-73.7129],[155,-76.9073],[156,-78.26],[158,-80.076],[159,-80.7628],[160,-81.2537],[162,-81.7729],[221,-81.7101],[223,-81.3665],[224,-81.0061],[225,-78.3553],[226,-73.7722],[227,-67.8813],[228,-61.2539],[229,-53.364],[230,-44.0659],[231,-31.7559],[232,-9.2465],[233,14.0999],[234,26.6278],[235,32.0972],[236,34.8666],[237,38.1216],[239,40.9678],[240,42.0299],[241,42.462],[242,42.0975],[243,40.3532],[244,36.5247],[245,35.9554],[246,36.5617],[247,37.7824],[248,39.1501],[249,40.1125],[250,40.5695],[251,40.7733],[252,40.753],[255,40.3911],[257,40.3362],[262,40.8121],[270,41.952],[278,42.385],[295,42.4439],[297,42.6483],[298,42.9213],[299,43.2914],[300,43.5116],[310,44.552],[311,44.3883],[312,43.9867],[313,43.4606],[315,42.1317],[318,39.5955],[319,38.8766],[320,38.3025],[321,37.9237],[322,37.6741],[323,37.5079],[326,37.3713],[328,37.6574],[334,38.985],[337,39.3339],[341,39.5493],[345,39.4331],[349,38.9619],[354,38.0918],[355,38.0373],[358,38.2338],[369,37.0561],[373,37.1755],[376,37.4229],[381,37.9505],[382,37.976],[383,37.9178],[384,37.7025],[386,37.0566],[387,36.8821],[389,36.8775],[391,37.0499],[393,37.4017],[400,38.8504],[403,39.2016],[408,39.2745],[422,38.8562]]
       }
       arm_l_pv_ik_ctrl.Rotation.Y {
        type=MovieSceneScriptingFloatChannel
        default=4.4598
        thinned_from=423
        keys [[0,4.4598],[1,4.4499],[3,4.4034],[5,4.3355],[7,4.2877],[11,4.2271],[12,4.2322],[13,4.2469],[14,4.2398],[15,4.2142],[16,4.0992],[17,3.7797],[18,3.1889],[19,2.346],[20,1.3685],[21,0.4828],[22,-0.2058],[23,-0.81],[24,-1.3526],[25,-1.7729],[26,-2.0693],[27,-2.2918],[28,-2.4555],[29,-2.5584],[30,-2.5987],[31,-2.582],[32,-2.513],[33,-2.4107],[34,-2.2859],[36,-2.0134],[37,-1.8886],[38,-1.7823],[39,-1.6969],[40,-1.6375],[43,-1.5098],[44,-1.4602],[45,-1.4333],[46,-1.3899],[47,-1.3508],[48,-1.3307],[49,-1.2984],[50,-1.269],[51,-1.2432],[52,-1.2211],[54,-1.1881],[55,-1.178],[56,-1.1707],[57,-1.168],[59,-1.2038],[60,-1.242],[61,-1.2988],[62,-1.3789],[63,-1.5124],[64,-1.705],[65,-1.9036],[66,-2.1572],[67,-2.4922],[68,-2.8202],[69,-3.0766],[70,-3.2558],[71,-3.3751],[72,-3.4277],[73,-3.3787],[74,-3.1962],[75,-2.8845],[76,-2.3724],[77,-1.4942],[78,-0.3533],[79,0.9133],[80,2.1923],[81,3.3742],[82,4.4034],[83,5.2459],[84,5.9575],[85,6.5546],[86,7.0481],[87,7.4455],[88,7.7667],[89,8.0313],[90,8.2467],[91,8.4127],[92,8.533],[93,8.6109],[94,8.667],[95,8.6717],[96,8.6342],[98,8.503],[100,8.2605],[103,7.8774],[105,7.7099],[108,7.5163],[111,7.3968],[115,7.2985],[119,7.2754],[120,7.2959],[121,7.3529],[123,7.5073],[124,7.5708],[125,7.6061],[126,7.5799],[127,7.4295],[128,7.2123],[129,6.928],[130,6.2283],[131,5.3384],[132,3.5877],[133,-2.785],[134,-4.0001],[135,-7.9955],[136,-9.4058],[137,-8.4135],[138,-4.2083],[139,0.3427],[140,4.3218],[141,19.0632],[142,46.2489],[143,59.3396],[144,25.4133],[145,14.0638],[146,10.7786],[147,6.0996],[148,1.1475],[149,-6.5083],[150,-7.1092],[151,-4.5657],[152,-3.1351],[153,-3.2256],[154,-2.7818],[155,-2.2392],[156,-1.7572],[157,-1.4028],[158,-1.1013],[159,-0.8498],[160,-0.6674],[161,-0.5456],[162,-0.4738],[163,-0.4396],[164,-0.4305],[166,-0.4383],[167,-0.4383],[168,-0.4476],[169,-0.4383],[170,-0.4456],[171,-0.436],[172,-0.4433],[173,-0.4435],[174,-0.442],[175,-0.443],[176,-0.4472],[177,-0.447],[178,-0.442],[179,-0.4429],[184,-0.4425],[187,-0.4427],[188,-0.4437],[189,-0.4428],[190,-0.4437],[191,-0.4423],[194,-0.4456],[195,-0.4495],[196,-0.4436],[199,-0.4424],[201,-0.4382],[202,-0.4443],[204,-0.4379],[205,-0.4445],[206,-0.4375],[207,-0.443],[208,-0.4356],[209,-0.443],[210,-0.4333],[211,-0.443],[214,-0.443],[215,-0.4415],[216,-0.4303],[217,-0.3919],[218,-0.3286],[219,-0.2242],[220,-0.068],[221,0.151],[222,0.4398],[223,0.6873],[224,0.9592],[225,0.6374],[226,0.1099],[227,-0.079],[228,0.6569],[229,2.5451],[230,6.7276],[231,12.8112],[232,20.974],[233,17.4456],[234,8.8943],[235,4.3846],[236,4.9904],[237,5.7046],[238,6.2647],[239,6.7744],[240,6.7906],[241,6.8851],[242,6.8054],[243,5.6806],[244,3.784],[245,1.436],[246,1.0778],[247,1.1435],[248,1.4603],[249,1.6635],[250,1.6608],[251,1.5762],[252,1.5349],[253,1.5932],[254,1.631],[255,1.6596],[256,1.6794],[257,1.687],[258,1.6629],[259,1.6102],[260,1.5393],[261,1.4572],[262,1.3651],[263,1.2588],[265,1.0293],[266,0.9172],[267,0.7996],[268,0.7107],[269,0.6178],[270,0.536],[271,0.4655],[272,0.4031],[273,0.3367],[274,0.2733],[275,0.2162],[276,0.1718],[277,0.1395],[278,0.1081],[279,0.092],[280,0.0729],[281,0.0639],[282,0.0545],[283,0.049],[284,0.052],[285,0.0621],[286,0.0736],[287,0.0889],[288,0.0863],[289,0.0969],[290,0.1067],[291,0.1133],[292,0.1138],[293,0.1057],[294,0.0893],[295,0.0633],[296,0.0065],[297,-0.1109],[298,-0.3712],[299,-0.672],[300,-0.8137],[301,-0.7179],[302,-0.4858],[303,-0.1523],[304,0.3031],[305,0.8654],[306,1.4818],[307,2.1357],[308,2.8249],[309,3.5958],[310,4.472],[311,5.3658],[312,6.2268],[313,7.0695],[314,7.8547],[315,8.6118],[316,9.3239],[317,9.9897],[318,10.5599],[319,11.0031],[320,11.3382],[321,11.5489],[322,11.6769],[323,11.7583],[325,11.8342],[326,11.839],[327,11.7816],[328,11.6648],[329,11.5199],[334,10.6412],[335,10.5074],[338,10.2426],[340,10.1192],[342,10.0613],[344,10.0671],[346,10.1352],[348,10.3054],[350,10.5261],[351,10.6105],[355,10.7934],[357,10.9405],[361,11.0992],[363,11.1328],[365,11.1088],[368,10.9737],[369,10.904],[370,10.8109],[372,10.5355],[375,10.0657],[376,9.8748],[377,9.6458],[379,9.1589],[381,8.7387],[382,8.5038],[383,8.2167],[384,7.8184],[385,7.312],[386,6.7849],[387,6.3349],[388,5.9742],[389,5.6547],[390,5.3796],[391,5.1504],[392,4.9869],[393,4.8907],[395,4.7887],[397,4.7199],[399,4.6721],[401,4.5893],[402,4.5633],[403,4.5483],[409,4.539],[418,4.4595],[422,4.4613]]
       }
       arm_l_pv_ik_ctrl.Rotation.Z {
        type=MovieSceneScriptingFloatChannel
        default=67.5868
        thinned_from=423
        keys [[0,67.5868],[12,67.8456],[17,67.2895],[19,67.7797],[20,68.1299],[21,69.2062],[23,72.4231],[24,73.8333],[25,74.8635],[28,77.1632],[30,78.1697],[35,79.3065],[60,81.3009],[62,81.6862],[69,84.4157],[72,85.8736],[74,86.5133],[75,86.5569],[76,86.0455],[77,84.6806],[78,82.7645],[79,80.5457],[81,75.716],[82,73.5172],[84,69.7103],[85,68.1459],[86,67.0291],[87,66.2826],[89,65.1592],[91,64.3963],[93,63.9526],[95,63.8373],[98,64.1316],[104,65.6862],[117,66.6449],[120,66.4551],[122,65.7949],[125,64.2456],[127,62.8995],[128,62.1186],[129,61.0294],[130,59.4768],[131,56.441],[132,51.0072],[133,40.2506],[134,3.0573],[135,-11.0086],[136,-12.6292],[137,-13.5375],[138,-13.5948],[139,-13.4304],[141,-12.0606],[142,-27.0005],[143,-68.2455],[144,-58.305],[145,-52.095],[146,-48.1736],[147,-46.3961],[148,-44.3434],[149,-39.6226],[150,-36.845],[151,-35.0352],[152,-34.3148],[153,-34.3647],[156,-33.9769],[217,-33.6451],[219,-33.3942],[220,-33.163],[221,-32.8417],[222,-32.417],[223,-32.0803],[224,-31.8706],[226,-31.5473],[227,-31.1563],[228,-30.4118],[229,-28.5487],[230,-23.3266],[231,-11.6471],[232,18.8322],[233,49.2473],[234,66.2072],[235,73.0775],[236,75.0844],[237,76.5915],[238,77.3349],[239,77.5639],[242,77.573],[244,77.3339],[245,77.5905],[246,76.1736],[247,75.2472],[248,74.8181],[251,75.0979],[253,74.7021],[256,74.5786],[259,74.9796],[269,77.7979],[278,78.6337],[292,78.5504],[294,78.8075],[297,79.6537],[299,80.7587],[302,81.99],[304,82.0986],[305,81.9605],[306,81.4239],[307,80.4597],[308,79.2775],[309,77.9269],[310,76.4097],[314,69.509],[316,66.4054],[317,65.0782],[318,63.917],[319,63.018],[320,62.3625],[321,61.9858],[323,61.6305],[326,61.4999],[329,62.0902],[334,63.5434],[340,64.2131],[345,64.0815],[348,63.5658],[353,61.8899],[354,61.4605],[355,61.2025],[357,61.1773],[358,61.0009],[363,58.6616],[366,57.6727],[369,57.0537],[380,57.5859],[382,58.1359],[383,58.5184],[384,59.0448],[385,59.7831],[388,62.4036],[389,63.1656],[392,64.853],[395,66.8056],[399,68.7186],[401,69.329],[403,69.3454],[416,67.6786],[422,67.5915]]
       }
       arm_l_pv_ik_ctrl.Scale.X {type=MovieSceneScriptingFloatChannel, default=1, const [1,423]}
       arm_l_pv_ik_ctrl.Scale.Y {type=MovieSceneScriptingFloatChannel, default=1, const [1,423]}
       arm_l_pv_ik_ctrl.Scale.Z {type=MovieSceneScriptingFloatChannel, default=1, const [1,423]}
       upperarm_r_fk_ctrl.Location.X {type=MovieSceneScriptingFloatChannel, default=0, thinned_from=423, keys [[0,0],[422,0]]}
       upperarm_r_fk_ctrl.Location.Y {type=MovieSceneScriptingFloatChannel, default=0, thinned_from=423, keys [[0,0],[422,0]]}
       upperarm_r_fk_ctrl.Location.Z {type=MovieSceneScriptingFloatChannel, default=0, thinned_from=423, keys [[0,0],[422,0]]}
       upperarm_r_fk_ctrl.Rotation.X {
        type=MovieSceneScriptingFloatChannel
        default=55.352
        thinned_from=423
        keys [[0,58.442],[12,58.0645],[15,58.3393],[16,58.2795],[17,57.874],[20,54.916],[22,52.539],[23,51.6668],[24,50.9081],[27,49.2195],[29,48.383],[30,48.1469],[32,47.9644],[39,48.2569],[45,48.2186],[59,48.5205],[62,48.2278],[63,47.9802],[68,46.1354],[69,45.6614],[70,44.962],[71,44.1603],[72,43.4493],[73,42.8888],[74,42.4702],[75,42.2583],[76,42.46],[77,43.2496],[78,44.4609],[79,46.1479],[82,52.5287],[84,56.5366],[85,58.3831],[86,60.046],[87,61.5214],[89,64.0961],[90,65.1643],[91,66.0623],[93,67.4133],[94,67.8927],[96,68.4294],[98,68.5588],[107,67.0231],[119,66.6034],[120,66.6675],[124,67.7085],[125,67.8832],[126,67.6655],[127,66.7437],[128,65.3797],[129,63.8424],[131,60.1359],[132,54.4982],[133,47.5419],[134,37.1499],[135,21.7237],[136,-8.0037],[137,-36.3367],[138,-35.7348],[139,-36.5439],[140,-39.1616],[141,-41.2096],[142,-61.7417],[143,-90.9515],[144,228.4824],[145,216.9413],[146,206.0924],[147,193.7398],[148,197.809],[149,207.5604],[150,223.7518],[151,228.7563],[152,232.5843],[153,231.6145],[155,228.4702],[159,224.7893],[165,222.9726],[221,223.3253],[224,224.2148],[226,226.6111],[227,228.4564],[228,231.0094],[229,234.5774],[230,-120.7831],[231,-113.3614],[232,-102.7271],[233,-94.5797],[234,-79.7793],[235,-38.8612],[236,-9.7251],[237,17.0383],[238,32.1675],[239,39.1255],[240,40.1611],[241,41.0164],[242,42.3729],[243,42.7407],[245,42.3609],[246,41.4462],[247,41.4653],[248,42.4801],[249,43.7937],[250,45.3296],[252,48.5672],[253,49.4626],[254,49.9578],[255,50.2817],[256,50.4172],[258,50.1655],[268,47.215],[272,46.5968],[284,46.1934],[294,46.0026],[296,45.79],[297,45.5301],[298,44.9685],[299,44.2968],[300,44.0051],[302,43.8072],[303,43.7798],[304,43.8892],[305,44.1248],[306,44.6097],[307,45.3685],[308,46.3098],[309,47.4138],[310,48.7146],[311,50.2325],[313,53.6835],[315,57.4112],[318,63.5316],[319,65.2556],[320,66.682],[321,67.7333],[322,68.5463],[323,69.2127],[325,70.1188],[326,70.4291],[327,70.5052],[329,70.1849],[335,68.3723],[340,67.864],[346,68.0733],[351,68.6744],[354,68.5755],[355,68.2587],[357,66.8135],[358,66.292],[361,65.6781],[364,65.5585],[369,66.0317],[375,65.435],[378,64.7],[381,64.2945],[382,64.3857],[383,64.9437],[384,66.5195],[385,68.8608],[386,70.9716],[387,71.7795],[388,71.1577],[389,69.9619],[390,68.5514],[391,67.2838],[394,63.9335],[396,62.0097],[398,60.4708],[400,59.2106],[401,58.737],[403,58.1907],[407,57.8446],[422,58.4332]]
       }
       upperarm_r_fk_ctrl.Rotation.Y {
        type=MovieSceneScriptingFloatChannel
        default=36.8163
        thinned_from=423
        keys [[0,36.8746],[15,36.441],[17,36.3962],[22,35.9622],[25,35.8779],[33,35.2005],[61,34.8136],[63,34.6107],[66,34.127],[68,33.9231],[69,33.7644],[74,32.2722],[75,32.0398],[76,31.9441],[77,32.0576],[78,32.3298],[79,32.7566],[82,34.4235],[84,35.3348],[86,36.042],[90,36.9078],[94,37.2889],[120,37.1725],[125,37.3979],[127,37.587],[129,37.4716],[130,37.3162],[131,36.8771],[132,34.783],[133,29.9391],[134,22.2522],[135,15.4953],[136,9.2233],[137,7.6851],[138,8.8475],[139,9.3771],[140,14.1736],[141,21.1196],[142,36.5532],[143,58.9487],[144,71.6981],[145,70.6825],[146,69.1029],[147,68.9142],[148,69.774],[149,71.6345],[150,71.9609],[152,68.5015],[153,69.5756],[154,71.1114],[155,72.353],[156,73.268],[157,73.84],[163,74.3148],[221,73.6039],[224,73.2322],[225,72.8278],[226,72.2747],[227,71.559],[228,70.5757],[229,69.2159],[230,67.4816],[231,63.689],[232,59.5307],[233,52.9881],[234,47.5103],[235,43.4175],[236,41.7033],[237,42.2337],[240,44.8331],[242,45.6435],[243,45.0124],[244,41.1634],[245,36.0285],[246,34.5604],[247,33.6745],[248,32.5346],[249,32.8082],[251,33.7406],[252,34.0839],[255,34.6813],[256,34.7817],[297,33.8692],[299,33.5902],[300,33.5088],[301,33.3214],[303,32.5403],[305,32.065],[306,31.9807],[307,32.0209],[308,32.128],[309,32.3013],[310,32.5529],[312,33.2502],[318,35.6166],[320,36.148],[322,36.4739],[326,36.7765],[347,36.3294],[349,36.1075],[350,35.9131],[351,35.6125],[353,34.8367],[357,33.7251],[359,33.4023],[362,33.1703],[368,33.2575],[371,33.4224],[373,33.7011],[377,34.5632],[379,35.1974],[381,36.0992],[383,37.1956],[385,38.7048],[386,39.3394],[387,39.6853],[388,39.7712],[389,39.725],[393,39.1053],[401,37.5995],[407,37.0616],[422,36.8614]]
       }
       upperarm_r_fk_ctrl.Rotation.Z {
        type=MovieSceneScriptingFloatChannel
        default=-3.1634
        thinned_from=423
        keys [[0,-7.6749],[2,-7.6678],[4,-7.6763],[11,-7.6167],[12,-7.62],[13,-7.5805],[14,-7.564],[15,-7.5647],[16,-7.6164],[17,-7.7306],[18,-7.9412],[19,-8.0932],[21,-8.2406],[22,-8.2614],[23,-8.2096],[24,-8.1756],[25,-8.2083],[26,-8.3224],[29,-8.8176],[30,-8.952],[35,-9.422],[41,-9.8062],[43,-9.8876],[45,-9.9335],[46,-9.9114],[47,-9.9306],[48,-9.8963],[49,-9.9197],[50,-9.8835],[51,-9.9002],[52,-9.8793],[54,-9.8741],[56,-9.8869],[58,-9.8489],[62,-9.8654],[65,-9.974],[66,-10.0541],[68,-10.3037],[69,-10.5577],[70,-11.0145],[71,-11.5606],[74,-13.0745],[75,-13.5425],[76,-13.9024],[77,-14.1008],[78,-14.1336],[79,-13.9591],[80,-13.5612],[81,-13.0165],[82,-12.3995],[83,-11.7522],[85,-10.364],[86,-9.7093],[88,-8.4937],[89,-7.9237],[90,-7.4179],[91,-6.9795],[92,-6.6102],[93,-6.2907],[94,-6.0249],[95,-5.8182],[96,-5.6676],[97,-5.55],[98,-5.4863],[99,-5.478],[100,-5.5055],[104,-5.718],[108,-5.8541],[109,-5.8594],[112,-5.9161],[114,-5.9344],[115,-5.9344],[117,-5.969],[118,-5.969],[119,-5.983],[120,-5.9725],[121,-5.9174],[124,-5.697],[125,-5.6484],[126,-5.9134],[127,-6.7101],[129,-8.6842],[130,-9.8031],[131,-11.0302],[132,-13.3649],[133,-14.9363],[134,-14.4325],[135,-14.4626],[136,-16.2334],[137,-18.5451],[138,-19.6096],[139,-21.2556],[140,-20.0874],[141,-19.3123],[142,-19.8493],[143,-28.7522],[144,-63.7091],[145,-71.6547],[146,-78.8239],[147,-82.1837],[148,-79.3829],[149,-71.1929],[150,-55.8272],[151,-50.2018],[152,-46.0641],[153,-46.8713],[154,-48.3931],[155,-49.7594],[156,-50.4767],[158,-52.3233],[159,-53.0285],[160,-53.5718],[161,-53.992],[163,-54.5128],[165,-54.7756],[196,-55.0577],[220,-55.3455],[223,-55.1086],[224,-54.9407],[225,-54.4962],[226,-53.8925],[227,-53.0969],[228,-52.0933],[230,-49.8031],[231,-46.9061],[232,-41.6605],[233,-37.9599],[234,-35.78],[236,-30.6329],[237,-25.4392],[238,-21.3042],[239,-17.7744],[240,-14.469],[241,-11.5224],[242,-8.3659],[243,-6.9474],[244,-6.8603],[245,-7.502],[246,-9.8147],[247,-11.3901],[248,-12.548],[249,-12.9551],[250,-12.9472],[251,-12.5756],[252,-11.9019],[253,-11.3655],[254,-10.993],[255,-10.6911],[256,-10.502],[257,-10.4334],[258,-10.419],[261,-10.4781],[274,-11.0273],[280,-11.1397],[287,-11.1752],[293,-11.1235],[296,-11.1366],[297,-11.1682],[299,-11.3341],[300,-11.3827],[301,-11.5296],[304,-12.427],[305,-12.761],[306,-13.0042],[307,-13.1389],[308,-13.2157],[309,-13.2138],[310,-13.1005],[311,-12.8616],[312,-12.5259],[313,-12.1148],[314,-11.6493],[315,-11.1263],[316,-10.539],[318,-9.2099],[319,-8.5847],[320,-8.0328],[321,-7.5779],[322,-7.1804],[323,-6.8125],[324,-6.4933],[325,-6.2196],[326,-5.9727],[327,-5.8145],[328,-5.7172],[329,-5.6501],[330,-5.614],[331,-5.6003],[332,-5.5984],[335,-5.6572],[340,-5.6435],[341,-5.6162],[343,-5.6005],[345,-5.5196],[346,-5.4677],[347,-5.392],[348,-5.2944],[349,-5.1775],[351,-4.9124],[352,-4.7887],[353,-4.6832],[354,-4.605],[355,-4.7718],[356,-5.2439],[357,-5.7898],[358,-6.1748],[359,-6.3777],[360,-6.5364],[361,-6.6553],[362,-6.7376],[363,-6.7856],[364,-6.8046],[365,-6.7792],[366,-6.7065],[369,-6.3674],[370,-6.2788],[372,-6.1658],[373,-6.1198],[374,-6.0908],[375,-6.0835],[376,-6.134],[377,-6.236],[378,-6.3196],[381,-6.3139],[382,-6.1318],[383,-5.5983],[384,-4.2955],[385,-2.4103],[386,-0.7051],[387,0.0038],[388,-0.3668],[389,-1.1546],[390,-2.0836],[391,-2.8848],[393,-4.2054],[394,-4.8291],[395,-5.3931],[396,-5.8637],[397,-6.2559],[398,-6.6029],[399,-6.9048],[400,-7.1607],[401,-7.357],[402,-7.4809],[403,-7.5558],[404,-7.6077],[407,-7.6906],[411,-7.7157],[417,-7.6791],[422,-7.6835]]
       }
       upperarm_r_fk_ctrl.Scale.X {type=MovieSceneScriptingFloatChannel, default=1, const [1,423]}
       upperarm_r_fk_ctrl.Scale.Y {type=MovieSceneScriptingFloatChannel, default=1, const [1,423]}
       upperarm_r_fk_ctrl.Scale.Z {type=MovieSceneScriptingFloatChannel, default=1, const [1,423]}
       lowerarm_r_fk_ctrl.Location.X {
        type=MovieSceneScriptingFloatChannel
        default=-0.0001
        thinned_from=423
        keys [[0,-0.0001],[55,-0.0001],[56,0.0012],[57,-0.0001],[114,-0.0001],[115,0.0008],[116,-0.0001],[117,-0.0001],[118,-0.0016],[119,-0.0001],[183,-0.0001],[184,-0.0011],[185,-0.0001],[186,-0.0001],[187,-0.0021],[188,-0.0001],[189,-0.0004],[190,-0.0013],[194,-0.0013],[195,-0.0001],[197,-0.0001],[199,0.0014],[201,0.0022],[202,-0.0001],[204,0.0012],[205,-0.0001],[206,0.0004],[207,-0.0001],[208,0.0006],[209,-0.0001],[210,0.0005],[211,-0.0001],[212,0.0005],[213,-0.0001],[214,0.0007],[215,-0.0001],[280,-0.0001],[281,-0.0003],[282,-0.0001],[283,-0.0004],[284,-0.0011],[285,-0.0001],[286,-0.0001],[287,0.0004],[288,-0.0001],[418,-0.0001],[419,0.0005],[422,0.0005]]
       }
       lowerarm_r_fk_ctrl.Location.Y {
        type=MovieSceneScriptingFloatChannel
        default=0
        thinned_from=423
        keys [[0,0],[55,0],[56,0.0024],[57,0],[114,0],[115,0.0016],[116,0],[117,0],[118,-0.0029],[119,0],[183,0],[184,-0.0017],[185,0],[186,0.0001],[187,-0.0035],[188,0],[189,-0.0006],[190,-0.0019],[194,-0.0019],[195,0],[197,0],[199,0.0025],[201,0.0042],[202,0],[204,0.0025],[205,0],[206,0.0011],[207,0],[208,0.0013],[209,0],[210,0.0012],[211,0],[212,0.0012],[213,0],[214,0.0016],[215,0],[280,0],[281,-0.0005],[282,0],[283,-0.0006],[284,-0.0019],[285,0],[286,0],[287,0.001],[288,0],[418,0],[419,0.0012],[422,0.0012]]
       }
       lowerarm_r_fk_ctrl.Location.Z {
        type=MovieSceneScriptingFloatChannel
        default=0
        thinned_from=423
        keys [[0,0],[55,0],[56,-0.003],[57,0],[114,0],[115,-0.0005],[116,0],[117,0],[118,-0.0019],[119,0],[183,0],[184,0.0035],[185,0],[186,0.0041],[187,0.0032],[188,0],[189,-0.0001],[190,0.005],[194,0.005],[195,0],[197,0],[198,-0.0044],[199,-0.0051],[200,-0.002],[201,-0.0001],[203,0.0001],[204,0.0011],[205,0],[206,0.0025],[207,0],[208,0.0003],[209,0],[210,0.0002],[211,0],[212,0.0002],[213,0],[214,0.0002],[215,0],[280,0],[281,-0.0027],[282,0],[283,-0.003],[284,-0.004],[285,0],[286,0],[287,0.0038],[288,0],[418,0],[419,0.0036],[422,0.0036]]
       }
       lowerarm_r_fk_ctrl.Rotation.X {
        type=MovieSceneScriptingFloatChannel
        default=-4.9625
        thinned_from=423
        keys [[0,-4.9659],[16,-4.8942],[17,-4.8828],[18,-4.8604],[21,-4.7478],[24,-4.6633],[30,-4.5706],[33,-4.5524],[37,-4.5718],[46,-4.6233],[55,-4.6249],[56,-4.6309],[62,-4.58],[65,-4.5129],[67,-4.4909],[69,-4.4894],[71,-4.5133],[73,-4.5631],[75,-4.6383],[76,-4.6899],[81,-5.0552],[82,-5.1159],[84,-5.2021],[86,-5.255],[90,-5.3016],[115,-5.2542],[125,-5.2616],[128,-5.3142],[129,-5.3142],[130,-5.2999],[131,-5.2205],[132,-4.9055],[133,-4.1835],[134,-2.4723],[135,-1.4135],[136,-0.6376],[137,-0.1861],[138,-0.1336],[139,-0.0711],[140,0.2073],[141,0.1632],[142,0.0525],[143,-0.5779],[144,-0.8801],[145,-1.6081],[146,-2.2215],[147,-2.4646],[148,-2.608],[149,-2.7394],[150,-3.6303],[151,-4.3529],[152,-4.6575],[153,-4.7378],[155,-4.4194],[156,-4.2436],[157,-4.1025],[158,-4.0448],[183,-4.0448],[184,-4.0381],[185,-4.0448],[186,-4.0368],[189,-4.0448],[190,-4.0349],[194,-4.0349],[195,-4.0448],[197,-4.0448],[199,-4.056],[206,-4.0412],[218,-4.0384],[221,-4.0108],[223,-3.975],[224,-3.9456],[225,-3.8736],[226,-3.7732],[227,-3.6401],[228,-3.3887],[229,-2.9807],[230,-2.4863],[231,-1.9703],[232,-1.4813],[233,-3.2697],[234,-6.5615],[235,-5.3802],[236,-9.8244],[237,-7.8664],[238,-2.8982],[239,-4.2433],[240,-4.7892],[241,-4.9539],[242,-5.0636],[243,-5.1292],[244,-4.9362],[245,-4.6934],[246,-4.7397],[248,-4.7582],[249,-4.8179],[250,-4.8436],[251,-4.8344],[253,-4.7655],[280,-4.6896],[284,-4.7048],[285,-4.6938],[297,-4.6848],[300,-4.6662],[301,-4.6448],[302,-4.5996],[303,-4.574],[304,-4.5848],[305,-4.6161],[306,-4.6641],[307,-4.7308],[309,-4.8819],[312,-5.0685],[316,-5.2504],[318,-5.3204],[320,-5.3612],[324,-5.3891],[348,-5.2739],[352,-5.2186],[354,-5.2169],[358,-5.311],[366,-5.379],[371,-5.3878],[373,-5.4168],[376,-5.4809],[381,-5.5261],[382,-5.5276],[383,-5.5078],[388,-5.2827],[392,-5.1422],[398,-5.0336],[406,-4.9662],[418,-4.9645],[422,-4.9582]]
       }
       lowerarm_r_fk_ctrl.Rotation.Y {
        type=MovieSceneScriptingFloatChannel
        default=-4.4175
        thinned_from=423
        keys [[0,-3.3019],[75,-3.3373],[86,-3.25],[114,-3.249],[115,-3.2378],[116,-3.2492],[125,-3.2486],[129,-3.237],[130,-3.2402],[131,-3.2573],[132,-3.3103],[133,-3.347],[134,-2.9481],[135,-3.7115],[136,-6.5986],[137,-2.5592],[138,7.9732],[139,12.2314],[140,13.1079],[141,5.7289],[142,17.2835],[143,7.1114],[144,1.3162],[145,-0.4413],[146,-2.4363],[147,-6.557],[148,-2.2334],[149,3.2568],[150,2.2242],[151,-3.349],[153,-3.3298],[155,-3.3481],[183,-3.3405],[184,-3.346],[185,-3.3405],[186,-3.3463],[189,-3.3402],[194,-3.3464],[199,-3.3406],[201,-3.3516],[202,-3.3405],[204,-3.3508],[205,-3.3405],[206,-3.351],[207,-3.3405],[208,-3.3496],[209,-3.3405],[210,-3.3518],[211,-3.3405],[212,-3.3542],[213,-3.3405],[214,-3.3583],[215,-3.3405],[224,-3.3332],[226,-3.3153],[227,-3.2968],[228,-3.2505],[229,-3.143],[230,-2.9544],[231,-6.3421],[232,-10.5423],[233,-20.6491],[234,-30.4227],[235,-21.7227],[236,-18.7273],[237,-14.244],[238,-9.796],[239,-6.069],[240,-4.1625],[241,-3.3042],[243,-3.2756],[245,-3.3331],[250,-3.318],[280,-3.3334],[281,-3.3235],[282,-3.3332],[283,-3.3309],[284,-3.3386],[286,-3.3329],[287,-3.3447],[288,-3.3329],[305,-3.3388],[319,-3.2299],[354,-3.258],[373,-3.2123],[382,-3.1828],[392,-3.2726],[422,-3.3046]]
       }
       lowerarm_r_fk_ctrl.Rotation.Z {
        type=MovieSceneScriptingFloatChannel
        default=83.7475
        thinned_from=423
        keys [[0,83.8227],[16,82.8923],[18,82.4552],[22,80.6075],[27,79.2112],[32,78.5205],[37,78.7324],[57,79.3753],[62,78.8382],[65,77.9748],[68,77.652],[70,77.7799],[73,78.6198],[75,79.5881],[76,80.2533],[81,84.9841],[83,86.3949],[85,87.2959],[89,88.1244],[125,87.6831],[128,88.3757],[129,88.3757],[130,88.1869],[131,87.1436],[132,83.0395],[133,73.7715],[134,51.0539],[135,34.8259],[136,19.9267],[137,7.542],[138,8.3085],[139,8.6277],[140,-3.0872],[141,-10.4546],[142,8.5103],[143,20.0306],[144,25.3484],[145,38.1923],[146,47.475],[147,50.8875],[149,55.0817],[150,66.7976],[151,75.948],[152,79.865],[153,80.8994],[155,76.8013],[156,74.5428],[157,72.73],[158,71.987],[218,71.9052],[221,71.549],[223,71.0884],[224,70.71],[225,69.7804],[226,68.483],[227,66.7567],[228,63.4728],[229,58.0505],[230,51.2506],[232,36.1535],[233,26.2805],[234,19.2833],[235,28.3357],[236,41.101],[237,53.5657],[238,63.2014],[239,73.8564],[240,80.7586],[241,83.6954],[242,85.1214],[243,85.9766],[244,83.4659],[245,80.2988],[246,80.8958],[248,81.1342],[249,81.9051],[250,82.2373],[251,82.1194],[253,81.2282],[298,80.0877],[300,79.9473],[301,79.6724],[302,79.0903],[303,78.7604],[304,78.8991],[305,79.3023],[306,79.9212],[311,84.3938],[315,86.9955],[318,88.4569],[322,89.2454],[348,87.8453],[352,87.1187],[354,87.0972],[359,88.5348],[372,89.4819],[377,90.7456],[382,91.2041],[383,90.94],[389,87.4532],[393,85.7833],[405,83.8726],[422,83.8213]]
       }
       lowerarm_r_fk_ctrl.Scale.X {type=MovieSceneScriptingFloatChannel, default=1, const [1,423]}
       lowerarm_r_fk_ctrl.Scale.Y {type=MovieSceneScriptingFloatChannel, default=1, const [1,423]}
       lowerarm_r_fk_ctrl.Scale.Z {type=MovieSceneScriptingFloatChannel, default=1, const [1,423]}
       hand_r_fk_ctrl.Location.X {type=MovieSceneScriptingFloatChannel, default=0.0042, thinned_from=423, keys [[0,-0.0001],[422,-0.0001]]}
       hand_r_fk_ctrl.Location.Y {
        type=MovieSceneScriptingFloatChannel
        default=-0.431
        thinned_from=423
        keys [[0,0],[110,0],[111,-0.0005],[112,-0.0005],[113,0],[271,0],[272,-0.0004],[273,0],[274,-0.0004],[279,-0.0004],[280,0.0002],[281,0],[369,0],[370,0.0004],[371,0],[416,0],[417,0.0004],[418,0.0004],[419,0],[422,0]]
       }
       hand_r_fk_ctrl.Location.Z {
        type=MovieSceneScriptingFloatChannel
        default=0.0668
        thinned_from=423
        keys [[0,0],[110,0],[111,-0.0058],[112,-0.0058],[113,0],[271,0],[272,-0.0056],[273,0],[274,-0.0055],[279,-0.0055],[280,0.0029],[281,0],[369,0],[370,0.0052],[371,0],[416,0],[417,0.0058],[418,0.0058],[419,0],[422,0]]
       }
       hand_r_fk_ctrl.Rotation.X {
        type=MovieSceneScriptingFloatChannel
        default=-11.1578
        thinned_from=423
        keys [[0,-11.1075],[15,-11.212],[16,-11.2543],[17,-11.3713],[18,-11.6159],[19,-12.0496],[20,-12.6064],[21,-13.1091],[22,-13.5138],[23,-13.8789],[24,-14.1635],[25,-14.3125],[26,-14.3391],[27,-14.3063],[31,-13.9535],[36,-13.3991],[40,-13.0965],[44,-12.9253],[65,-12.8746],[66,-12.8245],[68,-12.5749],[69,-12.3163],[70,-11.8584],[73,-10.1315],[75,-8.8788],[76,-8.2852],[77,-7.7482],[78,-7.2736],[79,-6.8782],[80,-6.553],[81,-6.2747],[82,-6.0428],[83,-5.8735],[84,-5.7566],[85,-5.6822],[86,-5.6358],[87,-5.6222],[88,-5.6447],[89,-5.689],[98,-6.2848],[100,-6.3962],[102,-6.4787],[109,-6.5926],[117,-6.6404],[120,-6.6853],[124,-6.6892],[125,-6.7026],[126,-6.5933],[127,-6.3099],[128,-6.1999],[129,-6.3427],[130,-6.6136],[131,-7.0133],[132,-12.8243],[133,-11.3341],[134,-10.2488],[135,-34.578],[136,-54.943],[137,-74.4164],[143,-74.4164],[144,-74.9905],[145,-75.1819],[147,-74.8708],[148,-74.8708],[149,-76.2574],[150,-77.4678],[151,-78.497],[152,-79.1173],[154,-79.5883],[223,-79.4411],[225,-79.1293],[227,-78.4063],[228,-77.7965],[229,-76.8834],[230,-75.6302],[231,-73.9975],[232,-71.9433],[233,-70.0844],[234,-68.5189],[235,-41.5668],[236,-37.0503],[237,-36.2965],[238,-32.1634],[239,-17.5394],[240,-9.5547],[241,-16.6311],[242,-21.8941],[243,-23.6446],[244,-23.6446],[245,-21.3773],[246,-18.2759],[247,-15.8696],[248,-13.2454],[249,-12.0852],[250,-11.3754],[251,-11.0056],[252,-11.1001],[253,-11.3992],[254,-11.7235],[255,-11.9908],[256,-12.1676],[257,-12.2665],[271,-12.5276],[299,-12.571],[300,-12.5512],[301,-12.6247],[302,-12.7695],[303,-12.7514],[304,-12.4309],[305,-11.9232],[306,-11.3465],[307,-10.6765],[308,-9.9134],[309,-9.2159],[310,-8.633],[311,-8.1046],[312,-7.6276],[313,-7.2239],[315,-6.5116],[316,-6.168],[317,-5.8584],[318,-5.6084],[319,-5.4471],[320,-5.3816],[321,-5.3961],[322,-5.4443],[323,-5.5147],[326,-5.8076],[327,-5.9246],[331,-6.477],[333,-6.7238],[334,-6.8271],[335,-6.908],[337,-7.0217],[340,-7.1361],[342,-7.1776],[350,-7.2079],[351,-7.197],[352,-7.1557],[353,-7.0916],[354,-7.01],[355,-6.7988],[356,-6.4319],[357,-6.0411],[358,-5.7531],[359,-5.5773],[360,-5.4375],[361,-5.3289],[362,-5.2466],[363,-5.185],[364,-5.1428],[365,-5.1298],[366,-5.1514],[367,-5.2048],[368,-5.2824],[369,-5.3711],[371,-5.5923],[372,-5.6749],[373,-5.7246],[375,-5.7719],[376,-5.8241],[377,-5.9405],[378,-6.1206],[379,-6.3596],[381,-6.7582],[382,-7.1251],[383,-7.8295],[384,-9.0269],[385,-10.5033],[386,-11.878],[387,-12.7531],[388,-13.0399],[389,-13.0343],[391,-12.7667],[392,-12.676],[393,-12.5348],[396,-12.0404],[398,-11.7737],[400,-11.5645],[402,-11.4388],[410,-11.1912],[422,-11.1092]]
       }
       hand_r_fk_ctrl.Rotation.Y {
        type=MovieSceneScriptingFloatChannel
        default=8.3334
        thinned_from=423
        keys [[0,8.1696],[3,8.2142],[13,8.5108],[15,8.6059],[17,8.7627],[18,8.8845],[19,9.124],[20,9.4256],[21,9.6508],[24,10.1031],[30,10.8056],[32,10.9754],[33,11.0272],[34,11.0489],[35,11.0473],[36,11.0221],[37,10.9735],[40,10.7519],[41,10.7001],[45,10.621],[55,10.6171],[57,10.6478],[59,10.7314],[62,10.9383],[63,11.0516],[64,11.1986],[65,11.3198],[67,11.5111],[68,11.583],[69,11.6237],[70,11.6018],[71,11.479],[72,11.2482],[73,11.1214],[75,11.1663],[77,11.1366],[79,11.2107],[81,11.1364],[82,11.1283],[84,11.2248],[86,11.3581],[95,11.5887],[98,11.5963],[109,11.4381],[112,11.4236],[118,11.4749],[120,11.5412],[125,11.8832],[126,11.8848],[127,11.9198],[128,11.8791],[129,11.7596],[130,11.806],[131,12.5451],[132,8.0165],[133,9.7378],[134,8.7067],[135,1.2849],[136,-4.4773],[137,-8.9806],[143,-8.9806],[144,-12.1241],[147,-15.284],[148,-15.284],[149,-5.2338],[150,4.8361],[151,14.1573],[152,19.9642],[153,22.8028],[154,24.3871],[155,25.0109],[222,24.937],[224,24.8666],[226,24.6159],[227,24.4224],[228,24.1387],[229,23.7186],[230,23.1507],[231,22.4254],[232,21.5354],[233,11.8985],[234,4.703],[235,8.0607],[236,8.7658],[237,8.8859],[238,11.4208],[239,16.5326],[240,19.6086],[241,12.9723],[242,7.9861],[243,6.3246],[244,6.3246],[245,9.4553],[246,9.5029],[247,10.0017],[249,10.1201],[250,10.6216],[251,11.3379],[252,12.0783],[253,12.4723],[254,12.5792],[257,12.5253],[271,12.4197],[272,12.397],[273,12.4078],[279,12.348],[280,12.3587],[283,12.3415],[292,12.3163],[296,12.337],[297,12.3244],[299,12.2175],[300,12.2399],[301,13.6797],[302,16.4823],[303,18.6648],[304,19.5519],[305,19.8436],[306,19.6479],[307,18.9363],[308,17.876],[309,16.8687],[310,16.0715],[311,15.3776],[312,14.779],[313,14.3288],[314,13.9815],[315,13.6912],[316,13.4666],[317,13.304],[318,13.2048],[319,13.1602],[321,13.2206],[327,13.1158],[334,12.7155],[338,12.6528],[342,12.6983],[346,12.8369],[350,13.1078],[351,13.1382],[352,13.0966],[353,12.9981],[354,12.8591],[355,12.6803],[357,12.2596],[358,12.0875],[359,11.9649],[361,11.7969],[365,11.5964],[368,11.5039],[370,11.4873],[371,11.3681],[372,11.0957],[373,10.6715],[374,10.1401],[375,9.5452],[376,8.9732],[377,8.4314],[378,7.8651],[379,7.2109],[380,6.1642],[381,4.8445],[382,3.8077],[383,3.6136],[384,4.1573],[386,5.5072],[387,6.0901],[388,6.497],[389,6.7622],[390,6.9633],[391,7.1813],[392,7.4249],[393,7.6401],[394,7.8208],[395,7.947],[396,8.0263],[397,8.0837],[399,8.1496],[405,8.2082],[412,8.1429],[416,8.1599],[418,8.1823],[422,8.1683]]
       }
       hand_r_fk_ctrl.Rotation.Z {
        type=MovieSceneScriptingFloatChannel
        default=-18.0061
        thinned_from=423
        keys [[0,-16.894],[2,-16.8309],[6,-16.5802],[10,-16.451],[11,-16.4408],[12,-16.4734],[14,-16.6328],[15,-16.6765],[16,-16.6152],[17,-16.2285],[18,-15.3477],[20,-13.3616],[21,-12.1915],[22,-11.1192],[23,-10.2967],[24,-9.5774],[25,-8.9924],[26,-8.4832],[27,-7.9957],[28,-7.5633],[29,-7.2198],[30,-6.9994],[31,-6.8833],[32,-6.824],[33,-6.8126],[34,-6.839],[35,-6.8973],[36,-6.9795],[38,-7.1657],[39,-7.2192],[40,-7.2271],[42,-7.21],[44,-7.2323],[46,-7.3027],[49,-7.4257],[53,-7.5208],[57,-7.5468],[58,-7.5371],[59,-7.4989],[60,-7.4263],[61,-7.3146],[62,-7.1554],[63,-6.8767],[65,-6.0968],[66,-5.745],[67,-5.4131],[68,-5.1451],[69,-4.7901],[70,-4.2468],[71,-3.6128],[72,-3.0572],[73,-2.664],[74,-2.4269],[75,-2.3835],[76,-2.7259],[77,-3.646],[78,-4.9678],[79,-6.6907],[80,-8.7158],[81,-10.7775],[82,-12.7822],[83,-14.6398],[84,-16.3947],[85,-18.0165],[86,-19.4612],[87,-20.7262],[88,-21.8695],[89,-22.9078],[90,-23.8132],[91,-24.5675],[92,-25.1847],[93,-25.6961],[94,-26.0942],[95,-26.3676],[96,-26.5404],[97,-26.6359],[98,-26.6394],[99,-26.5321],[103,-25.7878],[108,-25.2499],[117,-24.9648],[119,-24.9723],[120,-25.0317],[121,-25.2206],[124,-25.9317],[125,-26.0847],[126,-25.9316],[127,-25.1878],[128,-23.9825],[129,-22.5254],[130,-20.8153],[132,-17.0768],[133,-11.7758],[134,-5.1515],[135,-4.0734],[136,-1.3234],[137,2.6002],[143,2.6002],[144,2.5291],[145,2.5006],[147,2.266],[148,2.266],[149,3.2669],[150,4.024],[151,4.5494],[152,4.7997],[153,4.9008],[154,4.9511],[155,4.9696],[219,4.9598],[223,4.9147],[224,4.8916],[225,4.835],[226,4.7564],[227,4.6524],[228,4.5007],[229,4.2771],[230,3.9774],[231,3.5989],[232,3.1412],[233,2.6968],[234,2.2019],[235,-1.6355],[236,-2.0087],[237,-2.0627],[238,-3.2237],[239,-4.4325],[240,-4.8892],[241,-5.5217],[242,-5.9134],[243,-6.0294],[244,-6.0294],[245,-0.6555],[246,-0.3643],[247,-0.7317],[248,-1.7485],[249,-3.3779],[251,-6.9976],[252,-8.5555],[253,-9.3727],[254,-9.8198],[255,-10.0889],[256,-10.1876],[257,-10.0952],[258,-9.9389],[259,-9.7436],[260,-9.5078],[261,-9.2409],[266,-7.7576],[267,-7.4855],[268,-7.2421],[269,-7.0359],[270,-6.8728],[271,-6.7424],[272,-6.6289],[273,-6.5362],[275,-6.3988],[277,-6.297],[279,-6.2319],[281,-6.1973],[284,-6.1942],[286,-6.2369],[289,-6.2369],[292,-6.1632],[293,-6.1228],[294,-6.0701],[295,-5.9989],[296,-5.8669],[297,-5.6172],[298,-5.0711],[299,-4.4126],[300,-4.1252],[301,-4.1995],[302,-4.3026],[303,-4.4306],[304,-4.5435],[305,-4.6988],[306,-5.0479],[307,-5.6121],[308,-6.3131],[309,-7.1819],[310,-8.2701],[311,-9.5781],[312,-11.0393],[313,-12.5947],[314,-14.1799],[315,-15.8462],[316,-17.5893],[317,-19.3843],[318,-21.0938],[319,-22.556],[320,-23.7664],[321,-24.6598],[322,-25.3528],[323,-25.9139],[324,-26.3482],[325,-26.691],[326,-26.946],[327,-27.0143],[328,-26.9194],[329,-26.7571],[334,-25.4588],[335,-25.2569],[340,-24.8226],[342,-24.7875],[346,-24.9881],[350,-25.4537],[351,-25.5071],[354,-25.4369],[355,-25.178],[356,-24.5981],[357,-23.952],[358,-23.5058],[359,-23.2778],[360,-23.1047],[361,-22.98],[362,-22.9029],[364,-22.8827],[368,-23.2207],[369,-23.3408],[370,-23.3633],[375,-23.002],[378,-22.5087],[381,-22.2572],[382,-22.3284],[383,-22.7795],[384,-24.0563],[385,-25.9207],[386,-27.5723],[387,-28.1835],[388,-27.6627],[389,-26.6774],[390,-25.5166],[391,-24.4709],[394,-21.7361],[395,-20.8944],[396,-20.1425],[397,-19.4672],[398,-18.8379],[399,-18.2616],[400,-17.7463],[401,-17.3254],[402,-17.0251],[403,-16.816],[404,-16.6652],[405,-16.5616],[407,-16.4574],[409,-16.459],[418,-16.8799],[422,-16.8804]]
       }
       hand_r_fk_ctrl.Scale.X {type=MovieSceneScriptingFloatChannel, default=1, const [1,423]}
       hand_r_fk_ctrl.Scale.Y {type=MovieSceneScriptingFloatChannel, default=1, const [1,423]}
       hand_r_fk_ctrl.Scale.Z {type=MovieSceneScriptingFloatChannel, default=1, const [1,423]}
       thumb_01_r_ctrl.Location.X {
        type=MovieSceneScriptingFloatChannel
        default=0.0685
        thinned_from=423
        keys [[0,0.0685],[214,0.0685],[215,0.0683],[216,0.0679],[217,0.067],[218,0.0657],[219,0.064],[220,0.0618],[221,0.059],[222,0.0556],[223,0.0508],[224,0.0443],[226,0.0292],[227,0.0224],[228,0.017],[229,0.0148],[230,0.0145],[231,0.0128],[232,0.0128],[233,0.0107],[234,0.0055],[235,0.0028],[236,0.0006],[237,-0.0001],[238,-0.0001],[239,0.0001],[241,0.0304],[242,0.0423],[243,0.0483],[244,0.0574],[245,0.0596],[246,0.0604],[247,0.0617],[248,0.0651],[249,0.0674],[250,0.0682],[251,0.0685],[422,0.0685]]
       }
       thumb_01_r_ctrl.Location.Y {
        type=MovieSceneScriptingFloatChannel
        default=-0.3984
        thinned_from=423
        keys [[0,-0.3984],[214,-0.3984],[215,-0.3976],[216,-0.3948],[217,-0.3898],[218,-0.3824],[219,-0.3724],[220,-0.3594],[221,-0.3433],[222,-0.3237],[223,-0.2957],[224,-0.2577],[225,-0.2143],[226,-0.1703],[227,-0.1305],[228,-0.0995],[229,-0.0865],[230,-0.0846],[231,-0.0746],[232,-0.0746],[233,-0.063],[234,-0.0325],[235,-0.0166],[236,-0.0042],[237,-0.0001],[238,0.0001],[239,-0.001],[241,-0.1774],[242,-0.2464],[243,-0.2809],[244,-0.3338],[245,-0.347],[246,-0.3514],[247,-0.3592],[248,-0.3788],[249,-0.3919],[250,-0.3968],[251,-0.3984],[422,-0.3984]]
       }
       thumb_01_r_ctrl.Location.Z {
        type=MovieSceneScriptingFloatChannel
        default=-0.4179
        thinned_from=423
        keys [[0,-0.4179],[214,-0.4179],[215,-0.4169],[216,-0.414],[217,-0.4088],[218,-0.4011],[219,-0.3905],[220,-0.3769],[221,-0.36],[222,-0.3395],[223,-0.3102],[224,-0.2703],[225,-0.2248],[226,-0.1786],[227,-0.1369],[228,-0.1044],[229,-0.0907],[230,-0.0887],[231,-0.0783],[232,-0.0783],[233,-0.066],[234,-0.0341],[235,-0.0174],[236,-0.0044],[237,-0.0001],[238,0.0001],[239,-0.001],[241,-0.186],[242,-0.2584],[243,-0.2946],[244,-0.3501],[245,-0.3639],[246,-0.3685],[247,-0.3768],[248,-0.3973],[249,-0.411],[250,-0.4161],[251,-0.4179],[422,-0.4179]]
       }
       thumb_01_r_ctrl.Rotation.X {
        type=MovieSceneScriptingFloatChannel
        default=9.0456
        thinned_from=423
        keys [[0,9.0456],[215,9.0533],[216,9.0777],[217,9.121],[218,9.1854],[219,9.2732],[220,9.3864],[221,9.5273],[222,9.6981],[223,9.9423],[224,10.2746],[226,11.0375],[227,11.3853],[228,11.6554],[229,11.7692],[230,11.7863],[231,11.8729],[232,11.8729],[233,11.9748],[234,12.2405],[235,12.3795],[236,12.4879],[237,12.524],[239,12.5162],[241,10.9758],[242,10.3732],[243,10.0719],[244,9.6101],[245,9.4947],[246,9.4562],[247,9.3878],[248,9.2167],[249,9.1027],[250,9.0599],[422,9.0456]]
       }
       thumb_01_r_ctrl.Rotation.Y {
        type=MovieSceneScriptingFloatChannel
        default=-25.4575
        thinned_from=423
        keys [[0,-25.4575],[130,-25.4575],[131,-23.7775],[132,-21.6775],[134,-15.3775],[215,-15.3661],[217,-15.2656],[218,-15.1699],[219,-15.0396],[220,-14.8714],[221,-14.6622],[222,-14.4086],[223,-14.046],[224,-13.5526],[226,-12.4197],[227,-11.9033],[228,-11.5022],[229,-11.3332],[230,-11.3078],[231,-11.1792],[232,-11.1792],[233,-11.0278],[234,-10.6333],[235,-10.4269],[236,-10.2659],[237,-10.2123],[239,-10.2239],[241,-12.5114],[242,-13.4062],[243,-13.8536],[244,-14.5392],[245,-14.7106],[246,-14.7678],[247,-16.5494],[248,-21.0034],[249,-23.9728],[250,-25.0863],[251,-25.4575],[422,-25.4575]]
       }
       thumb_01_r_ctrl.Rotation.Z {
        type=MovieSceneScriptingFloatChannel
        default=10.3654
        thinned_from=423
        keys [[0,10.3654],[130,10.3654],[131,7.5454],[132,4.0204],[134,-6.5546],[214,-6.5546],[215,-6.5409],[216,-6.4972],[217,-6.4194],[218,-6.3038],[219,-6.1464],[220,-5.9433],[221,-5.6905],[222,-5.3842],[223,-4.946],[224,-4.35],[225,-3.6703],[226,-2.9813],[227,-2.3574],[228,-1.8728],[229,-1.6687],[231,10.1588],[232,17.3588],[233,14.4015],[234,7.0275],[235,3.3516],[236,0.6021],[237,-1.4336],[238,-1.9991],[239,-1.8572],[241,-3.092],[242,-4.1731],[243,-4.7136],[244,-5.5419],[245,-5.749],[246,-5.818],[247,-3.1208],[248,3.6223],[249,8.1177],[250,9.8034],[251,10.3654],[422,10.3654]]
       }
       thumb_01_r_ctrl.Scale.X {type=MovieSceneScriptingFloatChannel, default=1, const [1,423]}
       thumb_01_r_ctrl.Scale.Y {type=MovieSceneScriptingFloatChannel, default=1, const [1,423]}
       thumb_01_r_ctrl.Scale.Z {type=MovieSceneScriptingFloatChannel, default=1, const [1,423]}
       thumb_02_r_ctrl.Location.X {type=MovieSceneScriptingFloatChannel, default=0.0001, thinned_from=423, keys [[0,0.0001],[422,0.0001]]}
       thumb_02_r_ctrl.Location.Y {type=MovieSceneScriptingFloatChannel, default=0, thinned_from=423, keys [[0,0],[422,0]]}
       thumb_02_r_ctrl.Location.Z {type=MovieSceneScriptingFloatChannel, default=0, thinned_from=423, keys [[0,0],[422,0]]}
       thumb_02_r_ctrl.Rotation.X {
        type=MovieSceneScriptingFloatChannel
        default=-28.0109
        thinned_from=423
        keys [[0,-28.0109],[214,-28.0109],[215,-27.9174],[216,-27.619],[217,-27.0889],[218,-26.3005],[219,-25.227],[220,-23.8418],[221,-22.118],[222,-20.029],[223,-17.041],[224,-12.9762],[225,-8.3411],[226,-3.6428],[227,0.6122],[228,3.9169],[229,5.3088],[230,5.5182],[231,6.5775],[232,6.5775],[233,7.8247],[234,11.0752],[235,12.7758],[236,14.1019],[237,14.5439],[238,14.5595],[239,14.4486],[241,-4.3977],[242,-11.7699],[243,-15.456],[244,-21.1046],[245,-22.5168],[246,-22.9875],[247,-23.8248],[248,-25.9178],[249,-27.3132],[250,-27.8365],[251,-28.0109],[422,-28.0109]]
       }
       thumb_02_r_ctrl.Rotation.Y {
        type=MovieSceneScriptingFloatChannel
        default=2.0922
        thinned_from=423
        keys [[0,2.0922],[214,2.0922],[215,2.1804],[216,2.4619],[217,2.9619],[218,3.7056],[219,4.7182],[228,15.4775],[229,16.7905],[230,16.988],[231,17.9873],[232,17.9873],[233,19.1638],[234,22.23],[235,23.8342],[236,25.0851],[237,25.5021],[238,25.5168],[239,25.4123],[241,2.0922],[422,2.0922]]
       }
       thumb_02_r_ctrl.Rotation.Z {
        type=MovieSceneScriptingFloatChannel
        default=-0.1833
        thinned_from=423
        keys [[0,-0.1833],[214,-0.1833],[215,-0.0683],[216,0.2984],[217,0.9497],[218,1.9186],[219,3.2377],[220,4.9117],[222,8.2992],[229,20.2879],[231,33.4885],[232,40.6885],[233,39.0809],[234,35.2247],[235,33.3892],[236,32.0749],[237,30.4106],[238,29.9483],[239,29.6904],[241,-0.1833],[422,-0.1833]]
       }
       thumb_02_r_ctrl.Scale.X {type=MovieSceneScriptingFloatChannel, default=1, const [1,423]}
       thumb_02_r_ctrl.Scale.Y {type=MovieSceneScriptingFloatChannel, default=1, const [1,423]}
       thumb_02_r_ctrl.Scale.Z {type=MovieSceneScriptingFloatChannel, default=1, const [1,423]}
       thumb_03_r_ctrl.Location.X {type=MovieSceneScriptingFloatChannel, default=0, thinned_from=423, keys [[0,0],[422,0]]}
       thumb_03_r_ctrl.Location.Y {type=MovieSceneScriptingFloatChannel, default=0, thinned_from=423, keys [[0,0],[422,0]]}
       thumb_03_r_ctrl.Location.Z {type=MovieSceneScriptingFloatChannel, default=0, thinned_from=423, keys [[0,0],[422,0]]}
       thumb_03_r_ctrl.Rotation.X {
        type=MovieSceneScriptingFloatChannel
        default=-4.6289
        thinned_from=423
        keys [[0,-4.6289],[214,-4.6289],[215,-4.6209],[216,-4.5953],[217,-4.5499],[218,-4.4824],[219,-4.3904],[220,-4.2718],[221,-4.1241],[222,-3.9452],[223,-3.6893],[224,-3.3412],[226,-2.5418],[227,-2.1774],[228,-1.8943],[229,-1.7751],[230,-1.7572],[231,-1.6665],[232,-1.6665],[233,-1.5596],[234,-1.2812],[235,-1.1356],[236,-1.022],[237,-0.9842],[238,-0.9828],[239,-0.9923],[241,-2.6065],[242,-3.2379],[243,-3.5536],[244,-4.0374],[245,-4.1583],[246,-4.1986],[247,-4.2703],[248,-4.4496],[249,-4.5691],[250,-4.6139],[251,-4.6289],[422,-4.6289]]
       }
       thumb_03_r_ctrl.Rotation.Y {
        type=MovieSceneScriptingFloatChannel
        default=-5.8424
        thinned_from=423
        keys [[0,-5.8424],[215,-5.8377],[216,-5.8226],[217,-5.7959],[218,-5.7562],[219,-5.7021],[220,-5.6324],[221,-5.5455],[222,-5.4403],[223,-5.2898],[224,-5.085],[226,-4.6148],[227,-4.4005],[228,-4.234],[229,-4.1639],[230,-4.1533],[231,-4.0999],[232,-4.0999],[233,-4.0371],[234,-3.8734],[235,-3.7877],[236,-3.7209],[237,-3.6986],[239,-3.7034],[241,-4.6528],[242,-5.0242],[243,-5.2099],[244,-5.4945],[245,-5.5656],[246,-5.5893],[247,-5.6315],[248,-5.7369],[249,-5.8072],[250,-5.8336],[422,-5.8424]]
       }
       thumb_03_r_ctrl.Rotation.Z {
        type=MovieSceneScriptingFloatChannel
        default=-69.3636
        thinned_from=423
        keys [[0,-69.3636],[214,-69.3636],[215,-69.1404],[216,-68.4286],[217,-67.1644],[218,-65.2839],[219,-62.7234],[220,-59.4192],[221,-55.3076],[222,-50.3249],[223,-43.1979],[224,-33.5023],[225,-22.4468],[226,-11.2402],[227,-1.0912],[228,6.7912],[229,10.1112],[231,24.7788],[232,31.9789],[233,31.8134],[234,31.7159],[235,31.8468],[236,32.0659],[237,30.7986],[238,30.4465],[239,29.7615],[241,-13.0409],[242,-30.625],[243,-39.4174],[244,-52.8906],[245,-56.2589],[246,-57.3817],[247,-59.3787],[248,-64.3711],[249,-67.6994],[250,-68.9475],[251,-69.3636],[422,-69.3636]]
       }
       thumb_03_r_ctrl.Scale.X {type=MovieSceneScriptingFloatChannel, default=1, const [1,423]}
       thumb_03_r_ctrl.Scale.Y {type=MovieSceneScriptingFloatChannel, default=1, const [1,423]}
       thumb_03_r_ctrl.Scale.Z {type=MovieSceneScriptingFloatChannel, default=1, const [1,423]}
       index_metacarpal_r_ctrl.Location.X {type=MovieSceneScriptingFloatChannel, default=0.0001, thinned_from=423, keys [[0,0.0001],[422,0.0001]]}
       index_metacarpal_r_ctrl.Location.Y {type=MovieSceneScriptingFloatChannel, default=0.0002, thinned_from=423, keys [[0,0.0002],[422,0.0002]]}
       index_metacarpal_r_ctrl.Location.Z {type=MovieSceneScriptingFloatChannel, default=0, thinned_from=423, keys [[0,0],[422,0]]}
       index_metacarpal_r_ctrl.Rotation.X {
        type=MovieSceneScriptingFloatChannel
        default=-2.9693
        thinned_from=423
        keys [[0,-2.9693],[214,-2.9693],[215,-2.9637],[216,-2.9458],[217,-2.9141],[218,-2.8669],[219,-2.8027],[220,-2.7198],[221,-2.6166],[222,-2.4916],[223,-2.3127],[224,-2.0694],[225,-1.792],[226,-1.5108],[227,-1.2561],[228,-1.0583],[229,-0.975],[230,-0.9625],[231,-0.8991],[232,-0.8991],[233,-0.8244],[234,-0.6299],[235,-0.5281],[236,-0.4487],[237,-0.4222],[238,-0.4213],[239,-0.428],[241,-1.556],[242,-1.9972],[243,-2.2179],[244,-2.5559],[245,-2.6405],[246,-2.6686],[247,-2.7188],[248,-2.844],[249,-2.9275],[250,-2.9589],[251,-2.9693],[422,-2.9693]]
       }
       index_metacarpal_r_ctrl.Rotation.Y {
        type=MovieSceneScriptingFloatChannel
        default=3.5534
        thinned_from=423
        keys [[0,3.5534],[214,3.5534],[215,3.5354],[216,3.4778],[217,3.3756],[218,3.2235],[219,3.0165],[220,2.7493],[221,2.4168],[222,2.0139],[223,1.4376],[224,0.6535],[225,-0.2405],[226,-1.1467],[227,-1.9674],[228,-2.6048],[229,-2.8733],[230,-2.9137],[231,-3.118],[232,-3.118],[233,-3.3586],[234,-3.9855],[235,-4.3135],[236,-4.5693],[237,-4.6546],[238,-4.6576],[239,-4.6362],[241,-1.0011],[242,0.4208],[243,1.1318],[244,2.2214],[245,2.4937],[246,2.5845],[247,2.746],[248,3.1497],[249,3.4189],[250,3.5198],[251,3.5534],[422,3.5534]]
       }
       index_metacarpal_r_ctrl.Rotation.Z {
        type=MovieSceneScriptingFloatChannel
        default=-24.013
        thinned_from=423
        keys [[0,-24.013],[130,-24.013],[131,-22.2309],[132,-20.0033],[134,-13.3204],[214,-13.3204],[215,-13.2912],[216,-13.1983],[217,-13.0333],[218,-12.7879],[219,-12.4537],[220,-12.0224],[221,-11.4857],[222,-10.8354],[223,-9.9051],[224,-8.6396],[225,-7.1966],[226,-5.7339],[227,-4.4092],[228,-3.3804],[229,-2.947],[231,9.0894],[232,16.2894],[233,13.5375],[234,6.6988],[235,3.3029],[236,0.7718],[237,-1.2073],[238,-1.7571],[239,-1.6761],[241,-5.9689],[242,-8.2641],[243,-9.4117],[244,-11.1702],[245,-11.6099],[246,-11.7564],[247,-13.7992],[248,-18.9061],[249,-22.3107],[250,-23.5874],[251,-24.013],[422,-24.013]]
       }
       index_metacarpal_r_ctrl.Scale.X {type=MovieSceneScriptingFloatChannel, default=1, const [1,423]}
       index_metacarpal_r_ctrl.Scale.Y {type=MovieSceneScriptingFloatChannel, default=1, const [1,423]}
       index_metacarpal_r_ctrl.Scale.Z {type=MovieSceneScriptingFloatChannel, default=1, const [1,423]}
       index_01_r_ctrl.Location.X {
        type=MovieSceneScriptingFloatChannel
        default=-0.4363
        thinned_from=423
        keys [[0,-0.4363],[214,-0.4363],[215,-0.4353],[216,-0.4323],[217,-0.4268],[218,-0.4188],[219,-0.4078],[220,-0.3936],[221,-0.3759],[222,-0.3545],[223,-0.3239],[224,-0.2822],[225,-0.2347],[226,-0.1865],[227,-0.1429],[228,-0.1091],[229,-0.0948],[230,-0.0926],[231,-0.0818],[232,-0.0818],[233,-0.069],[234,-0.0357],[235,-0.0183],[236,-0.0047],[237,-0.0001],[238,0],[239,-0.0011],[241,-0.1943],[242,-0.2698],[243,-0.3076],[244,-0.3655],[245,-0.38],[246,-0.3848],[247,-0.3934],[248,-0.4148],[249,-0.4291],[250,-0.4345],[251,-0.4363],[422,-0.4363]]
       }
       index_01_r_ctrl.Location.Y {
        type=MovieSceneScriptingFloatChannel
        default=-1.0252
        thinned_from=423
        keys [[0,-1.0252],[214,-1.0252],[215,-1.0229],[216,-1.0157],[217,-1.003],[218,-0.984],[219,-0.9581],[220,-0.9248],[221,-0.8833],[222,-0.8329],[223,-0.761],[224,-0.6631],[225,-0.5515],[226,-0.4383],[227,-0.3359],[228,-0.2563],[229,-0.2228],[230,-0.2177],[231,-0.1922],[232,-0.1922],[233,-0.1622],[234,-0.0839],[235,-0.0429],[236,-0.011],[237,-0.0004],[238,0],[239,-0.0027],[241,-0.4565],[242,-0.6341],[243,-0.7228],[244,-0.8589],[245,-0.8929],[246,-0.9042],[247,-0.9244],[248,-0.9748],[249,-1.0084],[250,-1.021],[251,-1.0252],[422,-1.0252]]
       }
       index_01_r_ctrl.Location.Z {
        type=MovieSceneScriptingFloatChannel
        default=-0.034
        thinned_from=423
        keys [[0,-0.034],[215,-0.0339],[217,-0.0332],[218,-0.0326],[219,-0.0317],[220,-0.0306],[221,-0.0293],[222,-0.0276],[223,-0.0252],[224,-0.022],[226,-0.0145],[227,-0.0111],[228,-0.0085],[229,-0.0074],[230,-0.0072],[231,-0.0064],[232,-0.0064],[233,-0.0054],[234,-0.0028],[235,-0.0014],[236,-0.0004],[237,0],[239,-0.0001],[241,-0.0151],[242,-0.021],[243,-0.0239],[244,-0.0284],[245,-0.0296],[246,-0.03],[247,-0.0306],[248,-0.0323],[249,-0.0334],[250,-0.0338],[422,-0.034]]
       }
       index_01_r_ctrl.Rotation.X {
        type=MovieSceneScriptingFloatChannel
        default=8.4656
        thinned_from=423
        keys [[0,8.4656],[214,8.4656],[215,8.4352],[216,8.3382],[217,8.1659],[218,7.9097],[219,7.5608],[220,7.1105],[221,6.5503],[222,5.8713],[223,4.9001],[224,3.5789],[225,2.0724],[226,0.5454],[227,-0.8376],[228,-1.9117],[229,-2.3641],[230,-2.4322],[231,-2.7765],[232,-2.7765],[233,-3.1818],[234,-4.2383],[235,-4.791],[236,-5.2221],[237,-5.3657],[238,-5.3708],[239,-5.3348],[241,0.7907],[242,3.1869],[243,4.385],[244,6.2209],[245,6.6799],[246,6.8329],[247,7.105],[248,7.7853],[249,8.2388],[250,8.4089],[251,8.4656],[422,8.4656]]
       }
       index_01_r_ctrl.Rotation.Y {
        type=MovieSceneScriptingFloatChannel
        default=-1.3714
        thinned_from=423
        keys [[0,-1.3714],[214,-1.3715],[215,-1.373],[216,-1.378],[217,-1.3868],[218,-1.3999],[219,-1.4178],[220,-1.4408],[221,-1.4695],[222,-1.5042],[223,-1.5539],[224,-1.6215],[226,-1.7768],[227,-1.8476],[228,-1.9025],[229,-1.9257],[230,-1.9292],[231,-1.9468],[232,-1.9468],[233,-1.9675],[234,-2.0216],[235,-2.0499],[236,-2.072],[237,-2.0793],[239,-2.0777],[241,-1.7642],[242,-1.6416],[243,-1.5803],[244,-1.4863],[245,-1.4628],[246,-1.455],[247,-1.4411],[248,-1.4063],[249,-1.3831],[250,-1.3743],[251,-1.3714],[422,-1.3714]]
       }
       index_01_r_ctrl.Rotation.Z {
        type=MovieSceneScriptingFloatChannel
        default=-78.8307
        thinned_from=423
        keys [[0,-78.8307],[130,-78.8307],[131,-77.9793],[132,-76.1725],[134,-67.7822],[135,-66.4322],[214,-66.4322],[215,-66.2622],[216,-65.7199],[217,-64.7566],[218,-63.3238],[219,-61.373],[220,-58.8555],[221,-55.7228],[222,-51.9264],[223,-46.4962],[224,-39.109],[225,-30.6856],[226,-22.1471],[227,-14.4144],[228,-8.4087],[229,-5.8791],[231,8.068],[232,15.268],[233,14.3944],[234,12.4509],[235,11.6162],[236,11.0822],[237,9.6199],[238,9.2137],[239,8.7385],[241,-23.5191],[242,-36.9167],[243,-43.6158],[244,-56.6656],[245,-59.9281],[246,-61.0156],[247,-63.9847],[248,-71.4077],[249,-76.3564],[250,-78.2121],[251,-78.8307],[422,-78.8307]]
       }
       index_01_r_ctrl.Scale.X {type=MovieSceneScriptingFloatChannel, default=1, const [1,423]}
       index_01_r_ctrl.Scale.Y {type=MovieSceneScriptingFloatChannel, default=1, const [1,423]}
       index_01_r_ctrl.Scale.Z {type=MovieSceneScriptingFloatChannel, default=1, const [1,423]}
       index_02_r_ctrl.Location.X {type=MovieSceneScriptingFloatChannel, default=-0.0001, thinned_from=423, keys [[0,-0.0001],[422,-0.0001]]}
       index_02_r_ctrl.Location.Y {type=MovieSceneScriptingFloatChannel, default=-0.0001, thinned_from=423, keys [[0,-0.0001],[422,-0.0001]]}
       index_02_r_ctrl.Location.Z {type=MovieSceneScriptingFloatChannel, default=-0.0001, thinned_from=423, keys [[0,-0.0001],[422,-0.0001]]}
       index_02_r_ctrl.Rotation.X {
        type=MovieSceneScriptingFloatChannel
        default=2.7848
        thinned_from=423
        keys [[0,2.7848],[214,2.7848],[215,2.7787],[216,2.7592],[217,2.7245],[218,2.6729],[219,2.6027],[220,2.5121],[221,2.3993],[222,2.2627],[223,2.0672],[224,1.8013],[225,1.4981],[226,1.1907],[227,0.9124],[228,0.6962],[229,0.6052],[230,0.5915],[231,0.5222],[232,0.5222],[233,0.4406],[234,0.2279],[235,0.1167],[236,0.0299],[237,0.001],[238,0],[239,0.0073],[241,1.2401],[242,1.7224],[243,1.9635],[244,2.333],[245,2.4254],[246,2.4562],[247,2.511],[248,2.6479],[249,2.7392],[250,2.7734],[251,2.7848],[422,2.7848]]
       }
       index_02_r_ctrl.Rotation.Y {
        type=MovieSceneScriptingFloatChannel
        default=-2.6518
        thinned_from=423
        keys [[0,-2.6518],[214,-2.6518],[215,-2.6459],[216,-2.6273],[217,-2.5943],[218,-2.5452],[219,-2.4784],[220,-2.3921],[221,-2.2847],[222,-2.1546],[223,-1.9684],[224,-1.7152],[225,-1.4265],[226,-1.1338],[227,-0.8688],[228,-0.6629],[229,-0.5762],[230,-0.5632],[231,-0.4972],[232,-0.4972],[233,-0.4195],[234,-0.217],[235,-0.1111],[236,-0.0285],[237,-0.001],[238,0],[239,-0.0069],[241,-1.1809],[242,-1.6401],[243,-1.8697],[244,-2.2216],[245,-2.3095],[246,-2.3388],[247,-2.391],[248,-2.5214],[249,-2.6083],[250,-2.6409],[251,-2.6518],[422,-2.6518]]
       }
       index_02_r_ctrl.Rotation.Z {
        type=MovieSceneScriptingFloatChannel
        default=-67.2511
        thinned_from=423
        keys [[0,-67.2511],[214,-67.2511],[215,-67.0207],[216,-66.2857],[217,-64.9802],[218,-63.0384],[219,-60.3944],[220,-56.9826],[221,-52.7369],[222,-47.5917],[223,-40.2324],[224,-30.2207],[225,-18.8047],[226,-7.2327],[227,3.2471],[228,11.3865],[229,14.8147],[231,29.581],[232,36.781],[233,36.7126],[234,36.8678],[236,37.4532],[237,36.2125],[238,35.8679],[239,35.1542],[240,-6.4817],[241,-9.0921],[242,-27.2496],[243,-36.3286],[244,-50.2411],[245,-53.7192],[246,-54.8786],[247,-56.9407],[248,-62.0959],[249,-65.5327],[250,-66.8215],[251,-67.2511],[422,-67.2511]]
       }
       index_02_r_ctrl.Scale.X {type=MovieSceneScriptingFloatChannel, default=1, const [1,423]}
       index_02_r_ctrl.Scale.Y {type=MovieSceneScriptingFloatChannel, default=1, const [1,423]}
       index_02_r_ctrl.Scale.Z {type=MovieSceneScriptingFloatChannel, default=1, const [1,423]}
       index_03_r_ctrl.Location.X {type=MovieSceneScriptingFloatChannel, default=-0.0001, thinned_from=423, keys [[0,-0.0001],[422,-0.0001]]}
       index_03_r_ctrl.Location.Y {type=MovieSceneScriptingFloatChannel, default=0.0001, thinned_from=423, keys [[0,0.0001],[422,0.0001]]}
       index_03_r_ctrl.Location.Z {type=MovieSceneScriptingFloatChannel, default=-0.0001, thinned_from=423, keys [[0,-0.0001],[422,-0.0001]]}
       index_03_r_ctrl.Rotation.X {type=MovieSceneScriptingFloatChannel, default=0, thinned_from=423, keys [[0,0],[422,0]]}
       index_03_r_ctrl.Rotation.Y {type=MovieSceneScriptingFloatChannel, default=0, thinned_from=423, keys [[0,0],[422,0]]}
       index_03_r_ctrl.Rotation.Z {
        type=MovieSceneScriptingFloatChannel
        default=-90.9605
        thinned_from=423
        keys [[0,-90.9605],[214,-90.9605],[215,-90.6928],[216,-89.8387],[217,-88.3217],[218,-86.0654],[219,-82.9933],[220,-79.0289],[221,-74.0956],[222,-68.1171],[223,-59.5659],[224,-47.9327],[225,-34.6679],[226,-21.2218],[227,-9.0447],[228,0.413],[229,4.3965],[231,19.6688],[232,26.8688],[233,27.2979],[234,28.7497],[235,29.6912],[236,30.5424],[237,29.4387],[238,29.1321],[239,28.271],[241,-23.3823],[242,-44.4804],[243,-55.0299],[244,-71.1956],[245,-75.237],[246,-76.5842],[247,-78.9802],[248,-84.9703],[249,-88.9637],[250,-90.4613],[251,-90.9605],[422,-90.9605]]
       }
       index_03_r_ctrl.Scale.X {type=MovieSceneScriptingFloatChannel, default=1, const [1,423]}
       index_03_r_ctrl.Scale.Y {type=MovieSceneScriptingFloatChannel, default=1, const [1,423]}
       index_03_r_ctrl.Scale.Z {type=MovieSceneScriptingFloatChannel, default=1, const [1,423]}
       middle_metacarpal_r_ctrl.Location.X {type=MovieSceneScriptingFloatChannel, default=0, thinned_from=423, keys [[0,0],[422,0]]}
       middle_metacarpal_r_ctrl.Location.Y {type=MovieSceneScriptingFloatChannel, default=0.0001, thinned_from=423, keys [[0,0.0001],[422,0.0001]]}
       middle_metacarpal_r_ctrl.Location.Z {type=MovieSceneScriptingFloatChannel, default=0, thinned_from=423, keys [[0,0],[422,0]]}
       middle_metacarpal_r_ctrl.Rotation.X {
        type=MovieSceneScriptingFloatChannel
        default=0.2131
        thinned_from=423
        keys [[0,0.2131],[214,0.2131],[215,0.2126],[216,0.2111],[217,0.2085],[218,0.2045],[219,0.1992],[220,0.1922],[221,0.1836],[222,0.1731],[223,0.1582],[224,0.1378],[225,0.1146],[226,0.0911],[227,0.0698],[228,0.0533],[229,0.0463],[230,0.0453],[231,0.04],[232,0.04],[233,0.0337],[234,0.0174],[235,0.0089],[236,0.0023],[237,0.0001],[238,0],[239,0.0006],[241,0.0949],[242,0.1318],[243,0.1502],[244,0.1785],[245,0.1856],[246,0.1879],[247,0.1921],[248,0.2026],[249,0.2096],[250,0.2122],[251,0.2131],[422,0.2131]]
       }
       middle_metacarpal_r_ctrl.Rotation.Y {
        type=MovieSceneScriptingFloatChannel
        default=-1.1482
        thinned_from=423
        keys [[0,-1.1482],[214,-1.1482],[215,-1.1456],[216,-1.1376],[217,-1.1233],[218,-1.102],[219,-1.0731],[220,-1.0357],[221,-0.9892],[222,-0.9329],[223,-0.8523],[224,-0.7427],[225,-0.6176],[226,-0.4909],[227,-0.3762],[228,-0.287],[229,-0.2495],[230,-0.2438],[231,-0.2153],[232,-0.2153],[233,-0.1816],[234,-0.094],[235,-0.0481],[236,-0.0123],[237,-0.0004],[238,0],[239,-0.003],[241,-0.5113],[242,-0.7101],[243,-0.8095],[244,-0.9619],[245,-1],[246,-1.0127],[247,-1.0353],[248,-1.0917],[249,-1.1293],[250,-1.1435],[251,-1.1482],[422,-1.1482]]
       }
       middle_metacarpal_r_ctrl.Rotation.Z {
        type=MovieSceneScriptingFloatChannel
        default=-17.3656
        thinned_from=423
        keys [[0,-17.3656],[130,-17.3656],[131,-16.2238],[132,-14.7966],[134,-10.5148],[214,-10.5148],[215,-10.4917],[216,-10.418],[217,-10.2871],[218,-10.0924],[219,-9.8272],[220,-9.4851],[221,-9.0593],[222,-8.5433],[223,-7.8053],[224,-6.8013],[225,-5.6564],[226,-4.4959],[227,-3.445],[228,-2.6287],[229,-2.2849],[231,9.6699],[232,16.8699],[233,14.0378],[234,6.99],[235,3.4847],[236,0.8683],[237,-1.1329],[238,-1.6888],[239,-1.584],[241,-4.6824],[242,-6.5033],[243,-7.4138],[244,-8.809],[245,-9.1578],[246,-9.2741],[247,-10.6226],[248,-13.9941],[249,-16.2418],[250,-17.0847],[251,-17.3656],[422,-17.3656]]
       }
       middle_metacarpal_r_ctrl.Scale.X {type=MovieSceneScriptingFloatChannel, default=1, const [1,423]}
       middle_metacarpal_r_ctrl.Scale.Y {type=MovieSceneScriptingFloatChannel, default=1, const [1,423]}
       middle_metacarpal_r_ctrl.Scale.Z {type=MovieSceneScriptingFloatChannel, default=1, const [1,423]}
       middle_01_r_ctrl.Location.X {
        type=MovieSceneScriptingFloatChannel
        default=-0.553
        thinned_from=423
        keys [[0,-0.553],[214,-0.553],[215,-0.5518],[216,-0.5479],[217,-0.541],[218,-0.5308],[219,-0.5168],[220,-0.4988],[221,-0.4764],[222,-0.4493],[223,-0.4105],[224,-0.3576],[225,-0.2974],[226,-0.2363],[227,-0.1811],[228,-0.1381],[229,-0.12],[230,-0.1173],[231,-0.1035],[232,-0.1035],[233,-0.0873],[234,-0.0451],[235,-0.023],[236,-0.0058],[237,0],[238,0.0002],[239,-0.0013],[241,-0.2462],[242,-0.342],[243,-0.3899],[244,-0.4633],[245,-0.4816],[246,-0.4877],[247,-0.4986],[248,-0.5258],[249,-0.5439],[250,-0.5507],[251,-0.553],[422,-0.553]]
       }
       middle_01_r_ctrl.Location.Y {
        type=MovieSceneScriptingFloatChannel
        default=-0.9634
        thinned_from=423
        keys [[0,-0.9634],[214,-0.9634],[215,-0.9613],[216,-0.9545],[217,-0.9425],[218,-0.9247],[219,-0.9004],[220,-0.8691],[221,-0.83],[222,-0.7828],[223,-0.7151],[224,-0.6232],[225,-0.5183],[226,-0.4119],[227,-0.3156],[228,-0.2409],[229,-0.2094],[230,-0.2046],[231,-0.1806],[232,-0.1806],[233,-0.1524],[234,-0.0789],[235,-0.0404],[236,-0.0104],[237,-0.0004],[238,0],[239,-0.0025],[241,-0.429],[242,-0.5959],[243,-0.6793],[244,-0.8071],[245,-0.8391],[246,-0.8497],[247,-0.8687],[248,-0.916],[249,-0.9476],[250,-0.9595],[251,-0.9634],[422,-0.9634]]
       }
       middle_01_r_ctrl.Location.Z {
        type=MovieSceneScriptingFloatChannel
        default=-0.0914
        thinned_from=423
        keys [[0,-0.0914],[214,-0.0914],[215,-0.0912],[216,-0.0906],[217,-0.0895],[218,-0.0878],[219,-0.0855],[220,-0.0825],[221,-0.0788],[222,-0.0743],[223,-0.0679],[224,-0.0591],[226,-0.0391],[227,-0.03],[228,-0.0229],[229,-0.0199],[230,-0.0194],[231,-0.0171],[232,-0.0171],[233,-0.0145],[234,-0.0075],[235,-0.0038],[236,-0.001],[237,0],[238,0],[239,-0.0002],[241,-0.0407],[242,-0.0566],[243,-0.0645],[244,-0.0766],[245,-0.0796],[246,-0.0807],[247,-0.0825],[248,-0.0869],[249,-0.0899],[250,-0.0911],[251,-0.0914],[422,-0.0914]]
       }
       middle_01_r_ctrl.Rotation.X {type=MovieSceneScriptingFloatChannel, default=0, thinned_from=423, keys [[0,0],[422,0]]}
       middle_01_r_ctrl.Rotation.Y {type=MovieSceneScriptingFloatChannel, default=0, thinned_from=423, keys [[0,0],[422,0]]}
       middle_01_r_ctrl.Rotation.Z {
        type=MovieSceneScriptingFloatChannel
        default=-75.0797
        thinned_from=423
        keys [[0,-75.0797],[130,-75.0797],[131,-75.2309],[132,-75.9188],[134,-79.9786],[135,-80.8858],[214,-80.8858],[215,-80.7337],[216,-80.2483],[217,-79.3864],[218,-78.1042],[219,-76.3585],[220,-74.1058],[221,-71.3025],[222,-67.9053],[223,-63.0462],[224,-56.4358],[225,-48.8981],[226,-41.2575],[227,-34.338],[228,-28.9638],[229,-26.7002],[231,-12.9955],[232,-5.7956],[233,-6.9076],[234,-9.4722],[235,-10.6319],[236,-11.4193],[237,-12.9472],[239,-13.7762],[241,-42.4852],[242,-54.474],[243,-60.4686],[244,-67.7665],[245,-69.591],[246,-70.1992],[247,-71.0126],[248,-73.0461],[249,-74.4018],[250,-74.9102],[251,-75.0797],[422,-75.0797]]
       }
       middle_01_r_ctrl.Scale.X {type=MovieSceneScriptingFloatChannel, default=1, const [1,423]}
       middle_01_r_ctrl.Scale.Y {type=MovieSceneScriptingFloatChannel, default=1, const [1,423]}
       middle_01_r_ctrl.Scale.Z {type=MovieSceneScriptingFloatChannel, default=1, const [1,423]}
       middle_02_r_ctrl.Location.X {type=MovieSceneScriptingFloatChannel, default=0.0008, thinned_from=423, keys [[0,0.0008],[422,0.0008]]}
       middle_02_r_ctrl.Location.Y {type=MovieSceneScriptingFloatChannel, default=0.0009, thinned_from=423, keys [[0,0.0008],[422,0.0009]]}
       middle_02_r_ctrl.Location.Z {type=MovieSceneScriptingFloatChannel, default=0.0476, thinned_from=423, keys [[0,0.0476],[422,0.0476]]}
       middle_02_r_ctrl.Rotation.X {
        type=MovieSceneScriptingFloatChannel
        default=-3.6824
        thinned_from=423
        keys [[0,-3.6824],[214,-3.6824],[215,-3.6743],[216,-3.6486],[217,-3.6029],[218,-3.5349],[219,-3.4424],[220,-3.3231],[221,-3.1746],[222,-2.9946],[223,-2.7373],[224,-2.3873],[225,-1.9881],[226,-1.5834],[227,-1.2166],[228,-0.9316],[229,-0.8116],[230,-0.7945],[231,-0.7037],[232,-0.7044],[233,-0.5963],[234,-0.3146],[235,-0.1672],[236,-0.0522],[237,-0.0138],[238,-0.0124],[239,-0.022],[241,-1.6484],[242,-2.2834],[243,-2.6008],[244,-3.0873],[245,-3.2089],[246,-3.2495],[247,-3.3216],[248,-3.502],[249,-3.6222],[250,-3.6673],[251,-3.6824],[422,-3.6824]]
       }
       middle_02_r_ctrl.Rotation.Y {
        type=MovieSceneScriptingFloatChannel
        default=0.451
        thinned_from=423
        keys [[0,0.451],[214,0.451],[215,0.4499],[216,0.4466],[217,0.4407],[218,0.4319],[219,0.42],[220,0.4046],[221,0.3855],[222,0.3625],[223,0.3297],[224,0.2855],[226,0.1853],[227,0.1403],[228,0.1056],[229,0.091],[230,0.0901],[231,0.08],[232,0.0815],[233,0.068],[234,0.0326],[235,0.0142],[236,-0.0002],[237,-0.0052],[238,-0.0055],[239,-0.0044],[240,0.0938],[241,0.1934],[242,0.2724],[243,0.3124],[244,0.3743],[245,0.3899],[246,0.3951],[247,0.4044],[248,0.4277],[249,0.4432],[250,0.449],[251,0.451],[422,0.451]]
       }
       middle_02_r_ctrl.Rotation.Z {
        type=MovieSceneScriptingFloatChannel
        default=-62.569
        thinned_from=423
        keys [[0,-62.569],[214,-62.569],[215,-62.3468],[216,-61.6378],[217,-60.3786],[218,-58.5057],[219,-55.9556],[220,-52.6647],[221,-48.5697],[222,-43.607],[223,-36.5088],[224,-26.8523],[225,-15.8413],[226,-4.6799],[227,5.4281],[228,13.2788],[229,16.5854],[231,31.2408],[232,38.4409],[234,38.1347],[235,38.2493],[236,38.4556],[237,37.1849],[238,36.832],[239,36.1505],[241,-6.4733],[242,-23.9866],[243,-32.7435],[244,-46.1624],[245,-49.5172],[246,-50.6354],[247,-52.6243],[248,-57.5966],[249,-60.9115],[250,-62.1546],[251,-62.569],[422,-62.569]]
       }
       middle_02_r_ctrl.Scale.X {type=MovieSceneScriptingFloatChannel, default=1, const [1,423]}
       middle_02_r_ctrl.Scale.Y {type=MovieSceneScriptingFloatChannel, default=1, const [1,423]}
       middle_02_r_ctrl.Scale.Z {type=MovieSceneScriptingFloatChannel, default=1, const [1,423]}
       middle_03_r_ctrl.Location.X {type=MovieSceneScriptingFloatChannel, default=0.0005, thinned_from=423, keys [[0,0.0005],[422,0.0005]]}
       middle_03_r_ctrl.Location.Y {type=MovieSceneScriptingFloatChannel, default=0, thinned_from=423, keys [[0,0],[422,0]]}
       middle_03_r_ctrl.Location.Z {type=MovieSceneScriptingFloatChannel, default=0.0061, thinned_from=423, keys [[0,0.0061],[422,0.0061]]}
       middle_03_r_ctrl.Rotation.X {
        type=MovieSceneScriptingFloatChannel
        default=-0.0051
        thinned_from=423
        keys [[0,-0.0051],[215,-0.0051],[218,-0.0057],[221,-0.0072],[223,-0.0088],[225,-0.0105],[226,-0.011],[227,-0.011],[229,-0.0108],[231,-0.0098],[232,-0.009],[239,-0.0093],[240,-0.0108],[241,-0.0109],[243,-0.0092],[244,-0.0076],[245,-0.0071],[247,-0.0066],[248,-0.0059],[249,-0.0053],[422,-0.0051]]
       }
       middle_03_r_ctrl.Rotation.Y {
        type=MovieSceneScriptingFloatChannel
        default=-0.0098
        thinned_from=423
        keys [[0,-0.0098],[216,-0.0097],[219,-0.0092],[221,-0.0084],[222,-0.0077],[223,-0.0067],[224,-0.0052],[227,0.0005],[228,0.002],[229,0.0026],[232,0.0064],[236,0.0063],[239,0.0059],[240,0.0023],[241,-0.0017],[242,-0.0048],[243,-0.0062],[244,-0.0081],[245,-0.0085],[247,-0.0088],[248,-0.0094],[249,-0.0097],[422,-0.0098]]
       }
       middle_03_r_ctrl.Rotation.Z {
        type=MovieSceneScriptingFloatChannel
        default=-76.6019
        thinned_from=423
        keys [[0,-76.6019],[214,-76.6019],[215,-76.3877],[216,-75.7044],[217,-74.4906],[218,-72.6854],[219,-70.2273],[220,-67.0554],[221,-63.1082],[222,-58.3247],[223,-51.4829],[224,-42.1751],[225,-31.5618],[226,-20.8034],[227,-11.0604],[228,-3.4933],[229,-0.3061],[231,14.2405],[232,21.4405],[233,21.1561],[234,20.7484],[235,20.7172],[236,20.8097],[237,19.5096],[238,19.1485],[239,18.4987],[241,-22.5321],[242,-39.4129],[243,-47.8536],[244,-60.7878],[245,-64.0214],[246,-65.0993],[247,-67.0164],[248,-71.8092],[249,-75.0043],[250,-76.2025],[251,-76.6019],[422,-76.6019]]
       }
       middle_03_r_ctrl.Scale.X {type=MovieSceneScriptingFloatChannel, default=1, const [1,423]}
       middle_03_r_ctrl.Scale.Y {type=MovieSceneScriptingFloatChannel, default=1, const [1,423]}
       middle_03_r_ctrl.Scale.Z {type=MovieSceneScriptingFloatChannel, default=1, const [1,423]}
       ring_metacarpal_r_ctrl.Location.X {type=MovieSceneScriptingFloatChannel, default=0.0001, thinned_from=423, keys [[0,0.0001],[422,0.0001]]}
       ring_metacarpal_r_ctrl.Location.Y {type=MovieSceneScriptingFloatChannel, default=0.0002, thinned_from=423, keys [[0,0.0002],[422,0.0002]]}
       ring_metacarpal_r_ctrl.Location.Z {type=MovieSceneScriptingFloatChannel, default=0, thinned_from=423, keys [[0,0],[422,0]]}
       ring_metacarpal_r_ctrl.Rotation.X {
        type=MovieSceneScriptingFloatChannel
        default=5.3943
        thinned_from=423
        keys [[0,5.3943],[214,5.3943],[215,5.3828],[216,5.3461],[217,5.2808],[218,5.1837],[219,5.0515],[220,4.881],[221,4.6687],[222,4.4115],[223,4.0436],[224,3.543],[225,2.9723],[226,2.3938],[227,1.8698],[228,1.4629],[229,1.2915],[230,1.2657],[231,1.1353],[232,1.1353],[233,0.9817],[234,0.5815],[235,0.3721],[236,0.2088],[237,0.1543],[238,0.1524],[239,0.1661],[241,2.4867],[242,3.3945],[243,3.8484],[244,4.5439],[245,4.7178],[246,4.7758],[247,4.8789],[248,5.1366],[249,5.3084],[250,5.3729],[251,5.3943],[422,5.3943]]
       }
       ring_metacarpal_r_ctrl.Rotation.Y {
        type=MovieSceneScriptingFloatChannel
        default=-2.9431
        thinned_from=423
        keys [[0,-2.9431],[214,-2.9431],[215,-2.9367],[216,-2.9162],[217,-2.8798],[218,-2.8257],[219,-2.7521],[220,-2.657],[221,-2.5387],[222,-2.3953],[223,-2.1902],[224,-1.9112],[225,-1.593],[226,-1.2705],[227,-0.9785],[228,-0.7516],[229,-0.6561],[230,-0.6417],[231,-0.569],[232,-0.569],[233,-0.4834],[234,-0.2603],[235,-0.1436],[236,-0.0526],[237,-0.0222],[238,-0.0211],[239,-0.0288],[241,-1.3223],[242,-1.8284],[243,-2.0814],[244,-2.4691],[245,-2.566],[246,-2.5983],[247,-2.6558],[248,-2.7995],[249,-2.8952],[250,-2.9312],[251,-2.9431],[422,-2.9431]]
       }
       ring_metacarpal_r_ctrl.Rotation.Z {
        type=MovieSceneScriptingFloatChannel
        default=-10.4667
        thinned_from=423
        keys [[0,-10.4667],[214,-10.4667],[215,-10.4414],[216,-10.3605],[217,-10.2167],[218,-10.003],[219,-9.712],[220,-9.3364],[221,-8.8691],[222,-8.3027],[223,-7.4926],[224,-6.3906],[225,-5.1339],[226,-3.8601],[227,-2.7066],[228,-1.8106],[229,-1.4332],[231,10.5522],[232,17.7522],[233,14.9501],[234,7.9808],[235,4.5165],[236,1.932],[237,-0.0609],[238,-0.6145],[239,-0.5186],[241,-4.0648],[242,-6.0635],[243,-7.0629],[244,-8.5943],[245,-8.9772],[246,-9.1048],[247,-9.3318],[248,-9.8993],[249,-10.2776],[250,-10.4194],[251,-10.4667],[422,-10.4667]]
       }
       ring_metacarpal_r_ctrl.Scale.X {type=MovieSceneScriptingFloatChannel, default=1, const [1,423]}
       ring_metacarpal_r_ctrl.Scale.Y {type=MovieSceneScriptingFloatChannel, default=1, const [1,423]}
       ring_metacarpal_r_ctrl.Scale.Z {type=MovieSceneScriptingFloatChannel, default=1, const [1,423]}
       ring_01_r_ctrl.Location.X {
        type=MovieSceneScriptingFloatChannel
        default=-0.371
        thinned_from=423
        keys [[0,-0.371],[214,-0.371],[215,-0.3702],[216,-0.3676],[217,-0.363],[218,-0.3561],[219,-0.3468],[220,-0.3347],[221,-0.3197],[222,-0.3015],[223,-0.2755],[224,-0.2401],[225,-0.1997],[226,-0.1588],[227,-0.1217],[228,-0.0929],[229,-0.0808],[230,-0.0789],[231,-0.0697],[232,-0.0697],[233,-0.0589],[234,-0.0305],[235,-0.0157],[236,-0.0042],[237,-0.0003],[238,-0.0002],[239,-0.0011],[241,-0.1653],[242,-0.2296],[243,-0.2617],[244,-0.3109],[245,-0.3232],[246,-0.3273],[247,-0.3346],[248,-0.3528],[249,-0.365],[250,-0.3695],[251,-0.371],[422,-0.371]]
       }
       ring_01_r_ctrl.Location.Y {
        type=MovieSceneScriptingFloatChannel
        default=-1.0509
        thinned_from=423
        keys [[0,-1.0509],[214,-1.0509],[215,-1.0486],[216,-1.0413],[217,-1.0282],[218,-1.0087],[219,-0.9822],[220,-0.948],[221,-0.9055],[222,-0.8539],[223,-0.7801],[224,-0.6798],[225,-0.5654],[226,-0.4494],[227,-0.3443],[228,-0.2628],[229,-0.2284],[230,-0.2232],[231,-0.1971],[232,-0.1971],[233,-0.1663],[234,-0.086],[235,-0.0441],[236,-0.0113],[237,-0.0004],[238,0],[239,-0.0028],[241,-0.468],[242,-0.65],[243,-0.741],[244,-0.8804],[245,-0.9153],[246,-0.9269],[247,-0.9476],[248,-0.9993],[249,-1.0337],[250,-1.0466],[251,-1.0509],[422,-1.0509]]
       }
       ring_01_r_ctrl.Location.Z {
        type=MovieSceneScriptingFloatChannel
        default=-0.0243
        thinned_from=423
        keys [[0,-0.0243],[215,-0.0243],[217,-0.0238],[219,-0.0228],[221,-0.021],[222,-0.0198],[223,-0.0181],[224,-0.0157],[226,-0.0104],[227,-0.008],[228,-0.0061],[229,-0.0053],[230,-0.0052],[231,-0.0046],[232,-0.0046],[233,-0.0038],[234,-0.002],[235,-0.001],[236,-0.0002],[237,0],[239,0],[241,-0.0108],[242,-0.0151],[243,-0.0172],[244,-0.0204],[245,-0.0212],[246,-0.0215],[247,-0.0219],[248,-0.0231],[249,-0.0239],[251,-0.0243],[422,-0.0243]]
       }
       ring_01_r_ctrl.Rotation.X {
        type=MovieSceneScriptingFloatChannel
        default=-0.8239
        thinned_from=423
        keys [[0,-0.8239],[214,-0.8239],[215,-0.8302],[216,-0.8503],[217,-0.8861],[218,-0.9393],[219,-1.0117],[220,-1.1051],[221,-1.2214],[222,-1.3624],[223,-1.5639],[224,-1.8382],[226,-2.4679],[227,-2.7549],[228,-2.9779],[229,-3.0718],[230,-3.0859],[231,-3.1574],[232,-3.1574],[233,-3.2415],[234,-3.4608],[235,-3.5756],[236,-3.665],[237,-3.6948],[238,-3.6959],[239,-3.6884],[241,-2.4169],[242,-1.9196],[243,-1.6709],[244,-1.2898],[245,-1.1945],[246,-1.1628],[247,-1.1063],[248,-0.9651],[249,-0.8709],[250,-0.8356],[251,-0.8239],[422,-0.8239]]
       }
       ring_01_r_ctrl.Rotation.Y {
        type=MovieSceneScriptingFloatChannel
        default=-2.6856
        thinned_from=423
        keys [[0,-2.6856],[215,-2.6875],[217,-2.7047],[218,-2.7211],[219,-2.7435],[220,-2.7723],[221,-2.8081],[222,-2.8516],[223,-2.9137],[224,-2.9982],[226,-3.1923],[227,-3.2808],[228,-3.3496],[229,-3.3785],[230,-3.3829],[231,-3.4049],[232,-3.4049],[233,-3.4308],[234,-3.4984],[235,-3.5338],[236,-3.5614],[237,-3.5705],[239,-3.5686],[241,-3.1766],[242,-3.0233],[243,-2.9467],[244,-2.8292],[245,-2.7998],[246,-2.79],[247,-2.7726],[248,-2.7291],[249,-2.7001],[250,-2.6892],[422,-2.6856]]
       }
       ring_01_r_ctrl.Rotation.Z {
        type=MovieSceneScriptingFloatChannel
        default=-64.5197
        thinned_from=423
        keys [[0,-64.5197],[131,-64.4729],[132,-66.4382],[134,-80.4288],[135,-84.1083],[214,-84.1083],[215,-83.9385],[216,-83.397],[217,-82.4353],[218,-81.0047],[219,-79.0569],[220,-76.5434],[221,-73.4156],[222,-69.6251],[223,-64.2035],[224,-56.8278],[225,-48.4176],[226,-39.8925],[227,-32.1719],[228,-26.1756],[229,-23.65],[231,-9.7065],[232,-2.5064],[233,-3.3837],[234,-5.3363],[235,-6.176],[236,-6.7137],[237,-8.177],[238,-8.5834],[239,-9.0577],[241,-41.2623],[242,-54.6389],[243,-61.3274],[244,-63.9384],[245,-64.5912],[246,-64.8088],[422,-64.5197]]
       }
       ring_01_r_ctrl.Scale.X {type=MovieSceneScriptingFloatChannel, default=1, const [1,423]}
       ring_01_r_ctrl.Scale.Y {type=MovieSceneScriptingFloatChannel, default=1, const [1,423]}
       ring_01_r_ctrl.Scale.Z {type=MovieSceneScriptingFloatChannel, default=1, const [1,423]}
       ring_02_r_ctrl.Location.X {type=MovieSceneScriptingFloatChannel, default=0.0003, thinned_from=423, keys [[0,0.0003],[422,0.0003]]}
       ring_02_r_ctrl.Location.Y {type=MovieSceneScriptingFloatChannel, default=0.0001, thinned_from=423, keys [[0,0.0001],[422,0.0001]]}
       ring_02_r_ctrl.Location.Z {type=MovieSceneScriptingFloatChannel, default=0, thinned_from=423, keys [[0,0],[422,0]]}
       ring_02_r_ctrl.Rotation.X {
        type=MovieSceneScriptingFloatChannel
        default=-12.8636
        thinned_from=423
        keys [[0,-12.8636],[214,-12.8636],[215,-12.8354],[216,-12.7452],[217,-12.585],[218,-12.3468],[219,-12.0224],[220,-11.6038],[221,-11.0829],[222,-10.4517],[223,-9.5488],[224,-8.3205],[225,-6.9199],[226,-5.5002],[227,-4.2145],[228,-3.2159],[229,-2.7953],[230,-2.732],[231,-2.4119],[232,-2.4119],[233,-2.0351],[234,-1.0529],[235,-0.539],[236,-0.1383],[237,-0.0047],[238,0],[239,-0.0335],[241,-5.7283],[242,-7.956],[243,-9.0699],[244,-10.7767],[245,-11.2035],[246,-11.3457],[247,-11.5987],[248,-12.2312],[249,-12.6528],[250,-12.8109],[251,-12.8636],[422,-12.8636]]
       }
       ring_02_r_ctrl.Rotation.Y {
        type=MovieSceneScriptingFloatChannel
        default=-0.6911
        thinned_from=423
        keys [[0,-0.6911],[214,-0.6911],[215,-0.6895],[216,-0.6847],[217,-0.6761],[218,-0.6633],[219,-0.6459],[220,-0.6234],[221,-0.5954],[222,-0.5615],[223,-0.513],[224,-0.447],[225,-0.3718],[226,-0.2955],[227,-0.2264],[228,-0.1728],[229,-0.1502],[230,-0.1468],[231,-0.1296],[232,-0.1296],[233,-0.1093],[234,-0.0566],[235,-0.029],[236,-0.0074],[237,-0.0003],[238,0],[239,-0.0018],[241,-0.3077],[242,-0.4274],[243,-0.4873],[244,-0.579],[245,-0.6019],[246,-0.6095],[247,-0.6231],[248,-0.6571],[249,-0.6797],[250,-0.6882],[251,-0.6911],[422,-0.6911]]
       }
       ring_02_r_ctrl.Rotation.Z {
        type=MovieSceneScriptingFloatChannel
        default=-59.6606
        thinned_from=423
        keys [[0,-59.6606],[214,-59.6606],[215,-59.5256],[216,-59.0948],[217,-58.3296],[218,-57.1914],[219,-55.6417],[220,-53.6418],[221,-51.1533],[222,-48.1374],[223,-43.8238],[224,-37.9555],[225,-31.2641],[226,-24.4812],[227,-18.3386],[228,-13.5677],[229,-11.5582],[231,1.9148],[232,9.1148],[233,7.7751],[234,4.6171],[235,3.1468],[236,2.1173],[237,0.5267],[238,0.0849],[239,-0.2523],[241,-25.5711],[242,-36.214],[243,-41.5356],[244,-49.6903],[245,-51.729],[246,-52.4086],[247,-53.6173],[248,-56.639],[249,-58.6534],[250,-59.4088],[251,-59.6606],[422,-59.6606]]
       }
       ring_02_r_ctrl.Scale.X {type=MovieSceneScriptingFloatChannel, default=1, const [1,423]}
       ring_02_r_ctrl.Scale.Y {type=MovieSceneScriptingFloatChannel, default=1, const [1,423]}
       ring_02_r_ctrl.Scale.Z {type=MovieSceneScriptingFloatChannel, default=1, const [1,423]}
       ring_03_r_ctrl.Location.X {type=MovieSceneScriptingFloatChannel, default=0, thinned_from=423, keys [[0,0],[422,0]]}
       ring_03_r_ctrl.Location.Y {type=MovieSceneScriptingFloatChannel, default=-0.0001, thinned_from=423, keys [[0,-0.0001],[422,-0.0001]]}
       ring_03_r_ctrl.Location.Z {type=MovieSceneScriptingFloatChannel, default=0, thinned_from=423, keys [[0,0],[422,0]]}
       ring_03_r_ctrl.Rotation.X {type=MovieSceneScriptingFloatChannel, default=0, thinned_from=423, keys [[0,0],[422,0]]}
       ring_03_r_ctrl.Rotation.Y {type=MovieSceneScriptingFloatChannel, default=0, thinned_from=423, keys [[0,0],[422,0]]}
       ring_03_r_ctrl.Rotation.Z {
        type=MovieSceneScriptingFloatChannel
        default=-88.4683
        thinned_from=423
        keys [[0,-88.4683],[214,-88.4683],[215,-88.2281],[216,-87.4615],[217,-86.1001],[218,-84.075],[219,-81.3177],[220,-77.7596],[221,-73.3319],[222,-67.9661],[223,-60.2914],[224,-49.8505],[225,-37.9452],[226,-25.8771],[227,-14.948],[228,-6.4597],[229,-2.8845],[231,12.0158],[232,19.2158],[233,19.279],[234,19.7774],[236,20.6824],[237,19.478],[238,19.1434],[239,18.3906],[241,-27.8162],[242,-46.752],[243,-56.2202],[244,-70.7291],[245,-74.3564],[246,-75.5655],[247,-77.7159],[248,-83.0921],[249,-86.6762],[250,-88.0203],[251,-88.4683],[422,-88.4683]]
       }
       ring_03_r_ctrl.Scale.X {type=MovieSceneScriptingFloatChannel, default=1, const [1,423]}
       ring_03_r_ctrl.Scale.Y {type=MovieSceneScriptingFloatChannel, default=1, const [1,423]}
       ring_03_r_ctrl.Scale.Z {type=MovieSceneScriptingFloatChannel, default=1, const [1,423]}
       pinky_metacarpal_r_ctrl.Location.X {type=MovieSceneScriptingFloatChannel, default=-0.0001, thinned_from=423, keys [[0,-0.0001],[422,-0.0001]]}
       pinky_metacarpal_r_ctrl.Location.Y {type=MovieSceneScriptingFloatChannel, default=0, thinned_from=423, keys [[0,0],[422,0]]}
       pinky_metacarpal_r_ctrl.Location.Z {type=MovieSceneScriptingFloatChannel, default=0, thinned_from=423, keys [[0,0],[422,0]]}
       pinky_metacarpal_r_ctrl.Rotation.X {
        type=MovieSceneScriptingFloatChannel
        default=9.8795
        thinned_from=423
        keys [[0,9.8795],[214,9.8795],[215,9.8633],[216,9.8115],[217,9.7195],[218,9.5826],[219,9.3963],[220,9.1559],[221,8.8567],[222,8.4941],[223,7.9755],[224,7.2699],[226,5.6499],[227,4.9114],[228,4.3378],[229,4.0962],[230,4.0598],[231,3.876],[232,3.876],[233,3.6595],[234,3.0953],[235,2.8001],[236,2.5699],[237,2.4932],[238,2.4905],[239,2.5098],[241,5.7809],[242,7.0605],[243,7.7003],[244,8.6808],[245,8.9259],[246,9.0076],[247,9.1529],[248,9.5162],[249,9.7584],[250,9.8492],[251,9.8795],[422,9.8795]]
       }
       pinky_metacarpal_r_ctrl.Rotation.Y {
        type=MovieSceneScriptingFloatChannel
        default=-6.9881
        thinned_from=423
        keys [[0,-6.9881],[215,-6.9858],[218,-6.9464],[220,-6.8865],[222,-6.7936],[223,-6.7207],[224,-6.6216],[226,-6.3941],[227,-6.2904],[228,-6.2098],[229,-6.1759],[230,-6.1708],[231,-6.145],[232,-6.145],[233,-6.1146],[234,-6.0354],[236,-5.9616],[239,-5.9531],[241,-6.4125],[242,-6.5922],[243,-6.6821],[244,-6.8198],[245,-6.8542],[247,-6.8861],[248,-6.9371],[249,-6.9711],[422,-6.9881]]
       }
       pinky_metacarpal_r_ctrl.Rotation.Z {
        type=MovieSceneScriptingFloatChannel
        default=-8.607
        thinned_from=423
        keys [[0,-8.607],[215,-8.6015],[217,-8.5527],[219,-8.4429],[220,-8.3613],[221,-8.2597],[222,-8.1366],[223,-7.9605],[224,-7.721],[226,-7.171],[227,-6.9202],[228,-6.7255],[229,-6.6434],[231,5.0728],[232,12.2728],[233,9.2061],[234,1.547],[235,-2.2781],[236,-5.1439],[237,-7.2097],[238,-7.7835],[239,-7.6093],[241,-7.2154],[242,-7.6499],[243,-7.8671],[244,-8.2],[245,-8.2832],[246,-8.311],[247,-8.3603],[248,-8.4837],[249,-8.5659],[250,-8.5967],[422,-8.607]]
       }
       pinky_metacarpal_r_ctrl.Scale.X {type=MovieSceneScriptingFloatChannel, default=1, const [1,423]}
       pinky_metacarpal_r_ctrl.Scale.Y {type=MovieSceneScriptingFloatChannel, default=1, const [1,423]}
       pinky_metacarpal_r_ctrl.Scale.Z {type=MovieSceneScriptingFloatChannel, default=1, const [1,423]}
       pinky_01_r_ctrl.Location.X {
        type=MovieSceneScriptingFloatChannel
        default=-0.3834
        thinned_from=423
        keys [[0,-0.3834],[214,-0.3834],[215,-0.3826],[216,-0.3799],[217,-0.3751],[218,-0.368],[219,-0.3584],[220,-0.3459],[221,-0.3303],[222,-0.3115],[223,-0.2846],[224,-0.248],[225,-0.2062],[226,-0.1639],[227,-0.1255],[228,-0.0958],[229,-0.0832],[230,-0.0813],[231,-0.0718],[232,-0.0718],[233,-0.0606],[234,-0.0313],[235,-0.0159],[236,-0.004],[237,0],[238,0.0001],[239,-0.0009],[241,-0.1707],[242,-0.2371],[243,-0.2703],[244,-0.3212],[245,-0.3339],[246,-0.3382],[247,-0.3457],[248,-0.3646],[249,-0.3772],[250,-0.3819],[251,-0.3834],[422,-0.3834]]
       }
       pinky_01_r_ctrl.Location.Y {
        type=MovieSceneScriptingFloatChannel
        default=-1.045
        thinned_from=423
        keys [[0,-1.045],[214,-1.045],[215,-1.0427],[216,-1.0354],[217,-1.0224],[218,-1.003],[219,-0.9767],[220,-0.9427],[221,-0.9003],[222,-0.849],[223,-0.7757],[224,-0.6759],[225,-0.5621],[226,-0.4467],[227,-0.3422],[228,-0.2611],[229,-0.2269],[230,-0.2218],[231,-0.1958],[232,-0.1958],[233,-0.1651],[234,-0.0853],[235,-0.0436],[236,-0.011],[237,-0.0002],[238,0.0002],[239,-0.0025],[241,-0.4652],[242,-0.6463],[243,-0.7368],[244,-0.8755],[245,-0.9101],[246,-0.9217],[247,-0.9422],[248,-0.9936],[249,-1.0279],[250,-1.0407],[251,-1.045],[422,-1.045]]
       }
       pinky_01_r_ctrl.Location.Z {
        type=MovieSceneScriptingFloatChannel
        default=-0.0533
        thinned_from=423
        keys [[0,-0.0533],[214,-0.0533],[215,-0.0532],[216,-0.0528],[217,-0.0522],[218,-0.0512],[219,-0.0498],[220,-0.0481],[221,-0.0459],[222,-0.0433],[223,-0.0396],[224,-0.0345],[226,-0.0228],[227,-0.0175],[228,-0.0133],[229,-0.0116],[230,-0.0113],[231,-0.01],[232,-0.01],[233,-0.0084],[234,-0.0044],[235,-0.0022],[236,-0.0006],[237,0],[239,-0.0001],[241,-0.0237],[242,-0.033],[243,-0.0376],[244,-0.0447],[245,-0.0464],[246,-0.047],[247,-0.0481],[248,-0.0507],[249,-0.0524],[250,-0.0531],[251,-0.0533],[422,-0.0533]]
       }
       pinky_01_r_ctrl.Rotation.X {
        type=MovieSceneScriptingFloatChannel
        default=-7.5972
        thinned_from=423
        keys [[0,-7.5972],[214,-7.5971],[215,-7.6141],[216,-7.6683],[217,-7.7645],[218,-7.9077],[219,-8.1025],[220,-8.354],[221,-8.6669],[222,-9.0461],[223,-9.5886],[224,-10.3265],[226,-12.0208],[227,-12.7932],[228,-13.3931],[229,-13.6458],[230,-13.6838],[231,-13.8761],[232,-13.8761],[233,-14.1025],[234,-14.6926],[235,-15.0013],[236,-15.242],[237,-15.3223],[239,-15.305],[241,-11.8838],[242,-10.5455],[243,-9.8763],[244,-8.8509],[245,-8.5945],[246,-8.5091],[247,-8.3571],[248,-7.9771],[249,-7.7238],[250,-7.6288],[251,-7.5972],[422,-7.5972]]
       }
       pinky_01_r_ctrl.Rotation.Y {
        type=MovieSceneScriptingFloatChannel
        default=-7.928
        thinned_from=423
        keys [[0,-7.928],[214,-7.928],[215,-7.9094],[216,-7.8499],[217,-7.7443],[218,-7.5872],[219,-7.3733],[220,-7.0972],[221,-6.7537],[222,-6.3374],[223,-5.742],[224,-4.932],[225,-4.0084],[226,-3.0721],[227,-2.2242],[228,-1.5657],[229,-1.2883],[230,-1.2465],[231,-1.0355],[232,-1.0354],[233,-0.7869],[234,-0.1392],[235,0.1997],[236,0.464],[237,0.552],[238,0.5552],[239,0.5331],[241,-3.2225],[242,-4.6916],[243,-5.4262],[244,-6.5518],[245,-6.8332],[246,-6.927],[247,-7.0939],[248,-7.511],[249,-7.789],[250,-7.8933],[251,-7.928],[422,-7.928]]
       }
       pinky_01_r_ctrl.Rotation.Z {
        type=MovieSceneScriptingFloatChannel
        default=-52.7205
        thinned_from=423
        keys [[0,-52.7205],[130,-52.7205],[131,-51.1965],[132,-53.365],[134,-76.1647],[135,-83.5712],[214,-83.5712],[215,-83.4157],[216,-82.9198],[217,-82.039],[218,-80.7289],[219,-78.9451],[220,-76.6432],[221,-73.7787],[222,-70.3072],[223,-65.342],[224,-58.5872],[225,-50.885],[226,-43.0775],[227,-36.0069],[228,-30.5154],[229,-28.2024],[231,-14.4526],[232,-7.2526],[233,-8.3204],[234,-10.7695],[235,-11.8689],[236,-12.6092],[237,-14.1249],[239,-14.9637],[241,-44.332],[242,-56.5826],[243,-62.708],[244,-56.7275],[245,-55.2324],[246,-54.734],[247,-54.3984],[248,-53.5594],[249,-53.0001],[250,-52.7904],[422,-52.7205]]
       }
       pinky_01_r_ctrl.Scale.X {type=MovieSceneScriptingFloatChannel, default=1, const [1,423]}
       pinky_01_r_ctrl.Scale.Y {type=MovieSceneScriptingFloatChannel, default=1, const [1,423]}
       pinky_01_r_ctrl.Scale.Z {type=MovieSceneScriptingFloatChannel, default=1, const [1,423]}
       pinky_02_r_ctrl.Location.X {
        type=MovieSceneScriptingFloatChannel
        default=0.3528
        thinned_from=423
        keys [[0,0.3528],[214,0.3528],[215,0.352],[216,0.3496],[217,0.3452],[218,0.3386],[219,0.3297],[220,0.3182],[221,0.304],[222,0.2866],[223,0.2619],[224,0.2282],[225,0.1898],[226,0.1508],[227,0.1156],[228,0.0882],[229,0.0766],[230,0.0749],[231,0.0661],[232,0.0661],[233,0.0558],[234,0.0288],[235,0.0147],[236,0.0037],[237,0.0001],[238,-0.0001],[239,0.0009],[241,0.1571],[242,0.2182],[243,0.2487],[244,0.2956],[245,0.3073],[246,0.3112],[247,0.3181],[248,0.3355],[249,0.347],[250,0.3514],[251,0.3528],[422,0.3528]]
       }
       pinky_02_r_ctrl.Location.Y {
        type=MovieSceneScriptingFloatChannel
        default=0.0018
        thinned_from=423
        keys [[0,0.0018],[222,0.0015],[229,0.0004],[239,0.0001],[242,0.0011],[422,0.0018]]
       }
       pinky_02_r_ctrl.Location.Z {
        type=MovieSceneScriptingFloatChannel
        default=-0.0756
        thinned_from=423
        keys [[0,-0.0756],[214,-0.0756],[215,-0.0755],[216,-0.0749],[217,-0.074],[218,-0.0726],[219,-0.0707],[220,-0.0682],[221,-0.0652],[222,-0.0614],[223,-0.0561],[224,-0.0489],[226,-0.0323],[227,-0.0248],[228,-0.0189],[229,-0.0164],[230,-0.0161],[231,-0.0142],[232,-0.0142],[233,-0.012],[234,-0.0062],[235,-0.0032],[236,-0.0008],[237,0],[238,0],[239,-0.0002],[241,-0.0337],[242,-0.0468],[243,-0.0533],[244,-0.0634],[245,-0.0659],[246,-0.0667],[247,-0.0682],[248,-0.0719],[249,-0.0744],[250,-0.0753],[251,-0.0756],[422,-0.0756]]
       }
       pinky_02_r_ctrl.Rotation.X {
        type=MovieSceneScriptingFloatChannel
        default=1.9939
        thinned_from=423
        keys [[0,1.9939],[214,1.9939],[215,1.9896],[216,1.9756],[217,1.9508],[218,1.9138],[219,1.8635],[220,1.7987],[221,1.7179],[222,1.6201],[223,1.4801],[224,1.2897],[225,1.0726],[226,0.8526],[227,0.6533],[228,0.4985],[229,0.4333],[230,0.4235],[231,0.3739],[232,0.3739],[233,0.3154],[234,0.1632],[235,0.0835],[236,0.0214],[237,0.0007],[238,0],[239,0.0052],[241,0.8879],[242,1.2332],[243,1.4059],[244,1.6705],[245,1.7366],[246,1.7587],[247,1.7979],[248,1.8959],[249,1.9613],[250,1.9858],[251,1.9939],[422,1.9939]]
       }
       pinky_02_r_ctrl.Rotation.Y {
        type=MovieSceneScriptingFloatChannel
        default=10.0966
        thinned_from=423
        keys [[0,10.0966],[214,10.0966],[215,10.0744],[216,10.0037],[217,9.8779],[218,9.691],[219,9.4364],[220,9.1078],[221,8.699],[222,8.2035],[223,7.4948],[224,6.5308],[225,5.4314],[226,4.3171],[227,3.3079],[228,2.5242],[229,2.194],[230,2.1444],[231,1.8931],[232,1.8931],[233,1.5973],[234,0.8264],[235,0.4231],[236,0.1085],[237,0.0037],[238,0],[239,0.0263],[241,4.4962],[242,6.2446],[243,7.1189],[244,8.4586],[245,8.7936],[246,8.9052],[247,9.1038],[248,9.6002],[249,9.9312],[250,10.0553],[251,10.0966],[422,10.0966]]
       }
       pinky_02_r_ctrl.Rotation.Z {
        type=MovieSceneScriptingFloatChannel
        default=-57.9295
        thinned_from=423
        keys [[0,-57.9295],[214,-57.9295],[215,-57.799],[216,-57.3827],[217,-56.6433],[218,-55.5435],[219,-54.046],[220,-52.1135],[221,-49.7089],[222,-46.7946],[223,-42.6264],[224,-36.956],[225,-30.4901],[226,-23.9358],[227,-18.0003],[228,-13.3902],[229,-11.4484],[231,1.9628],[232,9.1629],[233,7.7625],[234,4.4462],[235,2.8932],[236,1.7992],[237,0.192],[238,-0.2545],[239,-0.5737],[241,-24.989],[242,-35.2732],[243,-40.4154],[244,-48.2952],[245,-50.2652],[246,-50.9219],[247,-52.0898],[248,-55.0096],[249,-56.9562],[250,-57.6862],[251,-57.9295],[422,-57.9295]]
       }
       pinky_02_r_ctrl.Scale.X {type=MovieSceneScriptingFloatChannel, default=1, const [1,423]}
       pinky_02_r_ctrl.Scale.Y {type=MovieSceneScriptingFloatChannel, default=1, const [1,423]}
       pinky_02_r_ctrl.Scale.Z {type=MovieSceneScriptingFloatChannel, default=1, const [1,423]}
       pinky_03_r_ctrl.Location.X {type=MovieSceneScriptingFloatChannel, default=0.0002, thinned_from=423, keys [[0,0.0002],[422,0.0002]]}
       pinky_03_r_ctrl.Location.Y {type=MovieSceneScriptingFloatChannel, default=-0.0002, thinned_from=423, keys [[0,-0.0002],[422,-0.0002]]}
       pinky_03_r_ctrl.Location.Z {type=MovieSceneScriptingFloatChannel, default=0, thinned_from=423, keys [[0,0],[422,0]]}
       pinky_03_r_ctrl.Rotation.X {
        type=MovieSceneScriptingFloatChannel
        default=3.1496
        thinned_from=423
        keys [[0,3.1496],[214,3.1496],[215,3.1427],[216,3.1206],[217,3.0814],[218,3.023],[219,2.9436],[220,2.8411],[221,2.7136],[222,2.559],[223,2.338],[224,2.0372],[225,1.6943],[226,1.3467],[227,1.0319],[228,0.7874],[229,0.6844],[230,0.6689],[231,0.5905],[232,0.5905],[233,0.4983],[234,0.2578],[235,0.132],[236,0.0339],[237,0.0012],[238,0],[239,0.0082],[241,1.4025],[242,1.948],[243,2.2207],[244,2.6386],[245,2.7431],[246,2.7779],[247,2.8399],[248,2.9947],[249,3.098],[250,3.1367],[251,3.1496],[422,3.1496]]
       }
       pinky_03_r_ctrl.Rotation.Y {
        type=MovieSceneScriptingFloatChannel
        default=-4.3172
        thinned_from=423
        keys [[0,-4.3172],[214,-4.3172],[215,-4.3078],[216,-4.2775],[217,-4.2237],[218,-4.1438],[219,-4.0349],[220,-3.8944],[221,-3.7196],[222,-3.5078],[223,-3.2047],[224,-2.7925],[225,-2.3224],[226,-1.846],[227,-1.4145],[228,-1.0793],[229,-0.9382],[230,-0.9169],[231,-0.8095],[232,-0.8095],[233,-0.683],[234,-0.3534],[235,-0.1809],[236,-0.0464],[237,-0.0016],[238,0],[239,-0.0112],[241,-1.9225],[242,-2.6702],[243,-3.044],[244,-3.6168],[245,-3.7601],[246,-3.8078],[247,-3.8927],[248,-4.105],[249,-4.2465],[250,-4.2995],[251,-4.3172],[422,-4.3172]]
       }
       pinky_03_r_ctrl.Rotation.Z {
        type=MovieSceneScriptingFloatChannel
        default=-111.6788
        thinned_from=423
        keys [[0,-111.6788],[214,-111.6788],[215,-111.3875],[216,-110.4583],[217,-108.8078],[218,-106.3529],[219,-103.0103],[220,-98.6969],[221,-93.3294],[222,-86.8247],[223,-77.5208],[224,-64.8636],[225,-50.4311],[226,-35.8014],[227,-22.5524],[228,-12.2623],[229,-7.9282],[231,7.6638],[232,14.8638],[233,15.607],[234,17.8777],[235,19.2476],[236,20.4329],[237,19.4156],[238,19.133],[239,18.1788],[241,-38.1521],[242,-61.1074],[243,-72.5854],[244,-90.1741],[245,-94.5713],[246,-96.037],[247,-98.6439],[248,-105.1614],[249,-109.5063],[250,-111.1357],[251,-111.6788],[422,-111.6788]]
       }
       pinky_03_r_ctrl.Scale.X {type=MovieSceneScriptingFloatChannel, default=1, const [1,423]}
       pinky_03_r_ctrl.Scale.Y {type=MovieSceneScriptingFloatChannel, default=1, const [1,423]}
       pinky_03_r_ctrl.Scale.Z {type=MovieSceneScriptingFloatChannel, default=1, const [1,423]}
       hand_r_ik_ctrl.Location.X {
        type=MovieSceneScriptingFloatChannel
        default=-28.9579
        thinned_from=423
        keys [[0,-26.3305],[12,-26.3504],[18,-26.0887],[20,-26.1757],[22,-26.4563],[26,-27.1625],[29,-27.4944],[66,-27.5862],[71,-27.8067],[74,-27.7632],[86,-26.8482],[93,-26.7465],[125,-27.3024],[127,-27.4887],[128,-27.5328],[129,-27.4794],[130,-27.3312],[131,-26.9115],[132,-25.0602],[133,-21.3595],[134,-14.6024],[135,-11.1356],[136,-10.9134],[137,-15.1978],[138,-19.4288],[139,-22.35],[140,-26.7464],[141,-31.4123],[142,-50.6454],[143,-69.3596],[144,-82.6469],[145,-85.5242],[146,-87.0211],[147,-87.549],[148,-85.6253],[149,-82.3037],[150,-77.0873],[151,-73.8485],[152,-72.9961],[153,-73.3883],[156,-76.1825],[157,-76.7537],[160,-77.282],[217,-77.2702],[224,-76.326],[228,-75.1721],[229,-74.4692],[230,-73.0004],[231,-68.9873],[232,-62.7577],[233,-53.2241],[234,-42.1189],[235,-32.6659],[236,-26.8744],[237,-24.5891],[238,-24.8403],[239,-26.0823],[240,-27.111],[241,-27.5535],[242,-27.5675],[243,-27.3725],[244,-25.8868],[245,-23.9866],[246,-23.6357],[247,-23.3684],[248,-23.4635],[249,-23.9753],[250,-24.5399],[251,-24.9628],[252,-25.2074],[255,-25.5521],[293,-25.7284],[300,-25.9705],[301,-25.8712],[303,-25.3752],[304,-25.2276],[306,-25.0717],[308,-25.0616],[309,-25.0103],[311,-24.7476],[314,-24.5382],[316,-24.5037],[356,-25.1142],[375,-24.9455],[379,-24.7446],[383,-24.8219],[384,-24.9173],[389,-26.0233],[391,-26.3139],[398,-26.7017],[402,-26.7272],[422,-26.3267]]
       }
       hand_r_ik_ctrl.Location.Y {
        type=MovieSceneScriptingFloatChannel
        default=2.9304
        thinned_from=423
        keys [[0,4],[4,3.9896],[7,3.9558],[12,3.8462],[14,3.7875],[15,3.7802],[16,3.7883],[17,3.8662],[18,4.0187],[19,4.2875],[20,4.6439],[21,5.0147],[22,5.3649],[23,5.6946],[24,5.9763],[25,6.2034],[26,6.3903],[27,6.5406],[28,6.6569],[29,6.7407],[31,6.8248],[33,6.8768],[38,6.8836],[53,6.7716],[63,6.6895],[65,6.6604],[66,6.6634],[69,6.7392],[70,6.7527],[71,6.7465],[72,6.7254],[75,6.5865],[81,6.1622],[82,6.111],[84,6.052],[86,6.0148],[92,5.9982],[98,6.0844],[104,6.216],[109,6.2682],[110,6.266],[116,6.2951],[125,6.2838],[127,6.2191],[130,6.08],[131,6.1186],[132,5.9756],[133,5.9493],[134,4.0862],[135,1.868],[136,-1.5031],[137,-4.8339],[138,-7.0441],[139,-8.4554],[140,-10.8188],[141,-14.4254],[142,-29.663],[143,-36.4435],[144,-34.9169],[145,-31.0661],[146,-27.4218],[147,-24.9071],[148,-24.028],[149,-23.675],[150,-22.2983],[151,-21.0096],[152,-20.1904],[153,-19.7697],[155,-20.7619],[156,-21.3667],[157,-21.8666],[158,-22.1083],[216,-22.0085],[220,-22.1543],[223,-22.405],[224,-22.5628],[225,-22.9883],[226,-23.5824],[227,-24.3688],[228,-25.7063],[229,-27.6682],[230,-29.8166],[231,-32.3646],[232,-33.865],[233,-33.6768],[234,-30.0848],[235,-23.3342],[236,-15.3964],[237,-8.2311],[238,-4.6753],[239,-3.0194],[240,-3.224],[241,-3.5316],[242,-3.616],[243,-3.1214],[244,-1.7569],[245,0.0352],[246,0.8514],[247,1.3424],[248,1.9767],[249,2.1128],[250,2.2431],[251,2.4097],[252,2.5314],[253,2.5809],[258,2.6922],[262,2.7519],[265,2.7757],[271,2.7804],[274,2.7806],[277,2.7864],[285,2.7746],[290,2.7841],[292,2.8043],[294,2.8392],[295,2.8772],[296,2.9014],[297,2.9006],[298,2.8628],[299,2.7968],[300,2.7165],[301,2.6436],[303,2.4824],[304,2.4416],[305,2.4631],[306,2.542],[307,2.6686],[308,2.8171],[309,2.9436],[311,3.1038],[312,3.1967],[313,3.2827],[315,3.426],[316,3.5058],[318,3.6886],[319,3.7727],[320,3.8471],[321,3.9112],[323,4.008],[325,4.0671],[332,4.1565],[333,4.1553],[334,4.164],[336,4.1457],[340,4.0692],[344,3.9637],[345,3.9439],[348,3.9273],[349,3.9289],[350,3.9394],[351,3.9636],[352,4.0209],[353,4.1149],[354,4.2386],[357,4.659],[359,4.9215],[361,5.2033],[362,5.3266],[363,5.4209],[365,5.5513],[367,5.6402],[369,5.6896],[371,5.699],[372,5.6888],[373,5.6666],[374,5.6311],[375,5.5805],[376,5.5077],[377,5.4074],[378,5.2809],[379,5.13],[380,4.9478],[381,4.7462],[382,4.5553],[383,4.4027],[384,4.323],[385,4.3178],[387,4.3968],[388,4.4091],[389,4.4026],[390,4.374],[393,4.263],[395,4.214],[397,4.2004],[400,4.2105],[401,4.2076],[402,4.1919],[406,4.0829],[408,4.0501],[417,4.0025],[418,4.0087],[422,4.0059]]
       }
       hand_r_ik_ctrl.Location.Z {
        type=MovieSceneScriptingFloatChannel
        default=-7.2233
        thinned_from=423
        keys [[0,-8.4942],[2,-8.5146],[9,-8.6434],[12,-8.6491],[14,-8.6003],[15,-8.5521],[16,-8.4754],[17,-8.3702],[18,-8.2829],[20,-8.1813],[21,-8.1556],[23,-8.1611],[24,-8.1377],[29,-7.959],[33,-7.8876],[43,-7.7765],[46,-7.7654],[47,-7.7782],[48,-7.7635],[49,-7.7768],[54,-7.7852],[58,-7.8323],[65,-7.9805],[66,-8.0236],[69,-8.2648],[70,-8.3339],[71,-8.3795],[76,-8.4742],[77,-8.5325],[78,-8.6189],[79,-8.685],[81,-8.7665],[82,-8.7821],[89,-8.6443],[97,-8.6149],[122,-8.762],[125,-8.8038],[126,-8.7814],[127,-8.7209],[128,-8.6277],[129,-8.5159],[130,-8.4472],[131,-8.6197],[132,-8.992],[133,-9.8652],[134,-13.2467],[135,-16.0829],[136,-21.3951],[137,-29.5621],[138,-32.6366],[139,-34.3542],[140,-36.7346],[141,-35.2839],[142,-28.9603],[143,-18.3691],[144,-8.0495],[145,-3.4172],[146,-0.5604],[147,-0.3287],[148,-1.592],[149,-3.5548],[150,-2.5906],[151,-1.4909],[152,-2.1397],[153,-1.8035],[154,-1.5905],[155,-1.2659],[156,-1.0974],[157,-0.9892],[158,-0.8007],[159,-0.5069],[160,-0.2586],[161,-0.0646],[162,0.0817],[163,0.2042],[164,0.3059],[165,0.3867],[166,0.4436],[167,0.4877],[168,0.5297],[170,0.5884],[171,0.6054],[172,0.6273],[173,0.632],[174,0.6543],[175,0.655],[176,0.6592],[177,0.6568],[178,0.6748],[179,0.6704],[180,0.6634],[181,0.676],[182,0.6697],[184,0.6614],[185,0.6722],[186,0.671],[187,0.6661],[188,0.6668],[189,0.666],[190,0.6597],[194,0.6589],[195,0.656],[196,0.6633],[197,0.6493],[199,0.6489],[201,0.6436],[202,0.6215],[204,0.6219],[205,0.6041],[206,0.6059],[207,0.5932],[208,0.5979],[209,0.5851],[210,0.5924],[211,0.5795],[213,0.5766],[214,0.5765],[215,0.5644],[216,0.5334],[217,0.4813],[218,0.4129],[219,0.3278],[220,0.2255],[221,0.1028],[222,-0.0452],[223,-0.1969],[224,-0.3846],[225,-0.7424],[226,-1.2037],[227,-1.7892],[228,-2.8272],[229,-4.7203],[230,-7.5352],[231,-11.851],[232,-17.1709],[233,-23.3341],[234,-27.1154],[235,-27.7578],[236,-25.2254],[237,-21.8208],[238,-18.7195],[239,-14.4621],[240,-10.8748],[241,-8.8086],[242,-7.4447],[243,-6.8802],[244,-7.2523],[245,-7.6974],[246,-8.1751],[247,-8.4697],[248,-8.482],[249,-8.3028],[250,-8.2598],[251,-8.3108],[252,-8.3903],[253,-8.4157],[256,-8.3498],[260,-8.3736],[272,-8.4995],[288,-8.3873],[293,-8.427],[295,-8.5157],[300,-8.7631],[302,-9.0047],[303,-9.0145],[304,-8.9071],[308,-8.1847],[309,-8.0433],[310,-7.9265],[312,-7.6597],[313,-7.5548],[315,-7.3888],[316,-7.3303],[317,-7.3113],[318,-7.3094],[319,-7.3224],[322,-7.424],[325,-7.459],[338,-7.2444],[342,-7.2279],[348,-7.3184],[349,-7.3206],[350,-7.3052],[352,-7.2305],[355,-7.0481],[361,-6.5101],[362,-6.4369],[364,-6.3338],[366,-6.2644],[368,-6.2261],[371,-6.2065],[372,-6.1744],[373,-6.1192],[374,-6.0393],[376,-5.8202],[377,-5.7198],[378,-5.6337],[379,-5.57],[380,-5.5336],[381,-5.5423],[382,-5.6349],[383,-5.8541],[384,-6.2645],[385,-6.8163],[386,-7.384],[387,-7.8481],[388,-8.1667],[389,-8.3859],[390,-8.5252],[392,-8.7172],[394,-8.856],[397,-8.9644],[400,-9.0213],[401,-9.0206],[403,-8.9573],[407,-8.7564],[417,-8.5072],[422,-8.4941]]
       }
       hand_r_ik_ctrl.Rotation.X {
        type=MovieSceneScriptingFloatChannel
        default=30.1298
        thinned_from=423
        keys [[0,30.1298],[7,30.0233],[15,30.2843],[16,30.2519],[17,30.1358],[18,29.8455],[19,29.273],[23,26.1541],[24,25.5397],[25,25.1131],[26,24.7706],[27,24.5114],[28,24.3384],[30,24.1818],[34,24.1996],[44,24.8146],[46,24.883],[47,24.8783],[48,24.9341],[49,24.9204],[54,24.9965],[58,24.8534],[65,24.2856],[68,24.1995],[69,24.2473],[70,24.3992],[71,24.7064],[72,25.178],[73,25.7982],[74,26.5697],[76,28.3585],[79,31.1698],[81,32.9042],[82,33.6002],[83,34.1345],[84,34.5798],[85,34.9349],[86,35.197],[88,35.4859],[92,35.5433],[98,35.0311],[103,34.3429],[108,34.0268],[126,33.7019],[127,33.6361],[128,33.3525],[129,32.8361],[131,31.5773],[132,19.9891],[133,18.1217],[134,10.0189],[135,-28.2865],[136,-71.7917],[137,-114.4044],[138,-110.2274],[139,-108.2634],[140,-109.2314],[141,-112.6574],[142,-95.0017],[143,-98.6528],[144,-96.8143],[145,-98.1967],[146,-99.2161],[147,-103.7909],[148,-106.867],[149,-115.2583],[150,-128.0693],[151,-141.9349],[152,206.7164],[153,198.7054],[154,196.7842],[155,197.0857],[156,199.413],[157,201.3377],[158,201.9479],[218,200.7986],[222,202.3643],[224,204.0784],[225,206.6474],[226,-149.7594],[227,-145.0012],[228,-137.2257],[229,-126.1772],[230,-114.2633],[231,-101.0257],[232,-92.5098],[233,-91.0341],[234,-92.6905],[235,-45.7461],[236,-33.0361],[237,-18.37],[238,-1.3133],[239,20.0501],[240,36.0339],[241,28.7063],[242,23.7453],[243,22.2627],[244,19.1541],[245,21.5965],[246,22.337],[247,23.314],[249,26.117],[250,27.1575],[251,27.8443],[252,28.18],[253,28.2381],[263,27.331],[272,27.1086],[287,27.6288],[291,27.5447],[293,27.4089],[294,27.245],[296,26.8103],[297,26.7353],[298,26.8257],[299,27.0548],[300,27.4323],[301,28.2801],[302,29.9396],[303,31.9341],[305,35.5275],[306,37.0282],[307,38.2042],[308,39.1617],[310,40.7323],[311,41.4172],[312,41.9649],[313,42.398],[315,42.9861],[316,43.1141],[318,42.9648],[323,42.1324],[325,41.9419],[345,41.5433],[347,41.6256],[349,41.9041],[351,42.4739],[352,42.8843],[357,45.664],[359,46.5124],[361,47.1743],[363,47.6219],[368,48.0471],[371,48.011],[373,47.7242],[375,47.1945],[376,46.8285],[377,46.3677],[378,45.8102],[379,45.1418],[380,44.2064],[381,43.0154],[383,40.4885],[384,38.9635],[385,36.9884],[386,34.8792],[387,33.0038],[388,31.5173],[389,30.3275],[390,29.4604],[391,28.8408],[392,28.3736],[393,28.0355],[394,27.8],[395,27.6403],[397,27.4872],[400,27.4861],[401,27.5596],[402,27.7359],[406,28.8302],[409,29.4078],[414,29.9563],[417,30.1066],[422,30.131]]
       }
       hand_r_ik_ctrl.Rotation.Y {
        type=MovieSceneScriptingFloatChannel
        default=42.2066
        thinned_from=423
        keys [[0,42.2066],[16,42.2326],[18,42.3794],[21,42.9962],[23,43.2184],[45,42.6205],[70,42.7425],[72,42.6286],[73,42.6617],[74,42.8735],[76,43.4841],[79,44.7382],[81,45.4115],[84,46.0298],[89,46.3776],[126,46.7437],[127,46.6985],[128,46.4661],[131,45.2038],[132,41.1979],[133,39.1024],[134,26.2473],[135,10.6243],[136,-9.1356],[137,-30.3519],[138,-36.1222],[139,-40.3124],[140,-50.3471],[141,-51.8805],[142,-39.7909],[143,-16.4842],[144,-1.2414],[145,10.1364],[146,18.0199],[147,19.6597],[148,19.7833],[149,27.2507],[150,42.455],[151,55.7905],[152,58.5558],[153,59.7365],[154,59.9016],[155,59.8549],[157,59.2789],[158,59.183],[226,59.6067],[227,59.3873],[228,58.3245],[229,55.4863],[230,50.57],[231,43.4114],[232,34.1431],[233,13.1316],[234,-1.3033],[235,0.7592],[236,7.5884],[237,14.162],[238,22.1797],[239,35.5592],[240,44.9142],[241,43.8629],[242,42.0765],[243,42.2894],[244,43.0462],[245,42.4527],[246,42.661],[247,42.7632],[248,42.7415],[249,42.9906],[251,44.0303],[252,44.371],[253,44.4731],[294,44.398],[297,44.5177],[300,44.4648],[301,45.4606],[302,47.3385],[303,48.7443],[304,49.2905],[305,49.459],[306,49.4334],[307,49.2737],[309,48.766],[315,48.0718],[319,48.1986],[325,48.6555],[349,47.8828],[352,47.4658],[358,45.9402],[362,45.2739],[371,44.7992],[373,44.498],[376,43.693],[378,42.954],[379,42.4767],[380,41.785],[381,40.9527],[382,40.3162],[383,40.2377],[384,40.7977],[386,42.4518],[387,43.0169],[388,43.2353],[389,43.2338],[391,42.9951],[422,42.205]]
       }
       hand_r_ik_ctrl.Rotation.Z {
        type=MovieSceneScriptingFloatChannel
        default=42.0043
        thinned_from=423
        keys [[0,42.0043],[9,41.9439],[15,42.2688],[17,42.193],[18,41.9584],[19,41.4688],[21,39.9939],[23,38.1849],[24,37.3668],[25,36.7221],[26,36.1796],[27,35.7427],[28,35.4032],[29,35.158],[30,34.9853],[32,34.8437],[36,34.8716],[44,35.3985],[48,35.5595],[55,35.6857],[66,35.0151],[68,35.0195],[69,35.0857],[70,35.2593],[71,35.5928],[72,36.0939],[73,36.7935],[74,37.73],[76,39.9883],[77,41.2482],[79,43.8845],[81,46.3197],[82,47.3043],[83,48.0688],[84,48.7122],[85,49.229],[86,49.605],[88,50.0429],[91,50.1778],[95,49.8544],[103,48.3879],[108,47.9093],[126,47.5053],[128,47.2573],[131,46.6945],[132,37.0454],[133,31.8124],[134,16.878],[135,-0.4495],[136,-13.5081],[137,-19.9473],[138,-30.4957],[139,-34.9533],[140,-33.9875],[141,-30.1719],[142,-85.8284],[143,-111.7203],[144,-125.7446],[145,-133.6436],[146,219.531],[147,213.1861],[148,210.0251],[149,201.7702],[150,189.5497],[151,174.6291],[152,162.1104],[153,153.0332],[154,150.9401],[155,151.3107],[156,153.9877],[157,156.2142],[158,156.9081],[163,155.4049],[217,155.3115],[220,156.0935],[223,158.0887],[224,159.3037],[225,162.2926],[226,166.46],[227,171.9608],[228,180.9613],[229,193.8291],[230,207.8526],[231,-133.5272],[232,-116.46],[233,-94.2908],[234,-68.0382],[235,-45.2246],[236,-20.7089],[237,3.0077],[238,17.7422],[239,29.5711],[240,37.9607],[241,34.6997],[242,32.794],[243,32.9248],[244,31.7858],[245,38.6015],[246,38.7153],[247,39.0018],[249,40.1617],[250,40.3572],[254,40.3884],[264,39.4555],[273,39.4004],[287,39.973],[292,39.8019],[293,39.7175],[296,39.0616],[297,38.9597],[298,39.0045],[299,39.1776],[300,39.5277],[301,40.8333],[302,43.4486],[303,46.2612],[304,48.5323],[305,50.5874],[306,52.3054],[307,53.6074],[309,55.5636],[310,56.6252],[311,57.5574],[313,58.8464],[314,59.3028],[315,59.6344],[316,59.7784],[324,58.567],[346,57.7785],[348,57.9186],[350,58.3399],[352,59.1152],[355,60.929],[357,62.2884],[360,63.7725],[362,64.4926],[365,65.0268],[370,65.1942],[371,65.1135],[373,64.5724],[375,63.6096],[377,62.3167],[378,61.4912],[379,60.4891],[380,59.0351],[382,55.3745],[384,52.4334],[385,50.5148],[386,48.4221],[387,46.4814],[388,44.8121],[389,43.3606],[390,42.2242],[391,41.3938],[392,40.7842],[393,40.3323],[394,40.0012],[396,39.5741],[399,39.3423],[401,39.3697],[403,39.7762],[407,40.8593],[414,41.8106],[422,42.0171]]
       }
       hand_r_ik_ctrl.Scale.X {type=MovieSceneScriptingFloatChannel, default=1, const [1,423]}
       hand_r_ik_ctrl.Scale.Y {type=MovieSceneScriptingFloatChannel, default=1, const [1,423]}
       hand_r_ik_ctrl.Scale.Z {type=MovieSceneScriptingFloatChannel, default=1, const [1,423]}
       Softness_2 {type=MovieSceneScriptingFloatChannel, default=0, const [0,423]}
       hand_r_ikorient_ctrl.Rotation.Roll {type=MovieSceneScriptingFloatChannel, default=0, const [0,423]}
       hand_r_ikorient_ctrl.Rotation.Pitch {type=MovieSceneScriptingFloatChannel, default=0, const [0,423]}
       hand_r_ikorient_ctrl.Rotation.Yaw {type=MovieSceneScriptingFloatChannel, default=0, const [0,423]}
       arm_r_pv_ik_ctrl.Location.X {
        type=MovieSceneScriptingFloatChannel
        default=18.2052
        thinned_from=423
        keys [[0,18.2052],[2,18.1646],[6,18.0231],[10,18.0987],[12,18.2004],[14,18.4172],[15,18.4633],[16,18.4042],[17,18.0857],[18,17.3471],[19,16.2725],[20,14.9643],[22,12.0665],[23,10.7263],[24,9.5414],[25,8.576],[26,7.7551],[27,7.0424],[28,6.4481],[29,5.9924],[30,5.6876],[31,5.5081],[32,5.4017],[33,5.3504],[34,5.3422],[35,5.3781],[36,5.4467],[38,5.6378],[39,5.7234],[40,5.785],[44,5.9793],[45,6.0155],[46,6.0826],[47,6.1224],[48,6.1888],[49,6.2167],[50,6.2813],[51,6.3058],[53,6.3763],[55,6.4119],[57,6.3996],[58,6.3629],[59,6.3025],[60,6.2208],[61,6.122],[62,6.007],[63,5.8305],[67,4.8862],[68,4.6909],[69,4.4814],[70,4.2442],[71,4.0818],[72,4.0944],[73,4.2795],[74,4.626],[75,5.1595],[76,5.9759],[77,7.1749],[78,8.621],[79,10.2623],[80,12.0462],[81,13.7938],[82,15.355],[83,16.6807],[84,17.859],[85,18.8829],[86,19.7424],[87,20.4502],[88,21.0474],[89,21.5539],[90,21.9684],[91,22.2939],[92,22.5379],[93,22.7116],[94,22.8192],[95,22.8578],[97,22.7926],[98,22.7019],[103,21.8128],[107,21.3906],[114,21.0534],[119,20.9732],[120,20.9924],[124,21.3332],[125,21.3733],[126,21.2629],[127,20.8695],[128,20.2948],[129,19.6355],[130,18.8444],[131,17.929],[132,15.5137],[133,11.9344],[134,6.2884],[135,-1.6545],[136,-14.8689],[137,-23.1355],[138,-23.6747],[139,-23.8171],[140,-20.3342],[141,-16.346],[142,-8.8468],[143,3.3664],[144,7.8728],[145,8.0008],[146,8.4053],[147,11.5079],[148,12.5049],[149,14.1532],[150,15.1222],[151,14.7501],[152,14.2091],[153,14.501],[154,14.8991],[155,15.1395],[158,15.4651],[160,15.5918],[163,15.6379],[173,15.5093],[196,15.5521],[215,15.5189],[220,15.6233],[222,15.5781],[223,15.5309],[224,15.4406],[225,15.1739],[226,14.7785],[227,14.2097],[228,13.3458],[229,12.0171],[230,10.1547],[231,6.9614],[232,3.2412],[233,-1.0956],[234,-6.2344],[235,-10.152],[236,-6.7101],[237,0.9401],[238,6.2171],[239,9.0707],[240,10.2926],[241,11.2785],[242,12.3719],[243,12.6357],[244,12.1163],[245,11.3153],[246,10.554],[247,10.4539],[248,10.8597],[249,11.6812],[250,12.2923],[252,13.3303],[253,13.6092],[254,13.7455],[255,13.8077],[256,13.7771],[257,13.6469],[259,13.3004],[264,12.2516],[266,11.8875],[267,11.7268],[269,11.4821],[271,11.3287],[273,11.2351],[277,11.1877],[287,11.3734],[289,11.3479],[291,11.2672],[293,11.1163],[294,10.9684],[297,10.3539],[298,10.1121],[299,9.899],[300,9.9674],[301,10.1916],[302,10.5377],[303,11.1471],[304,11.903],[305,12.7329],[306,13.6451],[307,14.6298],[308,15.6569],[309,16.7891],[311,19.3402],[312,20.5132],[313,21.6249],[314,22.6482],[315,23.6151],[316,24.4781],[317,25.2217],[318,25.8539],[319,26.3366],[320,26.6901],[321,26.9116],[323,27.1874],[326,27.4053],[327,27.3937],[330,27.0847],[334,26.5347],[338,26.3216],[343,26.3323],[348,26.5976],[354,27.2711],[355,27.2905],[358,27.0781],[370,26.9573],[380,26.5126],[382,26.2743],[384,25.9274],[385,25.65],[386,25.2032],[387,24.5209],[388,23.631],[389,22.6798],[390,21.7823],[391,21.0257],[392,20.3823],[393,19.7839],[394,19.2362],[395,18.7441],[396,18.3169],[397,17.9462],[398,17.6138],[399,17.3189],[400,17.0603],[401,16.8893],[402,16.844],[403,16.8895],[408,17.4733],[415,18.0656],[418,18.1983],[422,18.202]]
       }
       arm_r_pv_ik_ctrl.Location.Y {
        type=MovieSceneScriptingFloatChannel
        default=-17.9119
        thinned_from=423
        keys [[0,-17.9119],[1,-17.9268],[6,-18.2849],[10,-18.3595],[12,-18.3059],[15,-17.9634],[16,-17.9029],[17,-18.0446],[18,-18.5677],[19,-19.2803],[20,-20.0799],[22,-21.8668],[23,-22.6007],[24,-23.1772],[25,-23.5861],[27,-24.2219],[28,-24.457],[29,-24.6238],[30,-24.7223],[33,-24.758],[48,-24.2042],[57,-24.1774],[60,-24.3332],[62,-24.5171],[69,-25.8482],[71,-26.3436],[72,-26.4976],[73,-26.5646],[74,-26.5625],[75,-26.4794],[76,-26.2285],[77,-25.7477],[78,-25.0917],[79,-24.1919],[80,-23.0659],[83,-19.4176],[84,-18.2458],[85,-17.1367],[86,-16.1367],[87,-15.2592],[88,-14.4725],[89,-13.7635],[90,-13.153],[91,-12.6424],[92,-12.2333],[93,-11.9139],[94,-11.6914],[95,-11.5628],[96,-11.5065],[97,-11.5017],[98,-11.5792],[99,-11.7467],[102,-12.4399],[103,-12.6321],[104,-12.7984],[106,-13.0256],[108,-13.1989],[112,-13.4172],[118,-13.5945],[119,-13.6008],[120,-13.5723],[121,-13.4547],[123,-13.1523],[124,-13.0316],[125,-12.9605],[126,-13.0724],[127,-13.5075],[128,-14.1102],[129,-14.7416],[130,-15.4826],[131,-16.3225],[132,-18.6294],[133,-21.1427],[134,-23.2311],[135,-24.5076],[136,-22.9849],[137,-17.6553],[138,-19.4353],[139,-20.1473],[140,-20.3895],[141,-21.0621],[142,-19.3426],[143,-20.401],[144,-23.6861],[146,-24.2058],[147,-22.4514],[148,-22.2755],[149,-21.185],[150,-19.6737],[151,-19.4787],[152,-19.4001],[154,-20.2317],[155,-20.4463],[156,-20.5461],[157,-20.5605],[166,-19.9646],[172,-19.7833],[196,-19.6989],[219,-19.8932],[223,-20.1176],[224,-20.2073],[225,-20.421],[226,-20.6979],[227,-21.0715],[228,-21.6228],[229,-22.4803],[230,-23.6595],[231,-25.0123],[232,-26.3071],[233,-26.368],[234,-27.9701],[235,-33.5449],[236,-34.7467],[237,-32.7126],[238,-29.9324],[239,-27.9353],[240,-26.9254],[241,-26.0361],[242,-24.983],[243,-24.6281],[244,-24.884],[245,-25.012],[246,-25.3677],[247,-25.2244],[248,-24.7051],[249,-24.0408],[252,-22.5512],[253,-22.2487],[254,-22.0614],[255,-21.935],[256,-21.8945],[257,-21.9539],[259,-22.2008],[266,-23.4373],[269,-23.8285],[272,-24.0413],[290,-24.001],[293,-24.1379],[295,-24.4081],[297,-24.7925],[299,-25.4061],[300,-25.5291],[301,-25.5284],[302,-25.4366],[303,-25.1439],[304,-24.7216],[305,-24.1945],[307,-22.9484],[308,-22.2527],[309,-21.4703],[310,-20.5323],[311,-19.4442],[314,-15.9003],[316,-13.4457],[317,-12.2964],[318,-11.2326],[319,-10.3535],[320,-9.6616],[321,-9.1969],[322,-8.8625],[323,-8.5773],[324,-8.3555],[325,-8.1763],[326,-8.014],[327,-7.9773],[328,-8.0545],[329,-8.1736],[330,-8.3339],[334,-9.0638],[335,-9.1962],[336,-9.2695],[340,-9.4383],[343,-9.4382],[346,-9.3242],[347,-9.251],[348,-9.1431],[349,-8.9849],[351,-8.5616],[352,-8.3319],[353,-8.0477],[354,-7.7333],[355,-7.547],[356,-7.5262],[357,-7.549],[358,-7.4852],[359,-7.3202],[362,-6.7677],[363,-6.6064],[366,-6.1734],[367,-6.0466],[369,-5.8614],[370,-5.8471],[371,-5.9088],[372,-5.9344],[373,-5.9335],[375,-5.8966],[376,-5.9056],[378,-5.9919],[379,-6.0127],[380,-6.1165],[381,-6.3525],[383,-6.972],[384,-7.2267],[385,-7.5223],[386,-8.0486],[387,-9.0143],[388,-10.3513],[389,-11.7209],[390,-12.9492],[391,-13.9407],[392,-14.7761],[393,-15.558],[394,-16.2681],[395,-16.893],[396,-17.4266],[397,-17.8918],[398,-18.3106],[399,-18.6829],[400,-19.0078],[401,-19.2327],[402,-19.3213],[403,-19.3099],[405,-19.1273],[415,-18.1054],[418,-17.9233],[422,-17.9145]]
       }
       arm_r_pv_ik_ctrl.Location.Z {
        type=MovieSceneScriptingFloatChannel
        default=34.6784
        thinned_from=423
        keys [[0,31.57],[16,31.416],[17,31.3152],[21,30.451],[22,30.2834],[23,30.2363],[38,30.9882],[61,30.8721],[63,30.7306],[66,30.3836],[69,30.2216],[71,29.9101],[75,29.1364],[76,29.0119],[78,29.0118],[97,30.6309],[108,30.9917],[125,31.287],[128,31.7661],[129,31.8597],[130,31.8862],[131,31.6697],[132,29.957],[133,25.7625],[134,18.4415],[135,11.5548],[136,4.0769],[137,-1.2113],[138,-0.5919],[139,1.0962],[140,3.6079],[141,8.8127],[142,12.1305],[143,20.2092],[144,32.2749],[145,33.5392],[146,34.4706],[147,34.2921],[148,31.616],[149,27.3115],[150,22.363],[151,19.8779],[152,18.2433],[153,19.0858],[155,21.7209],[156,22.7518],[157,23.447],[158,23.894],[159,24.2492],[160,24.5242],[161,24.7278],[163,24.9805],[165,25.1127],[215,25.0864],[217,24.9501],[220,24.5473],[223,23.9281],[224,23.6617],[225,23.2461],[227,22.2248],[228,21.5858],[229,20.6279],[230,19.2505],[231,16.5452],[232,13.1978],[233,9.5815],[234,9.2462],[235,18.8059],[236,26.699],[237,32.4533],[238,35.0351],[239,35.95],[240,35.8901],[241,35.691],[242,35.4016],[243,34.4803],[244,31.3724],[245,27.5885],[246,26.8607],[247,26.7535],[248,26.9283],[250,28.803],[251,29.5953],[252,30.1836],[253,30.5279],[255,30.9736],[256,31.0831],[259,31.0716],[271,30.0572],[296,29.7467],[297,29.6733],[299,29.278],[301,29.0587],[303,28.5635],[306,28.0688],[308,27.6264],[310,26.9839],[311,26.7702],[313,26.6907],[314,26.7223],[315,26.8224],[319,27.5338],[322,27.8299],[349,28.5084],[351,28.4471],[354,28.1482],[355,28.0066],[357,27.5679],[360,27.1476],[363,26.9024],[368,26.8958],[373,27.0754],[376,27.3406],[378,27.6787],[380,28.2313],[382,28.9409],[383,29.3587],[384,29.9189],[386,31.2655],[387,31.7536],[388,32.0668],[390,32.4836],[392,32.6779],[395,32.5882],[404,31.738],[422,31.5578]]
       }
       arm_r_pv_ik_ctrl.Rotation.X {
        type=MovieSceneScriptingFloatChannel
        default=51.1805
        thinned_from=423
        keys [[0,51.1805],[12,50.5846],[16,50.6355],[17,50.4349],[20,48.8548],[22,47.6831],[24,47.0146],[28,46.4321],[33,46.3428],[40,46.9108],[59,46.9662],[62,46.6569],[65,45.7949],[67,45.4439],[70,45.1582],[72,44.8154],[74,44.7226],[75,44.7797],[76,45.0304],[77,45.5785],[78,46.2907],[81,48.8912],[83,50.374],[85,51.5213],[87,52.3643],[90,53.2526],[94,53.888],[99,54.0642],[120,53.8342],[125,54.2785],[127,54.6339],[129,54.73],[130,54.6702],[131,54.4122],[132,52.0622],[133,46.6561],[134,36.413],[135,25.3922],[136,10.1622],[137,-3.8324],[138,-3.9416],[139,-2.8748],[140,-7.4802],[141,-13.7725],[142,-40.3544],[143,-75.2096],[144,-97.7322],[145,-101.4557],[146,-104.9357],[147,-109.306],[148,-108.1405],[149,-105.4542],[150,-100.9252],[151,-98.4693],[152,-96.2202],[153,-97.1148],[154,-98.5302],[155,-99.724],[156,-100.5499],[157,-101.1554],[161,-102.0943],[221,-102.1207],[224,-101.639],[225,-101.1078],[226,-100.3709],[227,-99.3981],[228,-98.0176],[229,-96.016],[230,-93.3064],[231,-88.0338],[232,-80.628],[233,-71.625],[234,-58.3737],[235,-12.3392],[236,24.1626],[237,43.6215],[238,50.7248],[239,52.7696],[241,50.8718],[242,50.0639],[243,49.1315],[244,46.2612],[245,43.1112],[246,43.5553],[248,45.025],[249,46.1049],[250,47.0209],[251,47.766],[252,48.2561],[254,48.6597],[256,48.8423],[260,48.5262],[269,47.0077],[295,46.3728],[297,46.1338],[299,45.3457],[300,45.1302],[301,45.0194],[303,44.5912],[305,44.5635],[310,45.3468],[311,45.6902],[314,47.25],[318,49.8708],[319,50.4064],[320,50.8355],[322,51.3563],[326,51.7113],[337,50.9446],[349,50.7935],[354,50.0259],[362,49.4764],[372,49.7777],[377,50.5291],[380,51.3351],[383,52.497],[384,53.0928],[386,54.5408],[387,54.9295],[388,55.002],[391,54.5642],[401,51.8537],[405,51.2317],[422,51.169]]
       }
       arm_r_pv_ik_ctrl.Rotation.Y {
        type=MovieSceneScriptingFloatChannel
        default=5.4869
        thinned_from=423
        keys [[0,5.4869],[1,5.4938],[3,5.5441],[4,5.5581],[6,5.6017],[10,5.5946],[11,5.5808],[12,5.5396],[14,5.393],[15,5.3451],[16,5.3754],[17,5.5911],[18,6.1222],[19,6.8102],[20,7.5998],[21,8.4961],[22,9.3491],[23,10.1095],[24,10.7952],[25,11.3241],[26,11.7075],[27,12.0168],[28,12.2616],[29,12.433],[30,12.5265],[31,12.5407],[32,12.5039],[34,12.3444],[37,12.0392],[39,11.9015],[51,11.5329],[59,11.4185],[61,11.4529],[62,11.4932],[63,11.5663],[65,11.7801],[66,11.8553],[67,11.8844],[68,11.8767],[69,11.8142],[70,11.6754],[71,11.4834],[72,11.2416],[73,10.9199],[74,10.4637],[75,9.8882],[76,9.1737],[77,8.2536],[78,7.2101],[79,6.0903],[80,4.9304],[81,3.8114],[82,2.7732],[83,1.8536],[84,1.0172],[85,0.2649],[86,-0.3879],[87,-0.9443],[88,-1.4228],[89,-1.8427],[90,-2.1852],[91,-2.4694],[92,-2.6941],[93,-2.8679],[94,-2.9839],[95,-3.0351],[96,-3.0501],[97,-3.0134],[98,-2.9487],[99,-2.824],[100,-2.6728],[101,-2.514],[102,-2.3607],[103,-2.2331],[104,-2.1281],[105,-2.043],[106,-1.9714],[107,-1.9049],[108,-1.8476],[110,-1.7596],[111,-1.7362],[112,-1.7019],[113,-1.6831],[116,-1.6389],[118,-1.6192],[119,-1.6193],[120,-1.6468],[121,-1.7479],[122,-1.8787],[123,-1.9945],[124,-2.0852],[125,-2.1309],[126,-2.0622],[127,-1.8703],[128,-1.5239],[129,-1.08],[130,-0.5495],[131,-0.188],[132,1.4188],[133,2.6582],[134,4.5429],[135,7.0235],[136,10.3881],[137,11.6192],[138,11.683],[139,12.3839],[140,15.2289],[141,20.6558],[142,24.47],[143,20.1267],[144,19.2537],[145,15.535],[146,11.5884],[147,3.3291],[148,4.471],[149,6.0994],[150,6.6314],[151,5.2297],[152,4.2461],[153,4.4838],[154,4.993],[155,5.4272],[156,5.735],[157,5.9615],[158,6.0911],[159,6.185],[160,6.2542],[161,6.3079],[163,6.3755],[165,6.408],[173,6.4041],[177,6.4016],[178,6.4125],[180,6.4056],[186,6.4118],[195,6.4009],[201,6.422],[202,6.4103],[204,6.4201],[205,6.4099],[206,6.4184],[207,6.4111],[208,6.4202],[209,6.4112],[210,6.4224],[216,6.4164],[218,6.4467],[219,6.4798],[220,6.5294],[221,6.5987],[223,6.797],[224,6.9614],[225,7.371],[226,7.9556],[227,8.7571],[228,9.997],[229,11.9537],[230,14.7879],[231,18.0037],[232,21.8675],[233,23.9091],[234,32.4031],[235,49.5308],[236,43.185],[237,29.8859],[238,21.7399],[239,19.03],[240,20.8144],[241,22.2422],[242,23.3748],[243,23.439],[244,20.8545],[245,16.8389],[246,14.8767],[247,13.2954],[248,11.2135],[249,10.4646],[251,9.5168],[252,9.1379],[253,9.0004],[254,9.0242],[256,9.1543],[258,9.3624],[260,9.6139],[266,10.4814],[267,10.6064],[269,10.7921],[272,10.9556],[277,11.0929],[281,11.1405],[287,11.113],[291,11.1578],[295,11.2775],[296,11.3456],[297,11.4696],[298,11.7348],[299,12.054],[300,12.2059],[301,12.0844],[302,11.785],[303,11.4064],[304,10.976],[305,10.4569],[306,9.8132],[307,9.03],[308,8.1588],[309,7.1776],[310,6.1051],[311,5.0708],[312,4.1086],[313,3.1726],[314,2.2831],[315,1.402],[316,0.531],[317,-0.3239],[318,-1.1094],[319,-1.7529],[320,-2.2621],[321,-2.5992],[322,-2.8144],[323,-2.973],[324,-3.06],[325,-3.1214],[326,-3.1366],[327,-3.0823],[328,-2.9254],[329,-2.7342],[330,-2.5126],[331,-2.278],[332,-2.0515],[333,-1.8319],[334,-1.6331],[335,-1.4789],[336,-1.384],[337,-1.3119],[338,-1.2518],[339,-1.2102],[340,-1.1643],[341,-1.1464],[342,-1.1519],[343,-1.1732],[344,-1.2186],[345,-1.2583],[346,-1.3084],[347,-1.3784],[348,-1.4611],[349,-1.5545],[350,-1.6532],[351,-1.7369],[352,-1.7959],[353,-1.8303],[354,-1.8448],[355,-1.8419],[357,-1.8145],[358,-1.8171],[359,-1.8376],[363,-1.9596],[364,-1.9763],[365,-1.9812],[366,-1.9729],[367,-1.944],[368,-1.8944],[369,-1.8405],[370,-1.7369],[371,-1.5644],[372,-1.3719],[373,-1.1556],[374,-0.9235],[375,-0.684],[376,-0.4398],[377,-0.1794],[378,0.0997],[379,0.3948],[380,0.7251],[381,1.0851],[382,1.4373],[383,1.7329],[384,1.9498],[385,2.1033],[386,2.219],[387,2.3845],[388,2.6577],[389,2.9863],[390,3.3289],[391,3.6406],[393,4.1979],[394,4.4592],[395,4.7055],[396,4.9288],[397,5.1312],[398,5.3169],[399,5.4875],[400,5.6438],[401,5.7673],[402,5.8447],[403,5.8856],[404,5.8997],[406,5.8915],[407,5.8729],[408,5.8386],[409,5.8209],[417,5.5178],[418,5.4666],[419,5.4841],[422,5.4841]]
       }
       arm_r_pv_ik_ctrl.Rotation.Z {
        type=MovieSceneScriptingFloatChannel
        default=32.2279
        thinned_from=423
        keys [[0,32.2279],[2,32.1611],[6,31.8936],[11,31.9224],[12,31.9684],[14,32.1998],[15,32.2522],[16,32.1932],[17,31.8287],[18,30.9659],[19,29.8444],[20,28.5854],[22,25.7791],[23,24.5785],[24,23.5719],[25,22.7791],[26,22.0711],[27,21.4263],[28,20.8827],[29,20.466],[30,20.195],[31,20.0392],[32,19.9438],[34,19.8782],[36,19.9653],[39,20.2124],[45,20.4048],[48,20.5831],[54,20.744],[57,20.6953],[58,20.6416],[60,20.4115],[62,20.077],[63,19.8122],[67,18.449],[68,18.1751],[69,17.8197],[71,16.8117],[72,16.5049],[73,16.3991],[74,16.4407],[75,16.6765],[76,17.2773],[77,18.3736],[78,19.8075],[79,21.5906],[80,23.6697],[81,25.798],[82,27.8127],[83,29.6293],[84,31.3271],[85,32.876],[86,34.2328],[87,35.4034],[88,36.4431],[89,37.369],[90,38.1584],[91,38.8101],[92,39.327],[93,39.729],[94,40.0161],[95,40.1865],[97,40.2743],[98,40.1944],[103,38.9657],[108,38.2703],[119,37.7929],[120,37.825],[124,38.4585],[125,38.5596],[126,38.3391],[127,37.549],[128,36.4344],[129,35.1989],[130,33.7556],[131,32.1553],[132,27.7011],[133,21.5183],[134,11.995],[135,-2.4463],[136,-31.208],[137,-59.1706],[138,-59.2027],[139,-60.6091],[141,-65.0497],[142,-85.4642],[143,-105.3284],[144,-115.8361],[145,-118.2515],[146,-120.1758],[147,-118.3025],[148,-117.0651],[149,-113.9398],[150,-110.2175],[151,-109.676],[224,-109.975],[227,-110.9701],[231,-113.8385],[233,-110.3606],[234,-103.1522],[235,-60.564],[236,-23.2492],[237,0.8013],[238,11.7141],[239,16.7676],[240,18.3929],[241,19.6073],[242,21.0942],[243,21.3652],[244,19.7844],[245,17.815],[246,16.5528],[247,16.3636],[248,17.0515],[249,18.506],[251,21.2214],[252,22.4887],[253,23.2417],[254,23.6983],[255,23.996],[256,24.1089],[257,24.0224],[259,23.6359],[266,21.7052],[268,21.2589],[270,20.9442],[272,20.7412],[274,20.6474],[280,20.6559],[287,20.8529],[290,20.7983],[292,20.6691],[293,20.5698],[294,20.394],[297,19.5887],[299,18.828],[300,18.8201],[301,18.9771],[302,19.2076],[303,19.7575],[304,20.529],[305,21.4499],[306,22.4989],[307,23.6315],[308,24.8283],[309,26.1236],[310,27.5879],[315,35.7034],[316,37.2494],[317,38.688],[318,39.9991],[319,41.0764],[320,41.9258],[321,42.5144],[323,43.3409],[326,44.134],[327,44.2091],[329,44.0588],[335,43.0065],[340,42.7148],[345,42.8411],[347,43.0611],[349,43.4995],[352,44.5329],[354,45.4894],[355,45.7738],[356,45.756],[357,45.6409],[358,45.645],[369,47.418],[374,47.3115],[376,47.0684],[380,46.2921],[382,45.6034],[383,45.377],[384,45.4643],[385,45.6688],[386,45.5405],[387,44.6294],[388,42.9905],[389,41.1726],[390,39.453],[391,38.0334],[393,35.7171],[394,34.6895],[395,33.7805],[396,33.0106],[397,32.3536],[398,31.7685],[399,31.2529],[400,30.8051],[401,30.4972],[402,30.3838],[403,30.4107],[415,32.0153],[418,32.2118],[422,32.224]]
       }
       arm_r_pv_ik_ctrl.Scale.X {type=MovieSceneScriptingFloatChannel, default=1, const [1,423]}
       arm_r_pv_ik_ctrl.Scale.Y {type=MovieSceneScriptingFloatChannel, default=1, const [1,423]}
       arm_r_pv_ik_ctrl.Scale.Z {type=MovieSceneScriptingFloatChannel, default=1, const [1,423]}
       body_ctrl.Location.X {
        type=MovieSceneScriptingFloatChannel
        default=-4.9414
        thinned_from=423
        keys [[0,-4.7789],[6,-4.8304],[12,-4.9361],[14,-4.9798],[15,-4.9898],[16,-4.9748],[17,-4.9164],[18,-4.8031],[19,-4.6124],[20,-4.3627],[21,-4.104],[22,-3.874],[23,-3.6699],[24,-3.4877],[25,-3.3297],[26,-3.1995],[27,-3.0932],[28,-3.0068],[29,-2.9413],[30,-2.8979],[32,-2.8441],[34,-2.8076],[39,-2.7521],[43,-2.7331],[50,-2.7427],[55,-2.776],[57,-2.798],[58,-2.8155],[64,-2.9902],[65,-3.0158],[66,-3.0223],[67,-3.0063],[68,-2.9788],[69,-2.945],[70,-2.9041],[71,-2.8562],[72,-2.8],[73,-2.7348],[75,-2.5918],[76,-2.5268],[77,-2.4741],[78,-2.4329],[80,-2.3686],[83,-2.2853],[85,-2.2405],[86,-2.2251],[88,-2.2087],[114,-2.1543],[127,-2.1844],[128,-2.2011],[129,-2.2387],[130,-2.3023],[131,-2.3978],[132,-2.521],[133,-2.7056],[134,-3.0284],[135,-3.454],[136,-3.9818],[137,-4.5921],[138,-5.3827],[139,-6.469],[140,-7.3477],[141,-7.8374],[142,-8.1622],[143,-8.4016],[144,-8.5483],[145,-8.6601],[146,-8.7507],[148,-8.8787],[150,-8.9441],[152,-9.0615],[154,-9.2371],[156,-9.3416],[157,-9.3628],[225,-9.2606],[229,-9.3268],[231,-9.4068],[232,-9.5279],[233,-9.6212],[234,-9.6906],[235,-9.7239],[237,-9.733],[238,-9.7158],[239,-9.6721],[241,-9.5194],[242,-9.4174],[243,-9.2952],[244,-9.0953],[245,-8.8131],[249,-7.5344],[250,-7.2269],[251,-6.975],[252,-6.8163],[253,-6.7333],[254,-6.6762],[256,-6.6049],[258,-6.5804],[273,-6.6132],[292,-6.5114],[295,-6.5278],[297,-6.5617],[300,-6.6636],[301,-6.66],[302,-6.6042],[303,-6.5093],[304,-6.357],[305,-6.1328],[308,-5.3433],[309,-5.1183],[310,-4.9518],[312,-4.6395],[313,-4.5016],[314,-4.3881],[316,-4.1911],[317,-4.1051],[318,-4.0287],[319,-3.9639],[320,-3.9129],[321,-3.8744],[323,-3.819],[330,-3.7138],[334,-3.6805],[335,-3.6802],[337,-3.7032],[341,-3.7788],[344,-3.8562],[345,-3.8726],[346,-3.8767],[347,-3.8717],[348,-3.8587],[349,-3.8305],[350,-3.7809],[351,-3.7011],[352,-3.57],[353,-3.3822],[354,-3.1563],[356,-2.6667],[357,-2.4315],[358,-2.2206],[360,-1.8155],[361,-1.6227],[362,-1.4557],[363,-1.3288],[364,-1.2342],[365,-1.1551],[366,-1.0918],[367,-1.0436],[368,-1.0102],[369,-0.9884],[370,-0.9797],[371,-0.982],[372,-0.9954],[373,-1.0217],[374,-1.0629],[375,-1.121],[376,-1.2043],[377,-1.3193],[378,-1.4661],[379,-1.647],[380,-1.879],[381,-2.159],[382,-2.4596],[385,-3.3391],[386,-3.6226],[387,-3.8772],[388,-4.0976],[389,-4.2903],[390,-4.4556],[391,-4.5962],[392,-4.715],[393,-4.8121],[394,-4.8881],[395,-4.943],[396,-4.9785],[397,-4.9988],[400,-5.0082],[403,-4.9787],[413,-4.8073],[422,-4.7789]]
       }
       body_ctrl.Location.Y {
        type=MovieSceneScriptingFloatChannel
        default=-2.2548
        thinned_from=423
        keys [[0,-2.6549],[1,-2.6594],[3,-2.6882],[9,-2.7883],[10,-2.7998],[11,-2.8043],[12,-2.8022],[13,-2.7901],[14,-2.7593],[15,-2.7032],[16,-2.6038],[17,-2.4647],[18,-2.309],[19,-2.1269],[20,-1.9265],[21,-1.7565],[22,-1.626],[23,-1.5109],[24,-1.4117],[25,-1.3316],[26,-1.2732],[27,-1.2326],[28,-1.2052],[29,-1.1884],[30,-1.1798],[31,-1.1761],[33,-1.176],[43,-1.1957],[56,-1.2297],[63,-1.232],[64,-1.236],[65,-1.2428],[66,-1.2796],[67,-1.3567],[68,-1.4538],[69,-1.5594],[70,-1.6742],[71,-1.7984],[72,-1.9312],[73,-2.0728],[74,-2.2235],[75,-2.383],[76,-2.5506],[77,-2.7406],[78,-2.9422],[79,-3.1159],[80,-3.2701],[81,-3.4163],[82,-3.5195],[83,-3.5734],[84,-3.6006],[85,-3.61],[88,-3.5996],[103,-3.4211],[115,-3.3868],[123,-3.4097],[126,-3.4309],[127,-3.4283],[128,-3.3941],[129,-3.3262],[131,-3.1699],[133,-3.0457],[134,-2.8258],[135,-2.5303],[136,-2.2018],[137,-1.7598],[138,-1.0649],[139,-0.5695],[140,-0.4768],[141,-0.3684],[142,-0.3097],[143,-0.4039],[144,-0.5656],[145,-0.7512],[146,-0.9447],[147,-1.1598],[148,-1.3707],[149,-1.5623],[150,-1.7188],[151,-1.8562],[152,-1.9612],[153,-2.0229],[154,-2.0513],[155,-2.0495],[156,-2.0287],[160,-1.9053],[161,-1.8792],[164,-1.8162],[166,-1.7825],[169,-1.748],[174,-1.7133],[185,-1.6809],[201,-1.6977],[214,-1.7429],[215,-1.7576],[220,-1.858],[221,-1.8724],[222,-1.8814],[223,-1.8824],[224,-1.8739],[225,-1.8556],[226,-1.7984],[227,-1.6995],[228,-1.5612],[229,-1.3972],[230,-1.2153],[231,-1.0228],[232,-0.8368],[233,-0.6414],[234,-0.4421],[235,-0.2493],[236,-0.0749],[237,0.0399],[238,0.1106],[239,0.1699],[240,0.2113],[241,0.2478],[242,0.2791],[243,0.2934],[244,0.1855],[245,-0.0145],[246,-0.2256],[247,-0.3931],[248,-0.5393],[249,-0.7298],[250,-0.9408],[251,-1.1146],[252,-1.2335],[253,-1.294],[254,-1.3223],[255,-1.331],[256,-1.3316],[259,-1.3209],[263,-1.3181],[267,-1.3233],[275,-1.3422],[293,-1.3385],[294,-1.3402],[295,-1.3484],[296,-1.3679],[297,-1.4034],[298,-1.4785],[299,-1.5932],[301,-1.8398],[303,-2.0995],[305,-2.3773],[306,-2.5001],[307,-2.6079],[308,-2.709],[309,-2.7956],[310,-2.8513],[311,-2.875],[312,-2.8871],[313,-2.8926],[314,-2.8907],[315,-2.8799],[320,-2.7837],[329,-2.6708],[338,-2.6117],[341,-2.6099],[342,-2.6128],[343,-2.6215],[344,-2.6378],[345,-2.6625],[346,-2.6943],[347,-2.733],[349,-2.8265],[350,-2.8781],[351,-2.9379],[352,-3.0161],[353,-3.1173],[355,-3.3793],[356,-3.4866],[357,-3.5716],[358,-3.6367],[359,-3.6876],[360,-3.7284],[361,-3.7599],[363,-3.7991],[368,-3.819],[370,-3.8163],[371,-3.8059],[372,-3.7757],[373,-3.7221],[374,-3.6412],[375,-3.5292],[378,-3.1245],[379,-3.0023],[380,-2.8935],[381,-2.795],[382,-2.7078],[383,-2.6327],[384,-2.5711],[385,-2.5235],[386,-2.4885],[387,-2.4627],[388,-2.4433],[389,-2.4295],[390,-2.4209],[391,-2.4191],[392,-2.4256],[395,-2.4759],[400,-2.5957],[403,-2.6424],[406,-2.6694],[413,-2.6847],[418,-2.6566],[422,-2.6549]]
       }
       body_ctrl.Location.Z {
        type=MovieSceneScriptingFloatChannel
        default=2.4023
        thinned_from=423
        keys [[0,-0.7001],[1,-0.6989],[3,-0.6913],[5,-0.6819],[8,-0.6736],[10,-0.6707],[11,-0.6703],[12,-0.6723],[13,-0.6779],[14,-0.6859],[15,-0.6957],[16,-0.7071],[20,-0.7631],[21,-0.7755],[23,-0.7951],[24,-0.8027],[25,-0.8084],[27,-0.8144],[31,-0.8158],[50,-0.787],[58,-0.7847],[61,-0.7904],[64,-0.8018],[65,-0.8038],[67,-0.8034],[68,-0.8013],[69,-0.796],[70,-0.7866],[71,-0.7737],[72,-0.7577],[73,-0.7393],[75,-0.6972],[78,-0.6298],[79,-0.6084],[80,-0.5884],[81,-0.5702],[82,-0.5543],[83,-0.5419],[84,-0.534],[85,-0.5297],[86,-0.5284],[100,-0.5426],[104,-0.5419],[111,-0.5306],[119,-0.5101],[126,-0.4858],[127,-0.483],[128,-0.482],[129,-0.4839],[130,-0.4875],[131,-0.495],[132,-0.5737],[133,-0.6908],[134,-0.9185],[135,-1.184],[136,-1.3569],[137,-1.4247],[138,-1.4199],[139,-1.3421],[140,-0.9729],[141,-0.6382],[142,-0.4936],[143,-0.2716],[144,-0.1646],[145,-0.1346],[146,-0.1433],[147,-0.1831],[148,-0.4118],[149,-0.5938],[150,-0.723],[151,-0.817],[152,-0.8618],[153,-0.8166],[154,-0.6901],[155,-0.5602],[156,-0.4588],[157,-0.3921],[158,-0.3584],[159,-0.3394],[160,-0.3308],[161,-0.3287],[162,-0.3321],[163,-0.3422],[164,-0.3571],[165,-0.3752],[167,-0.4142],[168,-0.4321],[169,-0.4482],[170,-0.4627],[171,-0.4755],[172,-0.4868],[173,-0.4966],[174,-0.5048],[175,-0.5117],[177,-0.5224],[179,-0.5299],[181,-0.5347],[184,-0.5382],[190,-0.5376],[210,-0.5104],[216,-0.5097],[219,-0.5131],[221,-0.5189],[223,-0.5277],[225,-0.5377],[226,-0.5335],[227,-0.5016],[228,-0.4626],[229,-0.4301],[230,-0.4183],[231,-0.4188],[232,-0.424],[234,-0.437],[235,-0.4474],[236,-0.4647],[237,-0.4922],[238,-0.533],[239,-0.597],[240,-0.6859],[241,-0.7929],[243,-1.0273],[244,-1.1305],[245,-1.2083],[246,-1.2567],[247,-1.279],[248,-1.287],[249,-1.2574],[250,-1.1804],[251,-1.0784],[252,-0.9656],[253,-0.8553],[254,-0.7603],[255,-0.6938],[256,-0.6689],[304,-0.6685],[309,-0.6706],[316,-0.669],[321,-0.6759],[330,-0.695],[336,-0.7],[339,-0.6976],[343,-0.6861],[347,-0.6686],[354,-0.6257],[357,-0.6024],[358,-0.5989],[359,-0.599],[360,-0.6009],[361,-0.6044],[363,-0.6151],[365,-0.6297],[367,-0.6488],[370,-0.6823],[371,-0.6948],[374,-0.7364],[375,-0.7472],[376,-0.7548],[377,-0.7596],[378,-0.7623],[380,-0.7631],[382,-0.759],[383,-0.7546],[384,-0.7478],[387,-0.7179],[388,-0.706],[390,-0.6799],[391,-0.67],[392,-0.6632],[393,-0.6585],[394,-0.6555],[396,-0.654],[399,-0.6566],[402,-0.6645],[411,-0.6924],[415,-0.6988],[422,-0.7001]]
       }
       body_ctrl.Rotation.X {
        type=MovieSceneScriptingFloatChannel
        default=6.336
        thinned_from=423
        keys [[0,-0.4886],[1,-0.4903],[2,-0.4944],[3,-0.5002],[4,-0.4932],[5,-0.4992],[6,-0.5039],[7,-0.4949],[9,-0.5006],[10,-0.5022],[11,-0.5028],[12,-0.4989],[14,-0.4784],[15,-0.4887],[16,-0.5163],[17,-0.5706],[18,-0.6653],[19,-0.8295],[20,-1.0241],[21,-1.148],[22,-1.1944],[23,-1.2302],[24,-1.2604],[25,-1.2834],[27,-1.3178],[28,-1.3321],[29,-1.3431],[30,-1.35],[31,-1.3527],[32,-1.3518],[33,-1.3472],[35,-1.3287],[39,-1.2746],[41,-1.2564],[43,-1.2443],[44,-1.2444],[45,-1.235],[47,-1.235],[48,-1.2251],[58,-1.2246],[59,-1.2218],[60,-1.2287],[62,-1.2499],[63,-1.2703],[64,-1.2981],[65,-1.3225],[66,-1.2841],[67,-1.1658],[68,-1.0199],[69,-0.8646],[70,-0.691],[71,-0.5123],[72,-0.3387],[73,-0.1672],[74,0.007],[75,0.1757],[76,0.337],[77,0.5237],[78,0.7263],[79,0.8937],[80,1.0275],[81,1.1518],[82,1.2529],[83,1.3283],[84,1.3882],[85,1.4344],[86,1.4686],[87,1.4949],[89,1.5373],[91,1.5679],[93,1.5886],[95,1.5913],[96,1.6046],[98,1.6065],[102,1.5898],[103,1.5896],[104,1.5794],[105,1.5792],[106,1.5698],[108,1.5694],[109,1.5549],[111,1.5542],[112,1.541],[115,1.539],[116,1.5248],[117,1.5237],[118,1.5169],[120,1.5101],[121,1.5093],[122,1.5035],[123,1.4869],[124,1.4608],[125,1.4266],[126,1.3889],[127,1.3547],[128,1.3438],[129,1.3609],[130,1.387],[131,1.419],[132,1.3307],[133,1.0999],[134,0.454],[135,-0.3594],[136,-1.1912],[137,-2.3461],[138,-3.794],[139,-3.808],[140,-3.6718],[141,-3.5757],[142,-3.6095],[143,-3.7141],[146,-4.0844],[147,-4.2467],[148,-4.4589],[149,-4.7154],[150,-4.882],[151,-5.0812],[152,-5.204],[153,-5.1573],[154,-5.0522],[155,-4.925],[156,-4.7771],[157,-4.6521],[158,-4.5488],[159,-4.4569],[160,-4.3818],[161,-4.3291],[163,-4.2663],[165,-4.2334],[167,-4.2261],[169,-4.2347],[170,-4.2456],[171,-4.248],[173,-4.2699],[174,-4.2668],[177,-4.288],[178,-4.2809],[180,-4.2919],[184,-4.3009],[201,-4.318],[202,-4.3332],[204,-4.3342],[205,-4.3509],[206,-4.3513],[207,-4.3668],[208,-4.367],[209,-4.3864],[210,-4.3866],[211,-4.4104],[214,-4.4571],[221,-4.6222],[222,-4.6398],[223,-4.6422],[224,-4.6216],[225,-4.5789],[226,-4.5253],[227,-4.444],[230,-4.0106],[231,-3.8456],[232,-3.6655],[234,-3.2914],[235,-3.1233],[237,-2.8262],[238,-2.6447],[239,-2.4533],[240,-2.2439],[241,-1.9819],[242,-1.6479],[243,-1.2725],[244,-0.824],[245,-0.3967],[246,-0.0555],[247,0.1272],[248,0.2046],[249,0.0297],[250,-0.2713],[251,-0.4706],[252,-0.5797],[253,-0.5963],[254,-0.5744],[256,-0.4977],[257,-0.4615],[258,-0.4291],[259,-0.3993],[260,-0.3712],[261,-0.345],[262,-0.3239],[263,-0.3147],[264,-0.3133],[265,-0.3143],[266,-0.3179],[267,-0.3179],[269,-0.3556],[271,-0.4005],[272,-0.4204],[273,-0.4515],[274,-0.4852],[275,-0.5151],[276,-0.5349],[277,-0.5468],[278,-0.5467],[279,-0.566],[280,-0.566],[281,-0.5796],[283,-0.5796],[284,-0.5902],[287,-0.5902],[289,-0.5615],[291,-0.5278],[292,-0.5124],[293,-0.5007],[294,-0.4921],[295,-0.4858],[296,-0.4845],[297,-0.4916],[298,-0.519],[299,-0.5676],[300,-0.6238],[301,-0.6974],[302,-0.7956],[303,-0.9004],[304,-1.0102],[305,-1.1266],[307,-1.3493],[308,-1.4638],[309,-1.5622],[310,-1.6247],[311,-1.6514],[312,-1.659],[313,-1.6525],[314,-1.6376],[316,-1.5946],[318,-1.5463],[320,-1.5087],[321,-1.4966],[323,-1.477],[325,-1.4514],[326,-1.4361],[327,-1.4045],[333,-1.2675],[335,-1.2273],[338,-1.1602],[340,-1.124],[342,-1.0956],[343,-1.0835],[344,-1.0542],[345,-1.0304],[346,-1.0023],[347,-0.9709],[348,-0.9353],[349,-0.8913],[350,-0.8346],[351,-0.7615],[353,-0.5959],[354,-0.5258],[355,-0.4685],[356,-0.4188],[357,-0.3783],[358,-0.3487],[359,-0.3277],[360,-0.3113],[361,-0.2989],[362,-0.2899],[363,-0.2835],[364,-0.279],[367,-0.2691],[368,-0.2666],[369,-0.2602],[371,-0.2454],[372,-0.2415],[373,-0.2396],[374,-0.2385],[375,-0.2368],[377,-0.231],[378,-0.2286],[379,-0.2278],[380,-0.2301],[381,-0.2354],[382,-0.2428],[383,-0.2513],[384,-0.262],[386,-0.2893],[387,-0.3006],[389,-0.3218],[390,-0.3299],[391,-0.3328],[392,-0.3275],[393,-0.3162],[394,-0.3038],[395,-0.2956],[397,-0.2864],[398,-0.2838],[399,-0.2846],[400,-0.2905],[401,-0.3005],[404,-0.3341],[405,-0.3438],[406,-0.3511],[407,-0.3565],[408,-0.3605],[409,-0.3817],[411,-0.4078],[412,-0.4201],[414,-0.4418],[416,-0.4669],[417,-0.4779],[418,-0.4767],[419,-0.4886],[422,-0.4886]]
       }
       body_ctrl.Rotation.Y {
        type=MovieSceneScriptingFloatChannel
        default=-0.0058
        thinned_from=423
        keys [[0,-0.2145],[2,-0.2141],[3,-0.2133],[4,-0.2032],[5,-0.2013],[6,-0.1986],[7,-0.1883],[8,-0.184],[9,-0.1788],[10,-0.1724],[11,-0.1646],[12,-0.1516],[13,-0.1315],[14,-0.1263],[15,-0.1195],[16,-0.112],[17,-0.0887],[18,-0.0467],[19,-0.0082],[20,-0.0012],[21,-0.0348],[22,-0.1095],[23,-0.2399],[24,-0.409],[25,-0.5922],[26,-0.7575],[27,-0.8995],[28,-1.0337],[29,-1.1483],[30,-1.2314],[31,-1.2855],[32,-1.3243],[33,-1.3534],[35,-1.4053],[36,-1.4342],[37,-1.4681],[39,-1.5474],[40,-1.5804],[41,-1.6051],[42,-1.6195],[43,-1.6281],[44,-1.6211],[45,-1.63],[47,-1.6064],[48,-1.6072],[54,-1.5213],[56,-1.5002],[58,-1.4881],[61,-1.4248],[63,-1.3858],[69,-1.2361],[70,-1.1912],[71,-1.1067],[72,-0.9883],[73,-0.8448],[74,-0.6612],[75,-0.4274],[76,-0.171],[77,0.1084],[78,0.4081],[79,0.6793],[80,0.9382],[81,1.1905],[82,1.3743],[83,1.4907],[84,1.5793],[85,1.6449],[86,1.6859],[87,1.7064],[88,1.7108],[89,1.7001],[90,1.6802],[91,1.6566],[92,1.6281],[94,1.5592],[95,1.5154],[96,1.4671],[97,1.4095],[98,1.3723],[100,1.2805],[102,1.1951],[103,1.1454],[104,1.1217],[105,1.0833],[106,1.0631],[107,1.034],[108,1.0086],[109,1.0014],[110,0.982],[111,0.9647],[112,0.9575],[113,0.943],[114,0.9311],[115,0.9231],[116,0.9255],[117,0.9319],[118,0.927],[120,0.9301],[121,0.9237],[124,0.8894],[125,0.8803],[126,0.8476],[127,0.7889],[128,0.7086],[129,0.6133],[130,0.5244],[131,0.4575],[132,0.4518],[133,0.5317],[134,0.7025],[135,0.9243],[136,1.1822],[137,1.6417],[138,2.4274],[139,3.2639],[140,3.6312],[141,2.8411],[142,1.8152],[143,1.1402],[144,0.4869],[145,-0.0095],[146,-0.3422],[147,-0.5749],[148,-0.6522],[149,-0.5605],[150,-0.3667],[151,-0.1606],[152,0.0131],[153,0.1186],[154,0.1375],[155,0.103],[156,0.064],[157,0.0413],[158,0.0207],[159,-0.006],[160,-0.0351],[161,-0.0626],[162,-0.0884],[164,-0.1408],[165,-0.1664],[166,-0.1909],[167,-0.2124],[168,-0.2413],[169,-0.2504],[170,-0.2761],[171,-0.2883],[172,-0.3098],[173,-0.3287],[174,-0.3432],[175,-0.3562],[176,-0.3667],[177,-0.3757],[178,-0.3947],[179,-0.4011],[180,-0.4062],[181,-0.4236],[183,-0.4291],[184,-0.4307],[185,-0.4436],[201,-0.4445],[202,-0.4357],[204,-0.4357],[205,-0.4307],[206,-0.4307],[207,-0.4284],[208,-0.4284],[209,-0.4273],[214,-0.4294],[216,-0.4331],[219,-0.4423],[221,-0.449],[222,-0.4496],[223,-0.4482],[224,-0.4455],[225,-0.4479],[226,-0.4566],[227,-0.4764],[228,-0.5019],[229,-0.5228],[230,-0.5255],[231,-0.495],[232,-0.2914],[233,-0.1703],[234,-0.1225],[235,-0.1207],[236,-0.1495],[237,-0.1798],[238,-0.2309],[239,-0.2883],[240,-0.3599],[241,-0.4371],[242,-0.5195],[243,-0.6103],[244,-0.6897],[245,-0.7748],[246,-0.8831],[247,-1.0165],[248,-1.18],[249,-1.3221],[250,-1.4242],[251,-1.5018],[252,-1.5548],[253,-1.5868],[255,-1.632],[256,-1.6491],[257,-1.6591],[259,-1.6588],[260,-1.6545],[262,-1.6314],[264,-1.5959],[266,-1.5513],[267,-1.5366],[268,-1.4995],[269,-1.4753],[270,-1.458],[272,-1.4351],[273,-1.4293],[274,-1.4265],[277,-1.4333],[278,-1.4296],[279,-1.4385],[280,-1.435],[281,-1.4433],[283,-1.4377],[284,-1.4495],[287,-1.4463],[288,-1.453],[294,-1.4454],[295,-1.4404],[296,-1.432],[297,-1.4205],[298,-1.3965],[299,-1.3689],[300,-1.3649],[301,-1.3731],[302,-1.3584],[303,-1.3116],[304,-1.229],[305,-1.1073],[306,-0.8663],[307,-0.49],[308,-0.0528],[309,0.4944],[310,1.118],[311,1.6279],[312,1.9944],[313,2.3043],[314,2.5488],[315,2.7137],[316,2.8043],[317,2.8394],[318,2.8373],[322,2.7489],[323,2.7308],[324,2.7216],[325,2.7219],[326,2.7126],[327,2.7145],[328,2.7036],[329,2.6854],[334,2.5533],[337,2.4902],[338,2.4593],[339,2.4493],[340,2.4225],[341,2.4112],[346,2.3086],[347,2.2847],[348,2.2526],[349,2.2056],[350,2.1373],[351,2.0133],[352,1.8286],[353,1.6217],[354,1.4311],[355,1.2793],[356,1.1601],[357,1.0687],[358,0.9997],[359,0.9554],[360,0.9353],[361,0.9301],[362,0.9299],[363,0.9248],[364,0.9077],[367,0.8353],[368,0.822],[369,0.7823],[371,0.7267],[372,0.6957],[373,0.6621],[375,0.5912],[378,0.5001],[379,0.4771],[380,0.463],[382,0.4496],[383,0.4349],[384,0.3732],[385,0.2516],[386,0.117],[387,0.0134],[388,-0.073],[389,-0.1646],[390,-0.2465],[391,-0.3035],[392,-0.3367],[393,-0.3571],[394,-0.3629],[395,-0.3599],[396,-0.3499],[397,-0.333],[398,-0.3119],[400,-0.267],[401,-0.2426],[402,-0.2153],[403,-0.1869],[404,-0.1596],[405,-0.1387],[406,-0.1347],[407,-0.1311],[408,-0.1258],[409,-0.1422],[410,-0.1519],[411,-0.1624],[412,-0.1723],[413,-0.1803],[414,-0.1874],[416,-0.2024],[417,-0.2086],[418,-0.2082],[419,-0.2145],[422,-0.2145]]
       }
       body_ctrl.Rotation.Z {
        type=MovieSceneScriptingFloatChannel
        default=20.359
        thinned_from=423
        keys [[0,20.3359],[6,20.4576],[15,20.1798],[16,20.1899],[17,20.2729],[18,20.5011],[19,20.9569],[21,22.2243],[23,23.5854],[24,24.1488],[25,24.5621],[26,24.9023],[27,25.1691],[28,25.3628],[30,25.5706],[34,25.6236],[50,24.9605],[56,24.9352],[66,25.5295],[68,25.5408],[69,25.494],[70,25.3637],[71,25.1093],[72,24.7265],[73,24.2538],[75,23.1382],[78,21.3344],[80,20.2435],[81,19.7297],[82,19.3117],[83,18.9948],[84,18.7338],[85,18.5266],[86,18.3701],[87,18.2605],[89,18.1498],[92,18.1379],[95,18.3204],[98,18.6051],[103,19.2287],[109,19.5531],[126,19.8462],[128,19.9456],[134,20.5507],[135,20.6178],[136,20.7262],[137,20.995],[138,21.4556],[139,22.0522],[140,22.8696],[141,23.9516],[142,25.2555],[143,26.768],[144,27.2249],[145,27.3243],[146,27.1452],[147,26.7459],[148,26.2591],[150,24.8399],[151,24.5906],[152,24.5442],[153,24.8999],[154,25.4615],[156,26.4217],[157,26.6347],[226,25.6798],[228,25.7634],[230,25.9817],[232,26.2973],[233,26.3788],[234,26.2854],[237,25.5851],[238,25.3968],[241,25.1021],[244,24.8882],[245,24.7582],[246,24.5273],[247,24.1258],[248,23.4287],[249,22.8068],[250,22.5464],[251,22.4645],[254,22.4752],[264,23.1353],[272,23.2569],[286,22.8113],[291,22.8826],[293,23.0045],[294,23.141],[296,23.52],[297,23.603],[298,23.5614],[299,23.4239],[300,23.1954],[301,22.8903],[302,22.3354],[303,21.5401],[304,20.6433],[305,19.687],[306,18.8175],[307,18.0967],[308,17.4597],[310,16.2705],[311,15.739],[312,15.3043],[313,14.9464],[314,14.6711],[315,14.4587],[316,14.3681],[317,14.4274],[318,14.5426],[322,15.172],[324,15.3749],[343,15.5611],[346,15.5094],[347,15.4657],[348,15.3829],[349,15.2388],[350,15.0284],[351,14.7378],[352,14.3625],[353,13.8827],[355,12.8039],[356,12.2337],[357,11.692],[358,11.2282],[359,10.8326],[360,10.4691],[361,10.1484],[362,9.8809],[363,9.6768],[364,9.5265],[365,9.4106],[366,9.3248],[367,9.265],[369,9.206],[373,9.2057],[375,9.2551],[377,9.357],[378,9.4346],[379,9.5524],[380,9.7568],[381,10.0999],[382,10.6408],[383,11.4381],[384,12.7349],[385,14.4923],[386,16.3485],[387,17.9411],[388,19.1511],[389,20.0996],[390,20.7829],[391,21.2743],[392,21.6593],[393,21.9536],[394,22.1715],[395,22.3268],[397,22.4998],[400,22.5611],[401,22.512],[402,22.3741],[406,21.4729],[408,21.1211],[412,20.6594],[415,20.4314],[422,20.3359]]
       }
       body_ctrl.Scale.X {type=MovieSceneScriptingFloatChannel, default=1, const [1,423]}
       body_ctrl.Scale.Y {type=MovieSceneScriptingFloatChannel, default=1, const [1,423]}
       body_ctrl.Scale.Z {type=MovieSceneScriptingFloatChannel, default=1, const [1,423]}
       hips_ctrl.Location.X {type=MovieSceneScriptingFloatChannel, default=0, thinned_from=423, keys [[0,0],[422,0]]}
       hips_ctrl.Location.Y {type=MovieSceneScriptingFloatChannel, default=0, thinned_from=423, keys [[0,0],[422,0]]}
       hips_ctrl.Location.Z {
        type=MovieSceneScriptingFloatChannel
        default=0
        thinned_from=418
        keys [[0,0],[97,0],[98,0.0133],[99,0.0435],[100,0.0922],[101,0.1388],[102,0.1725],[103,0.1873],[134,0.1873],[135,0.2429],[136,0.5255],[137,0.5806],[138,0.5806],[139,0.5597],[140,0.4383],[141,0.2479],[142,0],[149,0],[150,0.0774],[151,0.2908],[152,0.3996],[153,0.3435],[154,0.229],[155,0.1125],[156,0.0281],[158,0],[240,0],[241,0.0533],[242,0.2368],[243,0.4321],[244,0.5268],[245,0.5573],[246,0.5298],[247,0.4575],[248,0.3543],[249,0.2064],[250,0.1238],[251,0.0754],[252,0.0419],[253,0.0164],[254,0],[348,0],[349,0.0377],[350,0.1127],[351,0.1978],[352,0.2632],[353,0.3114],[354,0.3509],[355,0.3794],[356,0.4013],[357,0.4188],[358,0.443],[360,0.5],[361,0.5329],[362,0.5636],[363,0.5877],[364,0.5987],[378,0.5987],[379,0.5877],[380,0.5482],[385,0.1184],[386,0.0614],[387,0.0197],[388,0],[422,0]]
       }
       hips_ctrl.Rotation.X {
        type=MovieSceneScriptingFloatChannel
        default=0
        thinned_from=423
        keys [[0,6.8247],[3,6.8247],[14,6.8016],[15,6.8248],[16,6.8655],[17,6.9348],[18,7.0445],[19,7.2271],[20,7.4403],[21,7.5728],[22,7.6155],[25,7.6731],[32,7.6746],[43,7.5453],[59,7.5248],[62,7.5633],[65,7.6602],[66,7.6363],[67,7.5354],[69,7.2682],[73,6.5862],[75,6.213],[76,6.0362],[77,5.836],[78,5.6208],[79,5.4441],[80,5.3051],[81,5.1788],[82,5.0778],[83,5.004],[84,4.9473],[85,4.9048],[87,4.852],[92,4.7855],[95,4.7781],[96,4.7657],[103,4.7822],[104,4.7923],[108,4.8016],[109,4.8159],[111,4.8159],[112,4.8287],[115,4.8287],[116,4.842],[125,4.8385],[126,4.8609],[127,4.9367],[128,5.2472],[129,5.8055],[130,6.4349],[131,7.1143],[132,7.8182],[133,8.5641],[134,9.6473],[135,10.8416],[136,11.9415],[138,13.7478],[139,13.9917],[140,13.9298],[141,13.7951],[142,13.7114],[143,13.6635],[146,13.4067],[147,13.3528],[148,13.3737],[150,13.5742],[151,13.7899],[152,13.9471],[153,13.9406],[155,13.8107],[160,13.4234],[165,13.2659],[201,13.1866],[210,13.2518],[227,13.7366],[230,13.7155],[234,13.5321],[235,13.505],[237,13.505],[240,13.422],[241,13.3717],[242,13.2758],[243,13.1099],[244,12.5411],[245,11.5213],[246,10.5762],[247,9.7926],[248,9.0786],[250,7.6843],[251,7.0595],[252,6.6943],[253,6.6008],[254,6.6172],[256,6.7056],[257,6.7331],[258,6.7278],[262,6.6313],[267,6.6249],[272,6.7271],[275,6.8219],[277,6.8536],[278,6.8536],[279,6.8729],[287,6.8971],[295,6.7925],[297,6.798],[300,6.8894],[301,6.939],[303,7.0876],[309,7.6588],[310,7.7147],[311,7.7351],[313,7.7227],[320,7.5337],[343,7.5185],[346,7.4543],[348,7.3651],[350,7.2209],[353,6.8994],[354,6.8027],[356,6.6484],[357,6.5878],[359,6.5103],[362,6.4484],[371,6.4193],[375,6.461],[386,6.7868],[390,6.8492],[392,6.844],[399,6.7087],[408,6.6965],[417,6.8124],[422,6.8247]]
       }
       hips_ctrl.Rotation.Y {
        type=MovieSceneScriptingFloatChannel
        default=0
        thinned_from=423
        keys [[0,0.2085],[3,0.2085],[4,0.1997],[6,0.1997],[7,0.1928],[13,0.1928],[14,0.2116],[15,0.2299],[16,0.2522],[17,0.2721],[18,0.2912],[19,0.3516],[20,0.4449],[21,0.5007],[22,0.4963],[23,0.4689],[24,0.4215],[26,0.3066],[27,0.2509],[28,0.1984],[29,0.1506],[30,0.1089],[31,0.0749],[32,0.0498],[33,0.0347],[34,0.0321],[35,0.0392],[36,0.0557],[37,0.0812],[38,0.1162],[39,0.1527],[40,0.1847],[41,0.2091],[42,0.2251],[43,0.2382],[44,0.2382],[45,0.2562],[47,0.2562],[48,0.2707],[58,0.2707],[59,0.2543],[60,0.2372],[61,0.2189],[62,0.2052],[63,0.1903],[64,0.1728],[66,0.1402],[67,0.1229],[68,0.1049],[69,0.0872],[70,0.0553],[71,-0.0097],[72,-0.1108],[73,-0.2479],[74,-0.4506],[75,-0.7178],[76,-0.9908],[77,-1.2541],[78,-1.5212],[79,-1.759],[80,-1.9872],[81,-2.2081],[82,-2.3659],[83,-2.4629],[84,-2.5353],[85,-2.5872],[86,-2.6159],[87,-2.6244],[88,-2.617],[89,-2.5961],[91,-2.5427],[92,-2.5219],[93,-2.5087],[97,-2.5122],[98,-2.5363],[102,-2.6012],[103,-2.6012],[104,-2.6214],[105,-2.6214],[106,-2.6347],[108,-2.6347],[109,-2.6496],[111,-2.6496],[112,-2.6579],[115,-2.6579],[117,-2.6619],[120,-2.609],[123,-2.5167],[124,-2.4894],[125,-2.4701],[126,-2.4399],[127,-2.4454],[128,-2.4635],[129,-2.4889],[130,-2.467],[132,-2.3136],[133,-2.1127],[134,-1.7025],[135,-1.1441],[136,-0.5628],[137,-0.2616],[138,-0.319],[139,0.0873],[140,0.5753],[141,1.0059],[142,1.2815],[143,1.3752],[144,1.466],[145,1.5493],[146,1.6087],[147,1.6125],[148,1.5088],[149,1.287],[150,0.9868],[151,0.7634],[152,0.7054],[153,0.8912],[154,1.1667],[155,1.3483],[156,1.4791],[157,1.5338],[158,1.5447],[159,1.546],[160,1.541],[161,1.5328],[162,1.5215],[166,1.4584],[167,1.4468],[168,1.4468],[169,1.4287],[170,1.4287],[171,1.4174],[177,1.4159],[178,1.4271],[180,1.4271],[181,1.4403],[184,1.4403],[185,1.4519],[201,1.4519],[202,1.4432],[204,1.4432],[205,1.4382],[212,1.4357],[215,1.4424],[221,1.4752],[224,1.4798],[225,1.4859],[226,1.4996],[227,1.5264],[228,1.5619],[229,1.5911],[230,1.61],[231,1.6256],[232,1.6337],[233,1.6333],[234,1.6253],[235,1.611],[237,1.611],[238,1.6366],[239,1.6797],[240,1.7469],[241,1.8302],[242,1.9238],[243,2.0248],[244,2.0908],[245,2.054],[246,1.9275],[247,1.7349],[248,1.466],[249,1.3004],[250,1.2209],[251,1.2052],[252,1.2618],[253,1.3265],[254,1.3634],[255,1.3501],[257,1.2859],[258,1.2562],[259,1.2291],[262,1.1603],[263,1.1407],[265,1.1086],[266,1.0959],[267,1.0959],[268,1.0781],[269,1.0738],[270,1.0732],[272,1.0782],[277,1.1102],[278,1.1102],[279,1.1226],[280,1.1226],[281,1.1343],[283,1.1343],[284,1.1482],[287,1.1482],[288,1.1534],[290,1.1416],[293,1.1139],[294,1.1003],[295,1.0772],[296,1.0397],[297,0.9819],[298,0.8898],[299,0.7702],[300,0.6485],[301,0.4606],[302,0.1653],[303,-0.1787],[304,-0.5553],[305,-0.9813],[306,-1.4471],[307,-1.9953],[308,-2.5863],[309,-3.0848],[310,-3.4725],[311,-3.811],[312,-4.0837],[313,-4.2744],[314,-4.413],[315,-4.532],[316,-4.6237],[317,-4.6828],[318,-4.7096],[319,-4.7059],[320,-4.6668],[321,-4.6171],[322,-4.5789],[323,-4.5522],[324,-4.5399],[326,-4.5399],[334,-4.7744],[337,-4.821],[338,-4.821],[339,-4.8398],[340,-4.8398],[341,-4.8529],[346,-4.8416],[347,-4.8302],[348,-4.809],[349,-4.7714],[350,-4.7108],[351,-4.5932],[352,-4.4134],[353,-4.2098],[354,-4.0208],[356,-3.6608],[357,-3.4942],[358,-3.3625],[359,-3.2656],[360,-3.1888],[361,-3.1287],[362,-3.0818],[363,-3.0441],[364,-3.0148],[366,-2.9771],[367,-2.9653],[368,-2.9653],[369,-2.9393],[371,-2.9047],[372,-2.878],[373,-2.8421],[374,-2.7967],[375,-2.7411],[376,-2.6318],[377,-2.4383],[378,-2.1894],[379,-1.9103],[380,-1.5697],[381,-1.1641],[382,-0.7497],[383,-0.3822],[384,-0.0517],[385,0.271],[386,0.5437],[387,0.7265],[388,0.8122],[389,0.8282],[390,0.8037],[391,0.7679],[392,0.7277],[393,0.6722],[394,0.6034],[396,0.4583],[397,0.387],[398,0.3201],[399,0.2613],[400,0.213],[401,0.1734],[402,0.1395],[403,0.1108],[404,0.087],[405,0.0711],[408,0.0711],[409,0.0939],[410,0.1109],[411,0.1289],[412,0.1463],[413,0.1612],[414,0.1738],[415,0.1852],[416,0.1947],[417,0.2021],[418,0.2021],[419,0.2085],[422,0.2085]]
       }
       hips_ctrl.Rotation.Z {
        type=MovieSceneScriptingFloatChannel
        default=0
        thinned_from=423
        keys [[0,0.025],[3,0.025],[4,0.0238],[6,0.0238],[7,0.023],[13,0.023],[15,0.0275],[18,0.036],[19,0.0446],[20,0.0581],[21,0.0666],[22,0.0664],[23,0.0629],[24,0.0567],[26,0.0413],[27,0.0339],[28,0.0268],[29,0.0203],[30,0.0147],[31,0.0101],[32,0.0067],[33,0.0047],[34,0.0043],[35,0.0053],[36,0.0075],[37,0.0108],[38,0.0155],[39,0.0203],[40,0.0245],[41,0.0277],[42,0.0298],[43,0.0315],[44,0.0315],[45,0.0339],[47,0.0339],[48,0.0358],[58,0.0358],[60,0.0314],[61,0.029],[63,0.0254],[66,0.0188],[69,0.0111],[70,0.0069],[71,-0.0012],[72,-0.0131],[73,-0.0286],[74,-0.0505],[75,-0.0781],[76,-0.1048],[77,-0.1282],[78,-0.1497],[79,-0.1676],[80,-0.1845],[81,-0.2001],[82,-0.2102],[83,-0.2156],[84,-0.2194],[85,-0.2219],[86,-0.223],[87,-0.2227],[88,-0.2212],[91,-0.2132],[92,-0.2111],[93,-0.2096],[97,-0.2094],[98,-0.2114],[102,-0.2175],[103,-0.2175],[104,-0.2197],[105,-0.2197],[106,-0.2212],[108,-0.2212],[109,-0.2232],[111,-0.2232],[112,-0.2245],[115,-0.2245],[116,-0.2254],[117,-0.2254],[119,-0.223],[120,-0.2213],[124,-0.2108],[126,-0.2074],[127,-0.2112],[128,-0.2262],[129,-0.253],[130,-0.2782],[131,-0.2984],[132,-0.3176],[133,-0.3181],[134,-0.2894],[135,-0.2191],[136,-0.119],[137,-0.0596],[138,-0.078],[139,0.0218],[140,0.1427],[141,0.247],[142,0.3126],[143,0.3343],[144,0.3543],[145,0.3718],[146,0.3834],[147,0.3827],[148,0.3587],[149,0.3084],[150,0.2382],[151,0.1874],[152,0.1752],[153,0.2212],[154,0.2883],[155,0.3314],[156,0.3611],[157,0.3721],[158,0.3726],[159,0.3708],[162,0.3607],[165,0.3472],[166,0.3434],[167,0.3405],[168,0.3405],[169,0.336],[170,0.336],[171,0.3331],[173,0.3331],[174,0.3324],[177,0.3324],[178,0.3347],[180,0.3347],[181,0.3376],[184,0.3376],[185,0.3402],[201,0.3402],[202,0.3385],[204,0.3385],[205,0.3377],[210,0.3378],[214,0.3407],[217,0.3458],[221,0.3549],[224,0.3591],[225,0.3615],[226,0.3659],[227,0.3731],[228,0.3818],[229,0.3886],[230,0.3929],[231,0.3957],[232,0.3962],[233,0.3945],[234,0.3911],[235,0.3869],[237,0.3869],[238,0.3923],[239,0.4018],[240,0.4168],[242,0.4538],[243,0.4715],[244,0.465],[245,0.4186],[246,0.3598],[247,0.2994],[248,0.2342],[249,0.1917],[250,0.1647],[251,0.1492],[252,0.1481],[253,0.1535],[254,0.1581],[255,0.1578],[256,0.1551],[257,0.1518],[259,0.1445],[261,0.1379],[262,0.1349],[263,0.1324],[264,0.1304],[265,0.1287],[266,0.1273],[267,0.1273],[268,0.1256],[269,0.1254],[270,0.1258],[272,0.1272],[275,0.1313],[277,0.1334],[278,0.1334],[279,0.1353],[280,0.1353],[281,0.137],[283,0.137],[284,0.1389],[287,0.1389],[288,0.1392],[289,0.1383],[293,0.133],[294,0.1312],[295,0.1283],[296,0.1238],[297,0.117],[298,0.1064],[299,0.0926],[300,0.0784],[301,0.0561],[302,0.0203],[303,-0.0222],[304,-0.0699],[305,-0.1252],[306,-0.1871],[307,-0.2614],[308,-0.3435],[309,-0.4146],[310,-0.4701],[311,-0.5172],[312,-0.5543],[313,-0.5791],[314,-0.5961],[315,-0.61],[316,-0.6198],[317,-0.625],[318,-0.626],[319,-0.6233],[321,-0.609],[322,-0.6035],[323,-0.5996],[324,-0.5977],[326,-0.5977],[328,-0.6038],[334,-0.6285],[337,-0.6352],[338,-0.6352],[339,-0.6379],[340,-0.6379],[341,-0.6397],[343,-0.6397],[345,-0.6358],[346,-0.6327],[347,-0.628],[348,-0.6209],[349,-0.6105],[350,-0.5962],[351,-0.5733],[352,-0.5422],[353,-0.5089],[354,-0.4792],[356,-0.4264],[357,-0.4033],[358,-0.3854],[359,-0.3725],[360,-0.3622],[361,-0.3543],[362,-0.3481],[363,-0.3433],[364,-0.3397],[365,-0.3371],[367,-0.3337],[368,-0.3337],[369,-0.3307],[371,-0.3267],[372,-0.3239],[373,-0.3203],[374,-0.3158],[375,-0.3103],[376,-0.2989],[377,-0.2781],[378,-0.2508],[379,-0.2198],[380,-0.1815],[381,-0.1353],[382,-0.0876],[383,-0.0449],[384,-0.0061],[385,0.0321],[386,0.0647],[387,0.0867],[388,0.0972],[389,0.0993],[390,0.0965],[391,0.0923],[392,0.0873],[393,0.0804],[394,0.072],[396,0.0543],[397,0.0457],[398,0.0377],[399,0.0307],[400,0.025],[401,0.0204],[402,0.0164],[403,0.013],[404,0.0102],[405,0.0083],[408,0.0083],[409,0.0111],[412,0.0173],[413,0.0191],[415,0.022],[416,0.0232],[417,0.0241],[418,0.0241],[419,0.025],[422,0.025]]
       }
       hips_ctrl.Scale.X {type=MovieSceneScriptingFloatChannel, default=1, const [1,423]}
       hips_ctrl.Scale.Y {type=MovieSceneScriptingFloatChannel, default=1, const [1,423]}
       hips_ctrl.Scale.Z {type=MovieSceneScriptingFloatChannel, default=1, const [1,423]}
       hips_tan_ctrl.Location.X {
        type=MovieSceneScriptingFloatChannel
        default=0.0827
        thinned_from=423
        keys [[0,0.0499],[2,0.05],[12,0.0506],[14,0.0504],[15,0.0505],[16,0.051],[17,0.0525],[18,0.0554],[19,0.0595],[21,0.0683],[22,0.0721],[23,0.0755],[24,0.0784],[25,0.0807],[26,0.0824],[27,0.084],[28,0.0853],[29,0.0864],[30,0.0872],[31,0.0877],[32,0.0881],[33,0.0881],[34,0.0879],[36,0.0869],[40,0.0839],[43,0.0827],[46,0.0821],[47,0.0821],[48,0.0818],[57,0.0815],[59,0.0819],[62,0.0832],[66,0.0857],[68,0.0865],[69,0.0865],[70,0.0861],[71,0.0851],[72,0.0835],[73,0.0812],[74,0.0777],[75,0.073],[76,0.0675],[77,0.0608],[78,0.0532],[79,0.045],[80,0.0365],[81,0.0282],[82,0.0207],[83,0.0145],[84,0.0089],[85,0.0043],[86,0.0004],[87,-0.0026],[88,-0.005],[89,-0.0069],[90,-0.0084],[91,-0.0094],[92,-0.0101],[93,-0.0107],[94,-0.011],[96,-0.011],[98,-0.0104],[103,-0.008],[108,-0.0069],[109,-0.0069],[115,-0.0065],[120,-0.0057],[124,-0.0063],[125,-0.0063],[126,-0.0056],[127,-0.0041],[128,-0.0018],[129,0.001],[130,0.0052],[131,0.0109],[132,0.0182],[133,0.0284],[134,0.0422],[136,0.074],[137,0.0919],[138,0.1158],[139,0.123],[140,0.1257],[142,0.123],[143,0.1226],[144,0.1208],[147,0.1127],[148,0.1112],[149,0.1119],[150,0.1154],[151,0.1183],[152,0.1198],[153,0.1196],[154,0.1182],[159,0.1149],[161,0.1143],[173,0.1147],[197,0.1138],[215,0.1145],[227,0.113],[228,0.113],[235,0.1158],[237,0.1157],[238,0.1149],[239,0.1137],[240,0.1119],[241,0.1097],[242,0.1072],[243,0.1043],[244,0.1008],[245,0.0961],[246,0.09],[247,0.0831],[248,0.0755],[249,0.0698],[250,0.0665],[251,0.0651],[252,0.0653],[254,0.0672],[265,0.0748],[268,0.0762],[272,0.077],[281,0.0767],[288,0.0762],[296,0.0777],[297,0.078],[299,0.0795],[300,0.079],[301,0.0767],[302,0.0728],[303,0.0679],[304,0.0621],[305,0.0553],[306,0.0478],[307,0.0392],[308,0.0299],[309,0.0215],[310,0.0141],[311,0.007],[312,0.0003],[313,-0.0055],[316,-0.0213],[317,-0.0263],[318,-0.0307],[319,-0.0341],[320,-0.0365],[321,-0.038],[322,-0.0391],[323,-0.0398],[324,-0.0402],[326,-0.0402],[327,-0.0396],[334,-0.0332],[336,-0.0321],[340,-0.0309],[343,-0.0308],[346,-0.0312],[348,-0.0321],[349,-0.0329],[350,-0.0341],[351,-0.0362],[352,-0.0394],[353,-0.0429],[356,-0.0524],[357,-0.0552],[358,-0.0575],[359,-0.0592],[360,-0.0605],[361,-0.0616],[363,-0.0631],[365,-0.0641],[366,-0.0643],[368,-0.0642],[369,-0.0639],[370,-0.0633],[372,-0.0613],[373,-0.0601],[374,-0.0586],[375,-0.0569],[376,-0.0544],[377,-0.0508],[378,-0.0462],[379,-0.0406],[380,-0.0338],[381,-0.0258],[382,-0.0175],[383,-0.0095],[385,0.0057],[386,0.0129],[387,0.0197],[388,0.0257],[389,0.0309],[390,0.0356],[391,0.0396],[392,0.043],[393,0.0456],[394,0.0476],[395,0.0493],[396,0.0506],[397,0.0517],[399,0.0531],[401,0.054],[404,0.0542],[407,0.0534],[413,0.051],[418,0.05],[422,0.05]]
       }
       hips_tan_ctrl.Location.Y {
        type=MovieSceneScriptingFloatChannel
        default=-1.1869
        thinned_from=423
        keys [[0,-0.1134],[13,-0.1139],[15,-0.1134],[16,-0.1127],[17,-0.1115],[18,-0.1096],[19,-0.1065],[20,-0.1028],[21,-0.1005],[22,-0.0998],[24,-0.099],[29,-0.0985],[33,-0.0989],[41,-0.1008],[58,-0.1015],[62,-0.1007],[65,-0.099],[66,-0.0981],[69,-0.094],[71,-0.0917],[73,-0.09],[74,-0.0893],[75,-0.0889],[76,-0.089],[77,-0.0902],[78,-0.0925],[80,-0.0974],[81,-0.0996],[82,-0.1013],[83,-0.1025],[84,-0.1034],[86,-0.1046],[93,-0.1061],[125,-0.1051],[126,-0.1055],[127,-0.1065],[128,-0.1101],[129,-0.1166],[130,-0.1239],[131,-0.1317],[132,-0.1399],[133,-0.1485],[135,-0.1687],[136,-0.1782],[137,-0.1853],[138,-0.1824],[139,-0.1978],[140,-0.2214],[141,-0.2343],[142,-0.2394],[143,-0.2381],[144,-0.2385],[146,-0.2417],[147,-0.2383],[148,-0.2258],[149,-0.2064],[150,-0.1946],[151,-0.1884],[152,-0.1881],[154,-0.1947],[155,-0.1987],[156,-0.2021],[160,-0.2073],[164,-0.2098],[203,-0.2112],[216,-0.2084],[221,-0.2055],[223,-0.2032],[225,-0.2003],[226,-0.1994],[227,-0.2002],[229,-0.2054],[231,-0.208],[234,-0.2107],[235,-0.2112],[238,-0.2116],[241,-0.2132],[242,-0.2134],[243,-0.2127],[244,-0.2056],[245,-0.191],[247,-0.1583],[248,-0.1447],[249,-0.1326],[250,-0.1225],[251,-0.1149],[252,-0.111],[253,-0.1108],[257,-0.1146],[262,-0.1167],[267,-0.1167],[272,-0.115],[275,-0.1133],[280,-0.1123],[286,-0.112],[295,-0.1138],[297,-0.1137],[300,-0.1121],[301,-0.1113],[303,-0.1087],[307,-0.1022],[308,-0.1005],[309,-0.0989],[310,-0.098],[311,-0.0976],[312,-0.0976],[314,-0.0982],[319,-0.1006],[321,-0.1011],[343,-0.1013],[346,-0.1023],[347,-0.103],[348,-0.1038],[349,-0.105],[350,-0.1063],[351,-0.108],[353,-0.1118],[354,-0.1134],[356,-0.1161],[357,-0.1171],[359,-0.1185],[362,-0.1195],[372,-0.12],[376,-0.119],[386,-0.1138],[390,-0.1128],[392,-0.1129],[399,-0.1154],[407,-0.1155],[418,-0.1134],[422,-0.1134]]
       }
       hips_tan_ctrl.Location.Z {
        type=MovieSceneScriptingFloatChannel
        default=-0.0298
        thinned_from=423
        keys [[0,0.0752],[17,0.0754],[19,0.0759],[25,0.0781],[32,0.0794],[33,0.0792],[34,0.0793],[35,0.0793],[36,0.0794],[65,0.079],[66,0.0776],[67,0.0742],[68,0.0703],[70,0.0621],[74,0.0449],[75,0.0408],[76,0.0373],[77,0.0346],[78,0.0327],[79,0.0318],[81,0.0312],[125,0.031],[126,0.0311],[127,0.0316],[128,0.0335],[129,0.037],[130,0.0414],[131,0.0465],[132,0.0525],[133,0.0603],[134,0.0771],[135,0.0991],[136,0.1222],[137,0.1425],[138,0.1157],[139,0.3133],[140,0.656],[141,0.8179],[142,0.8644],[143,0.8242],[144,0.8056],[145,0.8022],[146,0.8022],[147,0.7201],[148,0.4966],[149,0.1691],[150,-0.0199],[151,-0.0685],[152,-0.0247],[153,0.0373],[154,0.0832],[155,0.1368],[156,0.1728],[220,0.1738],[221,0.1714],[222,0.1627],[223,0.1509],[224,0.1347],[225,0.1211],[226,0.1159],[227,0.1388],[228,0.1842],[229,0.2335],[230,0.2579],[231,0.2691],[232,0.2719],[240,0.2721],[241,0.2717],[242,0.2706],[243,0.2679],[244,0.2214],[245,0.1409],[246,0.0625],[247,0.0142],[248,0.0185],[249,0.0432],[250,0.074],[251,0.0925],[252,0.1001],[253,0.101],[254,0.0984],[255,0.0936],[256,0.0894],[257,0.0855],[258,0.0822],[259,0.0802],[260,0.0792],[272,0.0795],[274,0.0788],[298,0.0786],[300,0.0782],[301,0.0776],[305,0.0739],[306,0.0733],[307,0.0729],[308,0.0728],[310,0.0735],[316,0.0763],[319,0.0771],[354,0.0773],[367,0.0786],[372,0.0781],[375,0.0773],[376,0.0768],[381,0.0722],[382,0.0715],[383,0.0712],[384,0.0712],[385,0.0713],[392,0.0748],[396,0.0757],[404,0.076],[422,0.0752]]
       }
       hips_tan_ctrl.Rotation.X {type=MovieSceneScriptingFloatChannel, default=0, const [0,423]}
       hips_tan_ctrl.Rotation.Y {type=MovieSceneScriptingFloatChannel, default=0, const [0,423]}
       hips_tan_ctrl.Rotation.Z {type=MovieSceneScriptingFloatChannel, default=0, const [0,423]}
       hips_tan_ctrl.Scale.X {type=MovieSceneScriptingFloatChannel, default=1, const [1,423]}
       hips_tan_ctrl.Scale.Y {type=MovieSceneScriptingFloatChannel, default=1, const [1,423]}
       hips_tan_ctrl.Scale.Z {type=MovieSceneScriptingFloatChannel, default=1, const [1,423]}
       spine_01_ctrl.Location.X {
        type=MovieSceneScriptingFloatChannel
        default=0.082
        thinned_from=423
        keys [[0,0.0556],[14,0.0559],[15,0.0562],[16,0.0567],[17,0.0578],[18,0.0597],[19,0.0627],[20,0.066],[21,0.0688],[22,0.0709],[23,0.0728],[24,0.0744],[25,0.0756],[28,0.0783],[30,0.0797],[32,0.0806],[34,0.0809],[36,0.0806],[41,0.0786],[46,0.0779],[57,0.0777],[65,0.0792],[67,0.0792],[69,0.0787],[70,0.0779],[71,0.0765],[72,0.0744],[73,0.0716],[74,0.0673],[75,0.0618],[76,0.0556],[77,0.0489],[78,0.0415],[79,0.0339],[81,0.0182],[82,0.0115],[83,0.0061],[84,0.0014],[85,-0.0024],[86,-0.0055],[87,-0.0078],[88,-0.0095],[89,-0.0107],[90,-0.0115],[92,-0.0123],[96,-0.0125],[109,-0.0106],[116,-0.0103],[125,-0.0081],[126,-0.0072],[127,-0.0058],[128,-0.0037],[129,-0.0011],[130,0.0029],[131,0.0086],[132,0.0159],[133,0.0267],[134,0.0421],[136,0.0775],[137,0.0948],[138,0.1107],[139,0.1181],[140,0.1211],[141,0.1208],[142,0.12],[143,0.12],[144,0.1192],[147,0.116],[148,0.1154],[149,0.1154],[150,0.116],[154,0.1149],[156,0.1157],[157,0.1157],[162,0.1158],[170,0.117],[178,0.117],[225,0.1161],[232,0.1149],[237,0.1151],[238,0.1149],[239,0.1144],[241,0.1125],[242,0.1113],[243,0.1098],[244,0.1075],[245,0.1032],[246,0.097],[247,0.0896],[248,0.0809],[249,0.0746],[250,0.0711],[251,0.0698],[252,0.0705],[254,0.0734],[257,0.0753],[264,0.0775],[270,0.0782],[288,0.0772],[295,0.0778],[297,0.0775],[298,0.0771],[299,0.0764],[300,0.075],[301,0.0717],[302,0.0663],[303,0.0598],[304,0.0524],[305,0.0439],[306,0.0346],[307,0.0238],[308,0.0123],[309,0.0022],[310,-0.0061],[311,-0.0137],[312,-0.0204],[313,-0.0258],[315,-0.0349],[316,-0.039],[317,-0.0427],[318,-0.0459],[319,-0.0482],[320,-0.0495],[322,-0.0507],[324,-0.0512],[326,-0.0511],[334,-0.048],[340,-0.0471],[346,-0.0472],[348,-0.0476],[349,-0.0481],[350,-0.049],[351,-0.0505],[352,-0.053],[353,-0.0557],[356,-0.0631],[357,-0.0653],[358,-0.0671],[359,-0.0684],[360,-0.0694],[362,-0.0709],[364,-0.0717],[366,-0.0722],[368,-0.072],[369,-0.0717],[370,-0.0712],[371,-0.0704],[372,-0.0695],[373,-0.0683],[374,-0.0668],[375,-0.065],[376,-0.0622],[377,-0.0577],[378,-0.0519],[379,-0.0449],[380,-0.0362],[381,-0.0261],[382,-0.0155],[383,-0.0055],[385,0.0134],[386,0.0221],[387,0.03],[388,0.0365],[389,0.0419],[390,0.0465],[391,0.0504],[392,0.0534],[393,0.0556],[394,0.0571],[395,0.0582],[396,0.059],[397,0.0594],[400,0.0597],[414,0.056],[422,0.0556]]
       }
       spine_01_ctrl.Location.Y {
        type=MovieSceneScriptingFloatChannel
        default=-1.0263
        thinned_from=423
        keys [[0,-0.1594],[13,-0.1599],[15,-0.1594],[16,-0.1589],[17,-0.1579],[18,-0.1563],[19,-0.1538],[20,-0.1507],[21,-0.1489],[23,-0.1479],[31,-0.1474],[58,-0.1497],[62,-0.1491],[65,-0.1477],[66,-0.1466],[71,-0.1372],[74,-0.1324],[75,-0.1313],[76,-0.1305],[77,-0.131],[78,-0.1325],[81,-0.138],[82,-0.1394],[84,-0.1412],[89,-0.143],[125,-0.1427],[126,-0.1432],[127,-0.1446],[128,-0.1499],[129,-0.1592],[130,-0.1698],[131,-0.1812],[132,-0.1929],[133,-0.2053],[135,-0.2362],[136,-0.2506],[137,-0.2616],[138,-0.263],[139,-0.2762],[140,-0.2936],[141,-0.3034],[142,-0.3076],[143,-0.3069],[144,-0.3074],[146,-0.3101],[147,-0.308],[148,-0.2998],[149,-0.287],[150,-0.2794],[151,-0.2748],[152,-0.274],[154,-0.2784],[156,-0.2837],[160,-0.2879],[214,-0.2894],[221,-0.2864],[226,-0.2818],[227,-0.2823],[229,-0.2856],[234,-0.2897],[241,-0.2918],[242,-0.2916],[243,-0.2901],[244,-0.2811],[245,-0.2628],[247,-0.2227],[248,-0.2047],[249,-0.188],[250,-0.1732],[251,-0.1618],[252,-0.1558],[253,-0.1552],[254,-0.1563],[256,-0.1591],[257,-0.1602],[262,-0.1622],[268,-0.162],[276,-0.1592],[287,-0.1585],[296,-0.1599],[298,-0.1595],[301,-0.1579],[304,-0.1545],[309,-0.1474],[310,-0.1465],[312,-0.1461],[320,-0.1488],[343,-0.149],[346,-0.1499],[348,-0.1512],[350,-0.1533],[354,-0.1594],[357,-0.1625],[361,-0.1643],[373,-0.1648],[389,-0.1592],[392,-0.1591],[400,-0.1612],[422,-0.1595]]
       }
       spine_01_ctrl.Location.Z {
        type=MovieSceneScriptingFloatChannel
        default=-0.0291
        thinned_from=423
        keys [[0,0.0739],[20,0.0738],[31,0.0765],[37,0.0766],[65,0.0762],[66,0.075],[67,0.0722],[69,0.0658],[74,0.0486],[75,0.0452],[76,0.0424],[77,0.0403],[78,0.039],[79,0.0385],[125,0.0385],[126,0.0386],[127,0.039],[128,0.0407],[129,0.0439],[130,0.0478],[131,0.0523],[132,0.0576],[133,0.0644],[134,0.0784],[135,0.0964],[136,0.1154],[137,0.1323],[138,0.1109],[139,0.2707],[140,0.5473],[141,0.678],[142,0.7156],[143,0.6832],[144,0.6683],[145,0.6657],[146,0.6658],[147,0.5997],[148,0.4195],[149,0.1554],[150,0.0032],[151,-0.0364],[152,-0.0015],[153,0.0484],[154,0.0853],[155,0.1285],[156,0.1577],[166,0.1605],[216,0.1599],[220,0.1591],[221,0.1571],[222,0.1499],[223,0.1403],[224,0.1272],[225,0.1161],[226,0.1119],[227,0.1303],[228,0.1667],[229,0.2064],[230,0.2261],[231,0.2352],[232,0.2377],[240,0.2382],[241,0.2379],[242,0.2369],[243,0.2348],[244,0.1971],[245,0.1318],[246,0.068],[247,0.028],[248,0.0304],[249,0.0493],[250,0.0734],[251,0.0879],[252,0.0938],[253,0.0946],[254,0.0925],[255,0.0888],[257,0.0825],[258,0.08],[259,0.0785],[260,0.0778],[268,0.0783],[272,0.078],[274,0.0773],[298,0.0772],[300,0.0767],[301,0.0761],[305,0.0722],[307,0.0707],[308,0.0703],[310,0.0704],[318,0.0729],[349,0.0737],[361,0.0768],[370,0.077],[374,0.0763],[376,0.0755],[381,0.0713],[382,0.0707],[383,0.0704],[385,0.0703],[394,0.074],[400,0.075],[422,0.074]]
       }
       spine_01_ctrl.Rotation.X {
        type=MovieSceneScriptingFloatChannel
        default=-9.1747
        thinned_from=423
        keys [[0,-2.3491],[13,-2.3585],[15,-2.3495],[16,-2.3394],[17,-2.3232],[18,-2.2987],[19,-2.257],[20,-2.2083],[21,-2.1808],[59,-2.2083],[63,-2.1948],[65,-2.18],[66,-2.1665],[72,-2.0335],[74,-1.9967],[75,-1.9816],[76,-1.9724],[77,-1.9807],[78,-2.0063],[81,-2.0958],[82,-2.1191],[84,-2.1505],[86,-2.1693],[94,-2.1969],[125,-2.1787],[126,-2.1853],[127,-2.2053],[128,-2.2804],[129,-2.4143],[130,-2.5648],[131,-2.7258],[133,-3.0639],[135,-3.4653],[136,-3.6503],[137,-3.7955],[138,-3.8789],[139,-3.8426],[140,-3.6972],[141,-3.6564],[142,-3.6679],[143,-3.7068],[144,-3.735],[146,-3.7699],[147,-3.831],[148,-3.9765],[149,-4.1968],[150,-4.3346],[151,-4.3316],[152,-4.2624],[153,-4.2157],[155,-4.1655],[156,-4.1577],[160,-4.2229],[213,-4.2459],[222,-4.1885],[225,-4.1923],[226,-4.1865],[227,-4.1637],[229,-4.0964],[230,-4.0837],[231,-4.0859],[234,-4.1264],[241,-4.1501],[242,-4.1455],[243,-4.1239],[244,-4.0474],[245,-3.8755],[246,-3.666],[247,-3.4158],[249,-2.8367],[250,-2.5741],[251,-2.3823],[252,-2.2839],[253,-2.2748],[254,-2.2953],[256,-2.3495],[257,-2.372],[258,-2.3865],[262,-2.4182],[267,-2.4242],[272,-2.4036],[276,-2.3762],[287,-2.3633],[296,-2.3908],[298,-2.387],[300,-2.3725],[301,-2.3588],[303,-2.3155],[307,-2.1993],[308,-2.167],[309,-2.1394],[310,-2.1218],[311,-2.1143],[312,-2.113],[314,-2.1229],[319,-2.1728],[321,-2.1847],[343,-2.1788],[346,-2.1951],[348,-2.2187],[350,-2.257],[354,-2.3707],[356,-2.4133],[358,-2.4428],[362,-2.4696],[371,-2.4762],[376,-2.4554],[387,-2.3502],[390,-2.3363],[392,-2.3371],[400,-2.3792],[409,-2.378],[418,-2.35],[422,-2.35]]
       }
       spine_01_ctrl.Rotation.Y {
        type=MovieSceneScriptingFloatChannel
        default=-1.1411
        thinned_from=423
        keys [[0,-0.9245],[12,-0.9204],[14,-0.9253],[16,-0.9366],[18,-0.9508],[20,-0.9815],[21,-0.9875],[24,-0.9943],[26,-1.0046],[29,-1.0309],[32,-1.0614],[33,-1.0698],[35,-1.0803],[36,-1.0825],[37,-1.0822],[41,-1.0662],[57,-1.0628],[61,-1.0722],[62,-1.0717],[63,-1.0689],[66,-1.0505],[69,-1.0263],[70,-1.0146],[71,-0.9941],[72,-0.9643],[73,-0.9256],[74,-0.8704],[75,-0.7984],[77,-0.6475],[79,-0.488],[80,-0.4054],[81,-0.3209],[82,-0.2502],[83,-0.195],[84,-0.1481],[85,-0.1109],[86,-0.0832],[87,-0.0639],[88,-0.0515],[89,-0.0447],[90,-0.0424],[91,-0.0438],[92,-0.0464],[94,-0.0521],[95,-0.0546],[96,-0.0546],[97,-0.0589],[98,-0.0602],[99,-0.0604],[100,-0.06],[101,-0.059],[103,-0.0566],[104,-0.0556],[108,-0.0531],[109,-0.0527],[110,-0.0527],[111,-0.0521],[112,-0.0521],[113,-0.0518],[115,-0.0518],[116,-0.0524],[117,-0.0546],[118,-0.0578],[119,-0.0621],[120,-0.0676],[121,-0.0748],[122,-0.0828],[125,-0.1075],[126,-0.1232],[127,-0.1459],[128,-0.1773],[129,-0.2171],[130,-0.2769],[131,-0.357],[132,-0.4583],[133,-0.6127],[134,-0.8252],[135,-1.0617],[136,-1.2864],[137,-1.4834],[138,-1.5294],[139,-1.5644],[140,-1.5181],[141,-1.4581],[142,-1.421],[143,-1.4183],[144,-1.4209],[145,-1.4332],[146,-1.4522],[147,-1.4839],[149,-1.5751],[150,-1.5646],[152,-1.468],[153,-1.4313],[154,-1.4155],[155,-1.4186],[157,-1.4422],[158,-1.4516],[161,-1.4876],[165,-1.5198],[170,-1.5421],[171,-1.5421],[172,-1.5466],[178,-1.5491],[221,-1.5368],[225,-1.5435],[226,-1.5403],[227,-1.5281],[228,-1.5078],[230,-1.4597],[231,-1.4393],[232,-1.4246],[233,-1.4153],[234,-1.4107],[237,-1.4108],[242,-1.4307],[243,-1.432],[244,-1.4287],[245,-1.4009],[246,-1.3431],[247,-1.2582],[248,-1.1436],[249,-1.0607],[250,-1.018],[251,-1.0061],[252,-1.0224],[253,-1.0485],[254,-1.0716],[255,-1.085],[257,-1.1036],[258,-1.1094],[259,-1.1119],[262,-1.1095],[278,-1.0569],[280,-1.053],[281,-1.053],[282,-1.05],[288,-1.048],[294,-1.0552],[295,-1.0545],[296,-1.0506],[297,-1.0414],[298,-1.0241],[299,-0.9993],[300,-0.9705],[301,-0.9214],[302,-0.8437],[303,-0.7524],[304,-0.6517],[305,-0.5379],[306,-0.4142],[307,-0.2707],[308,-0.1175],[309,0.0119],[310,0.1131],[311,0.2015],[312,0.2736],[313,0.3263],[314,0.3672],[315,0.404],[316,0.4359],[317,0.4617],[318,0.4808],[319,0.4928],[320,0.495],[321,0.494],[323,0.496],[324,0.4961],[327,0.4909],[329,0.4912],[335,0.5008],[341,0.5023],[343,0.5019],[347,0.5],[348,0.5018],[349,0.5071],[350,0.518],[351,0.5434],[352,0.5843],[353,0.6312],[354,0.6745],[356,0.7571],[357,0.7955],[358,0.8255],[359,0.847],[360,0.8638],[361,0.8768],[362,0.8868],[364,0.9005],[366,0.9072],[368,0.9071],[370,0.8993],[371,0.8932],[372,0.8841],[373,0.8715],[374,0.8546],[375,0.8328],[376,0.795],[377,0.7331],[378,0.6491],[379,0.5461],[380,0.4183],[381,0.2669],[382,0.1079],[383,-0.0428],[384,-0.1856],[385,-0.3254],[386,-0.4574],[387,-0.576],[388,-0.6754],[389,-0.7581],[390,-0.8272],[391,-0.8856],[392,-0.9307],[393,-0.961],[394,-0.9813],[395,-0.9944],[396,-1.0019],[397,-1.0041],[398,-1.0025],[410,-0.9394],[413,-0.9293],[422,-0.9246]]
       }
       spine_01_ctrl.Rotation.Z {
        type=MovieSceneScriptingFloatChannel
        default=0
        thinned_from=423
        keys [[0,-0.1107],[1,-0.1107],[3,-0.0875],[4,-0.0748],[5,-0.0629],[6,-0.0629],[7,-0.0467],[8,-0.0467],[9,-0.0359],[12,-0.0359],[13,-0.055],[14,-0.0714],[15,-0.0795],[16,-0.064],[17,0.0281],[18,0.2418],[19,0.5265],[20,0.8522],[21,1.2018],[22,1.5144],[23,1.7765],[24,1.9995],[25,2.1568],[26,2.2603],[27,2.3397],[28,2.3955],[29,2.4281],[30,2.4379],[31,2.4273],[32,2.4004],[33,2.361],[34,2.3128],[37,2.1504],[38,2.1025],[39,2.0633],[40,2.0365],[43,1.9838],[46,1.9438],[47,1.9438],[48,1.9238],[49,1.9238],[50,1.9086],[52,1.9086],[53,1.8939],[57,1.8939],[58,1.8903],[59,1.8999],[60,1.9171],[61,1.9425],[62,1.977],[63,2.0358],[67,2.3613],[68,2.43],[69,2.4717],[70,2.4874],[71,2.4919],[72,2.4839],[73,2.4472],[74,2.3708],[75,2.26],[76,2.0863],[77,1.8108],[78,1.4758],[79,1.099],[80,0.7131],[81,0.3575],[82,0.0277],[83,-0.2641],[84,-0.5235],[85,-0.7525],[86,-0.9512],[87,-1.118],[88,-1.26],[89,-1.3846],[90,-1.4899],[91,-1.576],[92,-1.6447],[93,-1.6985],[94,-1.7358],[95,-1.7545],[97,-1.7536],[98,-1.7355],[99,-1.7012],[102,-1.5717],[103,-1.5373],[104,-1.5087],[105,-1.4863],[107,-1.4494],[108,-1.4335],[109,-1.4208],[110,-1.4208],[111,-1.4027],[112,-1.4027],[113,-1.3905],[115,-1.3905],[116,-1.3801],[118,-1.3777],[119,-1.3796],[120,-1.391],[121,-1.4273],[123,-1.5221],[124,-1.5633],[125,-1.5919],[126,-1.6054],[127,-1.5846],[128,-1.5346],[129,-1.4724],[130,-1.3819],[131,-1.251],[132,-1.083],[133,-0.9054],[134,-0.6554],[135,-0.3417],[136,0.0246],[137,0.5799],[138,2.1477],[139,2.4265],[140,2.6056],[141,2.7224],[142,2.7959],[143,2.8277],[144,2.7194],[145,2.4973],[147,1.9645],[148,1.8348],[149,1.9306],[151,2.7277],[152,2.9922],[153,3.1139],[154,3.0898],[155,3.0166],[157,2.8278],[159,2.6459],[160,2.5603],[161,2.4952],[162,2.4491],[163,2.4116],[164,2.3813],[166,2.3367],[169,2.2901],[170,2.2778],[171,2.2778],[172,2.2574],[173,2.2574],[174,2.2424],[175,2.2424],[176,2.232],[178,2.232],[179,2.2234],[196,2.2163],[197,2.2303],[199,2.2303],[200,2.2443],[201,2.2443],[202,2.2547],[203,2.2547],[204,2.2652],[205,2.2652],[206,2.2751],[213,2.2956],[225,2.2458],[226,2.2533],[227,2.2829],[228,2.339],[229,2.4322],[230,2.5483],[231,2.6554],[232,2.7382],[233,2.7972],[234,2.8359],[235,2.8617],[236,2.8617],[237,2.848],[238,2.7825],[239,2.6808],[240,2.5492],[241,2.3933],[242,2.2201],[243,2.0364],[244,1.8597],[245,1.7002],[246,1.5454],[247,1.397],[248,1.2637],[249,1.1424],[250,1.0456],[251,0.9781],[252,0.9317],[253,0.9032],[254,0.89],[255,0.8858],[256,0.8906],[257,0.9084],[258,0.9383],[259,0.9784],[260,1.0265],[261,1.0809],[262,1.1396],[265,1.3256],[266,1.3855],[267,1.4415],[268,1.4921],[269,1.5354],[270,1.57],[271,1.5978],[272,1.6219],[274,1.6598],[276,1.6863],[278,1.7035],[280,1.7131],[281,1.7131],[282,1.7171],[286,1.7154],[287,1.7037],[290,1.7076],[292,1.7179],[294,1.7386],[295,1.755],[296,1.7853],[297,1.8417],[298,1.962],[299,2.1086],[300,2.1816],[301,2.1771],[302,2.1519],[303,2.0971],[304,2.0046],[305,1.8785],[306,1.7292],[307,1.5604],[308,1.3735],[309,1.1669],[310,0.9334],[311,0.6731],[312,0.3948],[313,0.1072],[314,-0.1791],[315,-0.478],[317,-1.0946],[318,-1.3791],[319,-1.6109],[320,-1.7933],[321,-1.912],[322,-1.9878],[323,-2.0372],[324,-2.0634],[325,-2.0759],[326,-2.078],[327,-2.0483],[328,-1.9881],[329,-1.9168],[330,-1.8357],[332,-1.6674],[333,-1.5893],[334,-1.515],[335,-1.4587],[336,-1.4208],[337,-1.3908],[338,-1.3665],[340,-1.3264],[341,-1.3264],[342,-1.3136],[343,-1.3136],[344,-1.3294],[345,-1.3411],[346,-1.3577],[347,-1.3828],[348,-1.4128],[350,-1.4799],[352,-1.5369],[355,-1.6123],[357,-1.6551],[360,-1.7078],[363,-1.7464],[365,-1.7637],[367,-1.7668],[368,-1.7476],[369,-1.7374],[370,-1.7097],[374,-1.5029],[376,-1.396],[378,-1.2952],[380,-1.2019],[382,-1.1149],[384,-1.0228],[386,-0.9261],[387,-0.8753],[388,-0.822],[390,-0.7102],[392,-0.5937],[395,-0.4126],[396,-0.3529],[397,-0.2939],[398,-0.236],[399,-0.1796],[400,-0.1252],[401,-0.0784],[402,-0.0437],[403,-0.0195],[404,-0.0041],[405,0.0042],[406,0.007],[407,0.0055],[408,0.0002],[409,-0.008],[410,-0.0182],[411,-0.0298],[412,-0.0417],[413,-0.0532],[414,-0.0651],[415,-0.0779],[416,-0.0902],[417,-0.1007],[418,-0.1079],[422,-0.1079]]
       }
       spine_01_ctrl.Scale.X {type=MovieSceneScriptingFloatChannel, default=1, const [1,423]}
       spine_01_ctrl.Scale.Y {type=MovieSceneScriptingFloatChannel, default=1, const [1,423]}
       spine_01_ctrl.Scale.Z {type=MovieSceneScriptingFloatChannel, default=1, const [1,423]}
       spine_02_ctrl.Location.X {
        type=MovieSceneScriptingFloatChannel
        default=0.0486
        thinned_from=423
        keys [[0,0.0486],[14,0.0487],[18,0.0497],[20,0.0507],[22,0.051],[24,0.0518],[32,0.0575],[34,0.0584],[36,0.0586],[37,0.0584],[41,0.0568],[44,0.0564],[57,0.0562],[61,0.0571],[63,0.0571],[66,0.0564],[69,0.0553],[70,0.0547],[71,0.0539],[72,0.0527],[73,0.0513],[74,0.0492],[77,0.0411],[79,0.0351],[81,0.0284],[82,0.0254],[83,0.0229],[84,0.0208],[85,0.019],[86,0.0177],[87,0.0167],[88,0.016],[89,0.0155],[91,0.0152],[117,0.0163],[125,0.0183],[126,0.0189],[127,0.0202],[128,0.0219],[129,0.0241],[130,0.0272],[131,0.031],[132,0.0361],[133,0.0434],[134,0.0529],[135,0.0631],[136,0.0728],[137,0.0827],[138,0.0867],[139,0.0894],[140,0.0887],[141,0.0849],[142,0.0818],[143,0.0807],[144,0.08],[145,0.08],[147,0.0811],[148,0.0815],[149,0.0812],[150,0.0802],[152,0.0769],[153,0.0748],[154,0.073],[155,0.0728],[158,0.0738],[165,0.0777],[170,0.0791],[177,0.0795],[224,0.0783],[226,0.0778],[228,0.0766],[231,0.0737],[232,0.073],[233,0.0725],[234,0.0724],[239,0.0725],[241,0.0718],[243,0.0706],[244,0.0694],[245,0.0669],[246,0.0633],[247,0.059],[248,0.0541],[249,0.0506],[250,0.0488],[251,0.0483],[252,0.0489],[254,0.051],[257,0.0531],[259,0.0538],[263,0.054],[281,0.0512],[288,0.0509],[295,0.0516],[297,0.0514],[298,0.051],[299,0.0503],[300,0.0494],[301,0.0477],[302,0.045],[303,0.0418],[304,0.0383],[305,0.0343],[306,0.0299],[307,0.025],[308,0.0197],[309,0.0152],[310,0.0117],[311,0.0086],[312,0.0061],[313,0.0041],[315,0.0011],[317,-0.0014],[318,-0.0023],[319,-0.003],[323,-0.0041],[325,-0.0042],[346,-0.0027],[348,-0.003],[349,-0.0034],[350,-0.0043],[351,-0.0063],[352,-0.0094],[353,-0.0129],[356,-0.0224],[357,-0.0253],[358,-0.0276],[359,-0.0293],[360,-0.0306],[361,-0.0316],[363,-0.033],[365,-0.0338],[367,-0.0341],[370,-0.0338],[372,-0.0332],[374,-0.032],[375,-0.0311],[376,-0.0296],[377,-0.0273],[378,-0.024],[379,-0.0199],[380,-0.0147],[381,-0.0086],[382,-0.0022],[383,0.0041],[385,0.0159],[386,0.0216],[387,0.0271],[388,0.032],[389,0.0364],[390,0.0402],[391,0.0436],[392,0.0462],[393,0.0482],[394,0.0496],[395,0.0507],[396,0.0515],[397,0.052],[399,0.0524],[405,0.0516],[413,0.0492],[422,0.0487]]
       }
       spine_02_ctrl.Location.Y {
        type=MovieSceneScriptingFloatChannel
        default=-0.0977
        thinned_from=423
        keys [[0,-0.0977],[2,-0.0977],[12,-0.0984],[14,-0.0981],[15,-0.0977],[16,-0.0969],[17,-0.0957],[18,-0.0938],[19,-0.0906],[20,-0.0869],[21,-0.0848],[22,-0.0843],[31,-0.0846],[44,-0.0866],[58,-0.0869],[62,-0.0862],[65,-0.0847],[66,-0.0839],[68,-0.0821],[71,-0.08],[73,-0.0791],[75,-0.0787],[76,-0.0789],[77,-0.0801],[80,-0.0872],[81,-0.0893],[82,-0.091],[83,-0.0923],[84,-0.0933],[85,-0.094],[88,-0.0952],[95,-0.0964],[109,-0.0956],[125,-0.095],[126,-0.0953],[127,-0.0961],[128,-0.0987],[129,-0.1035],[130,-0.1088],[131,-0.1147],[132,-0.1208],[133,-0.1271],[135,-0.1411],[136,-0.1479],[137,-0.1534],[138,-0.1519],[139,-0.1647],[140,-0.1838],[141,-0.1947],[142,-0.1994],[143,-0.1987],[144,-0.1992],[146,-0.202],[147,-0.1993],[148,-0.1897],[149,-0.1747],[150,-0.1661],[151,-0.1608],[152,-0.1599],[154,-0.1651],[156,-0.1713],[160,-0.1766],[163,-0.1787],[200,-0.1809],[214,-0.1784],[220,-0.175],[222,-0.1731],[225,-0.1692],[226,-0.1683],[227,-0.1688],[228,-0.1706],[229,-0.1729],[233,-0.1777],[234,-0.1787],[235,-0.1793],[239,-0.1799],[241,-0.1807],[242,-0.1808],[243,-0.1802],[244,-0.1748],[245,-0.1635],[246,-0.1503],[247,-0.1366],[248,-0.1247],[249,-0.1141],[250,-0.1054],[251,-0.0994],[252,-0.0964],[253,-0.0964],[257,-0.1],[262,-0.1027],[266,-0.1031],[268,-0.1029],[272,-0.1014],[275,-0.0996],[281,-0.0984],[287,-0.0981],[294,-0.1001],[296,-0.1003],[297,-0.1002],[298,-0.1],[300,-0.0988],[301,-0.0978],[302,-0.0963],[303,-0.0946],[304,-0.0927],[308,-0.0842],[309,-0.0823],[310,-0.0811],[311,-0.0806],[312,-0.0805],[314,-0.0811],[319,-0.084],[321,-0.0847],[329,-0.0852],[342,-0.0846],[344,-0.085],[346,-0.0859],[347,-0.0866],[348,-0.0876],[349,-0.0888],[350,-0.0903],[351,-0.0922],[353,-0.0966],[354,-0.0985],[356,-0.1017],[357,-0.1029],[358,-0.1039],[360,-0.1051],[363,-0.1061],[371,-0.1064],[375,-0.1054],[379,-0.103],[385,-0.0985],[387,-0.0975],[390,-0.0967],[391,-0.0966],[392,-0.0968],[399,-0.0998],[407,-0.1002],[418,-0.0977],[422,-0.0977]]
       }
       spine_02_ctrl.Location.Z {
        type=MovieSceneScriptingFloatChannel
        default=0.0735
        thinned_from=423
        keys [[0,0.0735],[18,0.0734],[20,0.0731],[33,0.0753],[65,0.0748],[66,0.0737],[67,0.0713],[69,0.0658],[74,0.0505],[75,0.0476],[76,0.0452],[77,0.0434],[78,0.0423],[79,0.0419],[125,0.0423],[126,0.0425],[127,0.043],[128,0.0448],[129,0.0483],[130,0.0524],[131,0.0572],[132,0.0627],[133,0.0694],[134,0.0824],[135,0.0988],[136,0.1159],[137,0.1309],[138,0.1137],[139,0.2457],[140,0.4739],[141,0.5818],[142,0.6129],[143,0.5863],[144,0.5741],[145,0.572],[146,0.5721],[147,0.5177],[148,0.3694],[149,0.1519],[150,0.0265],[151,-0.0062],[152,0.0224],[153,0.0635],[154,0.0939],[155,0.1295],[156,0.1536],[167,0.1562],[219,0.1551],[220,0.1549],[221,0.1532],[222,0.1474],[223,0.1394],[224,0.1286],[225,0.1195],[226,0.1161],[227,0.1312],[228,0.1611],[229,0.1938],[230,0.21],[231,0.2175],[232,0.2195],[240,0.2201],[241,0.2199],[242,0.219],[243,0.2171],[244,0.1853],[245,0.13],[246,0.0758],[247,0.0413],[248,0.0417],[249,0.0559],[250,0.0743],[251,0.0852],[252,0.0895],[253,0.0901],[254,0.0885],[255,0.0856],[257,0.0806],[258,0.0785],[259,0.0773],[260,0.0767],[272,0.0768],[274,0.0762],[298,0.0761],[300,0.0756],[301,0.0751],[306,0.071],[307,0.0704],[308,0.0701],[310,0.0701],[318,0.0725],[348,0.0731],[361,0.0761],[370,0.0763],[375,0.0754],[377,0.0744],[381,0.0715],[382,0.0709],[384,0.0705],[386,0.0708],[394,0.0736],[401,0.0744],[422,0.0735]]
       }
       spine_02_ctrl.Rotation.X {
        type=MovieSceneScriptingFloatChannel
        default=-1.6952
        thinned_from=423
        keys [[0,-1.6952],[2,-1.6952],[4,-1.7014],[5,-1.7014],[6,-1.7053],[12,-1.7112],[14,-1.7052],[15,-1.696],[16,-1.681],[17,-1.6575],[18,-1.6225],[19,-1.5625],[20,-1.4927],[21,-1.4549],[22,-1.4506],[44,-1.5042],[58,-1.5089],[62,-1.4969],[66,-1.4597],[68,-1.4387],[71,-1.4194],[73,-1.4141],[75,-1.4153],[76,-1.4212],[77,-1.4464],[79,-1.5361],[81,-1.6157],[82,-1.6472],[83,-1.6704],[84,-1.6887],[85,-1.7027],[88,-1.727],[94,-1.7485],[109,-1.7342],[125,-1.7215],[126,-1.7257],[127,-1.7379],[128,-1.7821],[129,-1.8605],[130,-1.949],[131,-2.044],[134,-2.3447],[136,-2.5259],[137,-2.6049],[138,-2.6393],[139,-2.6442],[140,-2.5949],[141,-2.6048],[144,-2.6992],[146,-2.7416],[147,-2.7847],[148,-2.8732],[149,-3.0007],[150,-3.0867],[151,-3.0547],[152,-2.9856],[153,-2.9567],[155,-2.9405],[156,-2.9503],[159,-3.0246],[161,-3.0611],[165,-3.0976],[202,-3.1193],[213,-3.0812],[223,-2.9759],[225,-2.9621],[226,-2.9497],[229,-2.8922],[230,-2.8897],[231,-2.9006],[234,-2.965],[235,-2.9769],[242,-2.9944],[243,-2.9829],[244,-2.943],[245,-2.8415],[246,-2.6951],[247,-2.4987],[249,-2.044],[250,-1.8501],[251,-1.7204],[252,-1.6583],[253,-1.6592],[254,-1.6785],[257,-1.7452],[261,-1.7985],[262,-1.8088],[265,-1.8197],[267,-1.8204],[269,-1.8119],[272,-1.7903],[275,-1.756],[281,-1.7322],[287,-1.7274],[294,-1.7668],[296,-1.7718],[298,-1.7681],[299,-1.761],[300,-1.7493],[301,-1.7289],[302,-1.6991],[303,-1.6645],[304,-1.6248],[307,-1.4911],[308,-1.4438],[309,-1.4028],[310,-1.3757],[311,-1.363],[312,-1.3589],[313,-1.3612],[314,-1.3676],[316,-1.3882],[319,-1.4261],[320,-1.4351],[324,-1.4479],[325,-1.4479],[327,-1.4521],[342,-1.4377],[344,-1.4452],[345,-1.452],[346,-1.462],[347,-1.4765],[348,-1.4962],[349,-1.5213],[350,-1.5521],[351,-1.5909],[353,-1.6815],[354,-1.7214],[356,-1.7864],[357,-1.8126],[358,-1.8323],[360,-1.8584],[362,-1.8745],[367,-1.8863],[371,-1.8852],[374,-1.8704],[376,-1.8531],[380,-1.7979],[385,-1.7172],[387,-1.6925],[389,-1.6769],[391,-1.6709],[392,-1.6743],[399,-1.7364],[406,-1.7474],[418,-1.6966],[422,-1.6966]]
       }
       spine_02_ctrl.Rotation.Y {
        type=MovieSceneScriptingFloatChannel
        default=-0.9514
        thinned_from=423
        keys [[0,-0.9514],[12,-0.9489],[14,-0.9518],[18,-0.9679],[20,-0.9822],[22,-0.9853],[23,-0.9919],[24,-1.0011],[26,-1.0261],[31,-1.1036],[32,-1.1168],[33,-1.1274],[34,-1.1341],[35,-1.1377],[36,-1.1378],[37,-1.1342],[40,-1.1084],[41,-1.1013],[44,-1.092],[45,-1.092],[46,-1.0899],[57,-1.0889],[61,-1.1064],[62,-1.1077],[64,-1.1038],[66,-1.0944],[69,-1.0739],[70,-1.0644],[71,-1.0494],[72,-1.0287],[73,-1.0028],[74,-0.9668],[75,-0.9208],[77,-0.8219],[78,-0.7678],[79,-0.7116],[80,-0.652],[81,-0.5901],[82,-0.5358],[83,-0.4908],[84,-0.4518],[85,-0.4205],[86,-0.3964],[87,-0.3786],[88,-0.366],[89,-0.3576],[90,-0.3531],[91,-0.3524],[92,-0.3534],[93,-0.3555],[98,-0.3707],[100,-0.3737],[115,-0.3723],[117,-0.3731],[118,-0.3766],[119,-0.3793],[120,-0.383],[122,-0.3926],[124,-0.4041],[125,-0.4107],[126,-0.424],[127,-0.4485],[128,-0.4835],[129,-0.5285],[130,-0.589],[131,-0.6648],[132,-0.7628],[133,-0.9033],[134,-1.081],[135,-1.2675],[136,-1.4402],[137,-1.6181],[138,-1.6753],[139,-1.6759],[140,-1.586],[141,-1.4774],[142,-1.4059],[143,-1.3887],[144,-1.3793],[145,-1.3831],[146,-1.3977],[147,-1.4274],[148,-1.4738],[149,-1.5262],[150,-1.5351],[151,-1.5068],[152,-1.4643],[153,-1.4051],[154,-1.3566],[155,-1.3414],[156,-1.3396],[158,-1.3554],[161,-1.3964],[165,-1.4395],[169,-1.4663],[172,-1.4766],[177,-1.4789],[222,-1.4603],[224,-1.4632],[225,-1.4623],[226,-1.4569],[227,-1.4422],[228,-1.4191],[230,-1.3656],[231,-1.3415],[232,-1.3237],[233,-1.3128],[234,-1.308],[237,-1.3089],[239,-1.3122],[241,-1.3036],[242,-1.2965],[243,-1.2867],[244,-1.2722],[245,-1.2405],[246,-1.1879],[247,-1.1175],[248,-1.0286],[249,-0.9624],[250,-0.9285],[251,-0.9188],[252,-0.9299],[254,-0.9701],[256,-0.9997],[257,-1.0122],[258,-1.0215],[259,-1.027],[261,-1.031],[264,-1.029],[280,-0.9735],[281,-0.9712],[282,-0.9712],[283,-0.9675],[284,-0.9675],[285,-0.9654],[288,-0.9655],[295,-0.9804],[296,-0.9804],[297,-0.9772],[298,-0.9691],[299,-0.9565],[300,-0.9401],[301,-0.9106],[302,-0.8635],[303,-0.8079],[304,-0.7463],[305,-0.6768],[306,-0.6014],[307,-0.5145],[308,-0.4222],[309,-0.3439],[310,-0.282],[311,-0.2275],[312,-0.1822],[313,-0.148],[314,-0.1203],[315,-0.0946],[316,-0.0712],[317,-0.0505],[318,-0.0334],[319,-0.0205],[320,-0.0138],[321,-0.0093],[322,-0.004],[323,0.0002],[324,0.0016],[325,0.0016],[326,-0.0035],[327,-0.0072],[328,-0.0105],[329,-0.0135],[330,-0.0155],[331,-0.0171],[333,-0.0209],[334,-0.0219],[335,-0.0226],[338,-0.0256],[340,-0.0271],[342,-0.0271],[343,-0.029],[344,-0.0294],[345,-0.0294],[346,-0.0292],[347,-0.0283],[348,-0.0245],[349,-0.0151],[350,0.0025],[351,0.0416],[352,0.104],[353,0.1754],[354,0.2416],[355,0.3038],[356,0.3679],[357,0.4267],[358,0.4728],[359,0.5062],[360,0.5324],[361,0.5527],[362,0.5684],[363,0.5807],[364,0.5902],[365,0.5971],[366,0.6015],[367,0.6038],[369,0.6026],[371,0.5958],[372,0.5897],[373,0.581],[374,0.5688],[375,0.5528],[376,0.5259],[377,0.4833],[378,0.4239],[379,0.3478],[380,0.2528],[381,0.1405],[382,0.0213],[383,-0.0947],[384,-0.2064],[385,-0.316],[386,-0.4231],[387,-0.5265],[388,-0.6209],[389,-0.706],[390,-0.7814],[391,-0.8469],[392,-0.8991],[393,-0.9376],[394,-0.967],[395,-0.9893],[396,-1.0055],[397,-1.0161],[398,-1.0221],[399,-1.0244],[401,-1.0234],[405,-1.0111],[412,-0.9664],[415,-0.9562],[418,-0.9517],[422,-0.9517]]
       }
       spine_02_ctrl.Rotation.Z {
        type=MovieSceneScriptingFloatChannel
        default=-0.1537
        thinned_from=423
        keys [[0,-0.1537],[2,-0.1537],[3,-0.1348],[4,-0.1244],[5,-0.1244],[6,-0.1071],[7,-0.1071],[8,-0.0965],[9,-0.0965],[10,-0.09],[12,-0.09],[13,-0.1082],[14,-0.1216],[15,-0.1282],[16,-0.1155],[17,-0.0402],[18,0.1347],[19,0.3682],[20,0.6361],[21,0.9234],[22,1.1794],[23,1.3934],[24,1.5749],[25,1.7023],[26,1.7854],[27,1.8488],[28,1.8928],[29,1.9178],[30,1.9242],[31,1.914],[32,1.8908],[33,1.8576],[34,1.8175],[37,1.6849],[38,1.6465],[39,1.6155],[40,1.5944],[43,1.5529],[44,1.5411],[45,1.5411],[46,1.5207],[47,1.5207],[48,1.5044],[49,1.5044],[50,1.492],[52,1.492],[53,1.48],[57,1.48],[58,1.4766],[59,1.484],[60,1.4976],[61,1.518],[62,1.546],[63,1.5942],[67,1.8635],[68,1.9215],[69,1.9572],[70,1.9719],[71,1.9772],[72,1.9722],[73,1.9436],[74,1.8823],[75,1.7926],[76,1.6514],[77,1.4264],[78,1.1528],[79,0.8452],[80,0.5305],[81,0.2412],[82,-0.0267],[83,-0.2637],[84,-0.4741],[85,-0.66],[86,-0.8213],[87,-0.9568],[88,-1.0722],[89,-1.1734],[90,-1.259],[91,-1.3292],[92,-1.3853],[93,-1.4293],[94,-1.4599],[95,-1.4754],[96,-1.4793],[97,-1.4751],[98,-1.4606],[99,-1.4328],[102,-1.3273],[103,-1.2992],[104,-1.2759],[105,-1.2577],[107,-1.2276],[108,-1.2146],[109,-1.2146],[110,-1.1959],[111,-1.1959],[112,-1.1844],[113,-1.1844],[114,-1.1757],[115,-1.1757],[116,-1.1712],[118,-1.1694],[119,-1.171],[120,-1.1804],[121,-1.2101],[123,-1.2879],[124,-1.3217],[125,-1.3454],[126,-1.357],[127,-1.3419],[128,-1.305],[129,-1.2605],[130,-1.1947],[131,-1.0977],[132,-0.9737],[133,-0.846],[134,-0.6653],[135,-0.4353],[136,-0.1624],[137,0.2568],[138,1.5017],[139,1.7318],[140,1.9017],[141,2.0195],[142,2.0927],[143,2.1212],[144,2.0359],[145,1.8574],[147,1.4215],[148,1.3052],[149,1.3645],[150,1.6721],[151,1.9916],[152,2.2115],[153,2.3234],[154,2.3178],[155,2.2658],[158,2.0434],[159,1.9658],[160,1.8943],[161,1.8394],[162,1.8003],[163,1.768],[164,1.7416],[165,1.72],[167,1.6871],[170,1.6509],[172,1.6335],[173,1.6335],[174,1.6211],[175,1.6211],[176,1.6127],[177,1.6127],[178,1.6079],[197,1.6024],[198,1.617],[200,1.617],[201,1.6289],[202,1.6289],[203,1.6374],[204,1.6374],[205,1.6457],[212,1.6663],[215,1.6663],[224,1.6332],[225,1.6318],[226,1.6392],[227,1.6663],[228,1.7163],[229,1.7965],[230,1.8941],[231,1.9836],[232,2.0521],[233,2.1001],[234,2.1311],[235,2.1511],[236,2.1511],[237,2.14],[238,2.0873],[239,2.0063],[240,1.9023],[241,1.7795],[242,1.6438],[243,1.5014],[244,1.3692],[245,1.2564],[246,1.1479],[247,1.0443],[248,0.9519],[249,0.8666],[250,0.7975],[251,0.7487],[252,0.7136],[253,0.6899],[254,0.6776],[255,0.6723],[256,0.6742],[257,0.687],[258,0.7101],[259,0.742],[260,0.7808],[261,0.8248],[262,0.8724],[265,1.0241],[266,1.0731],[267,1.1191],[268,1.1607],[269,1.1964],[270,1.225],[271,1.2481],[272,1.2683],[273,1.2856],[275,1.3127],[277,1.3312],[279,1.3427],[281,1.3486],[284,1.3505],[285,1.3451],[287,1.3451],[288,1.3397],[291,1.345],[293,1.3565],[294,1.3659],[295,1.3789],[296,1.4032],[297,1.449],[298,1.547],[299,1.6666],[300,1.7263],[301,1.7229],[302,1.7027],[303,1.6582],[304,1.5828],[305,1.4799],[306,1.3576],[307,1.2191],[308,1.0653],[309,0.8954],[310,0.7035],[311,0.4898],[312,0.2615],[313,0.0261],[314,-0.2081],[315,-0.4524],[317,-0.9557],[318,-1.1875],[319,-1.3761],[320,-1.5244],[321,-1.6205],[322,-1.6818],[323,-1.7216],[324,-1.7428],[325,-1.7428],[326,-1.755],[327,-1.7311],[328,-1.6824],[329,-1.6246],[330,-1.5587],[332,-1.4219],[333,-1.3584],[334,-1.2981],[335,-1.2523],[336,-1.2215],[337,-1.1971],[338,-1.1774],[340,-1.1449],[342,-1.1449],[343,-1.1388],[345,-1.1567],[346,-1.1699],[347,-1.1899],[350,-1.2639],[351,-1.2834],[355,-1.3372],[361,-1.4],[364,-1.4243],[365,-1.4295],[366,-1.4316],[367,-1.4263],[368,-1.4263],[369,-1.407],[370,-1.3844],[376,-1.1308],[378,-1.0514],[379,-1.0148],[380,-0.9809],[383,-0.8848],[386,-0.7804],[387,-0.7441],[388,-0.7061],[389,-0.6667],[390,-0.6258],[391,-0.5835],[392,-0.5394],[397,-0.3061],[398,-0.26],[399,-0.2147],[400,-0.1707],[401,-0.1328],[402,-0.1047],[403,-0.0849],[404,-0.0724],[405,-0.0655],[406,-0.0629],[407,-0.0637],[408,-0.0674],[409,-0.0734],[410,-0.0812],[411,-0.0899],[412,-0.0991],[413,-0.108],[414,-0.1174],[415,-0.1275],[416,-0.1373],[417,-0.1457],[418,-0.1515],[422,-0.1515]]
       }
       spine_02_ctrl.Scale.X {type=MovieSceneScriptingFloatChannel, default=1, const [1,423]}
       spine_02_ctrl.Scale.Y {type=MovieSceneScriptingFloatChannel, default=1, const [1,423]}
       spine_02_ctrl.Scale.Z {type=MovieSceneScriptingFloatChannel, default=1, const [1,423]}
       spine_03_ctrl.Location.X {
        type=MovieSceneScriptingFloatChannel
        default=0.1051
        thinned_from=423
        keys [[0,0.1051],[14,0.1053],[15,0.1055],[16,0.1059],[17,0.1067],[18,0.1083],[20,0.1124],[21,0.1142],[22,0.1164],[28,0.1321],[30,0.1367],[31,0.1386],[32,0.1401],[33,0.1412],[34,0.1417],[35,0.1417],[36,0.1412],[37,0.1403],[40,0.1359],[41,0.1348],[44,0.1331],[48,0.1323],[57,0.1322],[61,0.1348],[64,0.1359],[66,0.1361],[68,0.1355],[69,0.1349],[70,0.1341],[71,0.1327],[72,0.1309],[73,0.1287],[74,0.1255],[75,0.1215],[76,0.117],[77,0.1116],[78,0.1056],[79,0.0989],[81,0.0845],[82,0.0778],[83,0.0719],[84,0.0667],[85,0.0624],[86,0.0588],[87,0.056],[88,0.0538],[89,0.052],[90,0.0508],[91,0.05],[92,0.0496],[93,0.0495],[95,0.0498],[101,0.0526],[123,0.0552],[124,0.0555],[125,0.056],[126,0.0571],[127,0.06],[128,0.0642],[129,0.0696],[130,0.0766],[131,0.0851],[132,0.0964],[133,0.1117],[134,0.1302],[135,0.1493],[136,0.1673],[137,0.1896],[138,0.2076],[139,0.2121],[140,0.2079],[141,0.1961],[142,0.1868],[143,0.1833],[144,0.1801],[145,0.1781],[146,0.1774],[147,0.177],[148,0.1773],[149,0.1779],[150,0.1795],[151,0.1798],[152,0.1785],[153,0.1728],[154,0.1661],[155,0.1636],[156,0.1625],[157,0.1622],[158,0.1625],[166,0.1708],[170,0.1733],[176,0.1739],[224,0.1708],[226,0.1695],[229,0.1654],[231,0.1618],[232,0.1603],[233,0.1595],[234,0.1593],[236,0.1597],[238,0.1592],[239,0.1582],[240,0.1563],[241,0.1539],[242,0.1511],[243,0.1477],[244,0.1432],[245,0.137],[246,0.1292],[247,0.1208],[248,0.1119],[249,0.1052],[250,0.1018],[251,0.1006],[252,0.1012],[253,0.1029],[257,0.11],[258,0.1114],[259,0.1125],[262,0.1144],[265,0.1153],[268,0.1154],[273,0.114],[282,0.1109],[286,0.1102],[288,0.1103],[294,0.1123],[299,0.1144],[300,0.1139],[301,0.112],[302,0.1087],[303,0.1046],[304,0.0999],[305,0.0945],[306,0.0885],[307,0.0818],[308,0.0747],[309,0.0683],[310,0.0628],[311,0.0576],[312,0.0528],[313,0.0486],[317,0.0335],[318,0.0301],[319,0.0272],[320,0.0251],[321,0.0235],[322,0.0222],[323,0.0213],[324,0.0209],[325,0.021],[326,0.0214],[327,0.0221],[333,0.0277],[335,0.029],[340,0.0307],[343,0.031],[346,0.0307],[347,0.0304],[348,0.0297],[349,0.0282],[350,0.0256],[351,0.0202],[352,0.0117],[353,0.0021],[354,-0.0068],[355,-0.0152],[356,-0.0239],[357,-0.0318],[358,-0.0381],[359,-0.0426],[360,-0.0462],[361,-0.049],[362,-0.0512],[363,-0.053],[364,-0.0543],[365,-0.0553],[366,-0.0559],[367,-0.0562],[369,-0.0561],[370,-0.0558],[372,-0.0544],[373,-0.0534],[374,-0.0521],[375,-0.0503],[376,-0.0478],[377,-0.0441],[378,-0.039],[379,-0.0322],[380,-0.0237],[381,-0.0138],[383,0.0074],[386,0.0383],[387,0.0489],[388,0.0592],[389,0.069],[390,0.078],[391,0.086],[392,0.0925],[393,0.0976],[394,0.1018],[395,0.1052],[396,0.108],[397,0.1101],[398,0.1116],[399,0.1126],[400,0.1133],[401,0.1137],[403,0.1139],[405,0.1136],[407,0.1123],[412,0.1077],[415,0.106],[418,0.1052],[422,0.1051]]
       }
       spine_03_ctrl.Location.Y {
        type=MovieSceneScriptingFloatChannel
        default=-0.1499
        thinned_from=423
        keys [[0,-0.1499],[10,-0.152],[13,-0.1518],[14,-0.1512],[15,-0.15],[16,-0.148],[17,-0.1448],[18,-0.14],[19,-0.1319],[20,-0.1224],[21,-0.1172],[22,-0.1165],[23,-0.1164],[40,-0.1228],[48,-0.124],[58,-0.1242],[60,-0.1237],[62,-0.1226],[63,-0.1216],[65,-0.1189],[66,-0.1178],[67,-0.1169],[68,-0.1164],[69,-0.1161],[71,-0.1164],[72,-0.1169],[73,-0.1176],[74,-0.1186],[75,-0.1198],[76,-0.1215],[77,-0.1255],[79,-0.1381],[80,-0.1436],[81,-0.1488],[82,-0.1529],[83,-0.1559],[84,-0.1582],[85,-0.16],[87,-0.1622],[91,-0.1647],[96,-0.1659],[121,-0.1619],[125,-0.1622],[126,-0.1625],[127,-0.1634],[128,-0.1664],[129,-0.1717],[130,-0.1777],[131,-0.1843],[133,-0.1987],[134,-0.2044],[135,-0.209],[137,-0.22],[138,-0.2154],[139,-0.2373],[140,-0.2712],[141,-0.2917],[142,-0.3011],[143,-0.3011],[144,-0.303],[146,-0.309],[147,-0.305],[148,-0.2892],[149,-0.2641],[150,-0.2503],[151,-0.2395],[152,-0.236],[154,-0.2447],[156,-0.2568],[159,-0.267],[160,-0.2699],[161,-0.2722],[164,-0.2765],[167,-0.2784],[203,-0.2801],[212,-0.276],[215,-0.2731],[220,-0.2652],[222,-0.2607],[225,-0.2521],[226,-0.2498],[227,-0.2502],[228,-0.2532],[229,-0.2571],[234,-0.2717],[235,-0.2734],[239,-0.2745],[242,-0.2764],[243,-0.2759],[244,-0.2694],[245,-0.2551],[246,-0.2357],[247,-0.2133],[248,-0.193],[249,-0.1749],[250,-0.161],[251,-0.1526],[252,-0.1489],[253,-0.1498],[257,-0.1571],[262,-0.1646],[263,-0.1653],[266,-0.166],[267,-0.1659],[268,-0.1655],[272,-0.1617],[275,-0.157],[276,-0.156],[281,-0.1538],[286,-0.1531],[287,-0.1533],[289,-0.1546],[292,-0.1572],[295,-0.1588],[297,-0.159],[298,-0.1585],[299,-0.1575],[300,-0.1559],[301,-0.1533],[302,-0.1494],[303,-0.1449],[304,-0.1398],[307,-0.1228],[308,-0.1168],[309,-0.1115],[310,-0.108],[311,-0.1064],[312,-0.1057],[313,-0.1059],[314,-0.1065],[315,-0.1075],[319,-0.1129],[320,-0.114],[321,-0.1145],[327,-0.116],[341,-0.1148],[343,-0.1152],[344,-0.1158],[345,-0.1167],[346,-0.118],[347,-0.1199],[348,-0.1225],[349,-0.1258],[350,-0.1298],[351,-0.1349],[353,-0.1468],[354,-0.152],[355,-0.1565],[356,-0.1606],[357,-0.1641],[358,-0.1667],[359,-0.1686],[361,-0.1714],[363,-0.173],[370,-0.174],[372,-0.1735],[375,-0.171],[376,-0.1698],[379,-0.1645],[383,-0.1561],[386,-0.1507],[388,-0.1484],[389,-0.1475],[390,-0.147],[391,-0.1469],[392,-0.1474],[398,-0.1546],[399,-0.1556],[401,-0.1566],[406,-0.1569],[408,-0.1564],[417,-0.1506],[419,-0.1499],[422,-0.1499]]
       }
       spine_03_ctrl.Location.Z {
        type=MovieSceneScriptingFloatChannel
        default=0.0893
        thinned_from=423
        keys [[0,0.0893],[17,0.0895],[31,0.0925],[65,0.0923],[66,0.0909],[67,0.0877],[69,0.0802],[74,0.0597],[75,0.0558],[76,0.0524],[77,0.0499],[78,0.0482],[79,0.0474],[81,0.0469],[125,0.047],[126,0.0471],[127,0.0477],[128,0.05],[129,0.0542],[130,0.0593],[131,0.0652],[132,0.0719],[133,0.0803],[134,0.0972],[135,0.1189],[136,0.1414],[137,0.1609],[138,0.1373],[139,0.315],[140,0.6225],[141,0.7679],[142,0.8097],[143,0.7737],[144,0.7571],[145,0.7541],[146,0.7541],[147,0.6807],[148,0.4804],[149,0.1869],[150,0.0176],[151,-0.0262],[152,0.0129],[153,0.0685],[154,0.1097],[155,0.1577],[156,0.19],[220,0.1911],[221,0.1889],[222,0.1811],[223,0.1706],[224,0.156],[225,0.1438],[226,0.1392],[227,0.1598],[228,0.2004],[229,0.2445],[230,0.2663],[231,0.2763],[232,0.2789],[240,0.2791],[241,0.2788],[242,0.2778],[243,0.2752],[244,0.2326],[245,0.1587],[246,0.0866],[247,0.0414],[248,0.0434],[249,0.0638],[250,0.0898],[251,0.1051],[252,0.1113],[253,0.112],[254,0.1097],[255,0.1055],[257,0.0985],[258,0.0956],[259,0.0938],[260,0.0929],[272,0.0931],[274,0.0925],[298,0.0922],[300,0.0918],[301,0.0912],[305,0.0877],[306,0.0871],[307,0.0867],[308,0.0866],[310,0.0872],[317,0.0903],[319,0.0908],[353,0.0911],[366,0.0924],[372,0.092],[375,0.0912],[376,0.0907],[381,0.0865],[382,0.0859],[383,0.0856],[384,0.0856],[386,0.0861],[392,0.089],[397,0.09],[422,0.0893]]
       }
       spine_03_ctrl.Rotation.X {
        type=MovieSceneScriptingFloatChannel
        default=-1.9107
        thinned_from=423
        keys [[0,-1.9107],[1,-1.9107],[10,-1.9437],[12,-1.9437],[13,-1.9405],[14,-1.9311],[15,-1.9129],[16,-1.8834],[17,-1.8386],[18,-1.7735],[19,-1.6606],[20,-1.5296],[21,-1.4625],[22,-1.4613],[23,-1.4691],[28,-1.5198],[38,-1.5712],[48,-1.5893],[58,-1.5917],[60,-1.5851],[62,-1.5718],[63,-1.5595],[65,-1.5257],[66,-1.5152],[67,-1.5123],[68,-1.5144],[69,-1.5203],[71,-1.5419],[73,-1.5738],[75,-1.6148],[76,-1.6394],[77,-1.6943],[78,-1.7808],[79,-1.8632],[81,-2.0012],[82,-2.0551],[83,-2.0939],[84,-2.1242],[85,-2.1469],[87,-2.1748],[91,-2.2069],[97,-2.2229],[121,-2.165],[125,-2.1665],[126,-2.1698],[127,-2.1795],[128,-2.2119],[129,-2.2692],[130,-2.3349],[131,-2.4067],[133,-2.5606],[134,-2.598],[135,-2.6146],[136,-2.6485],[137,-2.7021],[138,-2.7399],[139,-2.8042],[140,-2.8456],[141,-2.9218],[142,-2.9902],[143,-3.0408],[144,-3.0839],[145,-3.1197],[146,-3.1482],[147,-3.1888],[148,-3.2563],[149,-3.3459],[150,-3.427],[151,-3.3531],[152,-3.2526],[153,-3.23],[154,-3.2286],[155,-3.2384],[156,-3.2718],[159,-3.4094],[160,-3.4481],[161,-3.478],[164,-3.536],[166,-3.5573],[203,-3.5846],[211,-3.5369],[214,-3.5019],[218,-3.4201],[227,-3.1983],[228,-3.1818],[229,-3.1756],[230,-3.1856],[231,-3.2148],[233,-3.3063],[234,-3.3465],[235,-3.3711],[242,-3.3919],[243,-3.3826],[244,-3.3603],[245,-3.2827],[246,-3.1265],[247,-2.8806],[248,-2.5925],[249,-2.3104],[250,-2.0851],[251,-1.9532],[252,-1.8978],[253,-1.9107],[257,-2.0338],[261,-2.1365],[262,-2.1582],[263,-2.1712],[266,-2.1865],[267,-2.1871],[268,-2.1825],[272,-2.1304],[275,-2.0627],[276,-2.0484],[281,-2.0153],[286,-2.005],[287,-2.008],[289,-2.0271],[292,-2.0665],[295,-2.0919],[297,-2.0974],[298,-2.094],[299,-2.0841],[300,-2.0634],[301,-2.024],[302,-1.9663],[303,-1.8985],[304,-1.82],[306,-1.6441],[307,-1.5535],[308,-1.4588],[309,-1.3751],[310,-1.3173],[311,-1.2862],[312,-1.2709],[313,-1.2675],[314,-1.2721],[315,-1.282],[316,-1.296],[319,-1.35],[320,-1.3639],[321,-1.3724],[326,-1.3909],[330,-1.3921],[341,-1.3793],[343,-1.3854],[344,-1.3937],[345,-1.4069],[346,-1.4261],[347,-1.4539],[348,-1.4921],[349,-1.5412],[350,-1.6018],[351,-1.6795],[353,-1.864],[354,-1.9457],[355,-2.016],[356,-2.0805],[357,-2.1354],[358,-2.1768],[359,-2.2072],[360,-2.2319],[361,-2.2512],[363,-2.2761],[367,-2.2915],[371,-2.2896],[373,-2.2731],[375,-2.2449],[376,-2.2256],[379,-2.1459],[383,-2.0142],[385,-1.9541],[386,-1.927],[387,-1.9045],[389,-1.8715],[390,-1.8616],[391,-1.8579],[392,-1.8644],[393,-1.8804],[398,-1.9749],[399,-1.9908],[400,-2.0015],[403,-2.0142],[407,-2.0141],[414,-1.9538],[417,-1.9207],[418,-1.9134],[422,-1.9107]]
       }
       spine_03_ctrl.Rotation.Y {
        type=MovieSceneScriptingFloatChannel
        default=-1.5486
        thinned_from=423
        keys [[0,-1.5486],[14,-1.5494],[17,-1.5599],[19,-1.5738],[21,-1.5779],[22,-1.5882],[23,-1.6077],[24,-1.6331],[26,-1.6945],[30,-1.8279],[31,-1.8572],[32,-1.8819],[33,-1.9007],[34,-1.9112],[35,-1.915],[36,-1.9117],[37,-1.9012],[40,-1.8414],[41,-1.8259],[43,-1.8094],[48,-1.7948],[56,-1.7929],[57,-1.7952],[61,-1.8318],[63,-1.8386],[65,-1.8354],[67,-1.8227],[69,-1.8029],[70,-1.7896],[71,-1.7712],[72,-1.7478],[73,-1.7201],[74,-1.684],[75,-1.6395],[76,-1.5907],[77,-1.5375],[78,-1.4783],[79,-1.4136],[80,-1.3438],[81,-1.2705],[82,-1.2024],[83,-1.1424],[84,-1.089],[85,-1.0459],[86,-1.0114],[87,-0.9845],[88,-0.9641],[89,-0.9487],[90,-0.9388],[91,-0.935],[92,-0.9348],[93,-0.9371],[99,-0.9717],[101,-0.9777],[116,-0.9823],[119,-0.9881],[122,-1.0001],[124,-1.012],[125,-1.0208],[126,-1.0389],[127,-1.0801],[128,-1.1399],[129,-1.2168],[130,-1.3134],[131,-1.4285],[132,-1.5812],[133,-1.7893],[135,-2.269],[136,-2.4847],[137,-2.7558],[138,-2.8956],[139,-2.8582],[140,-2.6597],[141,-2.4239],[142,-2.266],[143,-2.2206],[144,-2.1886],[145,-2.1803],[146,-2.1942],[147,-2.2359],[148,-2.3112],[149,-2.406],[150,-2.4525],[151,-2.4403],[152,-2.3868],[154,-2.1462],[155,-2.0963],[156,-2.0747],[157,-2.0758],[158,-2.0868],[165,-2.2325],[168,-2.2738],[171,-2.2987],[176,-2.3061],[223,-2.2714],[224,-2.2733],[225,-2.2702],[226,-2.2593],[227,-2.2332],[228,-2.1938],[231,-2.0625],[232,-2.0309],[233,-2.0114],[234,-2.0035],[238,-2.0063],[239,-2.0004],[240,-1.9841],[241,-1.962],[242,-1.9352],[243,-1.9024],[244,-1.8614],[245,-1.8019],[246,-1.7216],[247,-1.6252],[248,-1.5136],[249,-1.427],[250,-1.3825],[251,-1.3684],[252,-1.3789],[253,-1.4035],[256,-1.486],[257,-1.511],[258,-1.5302],[259,-1.5428],[261,-1.5554],[263,-1.5602],[266,-1.5545],[281,-1.4667],[286,-1.4536],[288,-1.4545],[296,-1.4928],[298,-1.4949],[299,-1.4907],[300,-1.4809],[301,-1.4583],[302,-1.4213],[303,-1.3768],[304,-1.3267],[305,-1.2696],[306,-1.2073],[307,-1.1362],[308,-1.061],[309,-0.9954],[310,-0.9406],[311,-0.89],[312,-0.8454],[313,-0.808],[316,-0.7093],[317,-0.6775],[318,-0.648],[319,-0.6226],[320,-0.6037],[322,-0.5731],[323,-0.5621],[324,-0.5579],[325,-0.5605],[326,-0.5674],[329,-0.597],[333,-0.6306],[334,-0.6369],[337,-0.6496],[341,-0.6602],[344,-0.6631],[346,-0.6619],[347,-0.6596],[348,-0.6515],[349,-0.6324],[350,-0.5972],[351,-0.5198],[352,-0.3962],[353,-0.2547],[354,-0.1234],[355,0.0004],[356,0.1283],[357,0.2456],[358,0.3379],[359,0.405],[360,0.4577],[361,0.4986],[362,0.5303],[363,0.5553],[364,0.5748],[365,0.5889],[366,0.5982],[367,0.6035],[368,0.6056],[369,0.6051],[370,0.6027],[371,0.5986],[372,0.5924],[373,0.583],[374,0.5694],[375,0.5506],[376,0.5211],[377,0.4767],[378,0.4117],[379,0.3232],[380,0.2115],[381,0.0799],[382,-0.062],[383,-0.2046],[384,-0.3453],[385,-0.4833],[386,-0.624],[387,-0.7707],[388,-0.9156],[389,-1.0542],[390,-1.1825],[391,-1.2962],[392,-1.3888],[393,-1.4609],[394,-1.5198],[395,-1.5673],[396,-1.6047],[397,-1.6325],[398,-1.652],[399,-1.6642],[400,-1.6713],[402,-1.6757],[404,-1.6728],[405,-1.6686],[407,-1.6497],[412,-1.5831],[415,-1.5603],[418,-1.5494],[422,-1.5486]]
       }
       spine_03_ctrl.Rotation.Z {
        type=MovieSceneScriptingFloatChannel
        default=-0.2975
        thinned_from=423
        keys [[0,-0.2975],[1,-0.2975],[3,-0.2725],[4,-0.2588],[5,-0.2459],[6,-0.236],[7,-0.2286],[8,-0.222],[9,-0.2169],[10,-0.2135],[12,-0.2135],[13,-0.2374],[14,-0.255],[15,-0.2635],[16,-0.2461],[17,-0.1453],[18,0.0884],[19,0.4019],[20,0.7632],[21,1.1494],[22,1.4909],[23,1.7752],[24,2.0154],[25,2.1828],[26,2.291],[27,2.3727],[28,2.4284],[29,2.459],[30,2.465],[31,2.4489],[32,2.4159],[33,2.3699],[34,2.3156],[37,2.1392],[38,2.0894],[39,2.0496],[40,2.023],[44,1.9547],[47,1.9165],[48,1.9064],[49,1.9064],[50,1.8899],[51,1.8899],[52,1.8782],[56,1.8782],[57,1.8664],[58,1.8687],[59,1.8779],[60,1.8954],[61,1.9219],[62,1.9591],[63,2.0235],[67,2.3868],[68,2.466],[69,2.5154],[70,2.5365],[71,2.5448],[72,2.5388],[73,2.5007],[74,2.4183],[75,2.2974],[76,2.1074],[77,1.8052],[78,1.4379],[79,1.0266],[80,0.607],[81,0.2223],[82,-0.1327],[83,-0.4456],[84,-0.7231],[85,-0.968],[86,-1.1806],[87,-1.3588],[88,-1.5106],[89,-1.6437],[90,-1.7563],[91,-1.8487],[92,-1.9228],[93,-1.9811],[94,-2.0219],[95,-2.0427],[96,-2.0483],[97,-2.0433],[98,-2.0246],[99,-1.9881],[102,-1.8494],[103,-1.8124],[104,-1.7816],[105,-1.7576],[107,-1.7179],[109,-1.6871],[111,-1.6678],[114,-1.6496],[118,-1.6413],[119,-1.6435],[120,-1.6558],[121,-1.6952],[123,-1.798],[124,-1.8429],[125,-1.8746],[126,-1.8909],[127,-1.874],[128,-1.8322],[129,-1.7849],[130,-1.7117],[131,-1.5995],[133,-1.3133],[134,-1.1051],[135,-0.8309],[136,-0.4973],[137,0.0162],[138,1.6282],[139,1.9409],[140,2.1999],[141,2.3898],[142,2.5074],[143,2.5497],[144,2.4412],[145,2.2079],[146,1.9217],[147,1.6275],[148,1.4595],[149,1.5137],[150,1.8957],[151,2.3128],[152,2.6128],[153,2.7825],[154,2.7989],[155,2.7418],[156,2.6495],[158,2.4495],[159,2.3426],[160,2.2434],[161,2.1669],[162,2.1115],[163,2.0651],[164,2.0265],[165,1.9946],[166,1.9684],[168,1.9268],[170,1.8939],[172,1.8693],[175,1.8463],[196,1.8316],[197,1.8461],[198,1.8461],[199,1.8555],[211,1.9159],[215,1.9218],[224,1.8931],[225,1.8932],[226,1.9063],[227,1.9474],[228,2.0197],[229,2.1313],[230,2.2647],[231,2.3856],[232,2.4767],[233,2.539],[234,2.5776],[235,2.6016],[236,2.611],[237,2.5868],[238,2.5184],[239,2.4142],[240,2.2818],[241,2.1264],[242,1.956],[243,1.7793],[244,1.6214],[245,1.4952],[246,1.3755],[247,1.2619],[248,1.1603],[249,1.0652],[250,0.9864],[251,0.9297],[252,0.8869],[253,0.8552],[254,0.8369],[255,0.8267],[256,0.826],[257,0.8399],[258,0.8682],[259,0.9088],[260,0.959],[261,1.0161],[262,1.0784],[266,1.3438],[267,1.405],[268,1.4605],[269,1.5086],[270,1.5474],[271,1.5789],[272,1.6066],[273,1.6307],[275,1.6691],[277,1.6954],[279,1.7119],[282,1.723],[284,1.7222],[287,1.71],[290,1.7118],[292,1.7204],[294,1.7404],[295,1.7568],[296,1.7881],[297,1.8479],[298,1.9768],[299,2.1345],[300,2.2131],[301,2.2078],[302,2.1795],[303,2.1186],[304,2.0163],[305,1.8768],[306,1.7108],[307,1.5217],[308,1.3115],[309,1.0802],[310,0.8206],[311,0.5321],[312,0.225],[313,-0.0905],[314,-0.4035],[315,-0.7297],[317,-1.3996],[318,-1.7074],[319,-1.957],[320,-2.1526],[321,-2.2789],[322,-2.3589],[323,-2.4107],[324,-2.4384],[325,-2.4522],[326,-2.4553],[327,-2.4243],[328,-2.3604],[329,-2.2845],[330,-2.1977],[332,-2.0175],[333,-1.934],[334,-1.8543],[335,-1.794],[336,-1.7534],[337,-1.7214],[338,-1.6955],[340,-1.6528],[341,-1.6409],[342,-1.6394],[343,-1.645],[345,-1.6684],[346,-1.6856],[347,-1.7114],[350,-1.8016],[351,-1.8194],[352,-1.8269],[358,-1.8502],[364,-1.8999],[366,-1.9072],[367,-1.8997],[369,-1.8736],[370,-1.8435],[376,-1.509],[377,-1.4551],[378,-1.4042],[379,-1.3571],[380,-1.3139],[383,-1.1937],[388,-0.9813],[390,-0.8914],[391,-0.843],[392,-0.791],[393,-0.7356],[395,-0.6205],[399,-0.3859],[400,-0.3289],[401,-0.2795],[402,-0.2427],[403,-0.2169],[404,-0.2005],[405,-0.1914],[406,-0.1875],[407,-0.1878],[408,-0.1917],[409,-0.1985],[410,-0.2074],[411,-0.2178],[413,-0.2398],[414,-0.2514],[416,-0.2766],[417,-0.2872],[418,-0.2947],[419,-0.2975],[422,-0.2975]]
       }
       spine_03_ctrl.Scale.X {type=MovieSceneScriptingFloatChannel, default=1, const [1,423]}
       spine_03_ctrl.Scale.Y {type=MovieSceneScriptingFloatChannel, default=1, const [1,423]}
       spine_03_ctrl.Scale.Z {type=MovieSceneScriptingFloatChannel, default=1, const [1,423]}
       spine_04_ctrl.Location.X {
        type=MovieSceneScriptingFloatChannel
        default=0.1927
        thinned_from=423
        keys [[0,0.1927],[11,0.1941],[15,0.1933],[16,0.1937],[17,0.1956],[18,0.1997],[20,0.2098],[21,0.2153],[22,0.222],[25,0.2442],[27,0.257],[29,0.2684],[30,0.2732],[31,0.2772],[32,0.2802],[33,0.2821],[34,0.2826],[35,0.2821],[36,0.2805],[37,0.2779],[40,0.267],[41,0.2644],[43,0.2613],[47,0.2584],[56,0.2571],[57,0.2573],[58,0.2584],[61,0.2637],[65,0.2695],[67,0.2708],[68,0.2707],[69,0.2701],[71,0.2669],[72,0.2646],[73,0.2618],[74,0.258],[75,0.2533],[76,0.2473],[77,0.2389],[78,0.229],[79,0.2175],[81,0.193],[82,0.1811],[83,0.1701],[84,0.1602],[85,0.1519],[86,0.1448],[87,0.1389],[88,0.1341],[89,0.13],[90,0.1268],[91,0.1247],[92,0.1234],[93,0.1227],[94,0.1226],[95,0.1231],[97,0.125],[101,0.1301],[103,0.1318],[108,0.1339],[120,0.1354],[123,0.1341],[124,0.134],[125,0.1344],[126,0.1361],[127,0.1417],[128,0.1499],[129,0.1603],[130,0.173],[131,0.1882],[132,0.2087],[133,0.2351],[134,0.2646],[135,0.2936],[136,0.3211],[137,0.3631],[138,0.4119],[139,0.4179],[140,0.406],[141,0.3782],[142,0.3565],[144,0.3394],[145,0.3325],[146,0.3281],[147,0.325],[148,0.3246],[149,0.3276],[150,0.3367],[151,0.3439],[152,0.3455],[153,0.3333],[154,0.3164],[155,0.3085],[156,0.3036],[157,0.3011],[158,0.3004],[159,0.301],[162,0.3055],[166,0.3136],[170,0.3184],[173,0.3196],[194,0.3152],[216,0.3165],[224,0.3136],[226,0.3111],[229,0.3043],[232,0.2962],[233,0.2947],[234,0.2945],[235,0.2954],[236,0.2956],[237,0.2951],[238,0.2936],[239,0.2902],[240,0.2847],[241,0.2777],[242,0.2696],[243,0.2604],[244,0.2492],[245,0.2362],[247,0.208],[248,0.1949],[249,0.1844],[250,0.1788],[251,0.1764],[252,0.1765],[253,0.1784],[254,0.1813],[257,0.1923],[258,0.1954],[259,0.198],[262,0.2034],[264,0.206],[267,0.2082],[269,0.2086],[273,0.207],[284,0.2002],[287,0.1995],[289,0.2001],[293,0.2032],[295,0.2056],[297,0.2091],[299,0.2155],[300,0.2169],[301,0.2162],[302,0.2146],[303,0.2122],[304,0.2087],[305,0.2044],[306,0.1995],[308,0.1884],[309,0.1824],[310,0.176],[312,0.1618],[314,0.1468],[315,0.1389],[316,0.1306],[317,0.122],[318,0.1137],[319,0.1065],[320,0.1005],[321,0.0959],[322,0.0923],[323,0.0898],[324,0.0886],[325,0.0888],[326,0.0897],[327,0.0916],[328,0.0943],[332,0.1057],[333,0.1084],[334,0.1108],[335,0.1126],[336,0.1139],[338,0.116],[340,0.1176],[341,0.1181],[343,0.1184],[345,0.118],[346,0.1175],[347,0.1167],[348,0.115],[349,0.1116],[350,0.106],[351,0.0944],[352,0.0763],[353,0.0556],[354,0.0365],[355,0.0184],[356,-0.0003],[357,-0.0174],[358,-0.0309],[359,-0.0408],[360,-0.0486],[361,-0.0547],[362,-0.0595],[363,-0.0633],[364,-0.0662],[365,-0.0684],[366,-0.0698],[367,-0.0705],[368,-0.0707],[369,-0.0706],[370,-0.0701],[372,-0.0677],[373,-0.066],[374,-0.064],[375,-0.0614],[376,-0.058],[377,-0.0539],[378,-0.0477],[379,-0.039],[380,-0.0281],[381,-0.0154],[382,-0.0016],[383,0.013],[384,0.0277],[385,0.0422],[386,0.0577],[387,0.075],[389,0.1116],[390,0.1291],[391,0.145],[392,0.1582],[393,0.169],[394,0.1784],[395,0.1863],[396,0.1929],[397,0.1983],[398,0.2025],[399,0.2057],[400,0.2081],[401,0.2098],[402,0.2109],[404,0.2118],[405,0.2115],[406,0.2106],[408,0.2073],[412,0.199],[415,0.195],[417,0.1933],[419,0.1927],[422,0.1927]]
       }
       spine_04_ctrl.Location.Y {
        type=MovieSceneScriptingFloatChannel
        default=-0.176
        thinned_from=423
        keys [[0,-0.176],[2,-0.1766],[9,-0.1803],[11,-0.1808],[12,-0.1807],[13,-0.1802],[14,-0.1789],[15,-0.1763],[16,-0.172],[17,-0.1654],[18,-0.1556],[19,-0.1389],[20,-0.1195],[21,-0.1094],[22,-0.1088],[23,-0.1097],[27,-0.1154],[29,-0.1177],[36,-0.1231],[40,-0.1252],[52,-0.1273],[57,-0.1275],[59,-0.127],[61,-0.1255],[62,-0.1244],[63,-0.1226],[65,-0.1176],[66,-0.1164],[67,-0.1168],[68,-0.1181],[69,-0.12],[70,-0.1226],[71,-0.1257],[72,-0.1291],[73,-0.133],[74,-0.1373],[76,-0.1468],[77,-0.1558],[78,-0.169],[79,-0.1812],[80,-0.1915],[81,-0.2013],[82,-0.209],[83,-0.2145],[84,-0.2187],[85,-0.2219],[86,-0.2241],[89,-0.2282],[93,-0.2314],[97,-0.2324],[120,-0.2242],[126,-0.2246],[128,-0.2259],[129,-0.2274],[131,-0.2318],[133,-0.2376],[134,-0.233],[135,-0.2254],[136,-0.2214],[137,-0.2217],[138,-0.2141],[139,-0.2496],[140,-0.3074],[141,-0.3436],[142,-0.361],[143,-0.3625],[144,-0.3663],[146,-0.3764],[147,-0.3699],[148,-0.344],[149,-0.303],[150,-0.2825],[151,-0.2635],[152,-0.2564],[154,-0.2706],[155,-0.2808],[156,-0.2916],[159,-0.3119],[160,-0.3177],[161,-0.3222],[163,-0.3285],[165,-0.3327],[167,-0.3348],[200,-0.3392],[207,-0.3356],[212,-0.3297],[214,-0.326],[216,-0.3207],[220,-0.3065],[221,-0.3023],[223,-0.2918],[225,-0.2806],[226,-0.2758],[227,-0.276],[228,-0.2809],[229,-0.288],[230,-0.2933],[231,-0.2993],[233,-0.3131],[234,-0.319],[235,-0.3226],[239,-0.3242],[242,-0.327],[243,-0.327],[244,-0.3226],[245,-0.3107],[246,-0.2886],[247,-0.2582],[248,-0.2299],[249,-0.205],[250,-0.1881],[251,-0.1807],[252,-0.1786],[258,-0.1971],[261,-0.2073],[262,-0.2104],[263,-0.2122],[265,-0.2138],[267,-0.2141],[268,-0.2134],[269,-0.2118],[272,-0.2056],[275,-0.1957],[276,-0.1936],[279,-0.1902],[282,-0.1882],[285,-0.1873],[286,-0.1872],[287,-0.1877],[288,-0.1888],[292,-0.1961],[294,-0.1988],[296,-0.2005],[297,-0.2005],[298,-0.2],[299,-0.1985],[300,-0.1955],[301,-0.19],[302,-0.1819],[303,-0.1725],[304,-0.1615],[306,-0.137],[307,-0.1244],[308,-0.1112],[309,-0.0994],[310,-0.0911],[311,-0.0865],[312,-0.0841],[313,-0.0832],[314,-0.0834],[315,-0.0844],[316,-0.0859],[317,-0.0879],[319,-0.0922],[320,-0.0939],[321,-0.095],[326,-0.0974],[328,-0.0979],[333,-0.0976],[340,-0.0969],[341,-0.097],[342,-0.0972],[343,-0.0979],[344,-0.099],[345,-0.1009],[346,-0.1037],[347,-0.1077],[348,-0.1131],[349,-0.1202],[350,-0.1289],[351,-0.1401],[353,-0.1667],[354,-0.1785],[355,-0.1886],[356,-0.1979],[357,-0.2059],[358,-0.2118],[359,-0.2162],[360,-0.2198],[361,-0.2226],[362,-0.2247],[363,-0.2262],[365,-0.2277],[370,-0.2286],[371,-0.2283],[372,-0.2274],[374,-0.2243],[375,-0.2222],[376,-0.2195],[378,-0.2123],[380,-0.2041],[382,-0.1947],[385,-0.1823],[386,-0.1787],[387,-0.1756],[388,-0.1732],[389,-0.1711],[390,-0.1697],[391,-0.1692],[392,-0.1701],[393,-0.1724],[398,-0.1856],[399,-0.1878],[400,-0.1893],[402,-0.1906],[406,-0.1912],[407,-0.1909],[409,-0.189],[414,-0.1822],[417,-0.1774],[418,-0.1764],[419,-0.176],[422,-0.176]]
       }
       spine_04_ctrl.Location.Z {
        type=MovieSceneScriptingFloatChannel
        default=0.101
        thinned_from=423
        keys [[0,0.101],[13,0.1008],[15,0.1011],[17,0.1019],[18,0.1026],[19,0.1039],[20,0.1054],[21,0.1063],[24,0.1067],[63,0.1066],[65,0.1068],[66,0.1051],[67,0.1009],[69,0.0914],[74,0.0647],[75,0.0596],[76,0.0552],[77,0.0515],[78,0.0487],[79,0.0471],[81,0.0456],[83,0.045],[125,0.045],[126,0.0452],[127,0.0458],[128,0.0486],[129,0.0538],[130,0.0601],[131,0.0674],[132,0.0755],[133,0.0856],[134,0.1069],[135,0.1343],[136,0.1624],[137,0.186],[138,0.1554],[139,0.3741],[140,0.7534],[141,0.9327],[142,0.9842],[143,0.9395],[144,0.9189],[145,0.9151],[146,0.915],[147,0.8242],[148,0.5769],[149,0.2145],[150,0.0049],[151,-0.0486],[152,0.0003],[153,0.0693],[154,0.1206],[155,0.1799],[156,0.2195],[220,0.2198],[221,0.2173],[222,0.208],[223,0.1953],[224,0.1776],[225,0.1628],[226,0.1575],[227,0.1831],[228,0.2332],[229,0.2877],[230,0.3145],[231,0.3266],[232,0.3294],[241,0.3284],[242,0.3272],[243,0.3242],[244,0.2715],[245,0.18],[246,0.0916],[247,0.0372],[248,0.0414],[249,0.0682],[250,0.1013],[251,0.1207],[252,0.1283],[253,0.129],[254,0.1259],[255,0.1206],[256,0.1158],[257,0.1113],[258,0.1075],[259,0.105],[260,0.1036],[268,0.1031],[272,0.1035],[274,0.1032],[288,0.1036],[300,0.1022],[304,0.1004],[306,0.1001],[307,0.1004],[308,0.1011],[312,0.1051],[316,0.1076],[318,0.1085],[320,0.1088],[345,0.1082],[348,0.1076],[351,0.1062],[354,0.1045],[357,0.1034],[374,0.1018],[376,0.1012],[381,0.0973],[382,0.0969],[383,0.0968],[384,0.097],[385,0.0974],[389,0.1002],[391,0.1012],[393,0.1016],[409,0.1004],[422,0.101]]
       }
       spine_04_ctrl.Rotation.X {
        type=MovieSceneScriptingFloatChannel
        default=-2.6097
        thinned_from=423
        keys [[0,-2.6097],[1,-2.6153],[2,-2.63],[6,-2.714],[9,-2.7624],[10,-2.7733],[11,-2.7774],[12,-2.7745],[13,-2.7568],[14,-2.7109],[15,-2.6233],[16,-2.4826],[17,-2.2717],[18,-1.9703],[19,-1.4458],[20,-0.8385],[21,-0.545],[22,-0.571],[23,-0.6408],[24,-0.7293],[25,-0.8105],[26,-0.8782],[27,-0.9418],[28,-0.9989],[29,-1.0474],[30,-1.0849],[31,-1.1134],[32,-1.1368],[33,-1.1558],[34,-1.1712],[35,-1.1835],[41,-1.229],[47,-1.2546],[52,-1.2648],[57,-1.2676],[58,-1.2641],[59,-1.2543],[60,-1.238],[61,-1.2149],[62,-1.1851],[63,-1.1355],[65,-1.0006],[66,-0.9972],[67,-1.0744],[68,-1.1862],[69,-1.3128],[70,-1.4616],[71,-1.6266],[72,-1.8001],[73,-1.9811],[75,-2.3607],[76,-2.5377],[77,-2.8179],[78,-3.2145],[79,-3.5573],[80,-3.8283],[81,-4.0892],[82,-4.2902],[83,-4.4271],[84,-4.5302],[85,-4.6034],[86,-4.6511],[89,-4.7447],[92,-4.8077],[95,-4.8382],[99,-4.8346],[120,-4.6018],[126,-4.5828],[127,-4.5617],[128,-4.4675],[131,-3.9601],[132,-3.8077],[133,-3.6412],[134,-3.1562],[135,-2.5764],[136,-2.1727],[137,-2.0278],[138,-2.1571],[139,-2.7985],[140,-3.759],[141,-4.4963],[142,-4.9256],[143,-5.0873],[144,-5.2307],[145,-5.3561],[146,-5.4516],[147,-5.4084],[148,-5.1559],[149,-4.7472],[150,-4.6995],[151,-4.3064],[152,-4.0125],[153,-4.078],[154,-4.1532],[155,-4.3084],[156,-4.5355],[159,-5.1537],[160,-5.3315],[161,-5.4698],[162,-5.5762],[163,-5.6684],[164,-5.7452],[165,-5.8054],[166,-5.848],[186,-6.0089],[200,-6.0027],[205,-5.9404],[208,-5.8631],[211,-5.7457],[213,-5.6412],[214,-5.5745],[215,-5.4916],[216,-5.3949],[218,-5.1692],[220,-4.9152],[221,-4.7794],[224,-4.3316],[226,-4.0444],[227,-3.9866],[228,-4.0289],[229,-4.1375],[230,-4.2644],[231,-4.4506],[232,-4.6811],[233,-4.9269],[234,-5.1323],[235,-5.2602],[236,-5.275],[241,-5.2536],[243,-5.2589],[244,-5.3533],[245,-5.4357],[246,-5.1865],[247,-4.5743],[248,-3.8826],[249,-3.2521],[250,-2.8675],[251,-2.7869],[252,-2.8209],[253,-2.9346],[254,-3.0344],[256,-3.2168],[257,-3.316],[258,-3.4281],[261,-3.7923],[262,-3.901],[263,-3.9678],[264,-4.006],[265,-4.0353],[266,-4.0541],[267,-4.0606],[268,-4.0419],[269,-3.9941],[272,-3.7973],[273,-3.6917],[274,-3.5746],[275,-3.47],[276,-3.4009],[278,-3.3232],[280,-3.2632],[282,-3.2207],[284,-3.1954],[286,-3.1871],[287,-3.2014],[288,-3.2396],[289,-3.2945],[292,-3.4883],[293,-3.5389],[295,-3.6174],[296,-3.6443],[297,-3.6562],[298,-3.6559],[299,-3.6294],[300,-3.5442],[301,-3.364],[302,-3.0985],[303,-2.7825],[304,-2.4094],[306,-1.5559],[307,-1.1089],[308,-0.6346],[309,-0.1985],[310,0.1285],[311,0.3379],[312,0.4802],[313,0.5689],[314,0.6176],[315,0.6423],[316,0.6459],[317,0.6263],[318,0.5909],[319,0.5479],[320,0.5065],[321,0.4772],[322,0.4585],[323,0.4416],[324,0.4268],[325,0.4139],[326,0.4028],[327,0.394],[328,0.3873],[329,0.3819],[330,0.3779],[331,0.375],[332,0.3733],[340,0.3715],[341,0.3699],[342,0.3619],[343,0.3419],[344,0.3041],[345,0.2429],[346,0.153],[347,0.0229],[348,-0.1584],[349,-0.396],[350,-0.6947],[351,-1.0915],[352,-1.5723],[353,-2.0654],[354,-2.4999],[355,-2.8757],[356,-3.2243],[357,-3.5227],[358,-3.7484],[359,-3.9131],[360,-4.0459],[361,-4.1499],[362,-4.2285],[363,-4.2846],[364,-4.3217],[366,-4.3619],[370,-4.3814],[371,-4.3726],[372,-4.3449],[373,-4.3007],[374,-4.2425],[375,-4.1727],[376,-4.0846],[377,-3.975],[379,-3.7224],[380,-3.5823],[382,-3.2743],[383,-3.1296],[385,-2.8617],[386,-2.7403],[387,-2.6361],[388,-2.5463],[389,-2.4691],[390,-2.4133],[391,-2.3875],[392,-2.4117],[393,-2.4823],[394,-2.5729],[396,-2.7374],[398,-2.9153],[399,-2.9909],[400,-3.0426],[401,-3.073],[402,-3.0933],[404,-3.1115],[406,-3.1144],[407,-3.1057],[408,-3.0812],[409,-3.0449],[414,-2.8184],[416,-2.7078],[417,-2.6584],[418,-2.6232],[419,-2.6097],[422,-2.6097]]
       }
       spine_04_ctrl.Rotation.Y {
        type=MovieSceneScriptingFloatChannel
        default=-5.7937
        thinned_from=423
        keys [[0,-5.7937],[11,-5.8442],[15,-5.8032],[16,-5.8035],[17,-5.8467],[18,-5.9537],[20,-6.1989],[21,-6.3569],[22,-6.5722],[23,-6.8114],[25,-7.309],[27,-7.7429],[28,-7.9437],[29,-8.127],[30,-8.2889],[31,-8.4222],[32,-8.522],[33,-8.5842],[34,-8.5995],[35,-8.5782],[36,-8.5213],[37,-8.4305],[39,-8.1705],[40,-8.0531],[41,-7.9628],[43,-7.8573],[45,-7.7924],[52,-7.7132],[56,-7.7086],[57,-7.7171],[58,-7.7523],[61,-7.9383],[65,-8.155],[66,-8.1878],[67,-8.1985],[68,-8.1909],[69,-8.163],[75,-7.8578],[76,-7.7707],[77,-7.6182],[78,-7.4296],[79,-7.1993],[83,-6.1995],[84,-5.9656],[85,-5.7664],[86,-5.5926],[87,-5.4429],[88,-5.3144],[89,-5.201],[90,-5.1102],[91,-5.0499],[92,-5.0131],[93,-4.9963],[94,-4.9989],[95,-5.017],[97,-5.0828],[101,-5.2409],[104,-5.3059],[117,-5.375],[120,-5.3635],[123,-5.3036],[124,-5.2951],[125,-5.3054],[126,-5.3469],[127,-5.513],[128,-5.7542],[129,-6.0555],[130,-6.4067],[131,-6.8071],[132,-7.3735],[134,-8.7658],[135,-9.3984],[136,-10.0077],[137,-11.2234],[138,-12.901],[139,-12.9851],[140,-12.4643],[141,-11.3985],[142,-10.5712],[144,-9.9083],[145,-9.6379],[146,-9.4656],[147,-9.3499],[148,-9.3414],[149,-9.4719],[150,-9.845],[151,-10.1929],[152,-10.3163],[153,-9.8732],[154,-9.229],[155,-8.8992],[156,-8.6718],[157,-8.5481],[158,-8.5037],[159,-8.5043],[161,-8.5741],[163,-8.6887],[166,-8.8893],[169,-9.0229],[172,-9.0897],[178,-9.0531],[188,-8.9385],[223,-8.9807],[224,-8.9721],[226,-8.8992],[229,-8.6792],[230,-8.5877],[232,-8.3795],[233,-8.3143],[234,-8.2926],[236,-8.3227],[237,-8.3026],[238,-8.2459],[239,-8.1228],[240,-7.9193],[241,-7.6646],[242,-7.3704],[243,-7.0396],[244,-6.6444],[245,-6.2357],[246,-5.8557],[247,-5.5311],[248,-5.2715],[249,-5.0435],[250,-4.9168],[251,-4.854],[252,-4.8343],[253,-4.8669],[254,-4.9392],[255,-5.048],[257,-5.2997],[258,-5.4049],[259,-5.4878],[261,-5.6105],[263,-5.708],[265,-5.7733],[267,-5.8106],[269,-5.8222],[273,-5.771],[283,-5.5663],[286,-5.5307],[288,-5.5318],[291,-5.586],[294,-5.6771],[295,-5.7242],[296,-5.7849],[297,-5.8631],[298,-5.985],[299,-6.127],[300,-6.2188],[301,-6.2804],[303,-6.4381],[305,-6.5537],[308,-6.7198],[309,-6.7293],[310,-6.6764],[311,-6.5836],[312,-6.4526],[313,-6.2832],[314,-6.0921],[315,-5.8827],[316,-5.6503],[318,-5.1464],[319,-4.9193],[320,-4.7192],[321,-4.5586],[322,-4.4328],[323,-4.3449],[324,-4.3062],[325,-4.3111],[326,-4.3438],[327,-4.4112],[333,-5.0093],[334,-5.0937],[335,-5.1577],[337,-5.2461],[340,-5.3341],[342,-5.3615],[345,-5.3553],[347,-5.3254],[348,-5.2778],[349,-5.1791],[350,-5.0065],[351,-4.6389],[352,-4.0551],[353,-3.3854],[354,-2.7604],[355,-2.1672],[356,-1.5526],[357,-0.9865],[358,-0.5376],[359,-0.2088],[360,0.0511],[361,0.254],[362,0.4124],[363,0.5392],[364,0.6389],[365,0.7112],[366,0.76],[367,0.7866],[368,0.797],[369,0.8023],[370,0.794],[372,0.7501],[373,0.7225],[374,0.6863],[376,0.5892],[377,0.548],[378,0.4696],[379,0.323],[380,0.1342],[381,-0.0815],[382,-0.3306],[383,-0.6195],[384,-0.9302],[385,-1.2375],[386,-1.5943],[387,-2.0436],[388,-2.5605],[389,-3.1076],[390,-3.6468],[391,-4.1399],[392,-4.5572],[393,-4.9088],[394,-5.2201],[395,-5.49],[396,-5.7195],[397,-5.9112],[398,-6.066],[399,-6.1853],[400,-6.275],[401,-6.3405],[402,-6.3848],[404,-6.4306],[405,-6.4299],[406,-6.4037],[407,-6.3574],[412,-6.0134],[414,-5.9148],[416,-5.8417],[418,-5.7995],[422,-5.7937]]
       }
       spine_04_ctrl.Rotation.Z {
        type=MovieSceneScriptingFloatChannel
        default=-0.2499
        thinned_from=423
        keys [[0,-0.2499],[1,-0.2455],[2,-0.234],[3,-0.2183],[4,-0.2008],[5,-0.1845],[6,-0.1719],[7,-0.1626],[8,-0.1544],[9,-0.1479],[10,-0.1437],[11,-0.1422],[12,-0.1522],[13,-0.1736],[14,-0.1936],[15,-0.1991],[16,-0.1679],[17,-0.0235],[18,0.3001],[19,0.741],[20,1.2495],[21,1.7767],[22,2.2297],[23,2.6023],[24,2.9124],[25,3.1239],[26,3.2569],[27,3.355],[28,3.4194],[29,3.4512],[30,3.4514],[31,3.4239],[32,3.3754],[33,3.3119],[36,3.0861],[37,3.0155],[38,2.9563],[39,2.9107],[40,2.8816],[46,2.7704],[50,2.7228],[52,2.7078],[53,2.7078],[54,2.6983],[56,2.6983],[57,2.6921],[58,2.6933],[59,2.7028],[60,2.7226],[61,2.7542],[62,2.8008],[63,2.8832],[66,3.2349],[67,3.3634],[68,3.4753],[69,3.5479],[70,3.5808],[71,3.5898],[72,3.5716],[73,3.5032],[74,3.3621],[75,3.1555],[76,2.8534],[77,2.3995],[78,1.8516],[79,1.2433],[80,0.6204],[81,0.0427],[82,-0.4826],[83,-0.9378],[84,-1.3391],[85,-1.6905],[86,-1.9915],[87,-2.2405],[88,-2.4492],[89,-2.629],[90,-2.7787],[91,-2.9003],[92,-2.9973],[93,-3.0741],[94,-3.1282],[95,-3.1564],[96,-3.1648],[97,-3.1592],[98,-3.1357],[99,-3.0887],[102,-2.9079],[103,-2.8594],[104,-2.8191],[107,-2.7351],[109,-2.6942],[113,-2.6508],[118,-2.627],[119,-2.6262],[120,-2.6381],[121,-2.685],[123,-2.81],[124,-2.8644],[125,-2.9022],[126,-2.9168],[127,-2.893],[129,-2.7932],[130,-2.687],[131,-2.4978],[132,-2.2373],[133,-1.9014],[134,-1.3901],[135,-0.7021],[136,0.1119],[137,1.1313],[138,3.5821],[139,4.2109],[140,4.6752],[141,5.0309],[142,5.2686],[143,5.3736],[144,5.2507],[145,4.9399],[147,4.1389],[148,3.9242],[149,4.0361],[151,5.1065],[152,5.488],[153,5.7633],[154,5.8545],[155,5.8176],[156,5.7244],[157,5.6052],[158,5.4703],[160,5.1824],[161,5.0733],[162,4.9931],[163,4.9247],[165,4.8183],[168,4.7136],[171,4.6427],[175,4.5934],[194,4.5813],[195,4.5908],[196,4.5908],[213,4.7002],[224,4.6565],[225,4.6595],[226,4.6796],[227,4.7344],[228,4.8301],[229,4.981],[230,5.1697],[231,5.3465],[232,5.484],[233,5.5805],[234,5.6409],[235,5.6767],[236,5.6901],[237,5.6567],[238,5.562],[239,5.4198],[240,5.2413],[241,5.0299],[242,4.7887],[243,4.5224],[244,4.1998],[245,3.7737],[247,2.8662],[248,2.4304],[249,2.0823],[250,1.821],[251,1.6447],[252,1.5497],[253,1.5129],[254,1.5035],[255,1.4986],[256,1.5009],[257,1.5207],[258,1.556],[259,1.6055],[260,1.6681],[261,1.7403],[262,1.8199],[266,2.1662],[267,2.2471],[268,2.3209],[269,2.3851],[270,2.4375],[271,2.4805],[273,2.5519],[275,2.6058],[277,2.6435],[279,2.668],[282,2.686],[284,2.6866],[287,2.6718],[291,2.6749],[293,2.6902],[295,2.7207],[296,2.7549],[297,2.8232],[298,2.9774],[299,3.1643],[300,3.2425],[301,3.1909],[302,3.0829],[303,2.9193],[304,2.6924],[305,2.4042],[306,2.072],[307,1.6909],[308,1.273],[309,0.8493],[310,0.4138],[311,-0.049],[312,-0.5218],[313,-0.9875],[314,-1.4384],[315,-1.9029],[316,-2.3736],[317,-2.8372],[318,-3.2576],[319,-3.5934],[320,-3.8486],[321,-4.008],[322,-4.108],[323,-4.1725],[324,-4.2076],[325,-4.2263],[326,-4.2322],[327,-4.1948],[328,-4.1153],[329,-4.0203],[332,-3.6832],[334,-3.4766],[335,-3.4],[336,-3.3488],[337,-3.3085],[340,-3.2221],[341,-3.2073],[342,-3.2059],[343,-3.2139],[345,-3.2467],[346,-3.2712],[350,-3.4297],[351,-3.4451],[352,-3.4374],[355,-3.3853],[357,-3.3409],[358,-3.3269],[361,-3.3225],[365,-3.3404],[366,-3.3398],[367,-3.3271],[369,-3.2855],[370,-3.2403],[372,-3.0789],[374,-2.9029],[375,-2.8102],[376,-2.7058],[377,-2.5836],[379,-2.3167],[380,-2.1711],[382,-1.8525],[383,-1.7014],[385,-1.4128],[386,-1.2798],[387,-1.1681],[388,-1.0781],[389,-1.0032],[392,-0.8034],[393,-0.738],[398,-0.4227],[399,-0.3583],[400,-0.2929],[401,-0.2356],[402,-0.1939],[403,-0.1659],[404,-0.1494],[405,-0.1409],[406,-0.1373],[407,-0.1371],[408,-0.1403],[409,-0.1462],[410,-0.1544],[411,-0.1641],[412,-0.1748],[413,-0.1859],[414,-0.1984],[416,-0.2263],[417,-0.2383],[418,-0.2467],[419,-0.2499],[422,-0.2499]]
       }
       spine_04_ctrl.Scale.X {type=MovieSceneScriptingFloatChannel, default=1, const [1,423]}
       spine_04_ctrl.Scale.Y {type=MovieSceneScriptingFloatChannel, default=1, const [1,423]}
       spine_04_ctrl.Scale.Z {type=MovieSceneScriptingFloatChannel, default=1, const [1,423]}
       head_ctrl.Location.X {type=MovieSceneScriptingFloatChannel, default=0, thinned_from=423, keys [[0,0],[422,0]]}
       head_ctrl.Location.Y {type=MovieSceneScriptingFloatChannel, default=0, thinned_from=423, keys [[0,0],[422,0]]}
       head_ctrl.Location.Z {type=MovieSceneScriptingFloatChannel, default=0, thinned_from=423, keys [[0,0],[422,0]]}
       head_ctrl.Rotation.X {
        type=MovieSceneScriptingFloatChannel
        default=-5.6359
        thinned_from=423
        keys [[0,1.1916],[1,1.2066],[2,1.2407],[3,1.287],[5,1.3993],[6,1.4457],[7,1.4791],[8,1.4901],[9,1.4678],[10,1.3994],[11,1.2856],[12,1.1138],[13,0.8833],[14,0.5786],[15,0.1996],[16,-0.2263],[17,-0.6706],[18,-1.1956],[19,-1.7485],[20,-2.1865],[21,-2.5249],[22,-2.8218],[23,-3.0211],[24,-3.1703],[25,-3.2882],[26,-3.3499],[27,-3.3162],[28,-3.2168],[29,-3.1256],[30,-3.0534],[31,-3.0016],[32,-2.9644],[41,-2.7437],[43,-2.7081],[44,-2.6979],[45,-2.6944],[47,-2.7099],[49,-2.7429],[52,-2.8108],[57,-2.9371],[58,-2.9715],[60,-3.0576],[61,-3.096],[62,-3.1199],[63,-3.0957],[64,-3.02],[65,-2.9334],[66,-2.8986],[67,-2.9236],[69,-3.0189],[70,-3.0912],[71,-3.1865],[72,-3.2993],[73,-3.4308],[75,-3.7392],[76,-3.8824],[77,-4.0421],[78,-4.2107],[79,-4.3269],[80,-4.4007],[81,-4.46],[82,-4.4804],[83,-4.4597],[84,-4.417],[85,-4.3589],[87,-4.215],[96,-3.4938],[101,-3.1076],[102,-3.0369],[103,-2.974],[104,-2.9207],[106,-2.8337],[108,-2.7614],[112,-2.6491],[116,-2.5575],[119,-2.5069],[122,-2.4735],[123,-2.4526],[124,-2.4247],[125,-2.391],[126,-2.302],[127,-2.1291],[128,-1.9104],[129,-1.6744],[130,-1.4266],[131,-1.1876],[132,-0.8784],[133,-0.5543],[134,0.0915],[135,0.7078],[136,0.1216],[137,-1.672],[138,-4.253],[139,-9.9794],[140,-14.2445],[141,-16.8348],[142,-18.7311],[143,-20.0279],[144,-21.045],[145,-21.8094],[146,-22.0931],[147,-20.249],[148,-16.4649],[149,-11.4332],[150,-6.133],[151,-2.2931],[152,0.6491],[153,2.0634],[154,2.5619],[155,2.0268],[156,1.3612],[157,0.8035],[158,0.3537],[159,0.1149],[160,-0.0139],[161,-0.0232],[162,0.0256],[163,0.0993],[164,0.1902],[165,0.2674],[166,0.3256],[167,0.3786],[168,0.4224],[169,0.4529],[170,0.475],[171,0.4961],[173,0.53],[175,0.5513],[176,0.5564],[177,0.5637],[181,0.5786],[183,0.5848],[186,0.5903],[188,0.593],[189,0.5954],[190,0.594],[191,0.594],[192,0.5976],[193,0.5976],[194,0.6011],[196,0.6011],[197,0.6048],[200,0.6048],[201,0.6167],[203,0.6167],[204,0.6308],[205,0.6308],[206,0.6463],[207,0.6463],[208,0.6638],[209,0.6638],[210,0.6855],[212,0.7119],[213,0.7269],[214,0.7449],[215,0.7715],[216,0.832],[217,1.0704],[218,1.4315],[219,1.7737],[220,2.049],[221,2.1326],[222,2.0549],[223,1.6801],[224,1.0026],[225,0.0261],[226,-1.238],[227,-2.9566],[228,-4.9888],[229,-7.2323],[230,-9.0652],[231,-10.3943],[232,-11.459],[233,-12.3261],[234,-13.0954],[235,-13.58],[238,-14.8564],[242,-16.3584],[243,-15.5857],[245,-9.6862],[246,-6.3815],[247,-3.7871],[248,-2.0625],[249,-0.6281],[250,0.3366],[251,0.8701],[252,1.1464],[253,1.1346],[254,0.8866],[255,0.3425],[256,-0.3431],[257,-1.0353],[258,-1.5986],[259,-2.0289],[260,-2.4088],[261,-2.7336],[262,-2.9947],[263,-3.1805],[264,-3.3162],[265,-3.4327],[266,-3.5307],[267,-3.6107],[268,-3.6701],[269,-3.7084],[270,-3.7293],[271,-3.7368],[272,-3.7347],[273,-3.7095],[274,-3.6677],[277,-3.4996],[279,-3.3619],[283,-3.0606],[285,-2.9265],[286,-2.8712],[287,-2.8259],[288,-2.7989],[289,-2.7837],[290,-2.7787],[293,-2.7974],[295,-2.817],[296,-2.8219],[297,-2.8168],[298,-2.7553],[299,-2.6141],[300,-2.424],[301,-2.1602],[302,-1.8352],[303,-1.5424],[304,-1.3057],[305,-1.1094],[306,-1.0078],[307,-1.0163],[308,-1.0959],[309,-1.2471],[310,-1.4522],[311,-1.6633],[312,-1.8651],[313,-2.0617],[314,-2.2346],[315,-2.3782],[316,-2.4678],[317,-2.5224],[318,-2.5532],[319,-2.5683],[320,-2.5757],[321,-2.5745],[324,-2.5475],[325,-2.5426],[326,-2.5461],[327,-2.52],[328,-2.4447],[329,-2.3331],[330,-2.1966],[332,-1.8987],[333,-1.7617],[334,-1.6482],[336,-1.45],[337,-1.3554],[338,-1.2709],[339,-1.2021],[340,-1.1547],[341,-1.1245],[342,-1.1011],[344,-1.0658],[345,-1.041],[346,-1.0056],[348,-0.9214],[350,-0.8138],[351,-0.7914],[352,-0.8478],[353,-0.9697],[354,-1.1292],[356,-1.4705],[357,-1.6262],[358,-1.7587],[359,-1.8671],[360,-1.9569],[361,-2.0327],[365,-2.293],[366,-2.3532],[367,-2.4063],[368,-2.4538],[369,-2.4962],[370,-2.5275],[371,-2.5481],[372,-2.5488],[373,-2.5089],[374,-2.4074],[375,-2.2234],[376,-1.8149],[377,-1.1635],[378,-0.4216],[379,0.264],[380,0.9084],[381,1.5705],[382,2.167],[383,2.6222],[384,2.9016],[385,3.0541],[386,3.1259],[387,3.1444],[388,3.099],[389,2.9927],[390,2.8543],[391,2.7087],[392,2.538],[393,2.3301],[394,2.1158],[395,1.9278],[396,1.7665],[397,1.614],[398,1.476],[399,1.358],[400,1.2657],[401,1.1961],[402,1.1412],[403,1.0989],[404,1.0673],[405,1.0453],[406,1.0318],[407,1.0261],[408,1.0276],[409,1.0345],[410,1.0453],[413,1.0837],[414,1.0991],[415,1.1212],[417,1.1684],[418,1.1851],[422,1.1851]]
       }
       head_ctrl.Rotation.Y {
        type=MovieSceneScriptingFloatChannel
        default=7.9933
        thinned_from=423
        keys [[0,7.0928],[12,7.0287],[13,7.0157],[14,6.8965],[15,6.5147],[16,5.789],[17,4.6265],[18,3.0019],[19,1.0406],[20,-1.0876],[21,-3.1907],[22,-5.1288],[23,-6.7899],[24,-8.0657],[25,-8.9944],[26,-9.7198],[27,-10.2694],[28,-10.658],[29,-10.9256],[30,-11.1057],[31,-11.2204],[32,-11.2889],[33,-11.3268],[35,-11.3272],[41,-11.0802],[46,-11.0258],[60,-11.1579],[65,-11.2823],[66,-11.2609],[67,-11.1038],[68,-10.7369],[69,-10.1218],[70,-9.316],[71,-8.3764],[72,-7.3342],[73,-6.1278],[75,-3.4196],[76,-2.1474],[77,-0.9044],[78,0.2952],[79,1.2811],[80,1.9721],[81,2.4437],[82,2.7651],[83,2.9663],[84,3.0587],[85,3.0738],[87,3.0248],[88,3.0155],[89,3.0205],[90,3.0343],[92,3.0789],[94,3.1385],[102,3.436],[105,3.5233],[108,3.5829],[114,3.6539],[116,3.658],[117,3.6504],[124,3.5129],[125,3.4885],[126,3.3698],[127,3.1187],[128,2.4605],[129,1.7219],[130,0.9159],[131,0.0617],[132,-0.8687],[133,-1.8707],[134,-2.9754],[135,-4.0677],[136,-5.0683],[137,-6.1706],[138,-7.4445],[139,-8.2695],[140,-8.5775],[141,-7.7475],[142,-6.1684],[143,-4.4758],[144,-2.7214],[145,-1.1699],[146,-0.0054],[147,0.9065],[148,1.5683],[149,1.9651],[150,2.163],[151,2.251],[152,2.2953],[153,2.4214],[154,2.6694],[155,2.9828],[156,3.2896],[157,3.5338],[158,3.7138],[159,3.8474],[160,3.939],[161,3.9923],[166,4.1634],[173,4.3439],[180,4.4522],[185,4.4863],[189,4.4909],[196,4.4991],[200,4.4917],[218,4.4614],[220,4.4708],[221,4.4436],[222,4.3859],[223,4.292],[224,4.1465],[225,3.9583],[226,3.7363],[227,3.5056],[228,3.2898],[229,3.095],[230,2.9292],[231,2.7871],[232,2.5262],[233,2.373],[234,2.3008],[235,2.2834],[236,2.3064],[237,2.3489],[238,2.4036],[239,2.4738],[240,2.5622],[241,2.6627],[242,2.7773],[244,3.0282],[245,3.5129],[246,4.4541],[247,5.479],[248,6.2952],[249,6.8297],[250,7.1873],[251,7.3617],[252,7.1221],[253,6.8143],[254,6.4635],[255,6.1374],[257,5.5228],[258,5.24],[259,5.0218],[260,4.8841],[261,4.8045],[262,4.7649],[267,4.6534],[269,4.5989],[273,4.5405],[288,4.5439],[290,4.5696],[292,4.6201],[297,4.8225],[299,4.8883],[300,4.9472],[301,5.0187],[302,5.0626],[303,5.0705],[304,5.04],[305,4.9668],[306,4.7782],[307,4.4591],[308,4.0798],[309,3.5951],[310,3.0413],[311,2.6082],[312,2.3205],[313,2.0829],[314,1.8935],[315,1.7581],[316,1.6784],[317,1.6548],[318,1.6679],[319,1.6981],[320,1.7329],[322,1.7923],[323,1.8183],[324,1.8354],[326,1.8511],[328,1.8808],[329,1.904],[333,2.0251],[343,2.2678],[345,2.326],[346,2.3611],[347,2.6908],[348,3.4458],[349,4.3962],[350,5.3077],[351,6.2002],[352,7.1662],[353,8.0815],[354,8.8351],[355,9.4088],[356,9.8509],[357,10.1596],[358,10.3304],[359,10.409],[360,10.4533],[375,10.691],[376,10.6638],[377,10.5501],[378,10.3557],[379,10.0816],[380,9.7408],[381,9.3527],[383,8.5211],[384,8.1538],[385,7.8681],[386,7.6336],[387,7.4161],[388,7.252],[389,7.176],[390,7.1653],[393,7.2655],[395,7.2757],[401,7.1632],[405,7.0508],[422,7.0917]]
       }
       head_ctrl.Rotation.Z {
        type=MovieSceneScriptingFloatChannel
        default=8.8749
        thinned_from=423
        keys [[0,9.7847],[4,9.73],[5,9.732],[6,9.7739],[7,9.8559],[8,9.9697],[9,10.1212],[10,10.3052],[11,10.5186],[13,11.0982],[14,11.9203],[15,13.5508],[16,15.7425],[17,18.2102],[18,21.4158],[19,25.1523],[20,28.2465],[21,30.4464],[22,32.1542],[23,33.3348],[24,34.2983],[25,35.1713],[26,35.7547],[27,35.9264],[29,35.7737],[30,35.6081],[31,35.3054],[32,34.8064],[35,32.952],[36,32.4005],[37,31.9187],[38,31.5276],[40,30.9246],[43,30.2461],[46,29.7681],[48,29.5536],[54,29.2355],[56,29.1987],[62,28.9374],[63,28.8438],[64,28.6722],[65,28.4274],[66,28.1229],[67,27.7542],[68,27.2659],[69,26.6273],[70,25.9077],[71,25.1143],[72,24.1096],[73,22.7927],[75,19.7334],[76,18.3008],[77,16.9616],[78,15.6739],[79,14.5819],[80,13.7531],[81,13.1417],[82,12.7269],[83,12.4213],[85,11.8934],[90,10.7349],[95,9.6626],[99,8.8695],[101,8.4527],[102,8.2595],[103,8.0884],[104,7.9479],[106,7.7309],[109,7.4848],[112,7.3375],[115,7.2383],[116,7.2444],[117,7.2743],[118,7.3378],[119,7.4245],[121,7.6498],[124,8.0336],[125,8.2465],[127,8.8454],[128,9.1628],[129,9.5231],[130,9.9604],[131,10.4562],[132,10.9784],[133,11.5411],[136,13.3573],[137,13.7357],[138,13.8038],[139,13.8143],[140,13.5726],[141,12.8147],[142,11.5637],[143,9.8281],[144,8.0316],[145,5.2352],[146,3.5999],[147,3.2593],[148,3.3601],[149,3.9471],[150,4.6213],[151,5.2089],[152,6.167],[153,7.0553],[154,7.4779],[155,7.5263],[156,7.373],[157,7.2702],[158,7.2675],[159,7.2418],[160,7.1931],[161,7.1209],[165,6.74],[170,6.215],[171,6.1242],[172,6.0148],[173,5.9372],[174,5.8378],[176,5.699],[179,5.5673],[182,5.4917],[190,5.345],[191,5.345],[192,5.3293],[193,5.3293],[194,5.3126],[196,5.3126],[197,5.3277],[200,5.3277],[201,5.3441],[205,5.3517],[215,5.4004],[222,5.5024],[225,5.499],[226,5.4832],[227,5.4404],[228,5.3534],[229,5.2366],[230,5.0933],[232,4.7321],[233,4.6316],[234,4.7127],[237,5.3962],[238,5.5788],[239,5.6996],[240,5.7865],[241,5.8534],[243,5.9537],[244,6.0175],[245,6.1264],[246,6.3339],[247,6.7162],[248,7.4033],[249,8.0459],[250,8.3451],[251,8.4548],[252,8.4813],[253,8.479],[254,8.4535],[256,8.341],[262,7.8681],[263,7.8051],[265,7.7226],[268,7.659],[272,7.6423],[273,7.6574],[284,8.0656],[287,8.1457],[288,8.2196],[291,8.5457],[292,8.6219],[293,8.6511],[294,8.5919],[296,8.3531],[297,8.3278],[298,8.4161],[299,8.5864],[300,8.8307],[301,9.1201],[302,9.6173],[303,10.3145],[304,11.0759],[305,11.8628],[306,12.5281],[307,13.0128],[308,13.3887],[310,13.9863],[311,14.205],[312,14.298],[313,14.2633],[314,14.1069],[315,13.8611],[316,13.4777],[317,12.9398],[322,9.9783],[323,9.4208],[324,8.9018],[325,8.4455],[326,8.0293],[327,7.6834],[328,7.4048],[329,7.1985],[330,7.0521],[331,6.952],[332,6.8856],[335,6.7621],[336,6.7039],[337,6.6159],[338,6.4854],[339,6.3],[340,6.0472],[341,5.6943],[342,5.297],[343,4.7711],[344,3.9793],[345,2.8294],[346,1.213],[347,-0.9792],[348,-3.8359],[349,-7.8916],[350,-13.229],[351,-19.2455],[352,-25.3472],[353,-30.9257],[354,-35.4419],[355,-38.5535],[356,-40.5496],[357,-41.8799],[359,-44.0581],[360,-44.9496],[361,-45.6924],[362,-46.31],[363,-46.826],[367,-48.2236],[370,-48.9311],[374,-49.4005],[375,-49.4533],[376,-49.2124],[377,-48.4356],[378,-47.176],[379,-45.4988],[380,-43.496],[381,-40.6491],[382,-36.7735],[383,-32.4652],[384,-27.9689],[385,-23.3261],[386,-18.8307],[387,-14.7792],[388,-11.1314],[389,-7.6911],[390,-4.5198],[391,-1.7571],[392,0.5248],[393,2.3347],[394,3.6079],[395,4.281],[397,4.9865],[398,5.3293],[399,5.6556],[400,5.9579],[401,6.2826],[402,6.6643],[404,7.497],[405,7.8978],[406,8.2551],[407,8.5558],[408,8.8077],[409,9.0167],[410,9.1887],[411,9.3297],[412,9.4456],[414,9.6221],[416,9.7303],[418,9.7791],[422,9.7791]]
       }
       head_ctrl.Scale.X {type=MovieSceneScriptingFloatChannel, default=1, const [1,423]}
       head_ctrl.Scale.Y {type=MovieSceneScriptingFloatChannel, default=1, const [1,423]}
       head_ctrl.Scale.Z {type=MovieSceneScriptingFloatChannel, default=1, const [1,423]}
       clavicle_l_ctrl.Location.X {type=MovieSceneScriptingFloatChannel, default=0, thinned_from=423, keys [[0,0],[422,0]]}
       clavicle_l_ctrl.Location.Y {type=MovieSceneScriptingFloatChannel, default=0, thinned_from=423, keys [[0,0],[422,0]]}
       clavicle_l_ctrl.Location.Z {type=MovieSceneScriptingFloatChannel, default=0, thinned_from=423, keys [[0,0],[422,0]]}
       clavicle_l_ctrl.Rotation.X {type=MovieSceneScriptingFloatChannel, default=0, thinned_from=423, keys [[0,0],[422,0]]}
       clavicle_l_ctrl.Rotation.Y {
        type=MovieSceneScriptingFloatChannel
        default=12.9795
        thinned_from=423
        keys [[0,12.9795],[1,13.0051],[2,13.0726],[12,14.0459],[16,14.3197],[18,14.3784],[19,14.3549],[21,14.2011],[24,13.7849],[26,13.5805],[28,13.4366],[29,13.3936],[31,13.3779],[37,13.5486],[47,14.1026],[57,14.4023],[62,14.4254],[65,14.3433],[66,14.359],[67,14.407],[68,14.4888],[72,15.0383],[74,15.2126],[75,15.2669],[76,15.287],[77,15.2103],[78,15.0249],[79,14.7975],[80,14.5238],[81,14.2083],[82,13.9305],[84,13.4867],[85,13.3046],[86,13.1706],[87,13.0846],[89,13.002],[93,13.002],[112,13.7278],[115,13.7856],[125,13.8286],[126,13.8159],[127,13.759],[128,13.6639],[129,13.5303],[130,13.3579],[131,13.1464],[132,12.8355],[133,12.4133],[134,11.9519],[135,11.5235],[136,11.1219],[137,10.4314],[138,10.3717],[139,11.0944],[140,16.2295],[141,25.4726],[142,33.1752],[143,34.2592],[144,34.6586],[145,34.7157],[146,33.6053],[147,30.8979],[148,27.529],[149,24.434],[150,21.4851],[151,20.4282],[152,20.0946],[153,20.3454],[157,22.1934],[159,23.0229],[160,23.3969],[161,23.707],[164,24.4163],[166,24.7593],[169,25.0584],[173,25.2085],[221,25.2012],[224,25.0656],[226,24.8534],[228,24.5035],[229,24.0643],[230,23.2697],[231,22.3175],[232,21.1637],[233,20.2196],[234,19.6435],[235,19.3664],[236,19.5584],[237,19.8733],[238,20.2567],[239,20.5913],[241,20.9904],[242,21.0646],[244,21.0646],[245,20.5627],[246,18.9987],[247,16.6499],[248,14.6657],[249,13.3737],[250,12.5506],[251,12.105],[252,11.9454],[253,11.9805],[255,12.2688],[256,12.3391],[298,12.3391],[300,12.4437],[308,13.1378],[310,13.2596],[313,13.3482],[317,13.3647],[321,13.248],[332,12.7055],[338,12.5592],[339,12.5592],[340,12.5304],[357,12.45],[365,12.739],[370,12.9795],[372,13.0299],[377,13.0528],[379,12.9795],[386,12.4037],[389,12.2306],[393,12.1259],[394,12.1259],[395,12.0983],[403,12.0798],[406,12.1627],[414,12.6803],[417,12.9125],[418,12.9612],[422,12.9795]]
       }
       clavicle_l_ctrl.Rotation.Z {
        type=MovieSceneScriptingFloatChannel
        default=-11.16
        thinned_from=423
        keys [[0,-11.16],[125,-11.16],[126,-11.1115],[127,-10.9464],[128,-10.6358],[129,-10.1504],[130,-9.4611],[131,-8.5388],[132,-6.7204],[133,-4.0828],[134,-1.7074],[135,-0.6753],[136,-2.26],[137,-8.6469],[138,-14.2762],[139,-14.2988],[140,-11.9925],[141,-7.8412],[142,-4.3819],[143,-2.9506],[144,-1.9243],[145,-1.2925],[146,-0.899],[147,-0.5892],[148,-0.3547],[149,-0.187],[150,-0.0776],[151,-0.0182],[152,-0.0003],[216,0],[217,-0.0192],[218,-0.0356],[219,-0.0579],[220,-0.0867],[221,-0.1225],[222,-0.1659],[223,-0.2101],[224,-0.2552],[225,-0.3121],[226,-0.3919],[227,-0.5054],[228,-0.6638],[229,-0.8371],[230,-1.1457],[231,-1.8253],[232,-4.4388],[233,-4.9742],[234,-5.0922],[235,-5.18],[237,-5.2193],[240,-5.5064],[241,-5.7493],[242,-6.285],[243,-7.9151],[244,-9.9413],[245,-10.8553],[246,-11.0838],[247,-11.1303],[422,-11.16]]
       }
       clavicle_l_ctrl.Scale.X {type=MovieSceneScriptingFloatChannel, default=1, const [1,423]}
       clavicle_l_ctrl.Scale.Y {type=MovieSceneScriptingFloatChannel, default=1, const [1,423]}
       clavicle_l_ctrl.Scale.Z {type=MovieSceneScriptingFloatChannel, default=1, const [1,423]}
       clavicle_r_ctrl.Location.X {type=MovieSceneScriptingFloatChannel, default=0, thinned_from=423, keys [[0,0],[422,0]]}
       clavicle_r_ctrl.Location.Y {type=MovieSceneScriptingFloatChannel, default=0, thinned_from=423, keys [[0,0],[422,0]]}
       clavicle_r_ctrl.Location.Z {type=MovieSceneScriptingFloatChannel, default=-0.0001, thinned_from=423, keys [[0,-0.0001],[422,-0.0001]]}
       clavicle_r_ctrl.Rotation.X {type=MovieSceneScriptingFloatChannel, default=0, thinned_from=423, keys [[0,0],[152,0.0001],[422,0]]}
       clavicle_r_ctrl.Rotation.Y {
        type=MovieSceneScriptingFloatChannel
        default=-5.7268
        thinned_from=423
        keys [[0,-5.7268],[1,-5.7055],[2,-5.6487],[3,-5.5674],[7,-5.1915],[10,-4.89],[11,-4.8109],[12,-4.7513],[13,-4.7084],[14,-4.6824],[16,-4.6824],[17,-4.7503],[18,-4.8345],[19,-4.9557],[21,-5.2474],[23,-5.5096],[24,-5.6294],[25,-5.7268],[27,-5.8788],[28,-5.9369],[29,-5.9758],[33,-5.9758],[34,-5.9425],[45,-5.7558],[46,-5.7558],[47,-5.7275],[48,-5.7275],[49,-5.7062],[50,-5.7062],[51,-5.6877],[52,-5.6877],[53,-5.6666],[54,-5.6666],[59,-5.5403],[60,-5.4968],[61,-5.4309],[62,-5.3338],[63,-5.1464],[64,-4.8887],[65,-4.6708],[66,-4.5349],[67,-4.4389],[68,-4.3668],[69,-4.3062],[70,-4.2576],[71,-4.2254],[73,-4.2254],[74,-4.2425],[75,-4.2907],[76,-4.3721],[77,-4.5177],[81,-5.326],[82,-5.4777],[83,-5.5686],[84,-5.6288],[85,-5.662],[89,-5.662],[104,-5.3422],[108,-5.2785],[109,-5.2785],[110,-5.2493],[111,-5.2493],[112,-5.2264],[117,-5.2264],[120,-5.3336],[125,-5.5811],[126,-5.7631],[127,-6.1097],[128,-6.6541],[129,-7.4295],[130,-8.4691],[131,-9.8062],[132,-12.153],[133,-15.463],[134,-18.6304],[135,-20.5496],[136,-21.2016],[137,-21.4762],[138,-18.0099],[139,-15.8571],[140,-10.191],[141,2.9055],[142,13.8193],[143,15.3553],[144,15.9212],[145,16.002],[146,15.021],[147,12.2269],[148,7.8435],[149,2.0947],[150,-2.4033],[151,-2.9496],[152,-2.6494],[153,-1.9705],[155,-0.5204],[156,0.2507],[157,0.8737],[158,1.3232],[159,1.6941],[160,1.9721],[161,2.1429],[162,2.2376],[163,2.301],[164,2.3396],[165,2.3593],[214,2.3593],[215,2.2991],[216,2.1092],[217,1.8216],[218,1.4599],[219,1.0479],[220,0.6092],[221,0.1673],[222,-0.2539],[223,-0.6219],[224,-0.9611],[225,-1.3338],[226,-1.8023],[227,-2.4287],[228,-3.2754],[229,-4.6619],[230,-6.6804],[231,-8.7333],[232,-10.2738],[233,-11.2347],[234,-11.4728],[235,-11.0457],[236,-10.3176],[237,-9.3545],[238,-8.4443],[239,-7.5091],[240,-6.781],[241,-5.9466],[242,-5.5826],[243,-5.9292],[244,-6.4753],[245,-7.371],[246,-8.3267],[247,-8.7817],[248,-8.3287],[249,-7.6089],[250,-6.8433],[251,-6.1444],[252,-5.5096],[253,-5.1887],[290,-5.1887],[291,-5.2045],[292,-5.2045],[293,-5.2248],[294,-5.2248],[295,-5.2539],[319,-5.8949],[321,-5.9312],[322,-5.9312],[323,-5.9588],[324,-5.9588],[325,-5.975],[332,-5.975],[333,-5.9563],[334,-5.9563],[337,-5.892],[339,-5.8252],[341,-5.7223],[342,-5.6542],[343,-5.5735],[344,-5.4789],[345,-5.3692],[346,-5.2432],[347,-5.0657],[348,-4.7837],[349,-4.3687],[350,-3.7924],[351,-2.9389],[353,-0.7762],[354,0.0876],[355,0.6727],[356,1.098],[357,1.4081],[358,1.6478],[359,1.8473],[360,2.0077],[361,2.1301],[362,2.216],[363,2.2665],[364,2.2831],[366,2.2831],[367,2.2606],[368,2.2297],[369,2.1789],[370,2.103],[371,1.9972],[372,1.7838],[373,1.431],[374,1.0019],[375,0.5598],[376,0.1249],[377,-0.3374],[378,-0.8471],[379,-1.4238],[380,-2.1112],[381,-2.886],[382,-3.6674],[383,-4.3747],[384,-5.0042],[385,-5.5929],[386,-6.1215],[387,-6.5704],[388,-6.9339],[389,-7.2227],[390,-7.4415],[391,-7.5948],[392,-7.6891],[393,-7.7374],[394,-7.7552],[395,-7.7552],[396,-7.7298],[397,-7.6558],[398,-7.5505],[401,-7.1749],[404,-6.7163],[405,-6.5724],[406,-6.4485],[408,-6.2428],[410,-6.0686],[412,-5.9259],[414,-5.8176],[416,-5.755],[417,-5.7383],[422,-5.7383]]
       }
       clavicle_r_ctrl.Rotation.Z {
        type=MovieSceneScriptingFloatChannel
        default=-10.4027
        thinned_from=423
        keys [[0,-10.4027],[125,-10.4027],[126,-10.2789],[127,-9.9388],[128,-9.429],[129,-8.796],[130,-8.0865],[131,-7.3472],[132,-6.4301],[133,-5.2793],[134,-4.0796],[135,-3.016],[136,-2.1357],[137,-0.809],[138,0.1359],[139,0.5861],[140,4.1526],[141,10.5723],[142,15.9219],[143,16.6748],[144,16.9522],[145,16.9919],[146,16.4319],[147,15.0863],[148,13.456],[149,12.0423],[150,11.2998],[151,11.152],[153,11.0794],[215,11.0416],[216,11.0071],[217,10.9459],[218,10.8549],[219,10.7309],[220,10.571],[221,10.372],[222,10.1307],[223,9.8515],[224,9.5268],[225,9.1363],[226,8.6595],[227,8.0762],[228,7.3659],[229,6.4136],[230,5.035],[231,3.1054],[232,-1.62],[233,-2.8346],[234,-3.1752],[235,-3.4123],[236,-3.505],[237,-3.5624],[239,-3.8262],[240,-3.9357],[241,-4.2037],[242,-4.8118],[243,-6.6813],[244,-9.0049],[245,-10.0532],[246,-10.3153],[247,-10.3685],[422,-10.4026]]
       }
       clavicle_r_ctrl.Scale.X {type=MovieSceneScriptingFloatChannel, default=1, const [1,423]}
       clavicle_r_ctrl.Scale.Y {type=MovieSceneScriptingFloatChannel, default=1, const [1,423]}
       clavicle_r_ctrl.Scale.Z {type=MovieSceneScriptingFloatChannel, default=1, const [1,423]}
       neck_01_ctrl.Location.X {type=MovieSceneScriptingFloatChannel, default=0, thinned_from=423, keys [[0,0],[422,0]]}
       neck_01_ctrl.Location.Y {type=MovieSceneScriptingFloatChannel, default=0, thinned_from=423, keys [[0,0],[422,0]]}
       neck_01_ctrl.Location.Z {type=MovieSceneScriptingFloatChannel, default=0, thinned_from=423, keys [[0,0],[422,0]]}
       neck_01_ctrl.Rotation.X {
        type=MovieSceneScriptingFloatChannel
        default=8.6965
        thinned_from=423
        keys [[0,8.6965],[15,8.6965],[16,8.0513],[17,6.6743],[18,5.4028],[19,4.4817],[20,3.6662],[21,3.0592],[22,2.6496],[23,2.3574],[24,2.1823],[25,2.124],[65,2.124],[66,2.1746],[67,2.3852],[68,2.8441],[69,3.5598],[70,4.4375],[71,5.4103],[72,6.4116],[73,7.505],[74,8.6988],[75,9.8438],[76,10.7908],[77,11.4538],[78,11.8453],[79,11.9637],[132,11.8596],[136,11.9358],[137,11.9816],[242,11.9991],[243,11.9817],[244,11.8602],[245,11.5302],[246,10.8878],[247,8.6852],[248,5.6976],[249,3.6767],[250,3.4338],[297,3.4338],[298,3.4533],[299,3.5447],[300,3.7577],[301,4.1411],[302,4.646],[303,5.1753],[304,5.7272],[305,6.3034],[306,6.8119],[307,7.2519],[308,7.664],[309,8.0175],[310,8.2818],[311,8.4589],[312,8.5785],[313,8.651],[314,8.6868],[350,8.6868],[351,8.6386],[352,8.5254],[353,8.4329],[354,8.3889],[381,8.3751],[384,8.4166],[391,8.658],[422,8.6965]]
       }
       neck_01_ctrl.Rotation.Y {
        type=MovieSceneScriptingFloatChannel
        default=10.5758
        thinned_from=423
        keys [[0,10.5758],[76,10.5758],[77,10.4968],[78,10.2908],[79,10.0046],[80,9.5961],[81,9.1074],[82,8.7112],[83,8.4495],[84,8.2429],[85,8.0824],[86,7.959],[87,7.8674],[88,7.8047],[89,7.7661],[90,7.7466],[127,7.7466],[129,7.7899],[130,7.8363],[131,7.9056],[132,8.0783],[133,8.3856],[135,9.1251],[136,9.5274],[137,10.298],[138,10.5758],[350,10.5758],[351,9.6216],[352,7.3892],[353,4.8249],[354,2.8773],[355,1.6837],[356,0.7773],[357,0.2016],[358,0],[375,0],[376,0.0574],[377,0.225],[378,0.4959],[379,0.8632],[380,1.3284],[381,1.8939],[382,2.5549],[383,3.3065],[384,4.21],[386,6.2932],[387,7.2235],[388,8.0389],[389,8.7945],[390,9.4471],[391,9.9535],[392,10.2864],[393,10.4741],[394,10.5571],[422,10.5758]]
       }
       neck_01_ctrl.Rotation.Z {
        type=MovieSceneScriptingFloatChannel
        default=3.5838
        thinned_from=423
        keys [[0,3.5838],[76,3.5838],[77,3.5563],[78,3.4849],[79,3.3858],[80,3.2448],[81,3.0765],[82,2.9404],[83,2.8507],[84,2.7799],[85,2.725],[86,2.6828],[87,2.6515],[88,2.6301],[89,2.617],[90,2.6103],[127,2.6103],[129,2.6251],[130,2.6409],[131,2.6646],[132,2.7236],[133,2.8288],[135,3.0826],[136,3.2211],[137,3.4874],[138,3.5838],[350,3.5838],[351,3.2536],[352,2.4884],[353,1.6194],[354,0.9642],[355,0.5639],[356,0.2603],[357,0.0675],[358,0],[375,0],[376,0.0192],[377,0.0753],[378,0.166],[379,0.289],[380,0.4448],[381,0.6343],[382,0.856],[383,1.1083],[384,1.4122],[385,1.7619],[386,2.1159],[387,2.432],[388,2.7101],[389,2.969],[390,3.1934],[391,3.3682],[392,3.4834],[393,3.5485],[394,3.5773],[422,3.5838]]
       }
       neck_01_ctrl.Scale.X {type=MovieSceneScriptingFloatChannel, default=1, const [1,423]}
       neck_01_ctrl.Scale.Y {type=MovieSceneScriptingFloatChannel, default=1, const [1,423]}
       neck_01_ctrl.Scale.Z {type=MovieSceneScriptingFloatChannel, default=1, const [1,423]}
       neck_02_ctrl.Location.X {type=MovieSceneScriptingFloatChannel, default=0, const [0,423]}
       neck_02_ctrl.Location.Y {type=MovieSceneScriptingFloatChannel, default=0, const [0,423]}
       neck_02_ctrl.Location.Z {type=MovieSceneScriptingFloatChannel, default=0, const [0,423]}
       neck_02_ctrl.Rotation.X {type=MovieSceneScriptingFloatChannel, default=0, const [0,423]}
       neck_02_ctrl.Rotation.Y {type=MovieSceneScriptingFloatChannel, default=0, const [0,423]}
       neck_02_ctrl.Rotation.Z {type=MovieSceneScriptingFloatChannel, default=0, const [0,423]}
       neck_02_ctrl.Scale.X {type=MovieSceneScriptingFloatChannel, default=1, const [1,423]}
       neck_02_ctrl.Scale.Y {type=MovieSceneScriptingFloatChannel, default=1, const [1,423]}
       neck_02_ctrl.Scale.Z {type=MovieSceneScriptingFloatChannel, default=1, const [1,423]}
       head_ik_ctrl.Location.X {
        type=MovieSceneScriptingFloatChannel
        default=-1.9637
        thinned_from=423
        keys [[0,-1.9637],[12,-1.9709],[16,-1.9666],[17,-1.9681],[19,-1.9795],[20,-1.9796],[24,-1.9549],[66,-1.9501],[67,-1.9579],[68,-1.9761],[69,-2.0052],[74,-2.1993],[75,-2.2298],[76,-2.2423],[77,-2.2139],[78,-2.141],[79,-2.037],[80,-1.9042],[81,-1.759],[82,-1.6335],[83,-1.538],[84,-1.4572],[85,-1.3896],[86,-1.3338],[87,-1.2891],[88,-1.2538],[89,-1.2258],[90,-1.2046],[91,-1.1903],[92,-1.1789],[93,-1.1701],[94,-1.1641],[95,-1.1612],[97,-1.1619],[98,-1.1651],[103,-1.1983],[108,-1.2155],[119,-1.2254],[120,-1.2238],[124,-1.197],[125,-1.1928],[126,-1.1914],[127,-1.1964],[128,-1.209],[129,-1.226],[130,-1.2517],[131,-1.2892],[132,-1.3527],[133,-1.4448],[134,-1.5638],[135,-1.6934],[136,-1.8349],[137,-2.0633],[138,-2.319],[139,-2.3518],[140,-2.3705],[142,-2.389],[143,-2.3919],[144,-2.379],[145,-2.353],[147,-2.2884],[148,-2.2722],[149,-2.284],[151,-2.3769],[152,-2.4085],[153,-2.4228],[154,-2.4213],[161,-2.3485],[226,-2.3305],[228,-2.3411],[231,-2.3747],[235,-2.3922],[237,-2.3906],[239,-2.372],[241,-2.339],[243,-2.295],[244,-2.2667],[245,-2.2316],[246,-2.1908],[247,-2.1171],[248,-2.0361],[249,-1.987],[250,-1.9796],[298,-1.9966],[300,-2.0098],[304,-2.0682],[305,-2.0788],[306,-2.0834],[307,-2.082],[308,-2.0756],[309,-2.0639],[310,-2.0459],[311,-2.0216],[314,-1.9293],[318,-1.7826],[319,-1.7524],[320,-1.7284],[321,-1.7125],[322,-1.702],[324,-1.6915],[326,-1.6899],[327,-1.6941],[334,-1.7643],[340,-1.7884],[346,-1.7839],[350,-1.7637],[351,-1.583],[352,-1.1701],[353,-0.6974],[354,-0.3375],[355,-0.115],[356,0.0552],[357,0.1647],[358,0.2057],[359,0.2094],[360,0.2125],[362,0.2171],[364,0.2206],[365,0.2217],[366,0.2223],[367,0.2218],[369,0.2191],[370,0.2159],[372,0.2044],[374,0.192],[375,0.1854],[376,0.1681],[377,0.1302],[378,0.0731],[379,-0.0024],[380,-0.0964],[381,-0.2094],[382,-0.3404],[383,-0.4881],[384,-0.6637],[385,-0.8637],[386,-1.0656],[387,-1.2466],[388,-1.4066],[389,-1.5554],[390,-1.6849],[391,-1.7871],[392,-1.8568],[393,-1.8991],[394,-1.9217],[401,-1.9672],[422,-1.9639]]
       }
       head_ik_ctrl.Location.Y {
        type=MovieSceneScriptingFloatChannel
        default=1.4009
        thinned_from=423
        keys [[0,1.3934],[12,1.3839],[14,1.3885],[15,1.3887],[16,1.2765],[18,0.779],[19,0.5757],[20,0.3836],[21,0.226],[22,0.1098],[23,0.0216],[24,-0.0407],[25,-0.0735],[26,-0.0885],[27,-0.0999],[28,-0.1078],[29,-0.1123],[30,-0.1134],[31,-0.1115],[32,-0.1073],[33,-0.1013],[34,-0.094],[35,-0.086],[36,-0.0777],[37,-0.0698],[38,-0.0627],[39,-0.057],[40,-0.053],[42,-0.0477],[43,-0.0453],[44,-0.0431],[45,-0.0415],[46,-0.0393],[47,-0.0384],[48,-0.0363],[49,-0.0359],[50,-0.034],[51,-0.0337],[52,-0.0331],[53,-0.0322],[54,-0.0318],[56,-0.0318],[57,-0.0313],[58,-0.0313],[59,-0.0327],[60,-0.0353],[61,-0.0391],[62,-0.0444],[63,-0.0534],[65,-0.0785],[66,-0.0821],[67,-0.0594],[68,0.0064],[69,0.1186],[70,0.2613],[71,0.4215],[72,0.5884],[73,0.775],[75,1.1936],[76,1.3809],[77,1.5451],[78,1.681],[79,1.7777],[80,1.8496],[81,1.9111],[82,1.9609],[83,1.9997],[84,2.0312],[85,2.0567],[86,2.0773],[88,2.1071],[91,2.1363],[94,2.151],[98,2.1511],[119,2.1165],[126,2.1363],[129,2.1262],[130,2.119],[132,2.0921],[133,2.0734],[134,2.0396],[135,1.99],[136,1.9267],[137,1.8234],[138,1.5442],[139,1.4919],[140,1.4626],[141,1.4461],[142,1.4359],[143,1.4314],[144,1.4528],[145,1.4946],[147,1.5914],[148,1.614],[149,1.5967],[150,1.5286],[151,1.4545],[152,1.4029],[153,1.3793],[154,1.3826],[155,1.396],[160,1.4884],[161,1.5015],[163,1.5183],[170,1.5442],[196,1.5557],[225,1.5335],[226,1.5306],[227,1.5245],[228,1.5145],[229,1.4986],[231,1.4614],[232,1.4488],[233,1.4407],[235,1.4323],[236,1.4314],[237,1.435],[238,1.447],[239,1.4656],[240,1.4896],[241,1.5179],[243,1.5798],[244,1.5942],[245,1.574],[246,1.4979],[247,1.1535],[248,0.6706],[249,0.3466],[250,0.321],[251,0.3325],[252,0.34],[253,0.3445],[254,0.3465],[255,0.3473],[256,0.347],[257,0.3447],[258,0.3408],[259,0.3355],[260,0.3289],[261,0.3214],[262,0.3132],[266,0.2769],[267,0.2684],[268,0.2606],[269,0.2538],[270,0.2482],[271,0.2436],[273,0.2359],[274,0.2326],[275,0.2298],[276,0.2276],[278,0.2245],[280,0.2227],[283,0.2217],[288,0.2239],[291,0.2238],[292,0.2232],[293,0.2222],[294,0.2206],[295,0.2184],[296,0.2141],[297,0.2059],[298,0.1913],[299,0.1848],[300,0.2098],[301,0.2754],[302,0.3649],[303,0.4629],[304,0.5703],[305,0.6868],[306,0.7957],[307,0.8964],[308,0.995],[309,1.086],[310,1.1664],[311,1.237],[312,1.3006],[313,1.3571],[314,1.4067],[317,1.5417],[318,1.5819],[319,1.614],[320,1.6387],[321,1.6543],[322,1.6642],[324,1.6742],[326,1.6764],[327,1.6728],[334,1.6036],[337,1.587],[342,1.5771],[346,1.5852],[350,1.6107],[351,1.6037],[352,1.5691],[353,1.5214],[354,1.4811],[355,1.4551],[356,1.4344],[357,1.421],[358,1.4164],[376,1.423],[379,1.4386],[385,1.4831],[386,1.4869],[388,1.4853],[390,1.4754],[401,1.3925],[404,1.3822],[422,1.3933]]
       }
       head_ik_ctrl.Location.Z {
        type=MovieSceneScriptingFloatChannel
        default=0.0274
        thinned_from=423
        keys [[0,-0.1464],[1,-0.1462],[6,-0.1422],[9,-0.1404],[10,-0.14],[11,-0.1399],[12,-0.1401],[13,-0.1408],[14,-0.1425],[15,-0.1454],[16,-0.1211],[17,-0.0678],[18,-0.0223],[19,0.006],[20,0.0302],[21,0.0522],[22,0.0724],[23,0.0899],[24,0.1044],[25,0.1153],[28,0.1416],[29,0.1498],[30,0.1574],[31,0.1641],[32,0.1694],[33,0.1732],[34,0.175],[35,0.1752],[36,0.1737],[37,0.1706],[38,0.1657],[39,0.1604],[40,0.1555],[41,0.1517],[42,0.1494],[43,0.1475],[44,0.1461],[46,0.1444],[49,0.1432],[52,0.1427],[56,0.1426],[57,0.143],[58,0.1447],[61,0.152],[62,0.1534],[64,0.1552],[65,0.1555],[66,0.1538],[67,0.1465],[68,0.1306],[69,0.1051],[70,0.0721],[71,0.0332],[72,-0.0092],[73,-0.0582],[74,-0.1153],[75,-0.1739],[76,-0.2259],[77,-0.2608],[78,-0.2763],[79,-0.2791],[81,-0.2755],[82,-0.2757],[87,-0.2883],[89,-0.2913],[90,-0.2921],[92,-0.2918],[99,-0.2844],[102,-0.2837],[121,-0.2918],[123,-0.2918],[125,-0.2904],[126,-0.2879],[127,-0.2808],[128,-0.2696],[129,-0.2542],[130,-0.235],[131,-0.2122],[132,-0.181],[133,-0.1404],[134,-0.1048],[135,-0.0722],[136,-0.0331],[137,0.0435],[138,0.1132],[139,0.1351],[140,0.1288],[141,0.0998],[142,0.0769],[143,0.0713],[144,0.0671],[145,0.0669],[146,0.0707],[147,0.0763],[148,0.0843],[149,0.0946],[150,0.1125],[151,0.0998],[152,0.0782],[153,0.0517],[154,0.0233],[155,0.0147],[156,0.0161],[157,0.0236],[158,0.0341],[160,0.059],[161,0.0695],[162,0.0783],[163,0.0869],[164,0.0948],[165,0.1017],[166,0.1072],[167,0.1112],[168,0.1145],[169,0.1173],[170,0.1197],[171,0.1215],[172,0.123],[174,0.1244],[176,0.1249],[178,0.1245],[185,0.1215],[190,0.1207],[194,0.1207],[199,0.121],[202,0.1202],[204,0.1192],[206,0.1175],[208,0.1151],[209,0.1136],[210,0.1118],[211,0.1099],[212,0.1076],[213,0.1051],[214,0.102],[215,0.0981],[216,0.0936],[217,0.0885],[218,0.083],[219,0.0772],[224,0.0466],[225,0.0399],[227,0.0225],[228,0.0151],[229,0.0096],[230,0.0055],[231,0.0046],[232,0.0077],[233,0.0141],[234,0.0214],[235,0.0276],[236,0.0284],[237,0.0282],[238,0.0274],[239,0.0245],[240,0.0186],[241,0.011],[242,0.0019],[243,-0.0096],[244,-0.0201],[245,-0.0332],[246,-0.0529],[247,-0.0363],[248,0.0011],[249,0.0211],[250,0.0118],[251,0.0055],[252,0.0054],[253,0.0102],[254,0.0165],[255,0.0235],[256,0.0309],[257,0.038],[258,0.0439],[259,0.0485],[260,0.052],[261,0.0552],[262,0.0578],[263,0.0595],[264,0.0604],[265,0.0609],[266,0.061],[267,0.0607],[268,0.0599],[269,0.0586],[270,0.0569],[272,0.0531],[275,0.0454],[276,0.0433],[278,0.04],[280,0.0371],[282,0.0348],[283,0.0339],[285,0.0325],[286,0.0322],[287,0.0322],[288,0.0329],[289,0.0341],[290,0.0357],[294,0.043],[296,0.0472],[297,0.0492],[298,0.0509],[299,0.0505],[300,0.0447],[301,0.0307],[302,0.0094],[303,-0.0164],[304,-0.0474],[305,-0.0842],[306,-0.1219],[307,-0.1603],[308,-0.2006],[309,-0.2386],[310,-0.2698],[311,-0.2934],[312,-0.3119],[313,-0.326],[314,-0.3367],[317,-0.36],[318,-0.3667],[319,-0.3724],[321,-0.3811],[322,-0.3848],[323,-0.3874],[324,-0.3882],[325,-0.3874],[326,-0.3853],[333,-0.3634],[335,-0.3592],[340,-0.3537],[343,-0.3516],[345,-0.3482],[346,-0.3454],[349,-0.3335],[350,-0.3313],[351,-0.3193],[352,-0.2958],[353,-0.2768],[354,-0.266],[355,-0.2582],[356,-0.2496],[357,-0.2404],[358,-0.2326],[359,-0.2269],[360,-0.2222],[361,-0.2186],[362,-0.2158],[363,-0.2138],[364,-0.2125],[366,-0.2111],[370,-0.2106],[371,-0.2111],[372,-0.2122],[373,-0.2139],[375,-0.2188],[376,-0.2226],[377,-0.2278],[381,-0.2541],[382,-0.2598],[383,-0.2641],[385,-0.271],[386,-0.2722],[387,-0.2682],[388,-0.2594],[389,-0.2471],[390,-0.232],[391,-0.2152],[393,-0.1805],[394,-0.164],[395,-0.1494],[396,-0.137],[397,-0.1262],[398,-0.1174],[399,-0.1107],[400,-0.1061],[401,-0.1032],[402,-0.1014],[403,-0.1005],[404,-0.1002],[405,-0.1006],[406,-0.1022],[407,-0.1048],[408,-0.1086],[409,-0.1131],[411,-0.1229],[412,-0.1274],[413,-0.1311],[415,-0.138],[416,-0.1412],[417,-0.1439],[418,-0.1457],[419,-0.1463],[422,-0.1463]]
       }
       head_ik_ctrl.Rotation.X {
        type=MovieSceneScriptingFloatChannel
        default=-5.6359
        thinned_from=423
        keys [[0,1.1916],[1,1.2066],[2,1.2407],[3,1.287],[5,1.3993],[6,1.4457],[7,1.4791],[8,1.4901],[9,1.4678],[10,1.3994],[11,1.2856],[12,1.1138],[13,0.8833],[14,0.5786],[15,0.1996],[16,-0.2263],[17,-0.6706],[18,-1.1956],[19,-1.7485],[20,-2.1865],[21,-2.5249],[22,-2.8218],[23,-3.0211],[24,-3.1703],[25,-3.2882],[26,-3.3499],[27,-3.3162],[28,-3.2168],[29,-3.1256],[30,-3.0534],[31,-3.0016],[32,-2.9644],[41,-2.7437],[43,-2.7081],[44,-2.6979],[45,-2.6944],[47,-2.7099],[49,-2.7429],[52,-2.8108],[57,-2.9371],[58,-2.9715],[60,-3.0576],[61,-3.096],[62,-3.1199],[63,-3.0957],[64,-3.02],[65,-2.9334],[66,-2.8986],[67,-2.9236],[69,-3.0189],[70,-3.0912],[71,-3.1865],[72,-3.2993],[73,-3.4308],[75,-3.7392],[76,-3.8824],[77,-4.0421],[78,-4.2107],[79,-4.3269],[80,-4.4007],[81,-4.46],[82,-4.4804],[83,-4.4597],[84,-4.417],[85,-4.3589],[87,-4.215],[96,-3.4938],[101,-3.1076],[102,-3.0369],[103,-2.974],[104,-2.9207],[106,-2.8337],[108,-2.7614],[112,-2.6491],[116,-2.5575],[119,-2.5069],[122,-2.4735],[123,-2.4526],[124,-2.4247],[125,-2.391],[126,-2.302],[127,-2.1291],[128,-1.9104],[129,-1.6744],[130,-1.4266],[131,-1.1876],[132,-0.8784],[133,-0.5543],[134,0.0915],[135,0.7078],[136,0.1216],[137,-1.672],[138,-4.253],[139,-9.9794],[140,-14.2445],[141,-16.8348],[142,-18.7311],[143,-20.0279],[144,-21.045],[145,-21.8094],[146,-22.0931],[147,-20.249],[148,-16.4649],[149,-11.4332],[150,-6.133],[151,-2.2931],[152,0.6491],[153,2.0634],[154,2.5619],[155,2.0268],[156,1.3612],[157,0.8035],[158,0.3537],[159,0.1149],[160,-0.0139],[161,-0.0232],[162,0.0256],[163,0.0993],[164,0.1902],[165,0.2674],[166,0.3256],[167,0.3786],[168,0.4224],[169,0.4529],[170,0.475],[171,0.4961],[173,0.53],[175,0.5513],[176,0.5564],[177,0.5637],[181,0.5786],[183,0.5848],[186,0.5903],[188,0.593],[189,0.5954],[190,0.594],[191,0.594],[192,0.5976],[193,0.5976],[194,0.6011],[195,0.5996],[196,0.603],[199,0.6092],[202,0.6206],[204,0.6308],[205,0.6375],[208,0.6638],[210,0.6855],[212,0.7119],[213,0.7269],[214,0.7449],[215,0.7715],[216,0.832],[217,1.0704],[218,1.4315],[219,1.7737],[220,2.049],[221,2.1326],[222,2.0549],[223,1.6801],[224,1.0026],[225,0.0261],[226,-1.238],[227,-2.9566],[228,-4.9888],[229,-7.2323],[230,-9.0652],[231,-10.3943],[232,-11.459],[233,-12.3261],[234,-13.0954],[235,-13.58],[238,-14.8564],[242,-16.3584],[243,-15.5857],[245,-9.6862],[246,-6.3815],[247,-3.7871],[248,-2.0625],[249,-0.6281],[250,0.3366],[251,0.8701],[252,1.1464],[253,1.1346],[254,0.8866],[255,0.3425],[256,-0.3431],[257,-1.0353],[258,-1.5986],[259,-2.0289],[260,-2.4088],[261,-2.7336],[262,-2.9947],[263,-3.1805],[264,-3.3162],[265,-3.4327],[266,-3.5307],[267,-3.6107],[268,-3.6701],[269,-3.7084],[270,-3.7293],[271,-3.7368],[272,-3.7347],[273,-3.7095],[274,-3.6677],[277,-3.4996],[279,-3.3619],[283,-3.0606],[285,-2.9265],[286,-2.8712],[287,-2.8259],[288,-2.7989],[289,-2.7837],[290,-2.7787],[293,-2.7974],[295,-2.817],[296,-2.8219],[297,-2.8168],[298,-2.7553],[299,-2.6141],[300,-2.424],[301,-2.1602],[302,-1.8352],[303,-1.5424],[304,-1.3057],[305,-1.1094],[306,-1.0078],[307,-1.0163],[308,-1.0959],[309,-1.2471],[310,-1.4522],[311,-1.6633],[312,-1.8651],[313,-2.0617],[314,-2.2346],[315,-2.3782],[316,-2.4678],[317,-2.5224],[318,-2.5532],[319,-2.5683],[320,-2.5757],[321,-2.5745],[324,-2.5475],[325,-2.5426],[326,-2.5461],[327,-2.52],[328,-2.4447],[329,-2.3331],[330,-2.1966],[332,-1.8987],[333,-1.7617],[334,-1.6482],[336,-1.45],[337,-1.3554],[338,-1.2709],[339,-1.2021],[340,-1.1547],[341,-1.1245],[342,-1.1011],[344,-1.0658],[345,-1.041],[346,-1.0056],[348,-0.9214],[350,-0.8138],[351,-0.7914],[352,-0.8478],[353,-0.9697],[354,-1.1292],[356,-1.4705],[357,-1.6262],[358,-1.7587],[359,-1.8671],[360,-1.9569],[361,-2.0327],[365,-2.293],[366,-2.3532],[367,-2.4063],[368,-2.4538],[369,-2.4962],[370,-2.5275],[371,-2.5481],[372,-2.5488],[373,-2.5089],[374,-2.4074],[375,-2.2234],[376,-1.8149],[377,-1.1635],[378,-0.4216],[379,0.264],[380,0.9084],[381,1.5705],[382,2.167],[383,2.6222],[384,2.9016],[385,3.0541],[386,3.1259],[387,3.1444],[388,3.099],[389,2.9927],[390,2.8543],[391,2.7087],[392,2.538],[393,2.3301],[394,2.1158],[395,1.9278],[396,1.7665],[397,1.614],[398,1.476],[399,1.358],[400,1.2657],[401,1.1961],[402,1.1412],[403,1.0989],[404,1.0673],[405,1.0453],[406,1.0318],[407,1.0261],[408,1.0276],[409,1.0345],[410,1.0453],[413,1.0837],[414,1.0991],[415,1.1212],[417,1.1684],[418,1.1851],[419,1.1893],[422,1.1893]]
       }
       head_ik_ctrl.Rotation.Y {
        type=MovieSceneScriptingFloatChannel
        default=7.9933
        thinned_from=423
        keys [[0,7.0928],[12,7.0287],[13,7.0157],[14,6.8965],[15,6.5147],[16,5.789],[17,4.6265],[18,3.0019],[19,1.0406],[20,-1.0876],[21,-3.1907],[22,-5.1288],[23,-6.7899],[24,-8.0657],[25,-8.9944],[26,-9.7198],[27,-10.2694],[28,-10.658],[29,-10.9256],[30,-11.1057],[31,-11.2204],[32,-11.2889],[33,-11.3268],[35,-11.3272],[41,-11.0802],[46,-11.0258],[60,-11.1579],[65,-11.2823],[66,-11.2609],[67,-11.1038],[68,-10.7369],[69,-10.1218],[70,-9.316],[71,-8.3764],[72,-7.3342],[73,-6.1278],[75,-3.4196],[76,-2.1474],[77,-0.9044],[78,0.2952],[79,1.2811],[80,1.9721],[81,2.4437],[82,2.7651],[83,2.9663],[84,3.0587],[85,3.0738],[87,3.0248],[88,3.0155],[89,3.0205],[90,3.0343],[92,3.0789],[94,3.1385],[102,3.436],[105,3.5233],[108,3.5829],[114,3.6539],[116,3.658],[117,3.6504],[124,3.5129],[125,3.4885],[126,3.3698],[127,3.1187],[128,2.4605],[129,1.7219],[130,0.9159],[131,0.0617],[132,-0.8687],[133,-1.8707],[134,-2.9754],[135,-4.0677],[136,-5.0683],[137,-6.1706],[138,-7.4445],[139,-8.2695],[140,-8.5775],[141,-7.7475],[142,-6.1684],[143,-4.4758],[144,-2.7214],[145,-1.1699],[146,-0.0054],[147,0.9065],[148,1.5683],[149,1.9651],[150,2.163],[151,2.251],[152,2.2953],[153,2.4214],[154,2.6694],[155,2.9828],[156,3.2896],[157,3.5338],[158,3.7138],[159,3.8474],[160,3.939],[161,3.9923],[166,4.1634],[173,4.3439],[180,4.4522],[185,4.4863],[189,4.4909],[194,4.4991],[218,4.4614],[220,4.4708],[221,4.4436],[222,4.3859],[223,4.292],[224,4.1465],[225,3.9583],[226,3.7363],[227,3.5056],[228,3.2898],[229,3.095],[230,2.9292],[231,2.7871],[232,2.5262],[233,2.373],[234,2.3008],[235,2.2834],[236,2.3064],[237,2.3489],[238,2.4036],[239,2.4738],[240,2.5622],[241,2.6627],[242,2.7773],[244,3.0282],[245,3.5129],[246,4.4541],[247,5.479],[248,6.2952],[249,6.8297],[250,7.1873],[251,7.3617],[252,7.1221],[253,6.8143],[254,6.4635],[255,6.1374],[257,5.5228],[258,5.24],[259,5.0218],[260,4.8841],[261,4.8045],[262,4.7649],[267,4.6534],[269,4.5989],[273,4.5405],[288,4.5439],[290,4.5696],[292,4.6201],[297,4.8225],[299,4.8883],[300,4.9472],[301,5.0187],[302,5.0626],[303,5.0705],[304,5.04],[305,4.9668],[306,4.7782],[307,4.4591],[308,4.0798],[309,3.5951],[310,3.0413],[311,2.6082],[312,2.3205],[313,2.0829],[314,1.8935],[315,1.7581],[316,1.6784],[317,1.6548],[318,1.6679],[319,1.6981],[320,1.7329],[322,1.7923],[323,1.8183],[324,1.8354],[326,1.8511],[328,1.8808],[329,1.904],[333,2.0251],[343,2.2678],[345,2.326],[346,2.3611],[347,2.6908],[348,3.4458],[349,4.3962],[350,5.3077],[351,6.2002],[352,7.1662],[353,8.0815],[354,8.8351],[355,9.4088],[356,9.8509],[357,10.1596],[358,10.3304],[359,10.409],[360,10.4533],[375,10.691],[376,10.6638],[377,10.5501],[378,10.3557],[379,10.0816],[380,9.7408],[381,9.3527],[383,8.5211],[384,8.1538],[385,7.8681],[386,7.6336],[387,7.4161],[388,7.252],[389,7.176],[390,7.1653],[393,7.2655],[395,7.2757],[401,7.1632],[405,7.0508],[422,7.0929]]
       }
       head_ik_ctrl.Rotation.Z {
        type=MovieSceneScriptingFloatChannel
        default=8.8749
        thinned_from=423
        keys [[0,9.7847],[4,9.73],[5,9.732],[6,9.7739],[7,9.8559],[8,9.9697],[9,10.1212],[10,10.3052],[11,10.5186],[13,11.0982],[14,11.9203],[15,13.5508],[16,15.7425],[17,18.2102],[18,21.4158],[19,25.1523],[20,28.2465],[21,30.4464],[22,32.1542],[23,33.3348],[24,34.2983],[25,35.1713],[26,35.7547],[27,35.9264],[29,35.7737],[30,35.6081],[31,35.3054],[32,34.8064],[35,32.952],[36,32.4005],[37,31.9187],[38,31.5276],[40,30.9246],[43,30.2461],[46,29.7681],[48,29.5536],[54,29.2355],[56,29.1987],[62,28.9374],[63,28.8438],[64,28.6722],[65,28.4274],[66,28.1229],[67,27.7542],[68,27.2659],[69,26.6273],[70,25.9077],[71,25.1143],[72,24.1096],[73,22.7927],[75,19.7334],[76,18.3008],[77,16.9616],[78,15.6739],[79,14.5819],[80,13.7531],[81,13.1417],[82,12.7269],[83,12.4213],[85,11.8934],[90,10.7349],[95,9.6626],[99,8.8695],[101,8.4527],[102,8.2595],[103,8.0884],[104,7.9479],[106,7.7309],[109,7.4848],[112,7.3375],[115,7.2383],[116,7.2444],[117,7.2743],[118,7.3378],[119,7.4245],[121,7.6498],[124,8.0336],[125,8.2465],[127,8.8454],[128,9.1628],[129,9.5231],[130,9.9604],[131,10.4562],[132,10.9784],[133,11.5411],[136,13.3573],[137,13.7357],[138,13.8038],[139,13.8143],[140,13.5726],[141,12.8147],[142,11.5637],[143,9.8281],[144,8.0316],[145,5.2352],[146,3.5999],[147,3.2593],[148,3.3601],[149,3.9471],[150,4.6213],[151,5.2089],[152,6.167],[153,7.0553],[154,7.4779],[155,7.5263],[156,7.373],[157,7.2702],[158,7.2675],[159,7.2418],[160,7.1931],[161,7.1209],[165,6.74],[170,6.215],[171,6.1242],[172,6.0148],[173,5.9372],[174,5.8378],[176,5.699],[179,5.5673],[182,5.4917],[190,5.345],[191,5.345],[192,5.3293],[193,5.3293],[194,5.3126],[195,5.3228],[196,5.3018],[197,5.3277],[214,5.3905],[222,5.5024],[225,5.499],[226,5.4832],[227,5.4404],[228,5.3534],[229,5.2366],[230,5.0933],[232,4.7321],[233,4.6316],[234,4.7127],[237,5.3962],[238,5.5788],[239,5.6996],[240,5.7865],[241,5.8534],[243,5.9537],[244,6.0175],[245,6.1264],[246,6.3339],[247,6.7162],[248,7.4033],[249,8.0459],[250,8.3451],[251,8.4548],[252,8.4813],[253,8.479],[254,8.4535],[256,8.341],[262,7.8681],[263,7.8051],[265,7.7226],[268,7.659],[272,7.6423],[273,7.6574],[284,8.0656],[287,8.1457],[288,8.2196],[291,8.5457],[292,8.6219],[293,8.6511],[294,8.5919],[296,8.3531],[297,8.3278],[298,8.4161],[299,8.5864],[300,8.8307],[301,9.1201],[302,9.6173],[303,10.3145],[304,11.0759],[305,11.8628],[306,12.5281],[307,13.0128],[308,13.3887],[310,13.9863],[311,14.205],[312,14.298],[313,14.2633],[314,14.1069],[315,13.8611],[316,13.4777],[317,12.9398],[322,9.9783],[323,9.4208],[324,8.9018],[325,8.4455],[326,8.0293],[327,7.6834],[328,7.4048],[329,7.1985],[330,7.0521],[331,6.952],[332,6.8856],[335,6.7621],[336,6.7039],[337,6.6159],[338,6.4854],[339,6.3],[340,6.0472],[341,5.6943],[342,5.297],[343,4.7711],[344,3.9793],[345,2.8294],[346,1.213],[347,-0.9792],[348,-3.8359],[349,-7.8916],[350,-13.229],[351,-19.2455],[352,-25.3472],[353,-30.9257],[354,-35.4419],[355,-38.5535],[356,-40.5496],[357,-41.8799],[359,-44.0581],[360,-44.9496],[361,-45.6924],[362,-46.31],[363,-46.826],[367,-48.2236],[370,-48.9311],[374,-49.4005],[375,-49.4533],[376,-49.2124],[377,-48.4356],[378,-47.176],[379,-45.4988],[380,-43.496],[381,-40.6491],[382,-36.7735],[383,-32.4652],[384,-27.9689],[385,-23.3261],[386,-18.8307],[387,-14.7792],[388,-11.1314],[389,-7.6911],[390,-4.5198],[391,-1.7571],[392,0.5248],[393,2.3347],[394,3.6079],[395,4.281],[397,4.9865],[398,5.3293],[399,5.6556],[400,5.9579],[401,6.2826],[402,6.6643],[404,7.497],[405,7.8978],[406,8.2551],[407,8.5558],[408,8.8077],[409,9.0167],[410,9.1887],[411,9.3297],[412,9.4456],[414,9.6221],[416,9.7303],[419,9.7893],[422,9.7893]]
       }
       head_ik_ctrl.Scale.X {type=MovieSceneScriptingFloatChannel, default=1, const [1,423]}
       head_ik_ctrl.Scale.Y {type=MovieSceneScriptingFloatChannel, default=1, const [1,423]}
       head_ik_ctrl.Scale.Z {type=MovieSceneScriptingFloatChannel, default=1, const [1,423]}
       chest_ctrl.Location.X {
        type=MovieSceneScriptingFloatChannel
        default=0.6396
        thinned_from=423
        keys [[0,0.563],[12,0.5666],[14,0.5663],[15,0.568],[16,0.5722],[17,0.5825],[18,0.6015],[19,0.6285],[20,0.6577],[21,0.6837],[22,0.7066],[23,0.7279],[24,0.7474],[25,0.7636],[27,0.7897],[29,0.8119],[30,0.8209],[31,0.8281],[32,0.8335],[33,0.8367],[34,0.8374],[35,0.8361],[36,0.8329],[37,0.828],[40,0.8081],[41,0.8034],[44,0.7958],[48,0.7915],[57,0.7898],[65,0.8124],[67,0.8144],[68,0.8138],[69,0.8114],[70,0.8054],[71,0.7944],[72,0.7783],[73,0.7562],[74,0.7237],[75,0.6811],[76,0.6326],[77,0.5773],[78,0.5156],[79,0.4499],[81,0.3127],[82,0.2519],[83,0.2009],[84,0.1563],[85,0.1191],[86,0.0889],[87,0.0654],[88,0.0473],[89,0.0335],[90,0.0235],[91,0.0172],[92,0.0131],[93,0.0105],[94,0.0094],[95,0.0098],[96,0.0107],[98,0.0158],[101,0.0257],[102,0.0285],[103,0.0306],[104,0.0324],[105,0.0338],[107,0.0362],[108,0.0372],[110,0.0382],[111,0.039],[112,0.0393],[113,0.0399],[114,0.0402],[115,0.0402],[117,0.0422],[120,0.0479],[121,0.0487],[122,0.049],[123,0.0496],[124,0.0508],[125,0.0531],[126,0.0606],[127,0.0761],[128,0.0987],[129,0.1269],[130,0.1677],[131,0.2223],[132,0.2935],[133,0.3946],[134,0.5309],[135,0.6837],[136,0.8351],[137,0.9997],[138,1.1698],[139,1.2302],[140,1.2445],[142,1.1964],[143,1.189],[147,1.13],[148,1.1228],[149,1.1266],[150,1.1437],[151,1.1525],[152,1.1535],[153,1.1421],[154,1.1247],[155,1.1196],[158,1.1133],[162,1.1176],[170,1.1357],[224,1.122],[232,1.104],[234,1.1057],[235,1.1084],[237,1.1077],[238,1.1036],[239,1.0956],[240,1.083],[241,1.0673],[242,1.0488],[243,1.0274],[244,0.9984],[245,0.9542],[246,0.8959],[247,0.8285],[248,0.7536],[249,0.698],[250,0.6671],[251,0.655],[252,0.6595],[254,0.6834],[257,0.7075],[259,0.7199],[263,0.737],[267,0.7478],[270,0.7502],[285,0.7354],[288,0.7348],[296,0.7475],[298,0.7497],[299,0.7494],[300,0.7412],[301,0.7172],[302,0.6779],[303,0.6297],[304,0.5742],[305,0.5101],[306,0.4398],[307,0.3591],[308,0.2729],[309,0.1967],[310,0.1321],[311,0.0718],[312,0.0179],[313,-0.0276],[314,-0.0678],[315,-0.1072],[316,-0.1452],[317,-0.1808],[318,-0.2118],[319,-0.2358],[320,-0.2518],[321,-0.2615],[322,-0.269],[323,-0.2742],[324,-0.2767],[325,-0.2767],[326,-0.275],[327,-0.2713],[333,-0.2359],[334,-0.2307],[335,-0.2269],[336,-0.224],[338,-0.2196],[340,-0.2164],[341,-0.2162],[343,-0.2149],[345,-0.2164],[346,-0.2177],[347,-0.22],[348,-0.2243],[349,-0.2318],[350,-0.2437],[351,-0.2668],[352,-0.3019],[353,-0.3417],[354,-0.3785],[356,-0.4486],[357,-0.481],[358,-0.5066],[359,-0.5254],[360,-0.5403],[361,-0.552],[362,-0.5611],[363,-0.5684],[364,-0.5741],[365,-0.5782],[366,-0.5806],[367,-0.5812],[368,-0.5802],[369,-0.5779],[370,-0.5736],[371,-0.5668],[372,-0.5583],[373,-0.5479],[374,-0.5351],[375,-0.5196],[376,-0.4962],[377,-0.4612],[378,-0.4146],[379,-0.3575],[380,-0.2875],[381,-0.2055],[382,-0.1191],[383,-0.0359],[384,0.0438],[385,0.1221],[386,0.1975],[387,0.2682],[388,0.331],[389,0.3862],[390,0.4346],[391,0.4768],[392,0.511],[393,0.5366],[394,0.5565],[395,0.5721],[396,0.584],[397,0.5926],[398,0.5985],[400,0.6051],[401,0.6068],[403,0.6063],[405,0.6029],[408,0.592],[412,0.5751],[416,0.5657],[418,0.5633],[422,0.5633]]
       }
       chest_ctrl.Location.Y {
        type=MovieSceneScriptingFloatChannel
        default=-3.7508
        thinned_from=423
        keys [[0,-1.2284],[13,-1.2351],[14,-1.233],[15,-1.2283],[16,-1.2202],[17,-1.2064],[18,-1.1845],[19,-1.1481],[20,-1.1054],[21,-1.0789],[22,-1.0704],[24,-1.0615],[29,-1.0566],[33,-1.0616],[41,-1.0825],[58,-1.0902],[62,-1.0814],[65,-1.0622],[66,-1.0531],[68,-1.0289],[71,-1.0009],[73,-0.9892],[74,-0.9862],[75,-0.9857],[76,-0.9894],[77,-1.0058],[79,-1.0639],[80,-1.0913],[81,-1.1161],[82,-1.1358],[83,-1.1502],[84,-1.1613],[85,-1.1695],[88,-1.1833],[95,-1.1961],[125,-1.1824],[126,-1.186],[127,-1.1962],[128,-1.2324],[129,-1.2966],[130,-1.3691],[131,-1.4469],[133,-1.6115],[134,-1.7044],[135,-1.7937],[136,-1.8761],[137,-1.9363],[138,-1.9142],[139,-1.9736],[140,-2.0441],[141,-2.0977],[142,-2.1302],[143,-2.1387],[146,-2.1925],[147,-2.2006],[148,-2.1871],[149,-2.1544],[150,-2.13],[151,-2.0872],[152,-2.0576],[153,-2.0596],[155,-2.0862],[159,-2.1515],[161,-2.1744],[165,-2.1963],[209,-2.1992],[217,-2.1702],[226,-2.1043],[227,-2.1006],[230,-2.1081],[234,-2.1458],[235,-2.1514],[238,-2.156],[241,-2.1748],[242,-2.178],[243,-2.1722],[244,-2.1237],[245,-2.015],[246,-1.8833],[248,-1.5888],[249,-1.4464],[250,-1.3226],[251,-1.2341],[252,-1.1895],[253,-1.1879],[254,-1.1991],[256,-1.2264],[257,-1.2378],[261,-1.2615],[263,-1.2676],[267,-1.2659],[272,-1.2468],[275,-1.2279],[280,-1.2162],[287,-1.2133],[294,-1.2325],[296,-1.2341],[297,-1.2328],[299,-1.2222],[300,-1.215],[301,-1.2054],[303,-1.1764],[307,-1.1016],[308,-1.0809],[309,-1.0626],[310,-1.0511],[311,-1.0467],[312,-1.0461],[314,-1.0523],[319,-1.0803],[321,-1.0864],[343,-1.0886],[345,-1.095],[346,-1.1002],[347,-1.1077],[348,-1.118],[349,-1.1309],[350,-1.1467],[351,-1.1661],[353,-1.2105],[354,-1.2297],[356,-1.2604],[357,-1.2725],[359,-1.2879],[362,-1.3003],[371,-1.3063],[375,-1.2984],[387,-1.2313],[390,-1.223],[392,-1.2241],[399,-1.2514],[407,-1.2535],[418,-1.229],[422,-1.2289]]
       }
       chest_ctrl.Location.Z {
        type=MovieSceneScriptingFloatChannel
        default=-0.3885
        thinned_from=423
        keys [[0,0.1054],[13,0.1048],[14,0.105],[15,0.1055],[16,0.1063],[17,0.1077],[18,0.11],[19,0.1138],[20,0.1183],[21,0.1212],[22,0.1223],[24,0.1235],[29,0.1243],[33,0.124],[41,0.1218],[58,0.1209],[62,0.1219],[65,0.1238],[66,0.1211],[67,0.114],[69,0.0974],[71,0.0787],[74,0.0516],[75,0.0432],[76,0.0359],[77,0.0293],[78,0.0237],[79,0.02],[80,0.0177],[81,0.0158],[82,0.0143],[83,0.0133],[84,0.0125],[86,0.0116],[93,0.0098],[99,0.0097],[119,0.0113],[126,0.0105],[127,0.0106],[128,0.0115],[129,0.0134],[130,0.016],[131,0.0195],[132,0.024],[133,0.0312],[134,0.0575],[135,0.0947],[136,0.1343],[137,0.1681],[138,0.0983],[139,0.5635],[140,1.3796],[141,1.7641],[142,1.8731],[143,1.7752],[144,1.7291],[145,1.7188],[146,1.7167],[147,1.5182],[148,0.981],[149,0.1946],[150,-0.2591],[151,-0.3704],[152,-0.2601],[153,-0.1098],[154,0.0003],[155,0.1276],[156,0.2118],[159,0.2065],[161,0.2039],[165,0.2014],[205,0.2003],[215,0.2033],[220,0.2066],[221,0.2018],[222,0.1821],[223,0.1549],[224,0.117],[225,0.0854],[226,0.0745],[227,0.1307],[228,0.24],[229,0.3586],[230,0.4172],[231,0.4433],[232,0.4489],[235,0.4458],[240,0.4438],[241,0.4426],[242,0.4402],[243,0.4356],[244,0.3316],[245,0.1546],[246,-0.0137],[247,-0.1079],[248,-0.0763],[249,0.0017],[250,0.09],[251,0.1439],[252,0.1662],[253,0.1682],[254,0.1602],[255,0.147],[256,0.1354],[257,0.1246],[258,0.1157],[259,0.1102],[260,0.1071],[262,0.1062],[266,0.1061],[268,0.1065],[272,0.108],[274,0.1079],[276,0.1087],[286,0.1098],[295,0.1078],[297,0.1077],[301,0.1092],[302,0.1098],[303,0.1106],[304,0.1118],[305,0.1134],[306,0.1153],[307,0.1178],[309,0.1243],[310,0.1269],[311,0.1288],[312,0.1303],[313,0.1311],[315,0.132],[317,0.132],[323,0.1301],[329,0.1301],[342,0.1314],[345,0.1307],[347,0.1293],[348,0.1282],[349,0.1268],[350,0.125],[351,0.1227],[353,0.117],[354,0.1146],[356,0.1106],[357,0.109],[358,0.1079],[360,0.1064],[362,0.1055],[375,0.104],[376,0.1038],[381,0.1009],[382,0.1007],[383,0.1009],[384,0.1014],[388,0.1046],[390,0.1057],[391,0.1059],[392,0.106],[399,0.1036],[407,0.1032],[418,0.1054],[422,0.1054]]
       }
       chest_ctrl.Rotation.X {
        type=MovieSceneScriptingFloatChannel
        default=-12.7859
        thinned_from=423
        keys [[0,-5.9487],[12,-6.0112],[14,-5.9875],[15,-5.9537],[16,-5.8998],[17,-5.8198],[18,-5.7058],[19,-5.5056],[20,-5.2725],[21,-5.1575],[22,-5.162],[29,-5.296],[48,-5.4066],[59,-5.4059],[62,-5.3745],[68,-5.1985],[71,-5.1384],[75,-5.1039],[76,-5.1114],[77,-5.1838],[79,-5.467],[81,-5.7199],[82,-5.8216],[83,-5.8964],[84,-5.9556],[85,-6.001],[88,-6.0788],[95,-6.1487],[125,-6.0444],[126,-6.057],[127,-6.0973],[128,-6.2471],[129,-6.5145],[130,-6.8163],[131,-7.1419],[134,-8.1719],[137,-9.1383],[138,-9.4117],[139,-9.4673],[140,-9.3201],[141,-9.3641],[143,-9.5977],[144,-9.6914],[146,-9.8066],[147,-9.9374],[148,-10.2341],[149,-10.6816],[150,-11.0122],[151,-10.9237],[152,-10.6948],[153,-10.5976],[155,-10.5305],[156,-10.5618],[160,-10.8831],[164,-11.0405],[205,-11.1147],[214,-10.9756],[223,-10.6136],[225,-10.5606],[229,-10.3251],[230,-10.324],[231,-10.3699],[234,-10.6121],[235,-10.6573],[241,-10.6853],[242,-10.6726],[243,-10.6195],[244,-10.4704],[245,-10.1084],[246,-9.5837],[247,-8.8776],[249,-7.2493],[250,-6.5612],[251,-6.1042],[252,-5.887],[253,-5.8915],[254,-5.9606],[257,-6.2007],[261,-6.4059],[262,-6.4475],[265,-6.4989],[267,-6.5085],[269,-6.4832],[272,-6.4087],[275,-6.286],[280,-6.2085],[286,-6.1811],[288,-6.2005],[294,-6.3279],[297,-6.3513],[298,-6.3457],[299,-6.3271],[300,-6.2866],[301,-6.2081],[302,-6.092],[303,-5.9556],[304,-5.7987],[307,-5.2691],[308,-5.0831],[309,-4.9226],[310,-4.8155],[311,-4.7621],[312,-4.7413],[313,-4.7454],[314,-4.7657],[316,-4.8365],[319,-4.9723],[320,-5.0049],[324,-5.0522],[329,-5.0631],[341,-5.0159],[343,-5.0254],[345,-5.0676],[346,-5.1039],[347,-5.1565],[348,-5.2285],[349,-5.3209],[350,-5.4347],[351,-5.5801],[353,-5.9254],[354,-6.0782],[356,-6.3298],[357,-6.4322],[358,-6.5096],[360,-6.6124],[362,-6.6758],[366,-6.7212],[371,-6.7155],[374,-6.6533],[376,-6.5826],[379,-6.4181],[385,-6.0205],[387,-5.9206],[389,-5.8573],[390,-5.8393],[391,-5.8332],[392,-5.8464],[399,-6.0931],[402,-6.136],[407,-6.1419],[418,-5.9539],[422,-5.9511]]
       }
       chest_ctrl.Rotation.Y {
        type=MovieSceneScriptingFloatChannel
        default=-3.6817
        thinned_from=423
        keys [[0,-3.4472],[12,-3.4311],[17,-3.474],[20,-3.4628],[21,-3.4355],[22,-3.4177],[23,-3.4199],[24,-3.4347],[25,-3.4643],[26,-3.5047],[28,-3.6052],[31,-3.775],[32,-3.8253],[33,-3.8671],[34,-3.8953],[35,-3.9122],[36,-3.9166],[37,-3.9076],[41,-3.7957],[44,-3.7649],[57,-3.7616],[61,-3.8204],[62,-3.823],[63,-3.8156],[66,-3.7543],[69,-3.6682],[70,-3.6341],[71,-3.5815],[72,-3.5102],[73,-3.4231],[74,-3.3048],[77,-2.8449],[79,-2.5156],[80,-2.3401],[81,-2.1551],[82,-1.9952],[83,-1.8647],[84,-1.7523],[85,-1.6647],[86,-1.5993],[87,-1.553],[88,-1.5226],[89,-1.5054],[90,-1.4999],[91,-1.506],[93,-1.5299],[95,-1.5576],[96,-1.5678],[97,-1.5816],[98,-1.5902],[99,-1.5935],[101,-1.5904],[108,-1.5661],[115,-1.5596],[117,-1.5647],[119,-1.5829],[120,-1.5965],[122,-1.6387],[125,-1.7146],[126,-1.7636],[127,-1.8515],[128,-1.9787],[129,-2.1442],[130,-2.3622],[131,-2.6291],[132,-2.9717],[133,-3.4626],[134,-4.0709],[135,-4.6954],[137,-5.8177],[138,-5.7874],[139,-5.736],[140,-5.3749],[141,-4.9462],[142,-4.6607],[143,-4.584],[144,-4.5598],[145,-4.6039],[146,-4.697],[147,-4.8434],[148,-5.0265],[149,-5.1915],[150,-5.1514],[151,-4.9855],[152,-4.7945],[153,-4.5544],[154,-4.3731],[155,-4.3254],[156,-4.3305],[158,-4.4113],[160,-4.5419],[161,-4.5997],[164,-4.7409],[166,-4.8164],[169,-4.8931],[172,-4.9373],[179,-4.9412],[221,-4.8811],[224,-4.8979],[225,-4.8948],[226,-4.8742],[227,-4.8166],[228,-4.7247],[231,-4.3872],[232,-4.3052],[233,-4.2518],[234,-4.2254],[236,-4.2195],[237,-4.2252],[239,-4.2637],[242,-4.2781],[243,-4.2716],[244,-4.2485],[245,-4.1658],[246,-4.0117],[247,-3.7965],[248,-3.5154],[249,-3.3093],[250,-3.2114],[251,-3.1907],[252,-3.2371],[254,-3.3833],[256,-3.4892],[257,-3.5327],[258,-3.5631],[259,-3.5789],[261,-3.5822],[263,-3.5682],[277,-3.3451],[283,-3.2916],[288,-3.2839],[295,-3.3307],[296,-3.3284],[297,-3.3122],[298,-3.273],[300,-3.1533],[301,-3.0546],[302,-2.8987],[303,-2.7169],[304,-2.5184],[305,-2.2957],[306,-2.0538],[307,-1.7734],[308,-1.4752],[309,-1.225],[310,-1.0314],[311,-0.864],[312,-0.7291],[313,-0.6327],[314,-0.5585],[315,-0.4925],[316,-0.4352],[317,-0.3879],[318,-0.3512],[319,-0.3248],[320,-0.3158],[321,-0.3089],[322,-0.2956],[323,-0.2845],[324,-0.2815],[325,-0.2866],[326,-0.3017],[327,-0.3132],[328,-0.3209],[329,-0.3264],[330,-0.3273],[331,-0.3264],[333,-0.3277],[335,-0.3238],[341,-0.3311],[342,-0.333],[344,-0.3394],[346,-0.3424],[347,-0.3427],[348,-0.3332],[349,-0.3042],[350,-0.2455],[351,-0.1085],[352,0.1138],[353,0.3693],[354,0.606],[355,0.829],[356,1.0596],[357,1.2712],[358,1.4371],[359,1.5566],[360,1.6502],[361,1.7226],[362,1.7783],[363,1.8219],[364,1.8557],[365,1.8799],[366,1.8958],[367,1.9042],[368,1.9073],[370,1.8976],[371,1.8896],[372,1.8746],[373,1.8503],[374,1.8144],[375,1.7646],[376,1.6772],[377,1.5351],[378,1.3334],[379,1.0719],[380,0.7433],[381,0.3538],[382,-0.0606],[383,-0.4643],[384,-0.8539],[385,-1.2357],[386,-1.6099],[387,-1.9732],[388,-2.3066],[389,-2.6077],[390,-2.8751],[391,-3.1072],[392,-3.2915],[393,-3.4267],[394,-3.5294],[395,-3.6065],[396,-3.6615],[397,-3.6961],[398,-3.7139],[399,-3.7183],[402,-3.6956],[405,-3.6566],[412,-3.4969],[415,-3.462],[422,-3.4475]]
       }
       chest_ctrl.Rotation.Z {
        type=MovieSceneScriptingFloatChannel
        default=0
        thinned_from=423
        keys [[0,-0.4131],[1,-0.4131],[2,-0.3889],[3,-0.346],[4,-0.309],[5,-0.2841],[6,-0.2568],[7,-0.2332],[8,-0.2161],[9,-0.2],[10,-0.1901],[12,-0.1901],[13,-0.2512],[14,-0.2986],[15,-0.3221],[16,-0.277],[17,-0.0098],[18,0.6109],[19,1.4397],[20,2.3909],[21,3.4118],[22,4.3221],[23,5.0838],[24,5.7304],[25,6.1848],[26,6.4823],[27,6.7096],[28,6.8682],[29,6.9594],[30,6.9847],[31,6.9505],[32,6.8697],[33,6.7531],[34,6.6116],[37,6.1405],[38,6.003],[39,5.8915],[40,5.8154],[43,5.6658],[44,5.6237],[45,5.5961],[46,5.5507],[47,5.5392],[48,5.4927],[49,5.4927],[50,5.4486],[51,5.4486],[52,5.4369],[53,5.4103],[56,5.4103],[58,5.3945],[59,5.4216],[60,5.4705],[61,5.5434],[62,5.6432],[63,5.8143],[67,6.7666],[68,6.9702],[69,7.0947],[70,7.1438],[71,7.1588],[72,7.1364],[73,7.0291],[74,6.8044],[75,6.4778],[76,5.9678],[77,5.161],[78,4.1818],[79,3.0828],[80,1.9585],[81,0.9244],[82,-0.0326],[83,-0.8782],[84,-1.629],[85,-2.2917],[86,-2.8666],[87,-3.3488],[88,-3.7592],[89,-4.1188],[90,-4.4227],[91,-4.6713],[92,-4.8698],[93,-5.0254],[94,-5.1337],[95,-5.1881],[96,-5.197],[97,-5.1862],[98,-5.1346],[99,-5.0356],[102,-4.6618],[103,-4.5626],[104,-4.48],[105,-4.4155],[107,-4.3091],[108,-4.2632],[111,-4.1808],[114,-4.1305],[115,-4.127],[116,-4.1098],[118,-4.1029],[119,-4.1083],[120,-4.141],[121,-4.2459],[123,-4.5199],[124,-4.639],[125,-4.7221],[126,-4.7615],[127,-4.7042],[128,-4.567],[129,-4.3997],[130,-4.1532],[131,-3.7923],[132,-3.3305],[133,-2.846],[134,-2.1615],[135,-1.2947],[136,-0.2734],[137,1.269],[138,5.7174],[139,6.5314],[140,7.0967],[141,7.4837],[142,7.7262],[143,7.8257],[144,7.5223],[145,6.8907],[147,5.3609],[148,4.9721],[149,5.2156],[150,6.3197],[151,7.4467],[152,8.2117],[153,8.5897],[154,8.5554],[155,8.3644],[159,7.306],[160,7.0562],[161,6.8652],[162,6.7292],[163,6.6177],[164,6.5267],[165,6.4527],[167,6.3403],[170,6.2158],[171,6.1943],[172,6.1557],[173,6.1468],[174,6.1124],[175,6.1063],[176,6.0827],[196,6.0433],[197,6.0716],[211,6.2634],[215,6.2689],[224,6.1554],[225,6.1511],[226,6.1765],[227,6.2677],[228,6.4361],[229,6.7101],[230,7.0479],[231,7.3579],[232,7.5955],[233,7.7625],[234,7.8704],[235,7.941],[236,7.9504],[237,7.9017],[238,7.716],[239,7.4296],[240,7.0605],[241,6.6246],[242,6.1424],[243,5.6346],[244,5.1576],[246,4.3295],[247,3.9327],[248,3.5708],[249,3.2408],[250,2.9763],[251,2.7915],[252,2.6633],[253,2.5817],[254,2.542],[255,2.5267],[256,2.5368],[257,2.5851],[258,2.6693],[259,2.7838],[260,2.9221],[261,3.0786],[262,3.2481],[265,3.786],[266,3.9595],[267,4.1222],[268,4.269],[269,4.3952],[270,4.4962],[271,4.5775],[272,4.6484],[274,4.7605],[276,4.8393],[278,4.8909],[280,4.9201],[283,4.9338],[286,4.9145],[288,4.8957],[291,4.9156],[293,4.9575],[294,4.9911],[295,5.0376],[296,5.1238],[297,5.2857],[298,5.632],[299,6.0543],[300,6.2634],[301,6.246],[302,6.166],[303,5.9983],[304,5.7203],[305,5.3433],[306,4.8966],[307,4.3906],[308,3.8301],[309,3.2142],[310,2.5233],[311,1.7561],[312,0.9386],[313,0.0974],[314,-0.7381],[315,-1.6095],[317,-3.4025],[318,-4.228],[319,-4.8991],[320,-5.426],[321,-5.7679],[322,-5.9859],[323,-6.1276],[324,-6.2028],[326,-6.2457],[327,-6.1604],[328,-5.9869],[329,-5.7814],[330,-5.5473],[332,-5.0613],[333,-4.8359],[334,-4.6215],[335,-4.4588],[336,-4.3494],[337,-4.2628],[338,-4.1928],[340,-4.0773],[342,-4.0509],[343,-4.0501],[345,-4.1185],[346,-4.1654],[347,-4.236],[350,-4.5008],[351,-4.5738],[355,-4.7964],[360,-5.0172],[364,-5.1445],[365,-5.1639],[366,-5.1718],[367,-5.1595],[369,-5.0849],[370,-5.0043],[374,-4.4041],[377,-3.9419],[379,-3.6536],[383,-3.1257],[386,-2.7083],[388,-2.4176],[390,-2.1114],[391,-1.9528],[392,-1.7892],[396,-1.1116],[397,-0.944],[398,-0.7791],[399,-0.6178],[400,-0.4616],[401,-0.3272],[402,-0.2275],[403,-0.1579],[404,-0.1138],[405,-0.09],[406,-0.0815],[407,-0.0852],[408,-0.0994],[409,-0.122],[410,-0.1505],[411,-0.1827],[412,-0.2161],[413,-0.2486],[414,-0.2825],[415,-0.319],[416,-0.3542],[417,-0.3843],[418,-0.4052],[419,-0.4081],[422,-0.4081]]
       }
       chest_ctrl.Scale.X {type=MovieSceneScriptingFloatChannel, default=1, const [1,423]}
       chest_ctrl.Scale.Y {type=MovieSceneScriptingFloatChannel, default=1, const [1,423]}
       chest_ctrl.Scale.Z {type=MovieSceneScriptingFloatChannel, default=1, const [1,423]}
       chest_tan_ctrl.Location.X {
        type=MovieSceneScriptingFloatChannel
        default=0.0754
        thinned_from=423
        keys [[0,0.0754],[12,0.0759],[14,0.0757],[15,0.0758],[16,0.0761],[17,0.077],[18,0.079],[21,0.0871],[24,0.0959],[25,0.0986],[27,0.1031],[29,0.1071],[30,0.1087],[31,0.1101],[32,0.1111],[33,0.1117],[34,0.1118],[35,0.1116],[36,0.111],[37,0.1101],[40,0.1063],[41,0.1054],[45,0.1035],[46,0.1035],[47,0.1032],[49,0.1032],[51,0.103],[52,0.1027],[57,0.103],[58,0.1033],[65,0.1073],[67,0.1079],[68,0.1079],[69,0.1077],[70,0.1072],[71,0.1063],[72,0.105],[73,0.1033],[74,0.1008],[75,0.0976],[76,0.0937],[77,0.0888],[78,0.0832],[79,0.077],[81,0.064],[82,0.0581],[83,0.0529],[84,0.0483],[85,0.0444],[86,0.0412],[87,0.0386],[88,0.0365],[89,0.0348],[90,0.0335],[91,0.0327],[92,0.0322],[93,0.0319],[95,0.032],[96,0.0322],[102,0.0349],[109,0.0362],[120,0.0369],[124,0.0366],[125,0.0368],[126,0.0376],[127,0.0398],[128,0.0431],[129,0.0472],[130,0.0525],[131,0.0589],[132,0.0675],[133,0.0789],[134,0.0924],[135,0.1064],[136,0.1197],[138,0.1543],[139,0.1581],[140,0.1556],[141,0.1471],[142,0.1403],[144,0.1349],[145,0.1327],[146,0.1314],[147,0.1304],[148,0.13],[149,0.1304],[150,0.1323],[151,0.1337],[152,0.1338],[153,0.1301],[154,0.1251],[155,0.123],[156,0.1218],[157,0.1212],[158,0.1211],[162,0.123],[166,0.1257],[170,0.1274],[172,0.1278],[178,0.1275],[189,0.1263],[196,0.1264],[197,0.1266],[198,0.1265],[223,0.1261],[225,0.1256],[227,0.1245],[229,0.1229],[232,0.1198],[233,0.1193],[234,0.1191],[236,0.1195],[238,0.1189],[239,0.1179],[240,0.1161],[241,0.114],[242,0.1115],[243,0.1087],[244,0.1052],[245,0.1006],[246,0.0952],[248,0.0833],[249,0.0787],[250,0.0763],[251,0.0754],[252,0.0758],[253,0.0769],[257,0.0819],[258,0.0831],[261,0.0853],[264,0.0868],[267,0.0877],[270,0.0879],[284,0.0852],[286,0.0849],[290,0.0854],[295,0.087],[297,0.088],[299,0.0897],[300,0.0897],[301,0.0885],[302,0.0863],[303,0.0835],[304,0.0801],[305,0.0761],[306,0.0717],[307,0.0665],[308,0.0611],[309,0.0559],[311,0.0465],[312,0.042],[313,0.0379],[317,0.0221],[318,0.0184],[319,0.0154],[320,0.0131],[321,0.0115],[322,0.0102],[323,0.0093],[324,0.0089],[325,0.0089],[326,0.0092],[327,0.0099],[333,0.016],[334,0.0168],[336,0.018],[341,0.0196],[342,0.0197],[346,0.0192],[347,0.019],[348,0.0184],[349,0.0172],[350,0.0153],[351,0.0113],[352,0.0052],[353,-0.0018],[354,-0.0083],[356,-0.0208],[357,-0.0266],[358,-0.0311],[359,-0.0345],[360,-0.0372],[361,-0.0392],[362,-0.0409],[363,-0.0422],[364,-0.0432],[365,-0.0439],[366,-0.0444],[367,-0.0446],[369,-0.0445],[370,-0.0441],[372,-0.0428],[373,-0.042],[374,-0.0409],[375,-0.0395],[376,-0.0376],[377,-0.0349],[378,-0.0312],[379,-0.0264],[380,-0.0203],[381,-0.0133],[384,0.0092],[386,0.024],[388,0.0392],[389,0.0465],[390,0.0532],[391,0.0592],[392,0.0642],[393,0.0681],[394,0.0714],[395,0.0742],[396,0.0764],[397,0.0782],[398,0.0796],[399,0.0806],[400,0.0813],[401,0.0818],[403,0.0822],[405,0.0821],[407,0.0811],[412,0.0775],[416,0.0758],[419,0.0753],[422,0.0753]]
       }
       chest_tan_ctrl.Location.Y {
        type=MovieSceneScriptingFloatChannel
        default=-0.1191
        thinned_from=423
        keys [[0,-0.1191],[1,-0.1191],[9,-0.1207],[13,-0.1206],[14,-0.1201],[15,-0.1192],[16,-0.1178],[17,-0.1156],[18,-0.1123],[19,-0.1067],[20,-0.1001],[21,-0.0967],[22,-0.0966],[29,-0.0994],[41,-0.1021],[52,-0.1029],[57,-0.1029],[60,-0.1024],[62,-0.1018],[63,-0.1011],[65,-0.0994],[66,-0.099],[67,-0.0991],[69,-0.1001],[71,-0.1018],[73,-0.1041],[75,-0.1068],[76,-0.1083],[77,-0.1113],[78,-0.1158],[79,-0.1201],[81,-0.127],[82,-0.1297],[83,-0.1317],[84,-0.1332],[85,-0.1343],[88,-0.1361],[94,-0.1377],[125,-0.1352],[126,-0.1353],[127,-0.1358],[128,-0.1373],[129,-0.1399],[131,-0.146],[133,-0.1526],[134,-0.1532],[135,-0.1523],[136,-0.1522],[137,-0.1532],[138,-0.1568],[139,-0.145],[140,-0.1215],[141,-0.1135],[142,-0.1138],[143,-0.1195],[144,-0.1232],[145,-0.1254],[146,-0.127],[147,-0.1352],[148,-0.155],[149,-0.1831],[150,-0.2005],[151,-0.2003],[152,-0.1922],[153,-0.1866],[155,-0.18],[156,-0.179],[159,-0.1859],[160,-0.1878],[162,-0.1904],[165,-0.1927],[201,-0.1948],[211,-0.1921],[215,-0.1896],[221,-0.1828],[223,-0.1818],[225,-0.1819],[226,-0.1809],[227,-0.178],[229,-0.17],[230,-0.1687],[231,-0.1693],[232,-0.1713],[233,-0.1737],[234,-0.1756],[235,-0.1769],[242,-0.1783],[243,-0.1781],[245,-0.1823],[246,-0.1803],[247,-0.1718],[248,-0.1574],[249,-0.1416],[250,-0.128],[251,-0.1199],[252,-0.1165],[253,-0.117],[254,-0.1187],[257,-0.1244],[259,-0.1274],[261,-0.1299],[262,-0.131],[263,-0.1316],[266,-0.1324],[268,-0.1322],[272,-0.1296],[275,-0.1263],[276,-0.1256],[280,-0.1242],[286,-0.1234],[288,-0.124],[293,-0.127],[296,-0.1279],[298,-0.1278],[299,-0.1272],[300,-0.1262],[301,-0.1243],[302,-0.1215],[303,-0.1183],[304,-0.1145],[307,-0.1016],[308,-0.0969],[309,-0.0928],[310,-0.0899],[311,-0.0884],[312,-0.0877],[313,-0.0875],[314,-0.0877],[316,-0.0888],[319,-0.0914],[320,-0.0921],[325,-0.0933],[341,-0.0929],[343,-0.0931],[345,-0.0942],[346,-0.0952],[347,-0.0966],[348,-0.0985],[349,-0.1009],[350,-0.1039],[351,-0.1077],[353,-0.1169],[354,-0.1209],[355,-0.1243],[356,-0.1275],[357,-0.1302],[358,-0.1322],[359,-0.1337],[361,-0.1359],[363,-0.1371],[370,-0.1379],[372,-0.1374],[375,-0.1357],[377,-0.1336],[380,-0.1296],[385,-0.1217],[387,-0.1192],[389,-0.1174],[390,-0.1169],[391,-0.1166],[392,-0.1169],[399,-0.123],[401,-0.1238],[406,-0.1243],[408,-0.1239],[418,-0.1193],[422,-0.1191]]
       }
       chest_tan_ctrl.Location.Z {
        type=MovieSceneScriptingFloatChannel
        default=0.1383
        thinned_from=423
        keys [[0,0.1383],[13,0.1377],[16,0.1385],[18,0.1399],[20,0.1426],[21,0.1431],[31,0.1391],[37,0.1386],[63,0.1392],[65,0.1396],[66,0.1414],[67,0.1453],[70,0.1588],[74,0.1774],[75,0.1818],[76,0.1854],[77,0.1877],[78,0.1885],[79,0.1884],[84,0.1854],[125,0.1848],[126,0.1845],[127,0.1838],[128,0.1809],[129,0.1756],[130,0.1693],[131,0.1619],[132,0.1534],[133,0.1429],[134,0.1228],[135,0.0972],[136,0.0702],[137,0.046],[138,0.0749],[139,-0.1456],[140,-0.5265],[141,-0.7071],[142,-0.7597],[143,-0.7157],[144,-0.6958],[145,-0.6926],[146,-0.693],[147,-0.6026],[148,-0.3556],[149,0.0066],[150,0.2151],[151,0.2703],[152,0.2233],[153,0.1549],[154,0.1041],[155,0.0446],[156,0.0041],[160,0.0002],[163,-0.0016],[166,-0.0027],[203,-0.0035],[213,-0.002],[218,0],[220,0.0011],[221,0.0043],[222,0.0143],[223,0.0278],[224,0.0461],[225,0.0616],[226,0.068],[227,0.0429],[228,-0.0074],[229,-0.0619],[230,-0.0891],[231,-0.1019],[232,-0.1058],[234,-0.1075],[235,-0.108],[240,-0.1079],[241,-0.1075],[242,-0.1062],[243,-0.103],[244,-0.0507],[245,0.0406],[246,0.1308],[247,0.1889],[248,0.1892],[249,0.1667],[250,0.1366],[251,0.1185],[252,0.1112],[253,0.11],[254,0.1125],[255,0.1173],[256,0.1214],[257,0.1252],[258,0.1284],[259,0.1302],[260,0.1309],[263,0.1297],[267,0.1294],[272,0.1304],[274,0.1318],[282,0.1333],[288,0.1332],[296,0.1317],[298,0.1319],[299,0.1322],[300,0.1329],[301,0.1341],[303,0.1381],[306,0.1447],[307,0.1466],[308,0.1481],[309,0.149],[310,0.1494],[312,0.1489],[319,0.1449],[344,0.1441],[347,0.1432],[349,0.142],[351,0.14],[356,0.1337],[358,0.1319],[361,0.1303],[366,0.1295],[371,0.1298],[374,0.131],[375,0.1316],[376,0.1325],[381,0.1398],[382,0.141],[383,0.1418],[385,0.1425],[387,0.1422],[397,0.1368],[400,0.1359],[407,0.136],[419,0.1383],[422,0.1383]]
       }
       chest_tan_ctrl.Rotation.X {
        type=MovieSceneScriptingFloatChannel
        default=0
        thinned_from=423
        keys [[0,8.735],[1,8.735],[2,1.9188],[3,1.9194],[5,1.9305],[10,1.9458],[12,1.9458],[14,1.9536],[15,1.953],[17,1.9392],[18,1.9231],[19,1.9006],[20,1.8],[21,1.6083],[22,1.4711],[23,1.447],[24,1.4389],[25,1.4354],[33,1.4553],[41,1.507],[45,1.5216],[46,1.5209],[47,1.5268],[48,1.5236],[49,1.5296],[50,1.5261],[51,1.5316],[53,1.5297],[54,1.533],[58,1.5342],[59,1.5367],[61,1.5337],[62,1.5294],[63,1.5283],[64,1.5213],[65,1.501],[66,1.4876],[67,1.4845],[69,1.4703],[70,1.4718],[72,1.4868],[73,1.4969],[75,1.5265],[76,1.5375],[77,1.5238],[78,1.5485],[79,1.6377],[80,1.7383],[82,1.9207],[83,2.0002],[84,2.0564],[85,2.1013],[86,2.1366],[87,2.162],[89,2.1947],[92,2.2279],[96,2.2529],[100,2.258],[109,2.2251],[114,2.2111],[115,2.2056],[116,2.2083],[117,2.202],[122,2.1908],[125,2.195],[126,2.1879],[127,2.1724],[128,2.0977],[129,2.0392],[130,2.0736],[131,2.1268],[132,2.1929],[133,2.2667],[135,2.3806],[136,2.4181],[137,2.5227],[138,2.8284],[139,2.7376],[140,2.9386],[141,2.799],[142,2.7934],[143,2.8215],[146,2.9635],[147,2.9326],[148,2.872],[149,2.8628],[150,3.1196],[151,3.4448],[152,3.4361],[153,3.2399],[154,3.1509],[155,3.1459],[156,3.1216],[157,3.1178],[161,3.3287],[162,3.3702],[164,3.4234],[166,3.4611],[168,3.4815],[171,3.4936],[172,3.4928],[173,3.5009],[174,3.4971],[175,3.5058],[176,3.5029],[177,3.51],[206,3.51],[214,3.4515],[217,3.4078],[220,3.341],[224,3.223],[225,3.2026],[226,3.1958],[227,3.1956],[228,3.1876],[229,3.1699],[230,3.1283],[231,3.1048],[232,3.1077],[233,3.1383],[235,3.2392],[236,3.2789],[239,3.2673],[240,3.2682],[241,3.2772],[242,3.2995],[243,3.3358],[244,3.4186],[245,3.5634],[246,3.5757],[247,3.5181],[248,3.3503],[249,3.0698],[250,2.7324],[251,2.3774],[252,2.0873],[253,1.8819],[254,1.8487],[255,1.8661],[257,1.941],[258,1.9845],[259,2.0199],[263,2.1261],[264,2.1423],[268,2.1593],[269,2.1573],[270,2.1473],[272,2.1153],[273,2.1082],[274,2.0874],[276,2.0322],[277,2.0126],[281,1.9783],[285,1.9637],[288,1.9562],[290,1.9727],[294,2.0266],[298,2.0675],[299,2.07],[301,2.0408],[302,2.0172],[303,1.9676],[304,1.9086],[305,1.8403],[306,1.7554],[308,1.5951],[310,1.3924],[311,1.3152],[312,1.2749],[313,1.2539],[314,1.2487],[315,1.253],[316,1.2637],[317,1.2789],[318,1.3012],[320,1.352],[321,1.3746],[322,1.3884],[325,1.4059],[326,1.406],[327,1.4131],[330,1.4193],[334,1.4121],[341,1.3978],[344,1.3948],[346,1.4085],[347,1.4203],[348,1.4394],[349,1.4684],[350,1.5078],[351,1.5541],[352,1.6187],[353,1.7059],[354,1.8042],[355,1.8917],[357,2.0318],[358,2.0939],[359,2.1413],[360,2.1741],[361,2.2012],[362,2.2229],[363,2.2398],[364,2.2521],[367,2.268],[372,2.2729],[374,2.2634],[377,2.2301],[381,2.1395],[384,2.046],[387,1.9603],[388,1.9365],[390,1.9007],[392,1.8767],[393,1.8759],[394,1.8886],[396,1.9284],[400,2.0004],[401,2.0125],[405,2.0228],[408,2.0241],[410,2.0141],[416,1.9586],[419,1.9221],[422,1.9221]]
       }
       chest_tan_ctrl.Rotation.Y {
        type=MovieSceneScriptingFloatChannel
        default=1.2922
        thinned_from=423
        keys [[0,1.8055],[1,1.8055],[2,1.5372],[13,1.5398],[16,1.529],[19,1.5031],[20,1.5155],[21,1.5429],[23,1.5818],[24,1.611],[31,1.8677],[32,1.9015],[33,1.9311],[34,1.9561],[35,1.9711],[36,1.9793],[37,1.9804],[38,1.9748],[39,1.9569],[41,1.9086],[42,1.8876],[45,1.858],[46,1.8569],[47,1.8497],[48,1.8512],[49,1.846],[50,1.848],[51,1.8441],[57,1.8434],[58,1.8384],[60,1.8529],[62,1.8774],[65,1.8927],[68,1.889],[69,1.8849],[71,1.889],[72,1.8879],[73,1.8828],[74,1.8866],[75,1.8804],[76,1.8483],[77,1.8102],[78,1.7679],[79,1.7092],[80,1.6429],[81,1.5614],[83,1.3415],[84,1.2513],[85,1.1656],[86,1.092],[87,1.0296],[88,0.9797],[89,0.9407],[90,0.9085],[91,0.8836],[92,0.8702],[93,0.8634],[94,0.8597],[95,0.8595],[96,0.8638],[97,0.8651],[101,0.8921],[103,0.9021],[109,0.9134],[115,0.9174],[116,0.9151],[117,0.9165],[118,0.9135],[120,0.9176],[122,0.9251],[125,0.931],[126,0.9219],[127,0.9153],[128,0.9316],[129,0.9704],[130,1.0064],[131,1.0607],[132,1.1274],[133,1.1865],[134,1.2886],[135,1.4913],[136,1.7545],[137,1.9634],[138,2.2836],[139,2.8668],[140,3.0343],[141,2.9003],[142,2.6301],[143,2.4049],[144,2.3842],[145,2.3536],[147,2.3022],[148,2.2685],[149,2.2737],[150,2.3878],[151,2.524],[152,2.5841],[153,2.588],[154,2.484],[155,2.3278],[156,2.2611],[157,2.2264],[159,2.2256],[160,2.2379],[167,2.3583],[171,2.402],[173,2.4141],[176,2.4185],[182,2.4051],[183,2.3999],[186,2.3987],[197,2.3943],[198,2.3897],[220,2.3821],[224,2.3725],[227,2.3795],[228,2.3639],[229,2.3265],[230,2.2807],[232,2.2021],[233,2.1716],[234,2.1519],[235,2.1428],[236,2.1527],[237,2.1555],[239,2.1711],[240,2.1725],[241,2.1574],[242,2.1334],[243,2.1036],[244,2.0657],[245,2.0489],[246,2.0278],[247,1.978],[248,1.9126],[249,1.7324],[250,1.5586],[251,1.4476],[252,1.3778],[253,1.3655],[254,1.3904],[255,1.4311],[258,1.5202],[259,1.5459],[260,1.5643],[263,1.591],[265,1.5998],[269,1.5999],[273,1.5832],[281,1.5323],[285,1.518],[286,1.518],[288,1.5116],[290,1.5126],[294,1.5318],[297,1.5596],[298,1.5667],[299,1.5784],[300,1.5992],[301,1.6371],[302,1.6619],[303,1.6484],[305,1.5897],[306,1.5442],[307,1.5077],[308,1.4458],[309,1.3271],[310,1.2124],[313,0.9714],[314,0.9086],[316,0.8178],[317,0.769],[318,0.7162],[320,0.6078],[321,0.5703],[323,0.5272],[324,0.5082],[325,0.4979],[326,0.4939],[327,0.4981],[328,0.5058],[329,0.5171],[330,0.5299],[332,0.551],[334,0.5746],[335,0.585],[336,0.5925],[338,0.6039],[340,0.612],[341,0.6179],[343,0.6195],[344,0.6227],[347,0.6233],[348,0.6248],[349,0.6243],[350,0.6167],[351,0.6139],[352,0.5718],[353,0.4599],[354,0.3069],[355,0.1664],[356,0.0433],[357,0],[384,0],[385,0.0496],[386,0.1997],[387,0.359],[388,0.5352],[389,0.7074],[390,0.8706],[391,1.0209],[392,1.1623],[393,1.2845],[394,1.3771],[395,1.4517],[396,1.5127],[397,1.5627],[398,1.6009],[399,1.6285],[400,1.6459],[402,1.6668],[403,1.6712],[406,1.671],[408,1.6551],[413,1.5829],[415,1.5628],[419,1.5386],[422,1.5386]]
       }
       chest_tan_ctrl.Rotation.Z {
        type=MovieSceneScriptingFloatChannel
        default=0.5498
        thinned_from=423
        keys [[0,0.1789],[1,0.1789],[2,0.3607],[3,0.3671],[4,0.3475],[5,0.3227],[6,0.3153],[7,0.3043],[8,0.2914],[9,0.2853],[10,0.2758],[12,0.2758],[13,0.2286],[14,0.2597],[15,0.2916],[16,0.3422],[17,0.4626],[18,0.5811],[19,0.4738],[20,0.2327],[21,0],[83,0],[84,0.0266],[85,0.3558],[86,0.6526],[87,0.9203],[88,1.1412],[89,1.323],[90,1.4894],[91,1.6356],[92,1.759],[93,1.8598],[94,1.9478],[95,2.0225],[96,2.074],[97,2.0884],[98,2.1113],[99,2.1226],[100,2.1009],[101,2.0599],[102,2.0094],[103,1.9483],[104,1.9008],[105,1.8587],[106,1.8281],[108,1.7833],[109,1.7501],[110,1.7423],[112,1.7156],[113,1.7096],[114,1.6998],[115,1.6859],[116,1.6971],[117,1.6816],[118,1.6801],[119,1.6735],[120,1.6584],[121,1.6255],[122,1.6432],[123,1.6985],[124,1.7589],[125,1.8267],[126,1.8846],[127,1.961],[128,1.9953],[129,1.9749],[131,1.9772],[132,1.9304],[133,1.798],[134,1.7795],[135,1.6902],[136,1.524],[137,1.5354],[138,2.8904],[139,0],[315,0],[316,0.1919],[317,0.5247],[318,0.9022],[319,1.3045],[320,1.6419],[321,1.9523],[322,2.1559],[323,2.2835],[324,2.3765],[325,2.4387],[326,2.4504],[327,2.5217],[328,2.5462],[329,2.5026],[330,2.4451],[331,2.3664],[333,2.1733],[334,2.0832],[335,1.9712],[336,1.8776],[337,1.8228],[338,1.7805],[339,1.7483],[340,1.7238],[341,1.6683],[342,1.6691],[343,1.661],[344,1.6363],[345,1.6509],[347,1.6567],[348,1.6746],[349,1.7],[350,1.7323],[351,1.7731],[352,1.7996],[353,1.8087],[355,1.8082],[357,1.8146],[359,1.8121],[363,1.842],[367,1.8873],[368,1.8932],[369,1.8882],[371,1.9074],[375,1.6911],[378,1.51],[380,1.4024],[381,1.3558],[382,1.3192],[384,1.258],[387,1.1545],[389,1.0913],[390,1.0566],[392,0.9796],[393,0.9362],[394,0.8864],[395,0.833],[397,0.7205],[399,0.6026],[400,0.5422],[402,0.4016],[403,0.3462],[404,0.3044],[405,0.2752],[406,0.2564],[407,0.2446],[408,0.2378],[409,0.2358],[410,0.2381],[411,0.2439],[412,0.2526],[413,0.2634],[415,0.2822],[416,0.2952],[417,0.3103],[418,0.3262],[419,0.3464],[422,0.3464]]
       }
       chest_tan_ctrl.Scale.X {type=MovieSceneScriptingFloatChannel, default=1, const [1,423]}
       chest_tan_ctrl.Scale.Y {type=MovieSceneScriptingFloatChannel, default=1, const [1,423]}
       chest_tan_ctrl.Scale.Z {type=MovieSceneScriptingFloatChannel, default=1, const [1,423]}
      ]
     }]
   }
  ]
 }]
display_rate {denominator=1, numerator=30}
tick_resolution {denominator=1, numerator=24000}
tracks []
